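; Base Blueprint for Fortnite player pet instances, handling animation layer linking and lobby hologram visuals.
; On BeginPlay it calls the parent, then links the configured LinkedLayerAnimClass onto the pet mesh when that class is valid.
; In the pet materials override event, if the owning player pawn is a front-end hologram, it sets every material slot to the lobby holo character material so variant materials do not replace it.
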(event EventBeginPlay
  (bind _linkedlayeranimclass (Variables|Default|GetLinkedLayerAnimClass))
  (|CallParentFunction)
  (if (Utilities|IsValidClass _linkedlayeranimclass)
    (Components|SkeletalMesh|AnimationBlueprintLinking|LinkAnimClassLayers (Variables|FortPlayerPet|GetPetMesh) _linkedlayeranimclass)))

(event EventPetMaterialsOverride
  (bind _petmesh (Variables|FortPlayerPet|GetPetMesh))
  ;[0 0 0] Called after pet materials reset, if they are overridden they arent stomped by variant materials
  (if (Variables|/Script/FortniteGame.FortPlayerPawn|GetbIsInFrontEndHologram (CallFunction|GetFortPlayerPawn))
    ;;[0 0 0] party in progress holo material
    (for _index (range (Rendering|Material|GetNumMaterials _petmesh))
      (Rendering|Material|SetMaterial _petmesh _index "/Game/UI/Frontend/Lobby/HeroSquad/M_CharacterShader_Holo_Inst.M_CharacterShader_Holo_Inst"))))

(fn ConstructionScript)
