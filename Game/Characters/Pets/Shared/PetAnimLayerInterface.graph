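# Animation Layer Interface for pets, with no target skeleton, declaring override layers that implementing anim blueprints fill in.
# PetFullOverrideLayer declares a pose in and a pose out.
# PetSkeletalControlLayer also takes a look-at override, look-at target location, and leg and hand IK alpha overrides alongside its input pose.
# Both layers share the PetOverride group, so a linked implementation can run them in one shared instance.
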AnimBlueprint /Game/Characters/Pets/Shared/PetAnimLayerInterface
target_skeleton=null
graphs [
 PetFullOverrideLayer {nodes [
   Input Pose()
   Output Pose() {layerGroup=PetOverride}
  ]}
 PetSkeletalControlLayer {nodes [
   Input Pose!() -> (Pose, LookAtOverrideInput:double=~, LookAtTargetLocationInput:Vector=~, LegIKAlphaOverride:double=~, HandIKAlphaOverride:double=~) {
    name=Pose
    inputs [
     LookAtOverrideInput {type {
       bIsConst=false
       bIsReference=false
       bIsUObjectWrapper=false
       bIsWeakPointer=false
       bSerializeAsSinglePrecisionFloat=true
       containerType=None
       pinCategory=real
       pinSubCategory=double
       pinSubCategoryObject=null
       pinSubCategoryMemberReference {memberGuid=#1, memberName=None, memberParent=null}
       pinValueType {
        bTerminalIsConst=false
        bTerminalIsUObjectWrapper=false
        bTerminalIsWeakPointer=false
        terminalCategory=None
        terminalSubCategory=None
        terminalSubCategoryObject=null
       }
      }}
     LookAtTargetLocationInput {type {
       bIsConst=false
       bIsReference=false
       bIsUObjectWrapper=false
       bIsWeakPointer=false
       bSerializeAsSinglePrecisionFloat=false
       containerType=None
       pinCategory=struct
       pinSubCategory=None
       pinSubCategoryObject=/Script/CoreUObject.Vector
       pinSubCategoryMemberReference {memberGuid=#1, memberName=None, memberParent=null}
       pinValueType {
        bTerminalIsConst=false
        bTerminalIsUObjectWrapper=false
        bTerminalIsWeakPointer=false
        terminalCategory=None
        terminalSubCategory=None
        terminalSubCategoryObject=null
       }
      }}
     LegIKAlphaOverride {type {
       bIsConst=false
       bIsReference=false
       bIsUObjectWrapper=false
       bIsWeakPointer=false
       bSerializeAsSinglePrecisionFloat=true
       containerType=None
       pinCategory=real
       pinSubCategory=double
       pinSubCategoryObject=null
       pinSubCategoryMemberReference {memberGuid=#1, memberName=None, memberParent=null}
       pinValueType {
        bTerminalIsConst=false
        bTerminalIsUObjectWrapper=false
        bTerminalIsWeakPointer=false
        terminalCategory=None
        terminalSubCategory=None
        terminalSubCategoryObject=null
       }
      }}
     HandIKAlphaOverride {type {
       bIsConst=false
       bIsReference=false
       bIsUObjectWrapper=false
       bIsWeakPointer=false
       bSerializeAsSinglePrecisionFloat=true
       containerType=None
       pinCategory=real
       pinSubCategory=double
       pinSubCategoryObject=null
       pinSubCategoryMemberReference {memberGuid=#1, memberName=None, memberParent=null}
       pinValueType {
        bTerminalIsConst=false
        bTerminalIsUObjectWrapper=false
        bTerminalIsWeakPointer=false
        terminalCategory=None
        terminalSubCategory=None
        terminalSubCategoryObject=null
       }
      }}
    ]
   }
   Output Pose() {layerGroup=PetOverride}
  ]}
]
variables []
