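# Master animation Blueprint for backpack pets on the Chameleon skeleton, reacting to the owner's movement and state.
# A FullBody state machine covers idle, locomotion, crouch, aiming, jumping, skydive, DBNO, hunker and lobby, with a crouch-pose alternative for non-player-AnimBP owners.
# Its pose passes through an interface IK/look-at layer, skydive root rotation and montage slot.
# Cached poses then feed face pose-asset blending, fox-dependent jaw/tongue masking and a full-override layer.
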AnimBlueprint /Game/Characters/Pets/Shared/B_MasterPetAnimBP
target_skeleton=/Game/Characters/Pets/Chameleon/Meshes/Chameleon_Skeleton
graphs [
 AnimGraph {nodes [
   "Transform (Modify) Bone - Bone: root"!(
    ComponentPose
    Alpha="To Float (Boolean)"(InBool=Get bEmotePlaying_WhereHunkeringIsBest() -> (bEmotePlaying_WhereHunkeringIsBest:bool))
    Translation=Get PlayerLocation() -> (PlayerLocation:Vector)
    Rotation="0,0,0"
   ) -> (Pose) {
    rotationMode=BMM_Replace
    rotationSpace=BCS_WorldSpace
    translationMode=BMM_Replace
    translationSpace=BCS_WorldSpace
    alphaBoolBlend {blendOption=Linear}
    boneToModify {boneName=root}
   }
   n1 = Layered blend per bone(
    BasePose=Use cached pose 'AfterFaceCurves'()
    BlendPoses_0=Use cached pose 'BeforeFaceCurves'()
    BlendWeights_0=Get FaceCurveOverrideValue() -> (FaceCurveOverrideValue:float)
   ) {layerSetup [{branchFilters [{boneName=C_jaw}, {boneName=L_lip_corner}, {boneName=R_lip_corner}]}]}
   n2 = Layered blend per bone(BasePose=@n1, BlendPoses_0=Local Space Ref Pose()) {layerSetup [{branchFilters [{boneName=tongue}]}]}
   n3 = Use cached pose 'AfterFaceCurves'()
   n4 = Get bIsFox() -> (bIsFox:bool)
   Output Pose(Result=PetAnimLayerInterface - PetFullOverrideLayer!(InPose:Pose Link Structure=Blend Poses by bool(BlendPose_0=@n2, BlendPose_1=@n3, bActiveValue=@n4) {blendType=Linear}) -> (Pose:Pose Link Structure) {
     is=title AnimGraphNode_LinkedAnimLayer
     interface=&/PetAnimLayerInterface.PetAnimLayerInterface_C
     layer=PetFullOverrideLayer
    }) {layerGroup=None}
   AnimGraphNode_StateMachine_0 = FullBody() {is=title AnimGraphNode_StateMachine}
   n5 = Blend Poses by bool(
    BlendPose_0=Apply Additive(
     Base=Sequence Player 'Crouch_Pose'() {sequence=/Game/Animation/Game/Pets/Chameleon/Backpack/Crouch/Crouch_Pose}
     Additive="Sequence Player (additive) 'IdleStraightLookAround_Add'"() {sequence=/Game/Animation/Game/Pets/Chameleon/Backpack/Idle/IdleStraightLookAround_Add}
    ) {alphaBoolBlend {blendOption=Linear}}
    BlendPose_1=@AnimGraphNode_StateMachine_0
    bActiveValue=Get bIsOwnerPawnBeingControlledByNonPlayerAnimBP() -> (bIsOwnerPawnBeingControlledByNonPlayerAnimBP:bool)
   ) {blendType=Linear}
   n6 = Get HandIKAlpha() -> (HandIKAlpha:float)
   n7 = Get LookAtOverride() -> (LookAtOverride:float)
   n8 = Get LegIKAlpha() -> (LegIKAlpha:float)
   n9 = Get LookAtTargetLocation() -> (LookAtTargetLocation:Vector)
   n10 = "Transform (Modify) Bone - Bone: root"!(
    ComponentPose=Local To Component(LocalPose=PetAnimLayerInterface - PetSkeletalControlLayer!(
      Pose:Pose Link Structure=@n5
      LookAtOverrideInput:double=@n7
      LookAtTargetLocationInput:Vector=@n9
      LegIKAlphaOverride:double=@n8
      HandIKAlphaOverride:double=@n6
     ) -> (Pose:Pose Link Structure) {
      is=title AnimGraphNode_LinkedAnimLayer
      interface=&PetAnimLayerInterface_C
      layer=PetSkeletalControlLayer
     })
    bAlphaBoolEnabled:bool=Get bUseSkydiveRotation() -> (bUseSkydiveRotation:bool)
    Rotation=Get SkydiveRotation() -> (SkydiveRotation:Rotator)
   ) -> (Pose) {
    alphaInputType=Bool
    rotationMode=BMM_Replace
    rotationSpace=BCS_WorldSpace
    alphaBoolBlend {blendInTime=0.2, blendOption=Linear, blendOutTime=0.2}
    alphaScaleBiasClamp {bInterpResult=true, interpSpeedDecreasing=1, interpSpeedIncreasing=1}
    boneToModify {boneName=root}
   }
   Save cached pose 'BeforeFaceCurves'(Pose=Slot 'DefaultSlot'(Source=Component To Local(ComponentPose=@n10))) {cacheName=BeforeFaceCurves}
   Save cached pose 'AfterFaceCurves'(Pose=Pose!(
     PoseAsset:Pose Asset Object Reference=Get FacePoseAsset() -> (FacePoseAsset:Pose Asset Object Reference)
     SourcePose:Pose Link Structure=Use cached pose 'BeforeFaceCurves'()
    ) -> (Pose:Pose Link Structure) {is=title AnimGraphNode_PoseBlendNode, blendOption=Linear}) {cacheName=AfterFaceCurves}
   AnimGraphNode_StateMachine_1 = FullBody_1() {is=title AnimGraphNode_StateMachine}
  ]}
 FullBody {
  entry=idle
  path=AnimGraph.AnimGraphNode_StateMachine_0.FullBody
  ids=id
  nodes [
   AnimStateConduitNode_0 = NotSkydiving() {is=title AnimStateConduitNode}
   AnimStateConduitNode_1 = IDS() {is=title AnimStateConduitNode}
   AnimStateConduitNode_2 = PopToSkydive() {is=title AnimStateConduitNode}
   AnimStateNode_0 = IdleLobby() {is=title AnimStateNode}
   AnimStateNode_1 = IdleToCrouch() {is=title AnimStateNode}
   AnimStateNode_10 = IdleToDBNO_Old() {is=title AnimStateNode}
   AnimStateNode_11 = DBNOToIdle() {is=title AnimStateNode}
   AnimStateNode_12 = DownSights() {is=title AnimStateNode}
   AnimStateNode_13 = IdleToDownSights() {is=title AnimStateNode}
   AnimStateNode_14 = DownSightsToIdle() {is=title AnimStateNode}
   AnimStateNode_15 = DBNO() {is=title AnimStateNode}
   AnimStateNode_16 = LocomotionToCrouch() {is=title AnimStateNode}
   AnimStateNode_17 = CrouchWalkingDownSights() {is=title AnimStateNode}
   AnimStateNode_18 = Crouching() {is=title AnimStateNode}
   AnimStateNode_19 = HunkerToCrouch() {is=title AnimStateNode}
   AnimStateNode_2 = LocomotionHappy() {is=title AnimStateNode}
   AnimStateNode_20 = idle() {is=title AnimStateNode}
   AnimStateNode_21 = Locomotion() {is=title AnimStateNode}
   AnimStateNode_22 = HunkerToLobby() {is=title AnimStateNode}
   AnimStateNode_23 = Hunker() {is=title AnimStateNode}
   AnimStateNode_24 = LobbyToHunker() {is=title AnimStateNode}
   AnimStateNode_25 = FrontendSkydive() {is=title AnimStateNode}
   AnimStateNode_26 = InAirHunker() {is=title AnimStateNode}
   AnimStateNode_27 = LobbyHunker() {is=title AnimStateNode}
   AnimStateNode_28 = CrouchToCrouchADS() {is=title AnimStateNode}
   AnimStateNode_29 = CrouchToCrouchADSNotMoving() {is=title AnimStateNode}
   AnimStateNode_3 = LocomotionToIdle() {is=title AnimStateNode}
   AnimStateNode_30 = ADSToCrouch() {is=title AnimStateNode}
   AnimStateNode_31 = CrouchToIdle() {is=title AnimStateNode}
   AnimStateNode_32 = DBNOToIdle_Old() {is=title AnimStateNode}
   AnimStateNode_34 = SkyDive() {is=title AnimStateNode}
   AnimStateNode_4 = StartJump() {
    is=title AnimStateNode
    stateEntered {guid=#1, notifyName=StartJump_Entered}
    stateFullyBlended {guid=#2}
    stateLeft {guid=#3}
   }
   AnimStateNode_5 = JumpLoop() {
    is=title AnimStateNode
    stateEntered {guid=#4}
    stateFullyBlended {guid=#5}
    stateLeft {guid=#6}
   }
   AnimStateNode_6 = JumpFall() {
    is=title AnimStateNode
    stateEntered {guid=#7}
    stateFullyBlended {guid=#8}
    stateLeft {guid=#9}
   }
   AnimStateNode_7 = JumpLand() {is=title AnimStateNode, stateLeft {guid=#10, notifyName=JumpLandExit}}
   AnimStateNode_8 = CrouchToLocomotion() {is=title AnimStateNode}
   AnimStateNode_9 = DBNO_old() {is=title AnimStateNode}
   AnimStateTransitionNode_0 = AnimStateTransitionNode()
   AnimStateTransitionNode_1 = AnimStateTransitionNode() {crossfadeDuration=0, priorityOrder=0}
   AnimStateTransitionNode_10 = AnimStateTransitionNode() {
    transitionEnd {guid=#11}
    transitionInterrupt {guid=#12}
    transitionStart {guid=#13}
   }
   AnimStateTransitionNode_100 = AnimStateTransitionNode() {bAutomaticRuleBasedOnSequencePlayerInState=true}
   AnimStateTransitionNode_101 = AnimStateTransitionNode()
   AnimStateTransitionNode_102 = AnimStateTransitionNode()
   AnimStateTransitionNode_103 = AnimStateTransitionNode()
   AnimStateTransitionNode_104 = AnimStateTransitionNode()
   AnimStateTransitionNode_105 = AnimStateTransitionNode()
   AnimStateTransitionNode_106 = AnimStateTransitionNode() {crossfadeDuration=0.25, priorityOrder=0}
   AnimStateTransitionNode_108 = AnimStateTransitionNode() {crossfadeDuration=0.25}
   AnimStateTransitionNode_109 = AnimStateTransitionNode()
   AnimStateTransitionNode_11 = AnimStateTransitionNode()
   AnimStateTransitionNode_110 = AnimStateTransitionNode() {bAutomaticRuleBasedOnSequencePlayerInState=true}
   AnimStateTransitionNode_12 = AnimStateTransitionNode() {bAutomaticRuleBasedOnSequencePlayerInState=true}
   AnimStateTransitionNode_13 = AnimStateTransitionNode()
   AnimStateTransitionNode_14 = AnimStateTransitionNode() {
    transitionEnd {guid=#14}
    transitionInterrupt {guid=#15}
    transitionStart {guid=#16}
   }
   AnimStateTransitionNode_16 = AnimStateTransitionNode()
   AnimStateTransitionNode_17 = AnimStateTransitionNode()
   AnimStateTransitionNode_18 = AnimStateTransitionNode()
   AnimStateTransitionNode_19 = AnimStateTransitionNode() {bAutomaticRuleBasedOnSequencePlayerInState=true, crossfadeDuration=0.3}
   AnimStateTransitionNode_2 = AnimStateTransitionNode() {crossfadeDuration=0}
   AnimStateTransitionNode_20 = AnimStateTransitionNode()
   AnimStateTransitionNode_21 = AnimStateTransitionNode() {bAutomaticRuleBasedOnSequencePlayerInState=true}
   AnimStateTransitionNode_22 = AnimStateTransitionNode()
   AnimStateTransitionNode_23 = AnimStateTransitionNode()
   AnimStateTransitionNode_24 = AnimStateTransitionNode()
   AnimStateTransitionNode_25 = AnimStateTransitionNode() {bAutomaticRuleBasedOnSequencePlayerInState=true}
   AnimStateTransitionNode_26 = AnimStateTransitionNode()
   AnimStateTransitionNode_27 = AnimStateTransitionNode() {bAutomaticRuleBasedOnSequencePlayerInState=true}
   AnimStateTransitionNode_28 = AnimStateTransitionNode() {bAutomaticRuleBasedOnSequencePlayerInState=true}
   AnimStateTransitionNode_29 = AnimStateTransitionNode()
   AnimStateTransitionNode_3 = AnimStateTransitionNode()
   AnimStateTransitionNode_30 = AnimStateTransitionNode()
   AnimStateTransitionNode_31 = AnimStateTransitionNode() {crossfadeDuration=0.3}
   AnimStateTransitionNode_32 = AnimStateTransitionNode()
   AnimStateTransitionNode_33 = AnimStateTransitionNode() {bAutomaticRuleBasedOnSequencePlayerInState=true, crossfadeDuration=0.3}
   AnimStateTransitionNode_34 = AnimStateTransitionNode()
   AnimStateTransitionNode_35 = AnimStateTransitionNode()
   AnimStateTransitionNode_36 = AnimStateTransitionNode() {bAutomaticRuleBasedOnSequencePlayerInState=true}
   AnimStateTransitionNode_37 = AnimStateTransitionNode()
   AnimStateTransitionNode_38 = AnimStateTransitionNode()
   AnimStateTransitionNode_39 = AnimStateTransitionNode()
   AnimStateTransitionNode_4 = AnimStateTransitionNode()
   AnimStateTransitionNode_40 = AnimStateTransitionNode()
   AnimStateTransitionNode_41 = AnimStateTransitionNode()
   AnimStateTransitionNode_42 = AnimStateTransitionNode()
   AnimStateTransitionNode_43 = AnimStateTransitionNode()
   AnimStateTransitionNode_44 = AnimStateTransitionNode()
   AnimStateTransitionNode_45 = AnimStateTransitionNode()
   AnimStateTransitionNode_46 = AnimStateTransitionNode() {priorityOrder=0}
   AnimStateTransitionNode_47 = AnimStateTransitionNode() {crossfadeDuration=0}
   AnimStateTransitionNode_48 = AnimStateTransitionNode()
   AnimStateTransitionNode_49 = AnimStateTransitionNode()
   AnimStateTransitionNode_5 = AnimStateTransitionNode() {bAutomaticRuleBasedOnSequencePlayerInState=true}
   AnimStateTransitionNode_50 = AnimStateTransitionNode()
   AnimStateTransitionNode_51 = AnimStateTransitionNode()
   AnimStateTransitionNode_52 = AnimStateTransitionNode()
   AnimStateTransitionNode_53 = AnimStateTransitionNode()
   AnimStateTransitionNode_54 = AnimStateTransitionNode() {bAutomaticRuleBasedOnSequencePlayerInState=true}
   AnimStateTransitionNode_55 = AnimStateTransitionNode() {bAutomaticRuleBasedOnSequencePlayerInState=true, crossfadeDuration=0.3}
   AnimStateTransitionNode_56 = AnimStateTransitionNode()
   AnimStateTransitionNode_57 = AnimStateTransitionNode() {bAutomaticRuleBasedOnSequencePlayerInState=true, crossfadeDuration=0.3}
   AnimStateTransitionNode_58 = AnimStateTransitionNode()
   AnimStateTransitionNode_59 = AnimStateTransitionNode()
   AnimStateTransitionNode_6 = AnimStateTransitionNode()
   AnimStateTransitionNode_60 = AnimStateTransitionNode() {crossfadeDuration=0}
   AnimStateTransitionNode_61 = AnimStateTransitionNode()
   AnimStateTransitionNode_62 = AnimStateTransitionNode()
   AnimStateTransitionNode_63 = AnimStateTransitionNode()
   AnimStateTransitionNode_64 = AnimStateTransitionNode() {crossfadeDuration=0, priorityOrder=0}
   AnimStateTransitionNode_65 = AnimStateTransitionNode()
   AnimStateTransitionNode_66 = AnimStateTransitionNode() {bAutomaticRuleBasedOnSequencePlayerInState=true}
   AnimStateTransitionNode_67 = AnimStateTransitionNode() {crossfadeDuration=0}
   AnimStateTransitionNode_68 = AnimStateTransitionNode()
   AnimStateTransitionNode_69 = AnimStateTransitionNode()
   AnimStateTransitionNode_7 = AnimStateTransitionNode() {crossfadeDuration=0}
   AnimStateTransitionNode_70 = AnimStateTransitionNode()
   AnimStateTransitionNode_71 = AnimStateTransitionNode()
   AnimStateTransitionNode_72 = AnimStateTransitionNode()
   AnimStateTransitionNode_73 = AnimStateTransitionNode()
   AnimStateTransitionNode_74 = AnimStateTransitionNode()
   AnimStateTransitionNode_75 = AnimStateTransitionNode()
   AnimStateTransitionNode_76 = AnimStateTransitionNode()
   AnimStateTransitionNode_77 = AnimStateTransitionNode() {bAutomaticRuleBasedOnSequencePlayerInState=true}
   AnimStateTransitionNode_78 = AnimStateTransitionNode()
   AnimStateTransitionNode_79 = AnimStateTransitionNode()
   AnimStateTransitionNode_8 = AnimStateTransitionNode() {crossfadeDuration=0.1}
   AnimStateTransitionNode_80 = AnimStateTransitionNode()
   AnimStateTransitionNode_81 = AnimStateTransitionNode()
   AnimStateTransitionNode_82 = AnimStateTransitionNode()
   AnimStateTransitionNode_83 = AnimStateTransitionNode()
   AnimStateTransitionNode_84 = AnimStateTransitionNode()
   AnimStateTransitionNode_85 = AnimStateTransitionNode()
   AnimStateTransitionNode_86 = AnimStateTransitionNode()
   AnimStateTransitionNode_87 = AnimStateTransitionNode()
   AnimStateTransitionNode_88 = AnimStateTransitionNode()
   AnimStateTransitionNode_89 = AnimStateTransitionNode()
   AnimStateTransitionNode_9 = AnimStateTransitionNode()
   AnimStateTransitionNode_90 = AnimStateTransitionNode()
   AnimStateTransitionNode_91 = AnimStateTransitionNode()
   AnimStateTransitionNode_92 = AnimStateTransitionNode()
   AnimStateTransitionNode_93 = AnimStateTransitionNode()
   AnimStateTransitionNode_94 = AnimStateTransitionNode()
   AnimStateTransitionNode_95 = AnimStateTransitionNode()
   AnimStateTransitionNode_96 = AnimStateTransitionNode() {bAutomaticRuleBasedOnSequencePlayerInState=true, crossfadeDuration=0.35}
   AnimStateTransitionNode_97 = AnimStateTransitionNode()
   AnimStateTransitionNode_98 = AnimStateTransitionNode() {bAutomaticRuleBasedOnSequencePlayerInState=true, crossfadeDuration=0.35}
   AnimStateTransitionNode_99 = AnimStateTransitionNode()
  ]
  links [
   {
    dst=Crouching
    rule=AnimGraph.AnimGraphNode_StateMachine_0.FullBody.AnimStateTransitionNode_100.Transition
    src=ADSToCrouch
    via=AnimStateTransitionNode_100
   }
   {
    dst=CrouchWalkingDownSights
    rule=AnimGraph.AnimGraphNode_StateMachine_0.FullBody.AnimStateTransitionNode_96.Transition
    src=CrouchToCrouchADS
    via=AnimStateTransitionNode_96
   }
   {
    dst=Crouching
    rule=AnimGraph.AnimGraphNode_StateMachine_0.FullBody.AnimStateTransitionNode_102.Transition
    src=CrouchToCrouchADS
    via=AnimStateTransitionNode_102
   }
   {
    dst=ADSToCrouch
    rule=AnimGraph.AnimGraphNode_StateMachine_0.FullBody.AnimStateTransitionNode_97.Transition
    src=CrouchToCrouchADSNotMoving
    via=AnimStateTransitionNode_97
   }
   {
    dst=DownSights
    rule=AnimGraph.AnimGraphNode_StateMachine_0.FullBody.AnimStateTransitionNode_98.Transition
    src=CrouchToCrouchADSNotMoving
    via=AnimStateTransitionNode_98
   }
   {
    dst=DownSights
    rule=AnimGraph.AnimGraphNode_StateMachine_0.FullBody.AnimStateTransitionNode_92.Transition
    src=CrouchToIdle
    via=AnimStateTransitionNode_92
   }
   {
    dst=IdleToCrouch
    rule=AnimGraph.AnimGraphNode_StateMachine_0.FullBody.AnimStateTransitionNode_85.Transition
    src=CrouchToIdle
    via=AnimStateTransitionNode_85
   }
   {
    dst=idle
    rule=AnimGraph.AnimGraphNode_StateMachine_0.FullBody.AnimStateTransitionNode_36.Transition
    src=CrouchToIdle
    via=AnimStateTransitionNode_36
   }
   {
    dst=DownSights
    rule=AnimGraph.AnimGraphNode_StateMachine_0.FullBody.AnimStateTransitionNode_94.Transition
    src=CrouchToLocomotion
    via=AnimStateTransitionNode_94
   }
   {
    dst=Locomotion
    rule=AnimGraph.AnimGraphNode_StateMachine_0.FullBody.AnimStateTransitionNode_55.Transition
    src=CrouchToLocomotion
    via=AnimStateTransitionNode_55
   }
   {
    dst=LocomotionToCrouch
    rule=AnimGraph.AnimGraphNode_StateMachine_0.FullBody.AnimStateTransitionNode_0.Transition
    src=CrouchToLocomotion
    via=AnimStateTransitionNode_0
   }
   {
    dst=ADSToCrouch
    rule=AnimGraph.AnimGraphNode_StateMachine_0.FullBody.AnimStateTransitionNode_101.Transition
    src=CrouchWalkingDownSights
    via=AnimStateTransitionNode_101
   }
   {
    dst=DownSights
    rule=AnimGraph.AnimGraphNode_StateMachine_0.FullBody.AnimStateTransitionNode_41.AnimationTransitionGraph_0
    src=CrouchWalkingDownSights
    via=AnimStateTransitionNode_41
   }
   {
    dst=CrouchToCrouchADS
    rule=AnimGraph.AnimGraphNode_StateMachine_0.FullBody.AnimStateTransitionNode_95.Transition
    src=Crouching
    via=AnimStateTransitionNode_95
   }
   {
    dst=CrouchToCrouchADSNotMoving
    rule=AnimGraph.AnimGraphNode_StateMachine_0.FullBody.AnimStateTransitionNode_39.Transition
    src=Crouching
    via=AnimStateTransitionNode_39
   }
   {
    dst=CrouchToIdle
    rule=AnimGraph.AnimGraphNode_StateMachine_0.FullBody.AnimStateTransitionNode_29.Transition
    src=Crouching
    via=AnimStateTransitionNode_29
   }
   {
    dst=CrouchToLocomotion
    rule=AnimGraph.AnimGraphNode_StateMachine_0.FullBody.AnimStateTransitionNode_44.Transition
    src=Crouching
    via=AnimStateTransitionNode_44
   }
   {
    dst=DBNO
    rule=AnimGraph.AnimGraphNode_StateMachine_0.FullBody.AnimStateTransitionNode_108.Transition
    src=Crouching
    via=AnimStateTransitionNode_108
   }
   {
    dst=Hunker
    rule=AnimGraph.AnimGraphNode_StateMachine_0.FullBody.AnimStateTransitionNode_34.Transition
    src=Crouching
    via=AnimStateTransitionNode_34
   }
   {
    dst=DBNOToIdle
    rule=AnimGraph.AnimGraphNode_StateMachine_0.FullBody.AnimStateTransitionNode_109.Transition
    src=DBNO
    via=AnimStateTransitionNode_109
   }
   {
    dst=idle
    rule=AnimGraph.AnimGraphNode_StateMachine_0.FullBody.AnimStateTransitionNode_110.Transition
    src=DBNOToIdle
    via=AnimStateTransitionNode_110
   }
   {
    dst=DBNOToIdle_Old
    rule=AnimGraph.AnimGraphNode_StateMachine_0.FullBody.AnimStateTransitionNode_105.AnimationTransitionGraph_2
    src=DBNO_old
    via=AnimStateTransitionNode_105
   }
   {
    dst=ADSToCrouch
    rule=AnimGraph.AnimGraphNode_StateMachine_0.FullBody.AnimStateTransitionNode_99.Transition
    src=DownSights
    via=AnimStateTransitionNode_99
   }
   {
    dst=CrouchWalkingDownSights
    rule=AnimGraph.AnimGraphNode_StateMachine_0.FullBody.AnimStateTransitionNode_40.AnimationTransitionGraph_0
    src=DownSights
    via=AnimStateTransitionNode_40
   }
   {
    dst=DownSightsToIdle
    rule=AnimGraph.AnimGraphNode_StateMachine_0.FullBody.AnimStateTransitionNode_20.Transition
    src=DownSights
    via=AnimStateTransitionNode_20
   }
   {
    dst=StartJump
    rule=AnimGraph.AnimGraphNode_StateMachine_0.FullBody.AnimStateTransitionNode_35.Transition
    src=DownSights
    via=AnimStateTransitionNode_35
   }
   {
    dst=StartJump
    rule=AnimGraph.AnimGraphNode_StateMachine_0.FullBody.AnimStateTransitionNode_59.Transition
    src=DownSights
    via=AnimStateTransitionNode_59
   }
   {
    dst=IdleToDownSights
    rule=AnimGraph.AnimGraphNode_StateMachine_0.FullBody.AnimStateTransitionNode_45.Transition
    src=DownSightsToIdle
    via=AnimStateTransitionNode_45
   }
   {
    dst=Locomotion
    rule=AnimGraph.AnimGraphNode_StateMachine_0.FullBody.AnimStateTransitionNode_50.Transition
    src=DownSightsToIdle
    via=AnimStateTransitionNode_50
   }
   {
    dst=idle
    rule=AnimGraph.AnimGraphNode_StateMachine_0.FullBody.AnimStateTransitionNode_21.Transition
    src=DownSightsToIdle
    via=AnimStateTransitionNode_21
   }
   {
    dst=idle
    rule=AnimGraph.AnimGraphNode_StateMachine_0.FullBody.AnimStateTransitionNode_89.Transition
    src=DownSightsToIdle
    via=AnimStateTransitionNode_89
   }
   {
    dst=IdleLobby
    rule=AnimGraph.AnimGraphNode_StateMachine_0.FullBody.AnimStateTransitionNode_2.AnimationTransitionGraph_1
    src=FrontendSkydive
    via=AnimStateTransitionNode_2
   }
   {
    dst=LobbyHunker
    rule=AnimGraph.AnimGraphNode_StateMachine_0.FullBody.AnimStateTransitionNode_64.AnimationTransitionGraph_0
    src=FrontendSkydive
    via=AnimStateTransitionNode_64
   }
   {
    dst=HunkerToCrouch
    rule=AnimGraph.AnimGraphNode_StateMachine_0.FullBody.AnimStateTransitionNode_22.Transition
    src=Hunker
    via=AnimStateTransitionNode_22
   }
   {
    dst=Locomotion
    rule=AnimGraph.AnimGraphNode_StateMachine_0.FullBody.AnimStateTransitionNode_70.Transition
    src=Hunker
    via=AnimStateTransitionNode_70
   }
   {
    dst=idle
    rule=AnimGraph.AnimGraphNode_StateMachine_0.FullBody.AnimStateTransitionNode_61.Transition
    src=Hunker
    via=AnimStateTransitionNode_61
   }
   {
    dst=Crouching
    rule=AnimGraph.AnimGraphNode_StateMachine_0.FullBody.AnimStateTransitionNode_28.Transition
    src=HunkerToCrouch
    via=AnimStateTransitionNode_28
   }
   {
    dst=Hunker
    rule=AnimGraph.AnimGraphNode_StateMachine_0.FullBody.AnimStateTransitionNode_103.Transition
    src=HunkerToCrouch
    via=AnimStateTransitionNode_103
   }
   {
    dst=IdleLobby
    rule=AnimGraph.AnimGraphNode_StateMachine_0.FullBody.AnimStateTransitionNode_33.AnimationTransitionGraph_0
    src=HunkerToLobby
    via=AnimStateTransitionNode_33
   }
   {
    dst=PopToSkydive
    rule=AnimGraph.AnimGraphNode_StateMachine_0.FullBody.AnimStateTransitionNode_71.AnimationTransitionGraph_0
    src=HunkerToLobby
    via=AnimStateTransitionNode_71
   }
   {
    dst=IdleToDownSights
    rule=AnimGraph.AnimGraphNode_StateMachine_0.FullBody.AnimStateTransitionNode_43.Transition
    src=IDS
    via=AnimStateTransitionNode_43
   }
   {
    dst=LobbyHunker
    rule=AnimGraph.AnimGraphNode_StateMachine_0.FullBody.AnimStateTransitionNode_1.AnimationTransitionGraph_0
    src=IdleLobby
    via=AnimStateTransitionNode_1
   }
   {
    dst=LobbyToHunker
    rule=AnimGraph.AnimGraphNode_StateMachine_0.FullBody.AnimStateTransitionNode_65.AnimationTransitionGraph_0
    src=IdleLobby
    via=AnimStateTransitionNode_65
   }
   {
    dst=PopToSkydive
    rule=AnimGraph.AnimGraphNode_StateMachine_0.FullBody.AnimStateTransitionNode_68.AnimationTransitionGraph_0
    src=IdleLobby
    via=AnimStateTransitionNode_68
   }
   {
    dst=idle
    rule=AnimGraph.AnimGraphNode_StateMachine_0.FullBody.AnimStateTransitionNode_60.Transition
    src=IdleLobby
    via=AnimStateTransitionNode_60
   }
   {
    dst=CrouchToIdle
    rule=AnimGraph.AnimGraphNode_StateMachine_0.FullBody.AnimStateTransitionNode_78.Transition
    src=IdleToCrouch
    via=AnimStateTransitionNode_78
   }
   {
    dst=CrouchWalkingDownSights
    rule=AnimGraph.AnimGraphNode_StateMachine_0.FullBody.AnimStateTransitionNode_91.Transition
    src=IdleToCrouch
    via=AnimStateTransitionNode_91
   }
   {
    dst=Crouching
    rule=AnimGraph.AnimGraphNode_StateMachine_0.FullBody.AnimStateTransitionNode_77.Transition
    src=IdleToCrouch
    via=AnimStateTransitionNode_77
   }
   {
    dst=DownSights
    rule=AnimGraph.AnimGraphNode_StateMachine_0.FullBody.AnimStateTransitionNode_90.Transition
    src=IdleToCrouch
    via=AnimStateTransitionNode_90
   }
   {
    dst=DBNO_old
    rule=AnimGraph.AnimGraphNode_StateMachine_0.FullBody.AnimStateTransitionNode_12.AnimationTransitionGraph_2
    src=IdleToDBNO_Old
    via=AnimStateTransitionNode_12
   }
   {
    dst=DownSights
    rule=AnimGraph.AnimGraphNode_StateMachine_0.FullBody.AnimStateTransitionNode_19.AnimationTransitionGraph_0
    src=IdleToDownSights
    via=AnimStateTransitionNode_19
   }
   {
    dst=DownSights
    rule=AnimGraph.AnimGraphNode_StateMachine_0.FullBody.AnimStateTransitionNode_88.Transition
    src=IdleToDownSights
    via=AnimStateTransitionNode_88
   }
   {
    dst=JumpLoop
    rule=AnimGraph.AnimGraphNode_StateMachine_0.FullBody.AnimStateTransitionNode_83.Transition
    src=InAirHunker
    via=AnimStateTransitionNode_83
   }
   {
    dst=InAirHunker
    rule=AnimGraph.AnimGraphNode_StateMachine_0.FullBody.AnimStateTransitionNode_81.Transition
    src=JumpFall
    via=AnimStateTransitionNode_81
   }
   {
    dst=JumpLand
    rule=AnimGraph.AnimGraphNode_StateMachine_0.FullBody.AnimStateTransitionNode_7.AnimationTransitionGraph_0
    src=JumpFall
    via=AnimStateTransitionNode_7
   }
   {
    dst=JumpLand
    rule=AnimGraph.AnimGraphNode_StateMachine_0.FullBody.AnimStateTransitionNode_8.Transition
    src=JumpFall
    via=AnimStateTransitionNode_8
   }
   {
    dst=SkyDive
    rule=AnimGraph.AnimGraphNode_StateMachine_0.FullBody.AnimStateTransitionNode_48.Transition
    src=JumpFall
    via=AnimStateTransitionNode_48
   }
   {
    dst=StartJump
    rule=AnimGraph.AnimGraphNode_StateMachine_0.FullBody.AnimStateTransitionNode_62.Transition
    src=JumpFall
    via=AnimStateTransitionNode_62
   }
   {
    dst=Locomotion
    rule=AnimGraph.AnimGraphNode_StateMachine_0.FullBody.AnimStateTransitionNode_63.Transition
    src=JumpLand
    via=AnimStateTransitionNode_63
   }
   {
    dst=StartJump
    rule=AnimGraph.AnimGraphNode_StateMachine_0.FullBody.AnimStateTransitionNode_49.Transition
    src=JumpLand
    via=AnimStateTransitionNode_49
   }
   {
    dst=idle
    rule=AnimGraph.AnimGraphNode_StateMachine_0.FullBody.AnimStateTransitionNode_52.Transition
    src=JumpLand
    via=AnimStateTransitionNode_52
   }
   {
    dst=idle
    rule=AnimGraph.AnimGraphNode_StateMachine_0.FullBody.AnimStateTransitionNode_53.Transition
    src=JumpLand
    via=AnimStateTransitionNode_53
   }
   {
    dst=idle
    rule=AnimGraph.AnimGraphNode_StateMachine_0.FullBody.AnimStateTransitionNode_54.Transition
    src=JumpLand
    via=AnimStateTransitionNode_54
   }
   {
    dst=InAirHunker
    rule=AnimGraph.AnimGraphNode_StateMachine_0.FullBody.AnimStateTransitionNode_76.Transition
    src=JumpLoop
    via=AnimStateTransitionNode_76
   }
   {
    dst=JumpFall
    rule=AnimGraph.AnimGraphNode_StateMachine_0.FullBody.AnimStateTransitionNode_11.AnimationTransitionGraph_0
    src=JumpLoop
    via=AnimStateTransitionNode_11
   }
   {
    dst=JumpFall
    rule=AnimGraph.AnimGraphNode_StateMachine_0.FullBody.AnimStateTransitionNode_13.AnimationTransitionGraph_0
    src=JumpLoop
    via=AnimStateTransitionNode_13
   }
   {
    dst=JumpFall
    rule=AnimGraph.AnimGraphNode_StateMachine_0.FullBody.AnimStateTransitionNode_14.Transition
    src=JumpLoop
    via=AnimStateTransitionNode_14
   }
   {
    dst=JumpFall
    rule=AnimGraph.AnimGraphNode_StateMachine_0.FullBody.AnimStateTransitionNode_6.AnimationTransitionGraph_0
    src=JumpLoop
    via=AnimStateTransitionNode_6
   }
   {
    dst=idle
    rule=AnimGraph.AnimGraphNode_StateMachine_0.FullBody.AnimStateTransitionNode_72.Transition
    src=JumpLoop
    via=AnimStateTransitionNode_72
   }
   {
    dst=HunkerToLobby
    rule=AnimGraph.AnimGraphNode_StateMachine_0.FullBody.AnimStateTransitionNode_31.AnimationTransitionGraph_0
    src=LobbyHunker
    via=AnimStateTransitionNode_31
   }
   {
    dst=PopToSkydive
    rule=AnimGraph.AnimGraphNode_StateMachine_0.FullBody.AnimStateTransitionNode_73.AnimationTransitionGraph_0
    src=LobbyHunker
    via=AnimStateTransitionNode_73
   }
   {
    dst=LobbyHunker
    rule=AnimGraph.AnimGraphNode_StateMachine_0.FullBody.AnimStateTransitionNode_66.AnimationTransitionGraph_0
    src=LobbyToHunker
    via=AnimStateTransitionNode_66
   }
   {
    dst=PopToSkydive
    rule=AnimGraph.AnimGraphNode_StateMachine_0.FullBody.AnimStateTransitionNode_69.AnimationTransitionGraph_0
    src=LobbyToHunker
    via=AnimStateTransitionNode_69
   }
   {
    dst=Hunker
    rule=AnimGraph.AnimGraphNode_StateMachine_0.FullBody.AnimStateTransitionNode_58.Transition
    src=Locomotion
    via=AnimStateTransitionNode_58
   }
   {
    dst=IDS
    rule=AnimGraph.AnimGraphNode_StateMachine_0.FullBody.AnimStateTransitionNode_37.Transition
    src=Locomotion
    via=AnimStateTransitionNode_37
   }
   {
    dst=InAirHunker
    rule=AnimGraph.AnimGraphNode_StateMachine_0.FullBody.AnimStateTransitionNode_104.Transition
    src=Locomotion
    via=AnimStateTransitionNode_104
   }
   {
    dst=LocomotionHappy
    rule=AnimGraph.AnimGraphNode_StateMachine_0.FullBody.AnimStateTransitionNode_9.Transition
    src=Locomotion
    via=AnimStateTransitionNode_9
   }
   {
    dst=LocomotionToCrouch
    rule=AnimGraph.AnimGraphNode_StateMachine_0.FullBody.AnimStateTransitionNode_56.Transition
    src=Locomotion
    via=AnimStateTransitionNode_56
   }
   {
    dst=LocomotionToIdle
    rule=AnimGraph.AnimGraphNode_StateMachine_0.FullBody.AnimStateTransitionNode_23.Transition
    src=Locomotion
    via=AnimStateTransitionNode_23
   }
   {
    dst=LocomotionToIdle
    rule=AnimGraph.AnimGraphNode_StateMachine_0.FullBody.AnimStateTransitionNode_26.Transition
    src=Locomotion
    via=AnimStateTransitionNode_26
   }
   {
    dst=StartJump
    rule=AnimGraph.AnimGraphNode_StateMachine_0.FullBody.AnimStateTransitionNode_82.Transition
    src=Locomotion
    via=AnimStateTransitionNode_82
   }
   {
    dst=Locomotion
    rule=AnimGraph.AnimGraphNode_StateMachine_0.FullBody.AnimStateTransitionNode_25.Transition
    src=LocomotionHappy
    via=AnimStateTransitionNode_25
   }
   {
    dst=CrouchWalkingDownSights
    rule=AnimGraph.AnimGraphNode_StateMachine_0.FullBody.AnimStateTransitionNode_93.Transition
    src=LocomotionToCrouch
    via=AnimStateTransitionNode_93
   }
   {
    dst=Crouching
    rule=AnimGraph.AnimGraphNode_StateMachine_0.FullBody.AnimStateTransitionNode_57.Transition
    src=LocomotionToCrouch
    via=AnimStateTransitionNode_57
   }
   {
    dst=IDS
    rule=AnimGraph.AnimGraphNode_StateMachine_0.FullBody.AnimStateTransitionNode_84.Transition
    src=LocomotionToIdle
    via=AnimStateTransitionNode_84
   }
   {
    dst=IdleToCrouch
    rule=AnimGraph.AnimGraphNode_StateMachine_0.FullBody.AnimStateTransitionNode_87.Transition
    src=LocomotionToIdle
    via=AnimStateTransitionNode_87
   }
   {
    dst=Locomotion
    rule=AnimGraph.AnimGraphNode_StateMachine_0.FullBody.AnimStateTransitionNode_86.Transition
    src=LocomotionToIdle
    via=AnimStateTransitionNode_86
   }
   {
    dst=StartJump
    rule=AnimGraph.AnimGraphNode_StateMachine_0.FullBody.AnimStateTransitionNode_30.Transition
    src=LocomotionToIdle
    via=AnimStateTransitionNode_30
   }
   {
    dst=idle
    rule=AnimGraph.AnimGraphNode_StateMachine_0.FullBody.AnimStateTransitionNode_27.Transition
    src=LocomotionToIdle
    via=AnimStateTransitionNode_27
   }
   {
    dst=JumpFall
    rule=AnimGraph.AnimGraphNode_StateMachine_0.FullBody.AnimStateTransitionNode_24.Transition
    src=NotSkydiving
    via=AnimStateTransitionNode_24
   }
   {
    dst=StartJump
    rule=AnimGraph.AnimGraphNode_StateMachine_0.FullBody.AnimStateTransitionNode_4.Transition
    src=NotSkydiving
    via=AnimStateTransitionNode_4
   }
   {
    dst=FrontendSkydive
    rule=AnimGraph.AnimGraphNode_StateMachine_0.FullBody.AnimStateTransitionNode_67.AnimationTransitionGraph_0
    src=PopToSkydive
    via=AnimStateTransitionNode_67
   }
   {
    dst=idle
    rule=AnimGraph.AnimGraphNode_StateMachine_0.FullBody.AnimStateTransitionNode_79.Transition
    src=SkyDive
    via=AnimStateTransitionNode_79
   }
   {
    dst=InAirHunker
    rule=AnimGraph.AnimGraphNode_StateMachine_0.FullBody.AnimStateTransitionNode_38.Transition
    src=StartJump
    via=AnimStateTransitionNode_38
   }
   {
    dst=JumpFall
    rule=AnimGraph.AnimGraphNode_StateMachine_0.FullBody.AnimStateTransitionNode_10.Transition
    src=StartJump
    via=AnimStateTransitionNode_10
   }
   {
    dst=JumpFall
    rule=AnimGraph.AnimGraphNode_StateMachine_0.FullBody.AnimStateTransitionNode_16.AnimationTransitionGraph_0
    src=StartJump
    via=AnimStateTransitionNode_16
   }
   {
    dst=JumpFall
    rule=AnimGraph.AnimGraphNode_StateMachine_0.FullBody.AnimStateTransitionNode_17.AnimationTransitionGraph_0
    src=StartJump
    via=AnimStateTransitionNode_17
   }
   {
    dst=JumpFall
    rule=AnimGraph.AnimGraphNode_StateMachine_0.FullBody.AnimStateTransitionNode_18.AnimationTransitionGraph_0
    src=StartJump
    via=AnimStateTransitionNode_18
   }
   {
    dst=JumpLoop
    rule=AnimGraph.AnimGraphNode_StateMachine_0.FullBody.AnimStateTransitionNode_5.AnimationTransitionGraph_0
    src=StartJump
    via=AnimStateTransitionNode_5
   }
   {
    dst=Crouching
    rule=AnimGraph.AnimGraphNode_StateMachine_0.FullBody.AnimStateTransitionNode_32.Transition
    src=idle
    via=AnimStateTransitionNode_32
   }
   {
    dst=DBNO
    rule=AnimGraph.AnimGraphNode_StateMachine_0.FullBody.AnimStateTransitionNode_106.Transition
    src=idle
    via=AnimStateTransitionNode_106
   }
   {
    dst=Hunker
    rule=AnimGraph.AnimGraphNode_StateMachine_0.FullBody.AnimStateTransitionNode_51.Transition
    src=idle
    via=AnimStateTransitionNode_51
   }
   {
    dst=IDS
    rule=AnimGraph.AnimGraphNode_StateMachine_0.FullBody.AnimStateTransitionNode_42.Transition
    src=idle
    via=AnimStateTransitionNode_42
   }
   {
    dst=IdleLobby
    rule=AnimGraph.AnimGraphNode_StateMachine_0.FullBody.AnimStateTransitionNode_47.Transition
    src=idle
    via=AnimStateTransitionNode_47
   }
   {
    dst=IdleToCrouch
    rule=AnimGraph.AnimGraphNode_StateMachine_0.FullBody.AnimStateTransitionNode_75.Transition
    src=idle
    via=AnimStateTransitionNode_75
   }
   {
    dst=InAirHunker
    rule=AnimGraph.AnimGraphNode_StateMachine_0.FullBody.AnimStateTransitionNode_46.Transition
    src=idle
    via=AnimStateTransitionNode_46
   }
   {
    dst=Locomotion
    rule=AnimGraph.AnimGraphNode_StateMachine_0.FullBody.AnimStateTransitionNode_80.Transition
    src=idle
    via=AnimStateTransitionNode_80
   }
   {
    dst=NotSkydiving
    rule=AnimGraph.AnimGraphNode_StateMachine_0.FullBody.AnimStateTransitionNode_3.Transition
    src=idle
    via=AnimStateTransitionNode_3
   }
   {
    dst=SkyDive
    rule=AnimGraph.AnimGraphNode_StateMachine_0.FullBody.AnimStateTransitionNode_74.AnimationTransitionGraph_1
    src=idle
    via=AnimStateTransitionNode_74
   }
  ]
 }
 NotSkydiving {
  path=AnimGraph.AnimGraphNode_StateMachine_0.FullBody.AnimStateConduitNode_0.NotSkydiving
  nodes [
   n1 = Get SkydivingState() -> (SkydivingState:Fort Anim Input Skydiving Structure)
   Result(bCanEnterTransition=NOT Boolean(A=Break Fort Anim Input Skydiving!(FortAnimInput_Skydiving:"Fort Anim Input Skydiving Structure (by ref)"=@n1) -> (
      ParaGilderRootOffset:Vector=~
      ParaGliderRootRotationOverride:Rotator=~
      bUseParaGlideRootModifier:bool=~
      bIsSkydivingFromLaunchPad:bool=~
      bIsSkydivingFromBus:bool=~
      bIsInVortex:bool=~
      bIsDiving:bool=~
      bIsParachuteOpen:bool=~
      bIsSkydiving:bool=~
      bIsParachuteLeaning:bool=~
      bIsSkydiveLeaning:bool=~
      bIsLeaning:bool=~
      bIsSkydiveDiveMode:bool=~
      bPlaySkydiveDrift:bool=~
      bSkydiveDriftDelayActive:bool=~
      bSkydiveDriftAnimAllowed:bool=~
      bIsGliderRight:bool=~
      bIsGliderCenter:bool=~
      bIsGliderLeft:bool=~
      bIsGliderForward:bool=~
      bIsGliderBack:bool=~
      bHasDeployedAndIsStandingUpright:bool=~
      LocalAccelForward:float=~
      LocalAccelRight:float=~
      LocalVelocityRight:float=~
      SkydiveAimPitch:float=~
      SkydiveAimYaw:float=~
      DeployChuteAnimRate:float=~
      SkydiveDriftAnimRate:float=~
      SkydiveDriftAnimRateCurrent:float=~
      SkydiveFidgetAnimRate:float=~
      SkydiveFidgetAnimRateCurrent:float=~
      SkydiveAdditiveAlpha:float=~
      SkydiveDriftDelay:float=~
      SkydiveDriftAnim:int
      SkydiveDriftAnimMax:int
      LaunchpadAnim:int
      LocalAccelDir:ESkydivingDirection Enum
      DirectionLast:ESkydivingDirection Enum
      PoseAdditive:Anim Sequence Object Reference
     ) {is=title K2Node_BreakStruct}.bIsSkydiving))
  ]
 }
 IDS {path=AnimGraph.AnimGraphNode_StateMachine_0.FullBody.AnimStateConduitNode_1.IDS, nodes [Result(bCanEnterTransition=true)]}
 PopToSkydive {path=AnimGraph.AnimGraphNode_StateMachine_0.FullBody.AnimStateConduitNode_2.PopToSkydive, nodes [Result(bCanEnterTransition=true)]}
 IdleLobby {
  path=AnimGraph.AnimGraphNode_StateMachine_0.FullBody.AnimStateNode_0.IdleLobby
  nodes [Output Animation Pose(Result=Sequence Player 'Idle_Lobby'() {sequence=/Game/Animation/Game/Pets/Chameleon/Backpack/Idle/Idle_Lobby})]
 }
 IdleToCrouch {
  path=AnimGraph.AnimGraphNode_StateMachine_0.FullBody.AnimStateNode_1.IdleToCrouch
  nodes [Output Animation Pose(Result=Sequence Player 'LocomotionToCrouch'() {bLoopAnimation=false, sequence=/Game/Animation/Game/Pets/Chameleon/Backpack/Sprint/LocomotionToCrouch})]
 }
 IdleToDBNO_Old {
  path=AnimGraph.AnimGraphNode_StateMachine_0.FullBody.AnimStateNode_10.IdleToDBNO_Old
  nodes [Output Animation Pose(Result=Sequence Player 'DBNO_Idle_Intro'() {sequence=/Game/Animation/Game/Pets/Chameleon/Backpack/DBNO/DBNO_Idle_Intro})]
 }
 DBNOToIdle {
  path=AnimGraph.AnimGraphNode_StateMachine_0.FullBody.AnimStateNode_11.DBNOToIdle
  nodes [Output Animation Pose(Result=Sequence Player 'DBNO_Idle_Outro'() {bLoopAnimation=false, sequence=/Game/Animation/Game/Pets/Chameleon/Backpack/DBNO/DBNO_Idle_Outro})]
 }
 DownSights {
  path=AnimGraph.AnimGraphNode_StateMachine_0.FullBody.AnimStateNode_12.DownSights
  nodes [Output Animation Pose(Result=Sequence Player 'DownSights'() {sequence=/Game/Animation/Game/Pets/Chameleon/Backpack/Downsights/DownSights})]
 }
 IdleToDownSights {
  path=AnimGraph.AnimGraphNode_StateMachine_0.FullBody.AnimStateNode_13.IdleToDownSights
  nodes [Output Animation Pose(Result=Sequence Player 'IdleToDownSights'() {bLoopAnimation=false, sequence=/Game/Animation/Game/Pets/Chameleon/Backpack/Downsights/IdleToDownSights})]
 }
 DownSightsToIdle {
  path=AnimGraph.AnimGraphNode_StateMachine_0.FullBody.AnimStateNode_14.DownSightsToIdle
  nodes [Output Animation Pose(Result=Sequence Player 'DownSightsToIdle'() {bLoopAnimation=false, sequence=/Game/Animation/Game/Pets/Chameleon/Backpack/Downsights/DownSightsToIdle})]
 }
 DBNO {
  path=AnimGraph.AnimGraphNode_StateMachine_0.FullBody.AnimStateNode_15.DBNO
  nodes [
   n1 = Sequence Player 'Hunker_Pose'() {sequence=/Game/Animation/Game/Pets/Chameleon/Backpack/Hunker/Hunker_Pose}
   n2 = "Sequence Player (additive) 'IdleStraightLookAround_Add'"() {sequence=&IdleStraightLookAround_Add}
   AnimGraphNode_StateMachine_194 = DBNO() {is=title AnimGraphNode_StateMachine}
   Output Animation Pose(Result=Blend Poses by bool(
     BlendPose_0=Apply Additive(Base=@n1, Additive=@n2) {alphaBoolBlend {blendOption=Linear}}
     BlendPose_1=@AnimGraphNode_StateMachine_194
     bActiveValue=Get bIsSurfaceSwimming()
    ) {blendType=Linear})
  ]
 }
 DBNO {
  entry=IdleToDBNO
  path=AnimGraph.AnimGraphNode_StateMachine_0.FullBody.AnimStateNode_15.DBNO.AnimGraphNode_StateMachine_194.DBNO
  ids=id
  nodes [
   AnimStateNode_0 = ToDBNOMove() {is=title AnimStateNode}
   AnimStateNode_10 = IdleToDBNO() {is=title AnimStateNode}
   AnimStateNode_3 = ToDBNOIdle() {is=title AnimStateNode}
   AnimStateNode_5 = DBNO Idle() {is=title AnimStateNode}
   AnimStateNode_9 = DBNO Move() {is=title AnimStateNode}
   AnimStateTransitionNode_0 = AnimStateTransitionNode()
   AnimStateTransitionNode_1 = AnimStateTransitionNode()
   AnimStateTransitionNode_2 = AnimStateTransitionNode() {bAutomaticRuleBasedOnSequencePlayerInState=true}
   AnimStateTransitionNode_3 = AnimStateTransitionNode() {bAutomaticRuleBasedOnSequencePlayerInState=true}
   AnimStateTransitionNode_6 = AnimStateTransitionNode()
   AnimStateTransitionNode_7 = AnimStateTransitionNode() {bAutomaticRuleBasedOnSequencePlayerInState=true}
  ]
  links [
   {
    dst=ToDBNOMove
    rule=AnimGraph.AnimGraphNode_StateMachine_0.FullBody.AnimStateNode_15.DBNO.AnimGraphNode_StateMachine_194.DBNO.AnimStateTransitionNode_1.Transition
    src=DBNO Idle
    via=AnimStateTransitionNode_1
   }
   {
    dst=ToDBNOIdle
    rule=AnimGraph.AnimGraphNode_StateMachine_0.FullBody.AnimStateNode_15.DBNO.AnimGraphNode_StateMachine_194.DBNO.AnimStateTransitionNode_6.Transition
    src=DBNO Move
    via=AnimStateTransitionNode_6
   }
   {
    dst=DBNO Idle
    rule=AnimGraph.AnimGraphNode_StateMachine_0.FullBody.AnimStateNode_15.DBNO.AnimGraphNode_StateMachine_194.DBNO.AnimStateTransitionNode_3.Transition
    src=IdleToDBNO
    via=AnimStateTransitionNode_3
   }
   {
    dst=DBNO Move
    rule=AnimGraph.AnimGraphNode_StateMachine_0.FullBody.AnimStateNode_15.DBNO.AnimGraphNode_StateMachine_194.DBNO.AnimStateTransitionNode_0.Transition
    src=IdleToDBNO
    via=AnimStateTransitionNode_0
   }
   {
    dst=DBNO Idle
    rule=AnimGraph.AnimGraphNode_StateMachine_0.FullBody.AnimStateNode_15.DBNO.AnimGraphNode_StateMachine_194.DBNO.AnimStateTransitionNode_7.Transition
    src=ToDBNOIdle
    via=AnimStateTransitionNode_7
   }
   {
    dst=DBNO Move
    rule=AnimGraph.AnimGraphNode_StateMachine_0.FullBody.AnimStateNode_15.DBNO.AnimGraphNode_StateMachine_194.DBNO.AnimStateTransitionNode_2.Transition
    src=ToDBNOMove
    via=AnimStateTransitionNode_2
   }
  ]
 }
 ToDBNOMove {
  path=AnimGraph.AnimGraphNode_StateMachine_0.FullBody.AnimStateNode_15.DBNO.AnimGraphNode_StateMachine_194.DBNO.AnimStateNode_0.ToDBNOMove
  nodes [Output Animation Pose(Result=Sequence Player 'DBNOToHunker'() {bLoopAnimation=false, sequence=/Game/Animation/Game/Pets/Chameleon/Backpack/DBNO/DBNOToHunker})]
 }
 IdleToDBNO {
  path=AnimGraph.AnimGraphNode_StateMachine_0.FullBody.AnimStateNode_15.DBNO.AnimGraphNode_StateMachine_194.DBNO.AnimStateNode_10.IdleToDBNO
  nodes [Output Animation Pose(Result=Sequence Player 'DBNO_Idle_Intro'() {sequence=&DBNO_Idle_Intro})]
 }
 ToDBNOIdle {
  path=AnimGraph.AnimGraphNode_StateMachine_0.FullBody.AnimStateNode_15.DBNO.AnimGraphNode_StateMachine_194.DBNO.AnimStateNode_3.ToDBNOIdle
  nodes [Output Animation Pose(Result=Sequence Player 'HunkerToDBNO'() {bLoopAnimation=false, sequence=/Game/Animation/Game/Pets/Chameleon/Backpack/DBNO/HunkerToDBNO})]
 }
 DBNO Idle {
  path=AnimGraph.AnimGraphNode_StateMachine_0.FullBody.AnimStateNode_15.DBNO.AnimGraphNode_StateMachine_194.DBNO.AnimStateNode_5.DBNO Idle
  nodes [
   Sequence Player 'AthenaDBNO_Loop_CMM'() {sequence=/Game/Animation/Game/MainPlayer/Locomotion/Medium/Male/DBNO/AthenaDBNO_Loop_CMM}
   Output Animation Pose(Result=Sequence Player 'DBNO_Idle'() {sequence=/Game/Animation/Game/Pets/Chameleon/Backpack/DBNO/DBNO_Idle})
  ]
 }
 DBNO Move {
  path=AnimGraph.AnimGraphNode_StateMachine_0.FullBody.AnimStateNode_15.DBNO.AnimGraphNode_StateMachine_194.DBNO.AnimStateNode_9.DBNO Move
  nodes [Output Animation Pose(Result=Sequence Player 'DBNO_Hunker'() {sequence=/Game/Animation/Game/Pets/Chameleon/Backpack/DBNO/DBNO_Hunker})]
 }
 Transition {
  path=AnimGraph.AnimGraphNode_StateMachine_0.FullBody.AnimStateNode_15.DBNO.AnimGraphNode_StateMachine_194.DBNO.AnimStateTransitionNode_0.Transition
  nodes [Result(bCanEnterTransition=Get bIsMoving2D())]
 }
 Transition {
  path=AnimGraph.AnimGraphNode_StateMachine_0.FullBody.AnimStateNode_15.DBNO.AnimGraphNode_StateMachine_194.DBNO.AnimStateTransitionNode_1.Transition
  nodes [Result(bCanEnterTransition=Get bIsMoving2D())]
 }
 Transition {
  path=AnimGraph.AnimGraphNode_StateMachine_0.FullBody.AnimStateNode_15.DBNO.AnimGraphNode_StateMachine_194.DBNO.AnimStateTransitionNode_2.Transition
  nodes [Result()]
 }
 Transition {
  path=AnimGraph.AnimGraphNode_StateMachine_0.FullBody.AnimStateNode_15.DBNO.AnimGraphNode_StateMachine_194.DBNO.AnimStateTransitionNode_3.Transition
  nodes [Result()]
 }
 Transition {
  path=AnimGraph.AnimGraphNode_StateMachine_0.FullBody.AnimStateNode_15.DBNO.AnimGraphNode_StateMachine_194.DBNO.AnimStateTransitionNode_6.Transition
  nodes [Result(bCanEnterTransition=NOT Boolean(A=Get bIsMoving2D()))]
 }
 Transition {
  path=AnimGraph.AnimGraphNode_StateMachine_0.FullBody.AnimStateNode_15.DBNO.AnimGraphNode_StateMachine_194.DBNO.AnimStateTransitionNode_7.Transition
  nodes [Result()]
 }
 LocomotionToCrouch {
  path=AnimGraph.AnimGraphNode_StateMachine_0.FullBody.AnimStateNode_16.LocomotionToCrouch
  nodes [Output Animation Pose(Result=Sequence Player 'LocomotionToCrouch'() {bLoopAnimation=false, sequence=&LocomotionToCrouch})]
 }
 CrouchWalkingDownSights {
  path=AnimGraph.AnimGraphNode_StateMachine_0.FullBody.AnimStateNode_17.CrouchWalkingDownSights
  nodes [Output Animation Pose(Result=Sequence Player 'CrouchWalkDownSights'() {sequence=/Game/Animation/Game/Pets/Chameleon/Backpack/Crouch/CrouchWalkDownSights})]
 }
 Crouching {
  path=AnimGraph.AnimGraphNode_StateMachine_0.FullBody.AnimStateNode_18.Crouching
  nodes [
   AnimGraphNode_StateMachine_0 = Crouch() {is=title AnimGraphNode_StateMachine}
   Output Animation Pose(Result=@AnimGraphNode_StateMachine_0)
  ]
 }
 Crouch {
  entry=Crouch
  path=AnimGraph.AnimGraphNode_StateMachine_0.FullBody.AnimStateNode_18.Crouching.AnimGraphNode_StateMachine_0.Crouch
  ids=id
  nodes [
   AnimStateNode_0 = Blank() {is=title AnimStateNode}
   AnimStateNode_1 = toCrouchWalk() {is=title AnimStateNode}
   AnimStateNode_2 = CrouchSprinting() {is=title AnimStateNode}
   AnimStateNode_27 = CrouchWalk() {is=title AnimStateNode}
   AnimStateNode_29 = Crouch() {is=title AnimStateNode}
   AnimStateNode_3 = LocomotionCrouchHappy() {is=title AnimStateNode}
   AnimStateTransitionNode_0 = AnimStateTransitionNode()
   AnimStateTransitionNode_1 = AnimStateTransitionNode() {bAutomaticRuleBasedOnSequencePlayerInState=true}
   AnimStateTransitionNode_11 = AnimStateTransitionNode()
   AnimStateTransitionNode_2 = AnimStateTransitionNode()
   AnimStateTransitionNode_3 = AnimStateTransitionNode()
   AnimStateTransitionNode_4 = AnimStateTransitionNode()
   AnimStateTransitionNode_47 = AnimStateTransitionNode()
   AnimStateTransitionNode_5 = AnimStateTransitionNode()
   AnimStateTransitionNode_55 = AnimStateTransitionNode()
   AnimStateTransitionNode_6 = AnimStateTransitionNode() {bAutomaticRuleBasedOnSequencePlayerInState=true}
   AnimStateTransitionNode_7 = AnimStateTransitionNode()
   AnimStateTransitionNode_8 = AnimStateTransitionNode()
  ]
  links [
   {
    dst=Crouch
    rule=AnimGraph.AnimGraphNode_StateMachine_0.FullBody.AnimStateNode_18.Crouching.AnimGraphNode_StateMachine_0.Crouch.AnimStateTransitionNode_11.Transition
    src=Blank
    via=AnimStateTransitionNode_11
   }
   {
    dst=CrouchWalk
    rule=AnimGraph.AnimGraphNode_StateMachine_0.FullBody.AnimStateNode_18.Crouching.AnimGraphNode_StateMachine_0.Crouch.AnimStateTransitionNode_7.Transition
    src=Blank
    via=AnimStateTransitionNode_7
   }
   {
    dst=toCrouchWalk
    rule=AnimGraph.AnimGraphNode_StateMachine_0.FullBody.AnimStateNode_18.Crouching.AnimGraphNode_StateMachine_0.Crouch.AnimStateTransitionNode_0.Transition
    src=Crouch
    via=AnimStateTransitionNode_0
   }
   {
    dst=Crouch
    rule=AnimGraph.AnimGraphNode_StateMachine_0.FullBody.AnimStateNode_18.Crouching.AnimGraphNode_StateMachine_0.Crouch.AnimStateTransitionNode_8.Transition
    src=CrouchSprinting
    via=AnimStateTransitionNode_8
   }
   {
    dst=CrouchWalk
    rule=AnimGraph.AnimGraphNode_StateMachine_0.FullBody.AnimStateNode_18.Crouching.AnimGraphNode_StateMachine_0.Crouch.AnimStateTransitionNode_55.AnimationTransitionGraph_1
    src=CrouchSprinting
    via=AnimStateTransitionNode_55
   }
   {
    dst=LocomotionCrouchHappy
    rule=AnimGraph.AnimGraphNode_StateMachine_0.FullBody.AnimStateNode_18.Crouching.AnimGraphNode_StateMachine_0.Crouch.AnimStateTransitionNode_3.Transition
    src=CrouchSprinting
    via=AnimStateTransitionNode_3
   }
   {
    dst=Crouch
    rule=AnimGraph.AnimGraphNode_StateMachine_0.FullBody.AnimStateNode_18.Crouching.AnimGraphNode_StateMachine_0.Crouch.AnimStateTransitionNode_4.Transition
    src=CrouchWalk
    via=AnimStateTransitionNode_4
   }
   {
    dst=CrouchSprinting
    rule=AnimGraph.AnimGraphNode_StateMachine_0.FullBody.AnimStateNode_18.Crouching.AnimGraphNode_StateMachine_0.Crouch.AnimStateTransitionNode_47.AnimationTransitionGraph_1
    src=CrouchWalk
    via=AnimStateTransitionNode_47
   }
   {
    dst=LocomotionCrouchHappy
    rule=AnimGraph.AnimGraphNode_StateMachine_0.FullBody.AnimStateNode_18.Crouching.AnimGraphNode_StateMachine_0.Crouch.AnimStateTransitionNode_2.Transition
    src=CrouchWalk
    via=AnimStateTransitionNode_2
   }
   {
    dst=CrouchWalk
    rule=AnimGraph.AnimGraphNode_StateMachine_0.FullBody.AnimStateNode_18.Crouching.AnimGraphNode_StateMachine_0.Crouch.AnimStateTransitionNode_6.Transition
    src=LocomotionCrouchHappy
    via=AnimStateTransitionNode_6
   }
   {
    dst=Crouch
    rule=AnimGraph.AnimGraphNode_StateMachine_0.FullBody.AnimStateNode_18.Crouching.AnimGraphNode_StateMachine_0.Crouch.AnimStateTransitionNode_5.Transition
    src=toCrouchWalk
    via=AnimStateTransitionNode_5
   }
   {
    dst=CrouchWalk
    rule=AnimGraph.AnimGraphNode_StateMachine_0.FullBody.AnimStateNode_18.Crouching.AnimGraphNode_StateMachine_0.Crouch.AnimStateTransitionNode_1.Transition
    src=toCrouchWalk
    via=AnimStateTransitionNode_1
   }
  ]
 }
 Blank {
  path=AnimGraph.AnimGraphNode_StateMachine_0.FullBody.AnimStateNode_18.Crouching.AnimGraphNode_StateMachine_0.Crouch.AnimStateNode_0.Blank
  nodes [Output Animation Pose()]
 }
 toCrouchWalk {
  path=AnimGraph.AnimGraphNode_StateMachine_0.FullBody.AnimStateNode_18.Crouching.AnimGraphNode_StateMachine_0.Crouch.AnimStateNode_1.toCrouchWalk
  nodes [Output Animation Pose(Result=Sequence Player 'CrouchIdleToWalk'() {bLoopAnimation=false, sequence=/Game/Animation/Game/Pets/Chameleon/Backpack/Crouch/CrouchIdleToWalk})]
 }
 CrouchSprinting {
  path=AnimGraph.AnimGraphNode_StateMachine_0.FullBody.AnimStateNode_18.Crouching.AnimGraphNode_StateMachine_0.Crouch.AnimStateNode_2.CrouchSprinting
  nodes [Output Animation Pose(Result=Sequence Player 'CrouchSprint'() {sequence=/Game/Animation/Game/Pets/Chameleon/Backpack/Crouch/CrouchSprint})]
 }
 CrouchWalk {
  path=AnimGraph.AnimGraphNode_StateMachine_0.FullBody.AnimStateNode_18.Crouching.AnimGraphNode_StateMachine_0.Crouch.AnimStateNode_27.CrouchWalk
  nodes [Output Animation Pose(Result=Sequence Player 'CrouchWalk'() {sequence=/Game/Animation/Game/Pets/Chameleon/Backpack/Crouch/CrouchWalk})]
 }
 Crouch {
  path=AnimGraph.AnimGraphNode_StateMachine_0.FullBody.AnimStateNode_18.Crouching.AnimGraphNode_StateMachine_0.Crouch.AnimStateNode_29.Crouch
  nodes [
   AnimGraphNode_StateMachine_0 = CrouchInternal() {is=title AnimGraphNode_StateMachine}
   Output Animation Pose(Result=@AnimGraphNode_StateMachine_0)
  ]
 }
 CrouchInternal {
  entry=Crouch
  path=AnimGraph.AnimGraphNode_StateMachine_0.FullBody.AnimStateNode_18.Crouching.AnimGraphNode_StateMachine_0.Crouch.AnimStateNode_29.Crouch.AnimGraphNode_StateMachine_0.CrouchInternal
  ids=id
  nodes [
   AnimStateNode_0 = Crouch()
   AnimStateNode_2 = Happy() {is=title AnimStateNode}
   AnimStateNode_3 = CrouchToHappy() {is=title AnimStateNode}
   AnimStateNode_4 = HappyToCrouch() {is=title AnimStateNode}
   AnimStateTransitionNode_4 = AnimStateTransitionNode()
   AnimStateTransitionNode_5 = AnimStateTransitionNode() {bAutomaticRuleBasedOnSequencePlayerInState=true}
   AnimStateTransitionNode_6 = AnimStateTransitionNode()
   AnimStateTransitionNode_7 = AnimStateTransitionNode() {bAutomaticRuleBasedOnSequencePlayerInState=true, crossfadeDuration=0}
  ]
  links [
   {
    dst=CrouchToHappy
    rule=AnimGraph.AnimGraphNode_StateMachine_0.FullBody.AnimStateNode_18.Crouching.AnimGraphNode_StateMachine_0.Crouch.AnimStateNode_29.Crouch.AnimGraphNode_StateMachine_0.CrouchInternal.AnimStateTransitionNode_4.Transition
    src=Crouch
    via=AnimStateTransitionNode_4
   }
   {
    dst=Happy
    rule=AnimGraph.AnimGraphNode_StateMachine_0.FullBody.AnimStateNode_18.Crouching.AnimGraphNode_StateMachine_0.Crouch.AnimStateNode_29.Crouch.AnimGraphNode_StateMachine_0.CrouchInternal.AnimStateTransitionNode_5.Transition
    src=CrouchToHappy
    via=AnimStateTransitionNode_5
   }
   {
    dst=HappyToCrouch
    rule=AnimGraph.AnimGraphNode_StateMachine_0.FullBody.AnimStateNode_18.Crouching.AnimGraphNode_StateMachine_0.Crouch.AnimStateNode_29.Crouch.AnimGraphNode_StateMachine_0.CrouchInternal.AnimStateTransitionNode_6.Transition
    src=Happy
    via=AnimStateTransitionNode_6
   }
   {
    dst=Crouch
    rule=AnimGraph.AnimGraphNode_StateMachine_0.FullBody.AnimStateNode_18.Crouching.AnimGraphNode_StateMachine_0.Crouch.AnimStateNode_29.Crouch.AnimGraphNode_StateMachine_0.CrouchInternal.AnimStateTransitionNode_7.Transition
    src=HappyToCrouch
    via=AnimStateTransitionNode_7
   }
  ]
 }
 Crouch {
  path=AnimGraph.AnimGraphNode_StateMachine_0.FullBody.AnimStateNode_18.Crouching.AnimGraphNode_StateMachine_0.Crouch.AnimStateNode_29.Crouch.AnimGraphNode_StateMachine_0.CrouchInternal.AnimStateNode_0.Crouch
  nodes [
   AnimGraphNode_StateMachine_0 = IdleLoop() {is=title AnimGraphNode_StateMachine}
   Output Animation Pose(Result=@AnimGraphNode_StateMachine_0)
  ]
 }
 IdleLoop {
  entry=Idle1
  path=AnimGraph.AnimGraphNode_StateMachine_0.FullBody.AnimStateNode_18.Crouching.AnimGraphNode_StateMachine_0.Crouch.AnimStateNode_29.Crouch.AnimGraphNode_StateMachine_0.CrouchInternal.AnimStateNode_0.Crouch.AnimGraphNode_StateMachine_0.IdleLoop
  ids=id
  nodes [
   AnimStateNode_0 = Idle1() {is=title AnimStateNode, stateEntered {guid=#17, notifyName=RollRandomIdle2}}
   AnimStateNode_1 = Idle2() {is=title AnimStateNode, stateEntered {guid=#18, notifyName=RollRandomIdle}}
   AnimStateNode_2 = State() {is=title AnimStateNode}
   AnimStateNode_3 = State_1() {is=title AnimStateNode}
   AnimStateNode_4 = State_2() {is=title AnimStateNode}
   AnimStateTransitionNode_0 = AnimStateTransitionNode() {bAutomaticRuleBasedOnSequencePlayerInState=true}
   AnimStateTransitionNode_1 = AnimStateTransitionNode() {bAutomaticRuleBasedOnSequencePlayerInState=true}
   AnimStateTransitionNode_2 = AnimStateTransitionNode() {bAutomaticRuleBasedOnSequencePlayerInState=true}
   AnimStateTransitionNode_3 = AnimStateTransitionNode() {bAutomaticRuleBasedOnSequencePlayerInState=true}
   AnimStateTransitionNode_4 = AnimStateTransitionNode() {bAutomaticRuleBasedOnSequencePlayerInState=true}
  ]
  links [
   {
    dst=Idle2
    rule=AnimGraph.AnimGraphNode_StateMachine_0.FullBody.AnimStateNode_18.Crouching.AnimGraphNode_StateMachine_0.Crouch.AnimStateNode_29.Crouch.AnimGraphNode_StateMachine_0.CrouchInternal.AnimStateNode_0.Crouch.AnimGraphNode_StateMachine_0.IdleLoop.AnimStateTransitionNode_0.Transition
    src=Idle1
    via=AnimStateTransitionNode_0
   }
   {
    dst=Idle1
    rule=AnimGraph.AnimGraphNode_StateMachine_0.FullBody.AnimStateNode_18.Crouching.AnimGraphNode_StateMachine_0.Crouch.AnimStateNode_29.Crouch.AnimGraphNode_StateMachine_0.CrouchInternal.AnimStateNode_0.Crouch.AnimGraphNode_StateMachine_0.IdleLoop.AnimStateTransitionNode_1.Transition
    src=Idle2
    via=AnimStateTransitionNode_1
   }
   {
    dst=State_1
    rule=AnimGraph.AnimGraphNode_StateMachine_0.FullBody.AnimStateNode_18.Crouching.AnimGraphNode_StateMachine_0.Crouch.AnimStateNode_29.Crouch.AnimGraphNode_StateMachine_0.CrouchInternal.AnimStateNode_0.Crouch.AnimGraphNode_StateMachine_0.IdleLoop.AnimStateTransitionNode_2.Transition
    src=State
    via=AnimStateTransitionNode_2
   }
   {
    dst=State_2
    rule=AnimGraph.AnimGraphNode_StateMachine_0.FullBody.AnimStateNode_18.Crouching.AnimGraphNode_StateMachine_0.Crouch.AnimStateNode_29.Crouch.AnimGraphNode_StateMachine_0.CrouchInternal.AnimStateNode_0.Crouch.AnimGraphNode_StateMachine_0.IdleLoop.AnimStateTransitionNode_3.Transition
    src=State_1
    via=AnimStateTransitionNode_3
   }
   {
    dst=State
    rule=AnimGraph.AnimGraphNode_StateMachine_0.FullBody.AnimStateNode_18.Crouching.AnimGraphNode_StateMachine_0.Crouch.AnimStateNode_29.Crouch.AnimGraphNode_StateMachine_0.CrouchInternal.AnimStateNode_0.Crouch.AnimGraphNode_StateMachine_0.IdleLoop.AnimStateTransitionNode_4.Transition
    src=State_2
    via=AnimStateTransitionNode_4
   }
  ]
 }
 Idle1 {
  path=AnimGraph.AnimGraphNode_StateMachine_0.FullBody.AnimStateNode_18.Crouching.AnimGraphNode_StateMachine_0.Crouch.AnimStateNode_29.Crouch.AnimGraphNode_StateMachine_0.CrouchInternal.AnimStateNode_0.Crouch.AnimGraphNode_StateMachine_0.IdleLoop.AnimStateNode_0.Idle1
  nodes [
   n1 = Blend Poses by int!(
    BlendPose_0:Pose Link Structure=Sequence Player 'CrouchBreak_01'() {bLoopAnimation=false, sequence=/Game/Animation/Game/Pets/Chameleon/Backpack/Crouch/CrouchBreak_01}
    BlendPose_1:Pose Link Structure=Sequence Player 'CrouchBreak_02'() {bLoopAnimation=false, sequence=/Game/Animation/Game/Pets/Chameleon/Backpack/Crouch/CrouchBreak_02}
    BlendPose_2:Pose Link Structure=Apply Additive(
     Base=Sequence Player 'Crouch_Pose'() {bIgnoreForRelevancyTest=true, bLoopAnimation=false, sequence=&Crouch_Pose}
     Additive="Sequence Player (additive) 'IdleStraightLookAround_Add'"() {bLoopAnimation=false, sequence=&IdleStraightLookAround_Add}
    ) {alphaBoolBlend {blendOption=Linear}}
    BlendTime_0:float=0
    BlendTime_1:float=0
    BlendTime_2:float=0
    ActiveChildIndex:int=Get RandomIdle() -> (RandomIdle:int=0)
   ) -> (Pose:Pose Link Structure) {
    is=title AnimGraphNode_BlendListByInt
    blendType=Linear
    blendPose [{linkId=-1, sourceLinkId=-1}, {linkId=-1, sourceLinkId=-1}, {linkId=-1, sourceLinkId=-1}]
    blendTime [0.1,0.1,0.1]
   }
   Output Animation Pose(Result=@n1)
  ]
 }
 Idle2 {
  path=AnimGraph.AnimGraphNode_StateMachine_0.FullBody.AnimStateNode_18.Crouching.AnimGraphNode_StateMachine_0.Crouch.AnimStateNode_29.Crouch.AnimGraphNode_StateMachine_0.CrouchInternal.AnimStateNode_0.Crouch.AnimGraphNode_StateMachine_0.IdleLoop.AnimStateNode_1.Idle2
  nodes [
   n1 = Blend Poses by int(
    BlendPose_0=Sequence Player 'CrouchBreak_01'() {bLoopAnimation=false, sequence=&CrouchBreak_01}
    BlendPose_1=Sequence Player 'CrouchBreak_02'() {bLoopAnimation=false, sequence=&CrouchBreak_02}
    BlendPose_2=Apply Additive(
     Base=Sequence Player 'Crouch_Pose'() {bIgnoreForRelevancyTest=true, bLoopAnimation=false, sequence=&Crouch_Pose}
     Additive="Sequence Player (additive) 'IdleStraightLookAround_Add'"() {bLoopAnimation=false, sequence=&IdleStraightLookAround_Add}
    ) {alphaBoolBlend {blendOption=Linear}}
    BlendTime_0=0
    BlendTime_1=0
    BlendTime_2=0
    ActiveChildIndex=Get RandomIdle2() -> (RandomIdle2:int=0)
   ) {
    blendType=Linear
    blendPose [{linkId=-1, sourceLinkId=-1}, {linkId=-1, sourceLinkId=-1}, {linkId=-1, sourceLinkId=-1}]
    blendTime [0.1,0.1,0.1]
   }
   Output Animation Pose(Result=@n1)
  ]
 }
 State {
  path=AnimGraph.AnimGraphNode_StateMachine_0.FullBody.AnimStateNode_18.Crouching.AnimGraphNode_StateMachine_0.Crouch.AnimStateNode_29.Crouch.AnimGraphNode_StateMachine_0.CrouchInternal.AnimStateNode_0.Crouch.AnimGraphNode_StateMachine_0.IdleLoop.AnimStateNode_2.State
  nodes [
   Sequence Player 'CrouchBreak_01'() {bLoopAnimation=false, sequence=&CrouchBreak_01}
   Sequence Player 'CrouchBreak_02'() {bLoopAnimation=false, sequence=&CrouchBreak_02}
   Output Animation Pose(Result=Apply Additive(
     Base=Sequence Player 'Crouch_Pose'() {bIgnoreForRelevancyTest=true, bLoopAnimation=false, sequence=&Crouch_Pose}
     Additive="Sequence Player (additive) 'IdleStraightLookAround_Add'"() {bLoopAnimation=false, sequence=&IdleStraightLookAround_Add}
    ) {alphaBoolBlend {blendOption=Linear}})
  ]
 }
 State_1 {
  path=AnimGraph.AnimGraphNode_StateMachine_0.FullBody.AnimStateNode_18.Crouching.AnimGraphNode_StateMachine_0.Crouch.AnimStateNode_29.Crouch.AnimGraphNode_StateMachine_0.CrouchInternal.AnimStateNode_0.Crouch.AnimGraphNode_StateMachine_0.IdleLoop.AnimStateNode_3.State_1
  nodes [
   Apply Additive(
    Base=Sequence Player 'Crouch_Pose'() {bIgnoreForRelevancyTest=true, bLoopAnimation=false, sequence=&Crouch_Pose}
    Additive="Sequence Player (additive) 'IdleStraightLookAround_Add'"() {bLoopAnimation=false, sequence=&IdleStraightLookAround_Add}
   ) {alphaBoolBlend {blendOption=Linear}}
   Sequence Player 'CrouchBreak_02'() {bLoopAnimation=false, sequence=&CrouchBreak_02}
   Output Animation Pose(Result=Sequence Player 'CrouchBreak_01'() {bLoopAnimation=false, sequence=&CrouchBreak_01})
  ]
 }
 State_2 {
  path=AnimGraph.AnimGraphNode_StateMachine_0.FullBody.AnimStateNode_18.Crouching.AnimGraphNode_StateMachine_0.Crouch.AnimStateNode_29.Crouch.AnimGraphNode_StateMachine_0.CrouchInternal.AnimStateNode_0.Crouch.AnimGraphNode_StateMachine_0.IdleLoop.AnimStateNode_4.State_2
  nodes [
   Apply Additive(
    Base=Sequence Player 'Crouch_Pose'() {bIgnoreForRelevancyTest=true, bLoopAnimation=false, sequence=&Crouch_Pose}
    Additive="Sequence Player (additive) 'IdleStraightLookAround_Add'"() {bLoopAnimation=false, sequence=&IdleStraightLookAround_Add}
   ) {alphaBoolBlend {blendOption=Linear}}
   Sequence Player 'CrouchBreak_01'() {bLoopAnimation=false, sequence=&CrouchBreak_01}
   Output Animation Pose(Result=Sequence Player 'CrouchBreak_02'() {bLoopAnimation=false, sequence=&CrouchBreak_02})
  ]
 }
 Transition {
  path=AnimGraph.AnimGraphNode_StateMachine_0.FullBody.AnimStateNode_18.Crouching.AnimGraphNode_StateMachine_0.Crouch.AnimStateNode_29.Crouch.AnimGraphNode_StateMachine_0.CrouchInternal.AnimStateNode_0.Crouch.AnimGraphNode_StateMachine_0.IdleLoop.AnimStateTransitionNode_0.Transition
  nodes [Result()]
 }
 Transition {
  path=AnimGraph.AnimGraphNode_StateMachine_0.FullBody.AnimStateNode_18.Crouching.AnimGraphNode_StateMachine_0.Crouch.AnimStateNode_29.Crouch.AnimGraphNode_StateMachine_0.CrouchInternal.AnimStateNode_0.Crouch.AnimGraphNode_StateMachine_0.IdleLoop.AnimStateTransitionNode_1.Transition
  nodes [Result()]
 }
 Transition {
  path=AnimGraph.AnimGraphNode_StateMachine_0.FullBody.AnimStateNode_18.Crouching.AnimGraphNode_StateMachine_0.Crouch.AnimStateNode_29.Crouch.AnimGraphNode_StateMachine_0.CrouchInternal.AnimStateNode_0.Crouch.AnimGraphNode_StateMachine_0.IdleLoop.AnimStateTransitionNode_2.Transition
  nodes [Result()]
 }
 Transition {
  path=AnimGraph.AnimGraphNode_StateMachine_0.FullBody.AnimStateNode_18.Crouching.AnimGraphNode_StateMachine_0.Crouch.AnimStateNode_29.Crouch.AnimGraphNode_StateMachine_0.CrouchInternal.AnimStateNode_0.Crouch.AnimGraphNode_StateMachine_0.IdleLoop.AnimStateTransitionNode_3.Transition
  nodes [Result()]
 }
 Transition {
  path=AnimGraph.AnimGraphNode_StateMachine_0.FullBody.AnimStateNode_18.Crouching.AnimGraphNode_StateMachine_0.Crouch.AnimStateNode_29.Crouch.AnimGraphNode_StateMachine_0.CrouchInternal.AnimStateNode_0.Crouch.AnimGraphNode_StateMachine_0.IdleLoop.AnimStateTransitionNode_4.Transition
  nodes [Result()]
 }
 Happy {
  path=AnimGraph.AnimGraphNode_StateMachine_0.FullBody.AnimStateNode_18.Crouching.AnimGraphNode_StateMachine_0.Crouch.AnimStateNode_29.Crouch.AnimGraphNode_StateMachine_0.CrouchInternal.AnimStateNode_2.Happy
  nodes [Output Animation Pose(Result=Sequence Player 'Happy'() {sequence=/Game/Animation/Game/Pets/Chameleon/Backpack/Emotions/Happy})]
 }
 CrouchToHappy {
  path=AnimGraph.AnimGraphNode_StateMachine_0.FullBody.AnimStateNode_18.Crouching.AnimGraphNode_StateMachine_0.Crouch.AnimStateNode_29.Crouch.AnimGraphNode_StateMachine_0.CrouchInternal.AnimStateNode_3.CrouchToHappy
  nodes [Output Animation Pose(Result=Sequence Player 'CrouchToHappy'() {sequence=/Game/Animation/Game/Pets/Chameleon/Backpack/Emotions/CrouchToHappy})]
 }
 HappyToCrouch {
  path=AnimGraph.AnimGraphNode_StateMachine_0.FullBody.AnimStateNode_18.Crouching.AnimGraphNode_StateMachine_0.Crouch.AnimStateNode_29.Crouch.AnimGraphNode_StateMachine_0.CrouchInternal.AnimStateNode_4.HappyToCrouch
  nodes [Output Animation Pose(Result=Sequence Player 'HappyToCrouch'() {bLoopAnimation=false, sequence=/Game/Animation/Game/Pets/Chameleon/Backpack/Emotions/HappyToCrouch})]
 }
 Transition {
  path=AnimGraph.AnimGraphNode_StateMachine_0.FullBody.AnimStateNode_18.Crouching.AnimGraphNode_StateMachine_0.Crouch.AnimStateNode_29.Crouch.AnimGraphNode_StateMachine_0.CrouchInternal.AnimStateTransitionNode_4.Transition
  nodes [Result(bCanEnterTransition=Get bIsHappy() -> (bIsHappy:bool))]
 }
 Transition {
  path=AnimGraph.AnimGraphNode_StateMachine_0.FullBody.AnimStateNode_18.Crouching.AnimGraphNode_StateMachine_0.Crouch.AnimStateNode_29.Crouch.AnimGraphNode_StateMachine_0.CrouchInternal.AnimStateTransitionNode_5.Transition
  nodes [Result()]
 }
 Transition {
  path=AnimGraph.AnimGraphNode_StateMachine_0.FullBody.AnimStateNode_18.Crouching.AnimGraphNode_StateMachine_0.Crouch.AnimStateNode_29.Crouch.AnimGraphNode_StateMachine_0.CrouchInternal.AnimStateTransitionNode_6.Transition
  nodes [Result(bCanEnterTransition=Get bTransition_Happy_HappyToCrouchInternal() -> (bTransition_Happy_HappyToCrouchInternal:bool))]
 }
 Transition {
  path=AnimGraph.AnimGraphNode_StateMachine_0.FullBody.AnimStateNode_18.Crouching.AnimGraphNode_StateMachine_0.Crouch.AnimStateNode_29.Crouch.AnimGraphNode_StateMachine_0.CrouchInternal.AnimStateTransitionNode_7.Transition
  nodes [Result()]
 }
 LocomotionCrouchHappy {
  path=AnimGraph.AnimGraphNode_StateMachine_0.FullBody.AnimStateNode_18.Crouching.AnimGraphNode_StateMachine_0.Crouch.AnimStateNode_3.LocomotionCrouchHappy
  nodes [Output Animation Pose(Result=Sequence Player 'LocomotionCrouchHappy'() {sequence=/Game/Animation/Game/Pets/Chameleon/Backpack/Emotions/LocomotionCrouchHappy})]
 }
 Transition {
  path=AnimGraph.AnimGraphNode_StateMachine_0.FullBody.AnimStateNode_18.Crouching.AnimGraphNode_StateMachine_0.Crouch.AnimStateTransitionNode_0.Transition
  nodes [Result(bCanEnterTransition=Get bIsMoving2D())]
 }
 Transition {
  path=AnimGraph.AnimGraphNode_StateMachine_0.FullBody.AnimStateNode_18.Crouching.AnimGraphNode_StateMachine_0.Crouch.AnimStateTransitionNode_1.Transition
  nodes [Result()]
 }
 Transition {
  path=AnimGraph.AnimGraphNode_StateMachine_0.FullBody.AnimStateNode_18.Crouching.AnimGraphNode_StateMachine_0.Crouch.AnimStateTransitionNode_11.Transition
  nodes [Result(bCanEnterTransition=NOT Boolean(A=Get bIsMoving2D()))]
 }
 Transition {
  path=AnimGraph.AnimGraphNode_StateMachine_0.FullBody.AnimStateNode_18.Crouching.AnimGraphNode_StateMachine_0.Crouch.AnimStateTransitionNode_2.Transition
  nodes [Result(bCanEnterTransition=Get bIsHappy())]
 }
 Transition {
  path=AnimGraph.AnimGraphNode_StateMachine_0.FullBody.AnimStateNode_18.Crouching.AnimGraphNode_StateMachine_0.Crouch.AnimStateTransitionNode_3.Transition
  nodes [Result(bCanEnterTransition=Get bIsHappy())]
 }
 Transition {
  path=AnimGraph.AnimGraphNode_StateMachine_0.FullBody.AnimStateNode_18.Crouching.AnimGraphNode_StateMachine_0.Crouch.AnimStateTransitionNode_4.Transition
  nodes [Result(bCanEnterTransition=NOT Boolean(A=Get bIsMoving2D()))]
 }
 AnimationTransitionGraph_1 {
  path=AnimGraph.AnimGraphNode_StateMachine_0.FullBody.AnimStateNode_18.Crouching.AnimGraphNode_StateMachine_0.Crouch.AnimStateTransitionNode_47.AnimationTransitionGraph_1
  nodes [Result(bCanEnterTransition=Get bIsCrouchSprinting())]
 }
 Transition {
  path=AnimGraph.AnimGraphNode_StateMachine_0.FullBody.AnimStateNode_18.Crouching.AnimGraphNode_StateMachine_0.Crouch.AnimStateTransitionNode_5.Transition
  nodes [Result(bCanEnterTransition=NOT Boolean(A=Get bIsMoving2D()))]
 }
 AnimationTransitionGraph_1 {
  path=AnimGraph.AnimGraphNode_StateMachine_0.FullBody.AnimStateNode_18.Crouching.AnimGraphNode_StateMachine_0.Crouch.AnimStateTransitionNode_55.AnimationTransitionGraph_1
  nodes [Result(bCanEnterTransition=NOT Boolean(A=Get bIsCrouchSprinting()))]
 }
 Transition {
  path=AnimGraph.AnimGraphNode_StateMachine_0.FullBody.AnimStateNode_18.Crouching.AnimGraphNode_StateMachine_0.Crouch.AnimStateTransitionNode_6.Transition
  nodes [Result()]
 }
 Transition {
  path=AnimGraph.AnimGraphNode_StateMachine_0.FullBody.AnimStateNode_18.Crouching.AnimGraphNode_StateMachine_0.Crouch.AnimStateTransitionNode_7.Transition
  nodes [Result(bCanEnterTransition=Get bIsMoving2D())]
 }
 Transition {
  path=AnimGraph.AnimGraphNode_StateMachine_0.FullBody.AnimStateNode_18.Crouching.AnimGraphNode_StateMachine_0.Crouch.AnimStateTransitionNode_8.Transition
  nodes [Result(bCanEnterTransition=NOT Boolean(A=Get bIsMoving2D()))]
 }
 HunkerToCrouch {
  path=AnimGraph.AnimGraphNode_StateMachine_0.FullBody.AnimStateNode_19.HunkerToCrouch
  nodes [Output Animation Pose(Result=Sequence Player 'HunkerToCrouch'() {bLoopAnimation=false, sequence=/Game/Animation/Game/Pets/Chameleon/Backpack/Hunker/HunkerToCrouch})]
 }
 LocomotionHappy {
  path=AnimGraph.AnimGraphNode_StateMachine_0.FullBody.AnimStateNode_2.LocomotionHappy
  nodes [Output Animation Pose(Result=Sequence Player 'LocomotionHappy'() {sequence=/Game/Animation/Game/Pets/Chameleon/Backpack/Emotions/LocomotionHappy})]
 }
 idle {
  path=AnimGraph.AnimGraphNode_StateMachine_0.FullBody.AnimStateNode_20.idle
  nodes [
   AnimGraphNode_StateMachine_0 = Idle() {is=title AnimGraphNode_StateMachine}
   Output Animation Pose(Result=@AnimGraphNode_StateMachine_0)
  ]
 }
 Idle {
  entry=InnerIdle
  path=AnimGraph.AnimGraphNode_StateMachine_0.FullBody.AnimStateNode_20.idle.AnimGraphNode_StateMachine_0.Idle
  ids=id
  nodes [
   AnimStateNode_0 = TestIdle_01_2() {is=title AnimStateNode}
   AnimStateNode_1 = Emote() {is=title AnimStateNode}
   AnimStateNode_2 = Happy()
   AnimStateNode_3 = IdleToHappy() {is=title AnimStateNode}
   AnimStateNode_4 = HappyToIdle() {is=title AnimStateNode}
   AnimStateNode_5 = InnerIdle() {is=title AnimStateNode}
   AnimStateNode_6 = TestIdle_01() {is=title AnimStateNode}
   AnimStateNode_7 = TestIdle_01_1() {is=title AnimStateNode}
   AnimStateTransitionNode_0 = AnimStateTransitionNode() {bAutomaticRuleBasedOnSequencePlayerInState=true}
   AnimStateTransitionNode_1 = AnimStateTransitionNode() {bAutomaticRuleBasedOnSequencePlayerInState=true}
   AnimStateTransitionNode_2 = AnimStateTransitionNode() {bAutomaticRuleBasedOnSequencePlayerInState=true}
   AnimStateTransitionNode_3 = AnimStateTransitionNode()
   AnimStateTransitionNode_4 = AnimStateTransitionNode() {bAutomaticRuleBasedOnSequencePlayerInState=true}
   AnimStateTransitionNode_6 = AnimStateTransitionNode()
   AnimStateTransitionNode_7 = AnimStateTransitionNode() {bAutomaticRuleBasedOnSequencePlayerInState=true}
   AnimStateTransitionNode_8 = AnimStateTransitionNode() {crossfadeDuration=0.1}
   AnimStateTransitionNode_9 = AnimStateTransitionNode() {bAutomaticRuleBasedOnSequencePlayerInState=true, crossfadeDuration=0}
  ]
  links [
   {
    dst=InnerIdle
    rule=AnimGraph.AnimGraphNode_StateMachine_0.FullBody.AnimStateNode_20.idle.AnimGraphNode_StateMachine_0.Idle.AnimStateTransitionNode_4.Transition
    src=Emote
    via=AnimStateTransitionNode_4
   }
   {
    dst=HappyToIdle
    rule=AnimGraph.AnimGraphNode_StateMachine_0.FullBody.AnimStateNode_20.idle.AnimGraphNode_StateMachine_0.Idle.AnimStateTransitionNode_8.Transition
    src=Happy
    via=AnimStateTransitionNode_8
   }
   {
    dst=InnerIdle
    rule=AnimGraph.AnimGraphNode_StateMachine_0.FullBody.AnimStateNode_20.idle.AnimGraphNode_StateMachine_0.Idle.AnimStateTransitionNode_9.Transition
    src=HappyToIdle
    via=AnimStateTransitionNode_9
   }
   {
    dst=Happy
    rule=AnimGraph.AnimGraphNode_StateMachine_0.FullBody.AnimStateNode_20.idle.AnimGraphNode_StateMachine_0.Idle.AnimStateTransitionNode_7.Transition
    src=IdleToHappy
    via=AnimStateTransitionNode_7
   }
   {
    dst=Emote
    rule=AnimGraph.AnimGraphNode_StateMachine_0.FullBody.AnimStateNode_20.idle.AnimGraphNode_StateMachine_0.Idle.AnimStateTransitionNode_3.Transition
    src=InnerIdle
    via=AnimStateTransitionNode_3
   }
   {
    dst=IdleToHappy
    rule=AnimGraph.AnimGraphNode_StateMachine_0.FullBody.AnimStateNode_20.idle.AnimGraphNode_StateMachine_0.Idle.AnimStateTransitionNode_6.Transition
    src=InnerIdle
    via=AnimStateTransitionNode_6
   }
   {
    dst=TestIdle_01_1
    rule=AnimGraph.AnimGraphNode_StateMachine_0.FullBody.AnimStateNode_20.idle.AnimGraphNode_StateMachine_0.Idle.AnimStateTransitionNode_0.Transition
    src=TestIdle_01
    via=AnimStateTransitionNode_0
   }
   {
    dst=TestIdle_01_2
    rule=AnimGraph.AnimGraphNode_StateMachine_0.FullBody.AnimStateNode_20.idle.AnimGraphNode_StateMachine_0.Idle.AnimStateTransitionNode_1.Transition
    src=TestIdle_01_1
    via=AnimStateTransitionNode_1
   }
   {
    dst=TestIdle_01
    rule=AnimGraph.AnimGraphNode_StateMachine_0.FullBody.AnimStateNode_20.idle.AnimGraphNode_StateMachine_0.Idle.AnimStateTransitionNode_2.Transition
    src=TestIdle_01_2
    via=AnimStateTransitionNode_2
   }
  ]
 }
 TestIdle_01_2 {
  path=AnimGraph.AnimGraphNode_StateMachine_0.FullBody.AnimStateNode_20.idle.AnimGraphNode_StateMachine_0.Idle.AnimStateNode_0.TestIdle_01_2
  nodes [Output Animation Pose(Result=Sequence Player 'IdleBreak_02'() {sequence=/Game/Animation/Game/Pets/Chameleon/Backpack/Idle/IdleBreak_02})]
 }
 Emote {
  path=AnimGraph.AnimGraphNode_StateMachine_0.FullBody.AnimStateNode_20.idle.AnimGraphNode_StateMachine_0.Idle.AnimStateNode_1.Emote
  nodes [Output Animation Pose(Result=Sequence Player 'Emote'() {bLoopAnimation=false, sequence=/Game/Animation/Game/Pets/Chameleon/Backpack/Emotions/Emote})]
 }
 Happy {
  path=AnimGraph.AnimGraphNode_StateMachine_0.FullBody.AnimStateNode_20.idle.AnimGraphNode_StateMachine_0.Idle.AnimStateNode_2.Happy
  nodes [Output Animation Pose(Result=Sequence Player 'Happy'() {sequence=&Happy})]
 }
 IdleToHappy {
  path=AnimGraph.AnimGraphNode_StateMachine_0.FullBody.AnimStateNode_20.idle.AnimGraphNode_StateMachine_0.Idle.AnimStateNode_3.IdleToHappy
  nodes [Output Animation Pose(Result=Sequence Player 'IdleToHappy'() {sequence=/Game/Animation/Game/Pets/Chameleon/Backpack/Emotions/IdleToHappy})]
 }
 HappyToIdle {
  path=AnimGraph.AnimGraphNode_StateMachine_0.FullBody.AnimStateNode_20.idle.AnimGraphNode_StateMachine_0.Idle.AnimStateNode_4.HappyToIdle
  nodes [Output Animation Pose(Result=Sequence Player 'HappyToIdle'() {bLoopAnimation=false, sequence=/Game/Animation/Game/Pets/Chameleon/Backpack/Emotions/HappyToIdle})]
 }
 InnerIdle {
  path=AnimGraph.AnimGraphNode_StateMachine_0.FullBody.AnimStateNode_20.idle.AnimGraphNode_StateMachine_0.Idle.AnimStateNode_5.InnerIdle
  nodes [
   n1 = Random Sequence Player!() -> (Pose:Pose Link Structure) {
    is=title AnimGraphNode_RandomPlayer
    bShuffleMode=true
    entries [
     {
      chanceToPlay=0.2
      maxLoopCount=2
      maxPlayRate=1
      minLoopCount=0
      minPlayRate=1
      sequence=/Game/Animation/Game/Pets/Chameleon/Backpack/Idle/Idle
      blendIn {blendOption=Linear, blendTime=0.2, customCurve=null}
     }
     {
      chanceToPlay=0.1
      maxLoopCount=0
      maxPlayRate=1
      minLoopCount=0
      minPlayRate=1
      sequence=&IdleBreak_02
      blendIn {blendOption=Linear, blendTime=0.2, customCurve=null}
     }
     {
      chanceToPlay=0.1
      maxLoopCount=0
      maxPlayRate=1
      minLoopCount=0
      minPlayRate=1
      sequence=/Game/Animation/Game/Pets/Chameleon/Backpack/Idle/IdleBreak_01
      blendIn {blendOption=Linear, blendTime=0.2, customCurve=null}
     }
     {
      chanceToPlay=0.2
      maxLoopCount=2
      maxPlayRate=1
      minLoopCount=0
      minPlayRate=1
      sequence=&Idle
      blendIn {blendOption=Linear, blendTime=0.2, customCurve=null}
     }
     {
      chanceToPlay=0.2
      maxLoopCount=2
      maxPlayRate=1
      minLoopCount=0
      minPlayRate=1
      sequence=&Idle
      blendIn {blendOption=Linear, blendTime=0.2, customCurve=null}
     }
     {
      chanceToPlay=0.2
      maxLoopCount=2
      maxPlayRate=1
      minLoopCount=0
      minPlayRate=1
      sequence=&Idle
      blendIn {blendOption=Linear, blendTime=0.2, customCurve=null}
     }
     {
      chanceToPlay=1
      maxLoopCount=0
      maxPlayRate=1
      minLoopCount=0
      minPlayRate=1
      sequence=&Idle
      blendIn {blendOption=Linear, blendTime=0.2, customCurve=null}
     }
     {
      chanceToPlay=1
      maxLoopCount=0
      maxPlayRate=1
      minLoopCount=0
      minPlayRate=1
      sequence=&Idle
      blendIn {blendOption=Linear, blendTime=0.2, customCurve=null}
     }
     {
      chanceToPlay=1
      maxLoopCount=0
      maxPlayRate=1
      minLoopCount=0
      minPlayRate=1
      sequence=&Idle
      blendIn {blendOption=Linear, blendTime=0.2, customCurve=null}
     }
     {
      chanceToPlay=1
      maxLoopCount=0
      maxPlayRate=1
      minLoopCount=0
      minPlayRate=1
      sequence=&Idle
      blendIn {blendOption=Linear, blendTime=0.2, customCurve=null}
     }
     {
      chanceToPlay=1
      maxLoopCount=0
      maxPlayRate=1
      minLoopCount=0
      minPlayRate=1
      sequence=&Idle
      blendIn {blendOption=Linear, blendTime=0.2, customCurve=null}
     }
    ]
   }
   Output Animation Pose(Result=@n1)
  ]
 }
 TestIdle_01 {
  path=AnimGraph.AnimGraphNode_StateMachine_0.FullBody.AnimStateNode_20.idle.AnimGraphNode_StateMachine_0.Idle.AnimStateNode_6.TestIdle_01
  nodes [Output Animation Pose(Result=Sequence Player 'Idle'() {sequence=&Idle})]
 }
 TestIdle_01_1 {
  path=AnimGraph.AnimGraphNode_StateMachine_0.FullBody.AnimStateNode_20.idle.AnimGraphNode_StateMachine_0.Idle.AnimStateNode_7.TestIdle_01_1
  nodes [Output Animation Pose(Result=Sequence Player 'IdleBreak_01'() {sequence=&IdleBreak_01})]
 }
 Transition {
  path=AnimGraph.AnimGraphNode_StateMachine_0.FullBody.AnimStateNode_20.idle.AnimGraphNode_StateMachine_0.Idle.AnimStateTransitionNode_0.Transition
  nodes [Result()]
 }
 Transition {
  path=AnimGraph.AnimGraphNode_StateMachine_0.FullBody.AnimStateNode_20.idle.AnimGraphNode_StateMachine_0.Idle.AnimStateTransitionNode_1.Transition
  nodes [Result()]
 }
 Transition {
  path=AnimGraph.AnimGraphNode_StateMachine_0.FullBody.AnimStateNode_20.idle.AnimGraphNode_StateMachine_0.Idle.AnimStateTransitionNode_2.Transition
  nodes [Result()]
 }
 Transition {
  path=AnimGraph.AnimGraphNode_StateMachine_0.FullBody.AnimStateNode_20.idle.AnimGraphNode_StateMachine_0.Idle.AnimStateTransitionNode_3.Transition
  nodes [Result(bCanEnterTransition=Get bEmotePlaying_WhereHunkeringIsBest())]
 }
 Transition {
  path=AnimGraph.AnimGraphNode_StateMachine_0.FullBody.AnimStateNode_20.idle.AnimGraphNode_StateMachine_0.Idle.AnimStateTransitionNode_4.Transition
  nodes [Result(bCanEnterTransition=NOT Boolean(A=Get bEmotePlaying_WhereHunkeringIsBest()))]
 }
 Transition {
  path=AnimGraph.AnimGraphNode_StateMachine_0.FullBody.AnimStateNode_20.idle.AnimGraphNode_StateMachine_0.Idle.AnimStateTransitionNode_6.Transition
  nodes [Result(bCanEnterTransition=Get bTransition_InnerIdle_IdleToHappy() -> (bTransition_InnerIdle_IdleToHappy:bool))]
 }
 Transition {
  path=AnimGraph.AnimGraphNode_StateMachine_0.FullBody.AnimStateNode_20.idle.AnimGraphNode_StateMachine_0.Idle.AnimStateTransitionNode_7.Transition
  nodes [Result()]
 }
 Transition {
  path=AnimGraph.AnimGraphNode_StateMachine_0.FullBody.AnimStateNode_20.idle.AnimGraphNode_StateMachine_0.Idle.AnimStateTransitionNode_8.Transition
  nodes [Result(bCanEnterTransition=Get bTransition_Happy_HappyToIdle() -> (bTransition_Happy_HappyToIdle:bool))]
 }
 Transition {
  path=AnimGraph.AnimGraphNode_StateMachine_0.FullBody.AnimStateNode_20.idle.AnimGraphNode_StateMachine_0.Idle.AnimStateTransitionNode_9.Transition
  nodes [Result()]
 }
 Locomotion {
  path=AnimGraph.AnimGraphNode_StateMachine_0.FullBody.AnimStateNode_21.Locomotion
  nodes [
   AnimGraphNode_StateMachine_0 = TestInOrder() {is=title AnimGraphNode_StateMachine}
   n1 = Random Sequence Player() {
    bShuffleMode=true
    entries [
     {
      chanceToPlay=0.2
      maxLoopCount=4
      maxPlayRate=1
      minLoopCount=2
      minPlayRate=1
      sequence=/Game/Animation/Game/Pets/Chameleon/Backpack/Sprint/LocomotionPose
      blendIn {blendOption=Linear, blendTime=0.2, customCurve=null}
     }
     {
      chanceToPlay=0.1
      maxLoopCount=0
      maxPlayRate=1
      minLoopCount=0
      minPlayRate=1
      sequence=/Game/Animation/Game/Pets/Chameleon/Backpack/Sprint/SprintBreak_01
      blendIn {blendOption=Linear, blendTime=0.2, customCurve=null}
     }
     {
      chanceToPlay=0.1
      maxLoopCount=0
      maxPlayRate=1
      minLoopCount=0
      minPlayRate=1
      sequence=/Game/Animation/Game/Pets/Chameleon/Backpack/Sprint/SprintBreak_02
      blendIn {blendOption=Linear, blendTime=0.2, customCurve=null}
     }
     {
      chanceToPlay=0.2
      maxLoopCount=4
      maxPlayRate=1
      minLoopCount=2
      minPlayRate=1
      sequence=&LocomotionPose
      blendIn {blendOption=Linear, blendTime=0.2, customCurve=null}
     }
     {
      chanceToPlay=0.2
      maxLoopCount=4
      maxPlayRate=1
      minLoopCount=2
      minPlayRate=1
      sequence=&LocomotionPose
      blendIn {blendOption=Linear, blendTime=0.2, customCurve=null}
     }
     {
      chanceToPlay=0.2
      maxLoopCount=4
      maxPlayRate=1
      minLoopCount=2
      minPlayRate=1
      sequence=&LocomotionPose
      blendIn {blendOption=Linear, blendTime=0.2, customCurve=null}
     }
     {
      chanceToPlay=1
      maxLoopCount=0
      maxPlayRate=1
      minLoopCount=0
      minPlayRate=1
      sequence=&LocomotionPose
      blendIn {blendOption=Linear, blendTime=0.2, customCurve=null}
     }
     {
      chanceToPlay=1
      maxLoopCount=0
      maxPlayRate=1
      minLoopCount=0
      minPlayRate=1
      sequence=&LocomotionPose
      blendIn {blendOption=Linear, blendTime=0.2, customCurve=null}
     }
     {
      chanceToPlay=1
      maxLoopCount=0
      maxPlayRate=1
      minLoopCount=0
      minPlayRate=1
      sequence=&LocomotionPose
      blendIn {blendOption=Linear, blendTime=0.2, customCurve=null}
     }
     {
      chanceToPlay=1
      maxLoopCount=0
      maxPlayRate=1
      minLoopCount=0
      minPlayRate=1
      sequence=&LocomotionPose
      blendIn {blendOption=Linear, blendTime=0.2, customCurve=null}
     }
     {
      chanceToPlay=1
      maxLoopCount=0
      maxPlayRate=1
      minLoopCount=0
      minPlayRate=1
      sequence=&LocomotionPose
      blendIn {blendOption=Linear, blendTime=0.2, customCurve=null}
     }
    ]
   }
   Output Animation Pose(Result=@n1)
  ]
 }
 TestInOrder {
  entry=State
  path=AnimGraph.AnimGraphNode_StateMachine_0.FullBody.AnimStateNode_21.Locomotion.AnimGraphNode_StateMachine_0.TestInOrder
  ids=id
  nodes [
   AnimStateNode_0 = State()
   AnimStateNode_1 = State_1()
   AnimStateNode_2 = State_2()
   AnimStateTransitionNode_0 = AnimStateTransitionNode() {bAutomaticRuleBasedOnSequencePlayerInState=true}
   AnimStateTransitionNode_1 = AnimStateTransitionNode() {bAutomaticRuleBasedOnSequencePlayerInState=true}
   AnimStateTransitionNode_2 = AnimStateTransitionNode() {bAutomaticRuleBasedOnSequencePlayerInState=true}
  ]
  links [
   {
    dst=State_1
    rule=AnimGraph.AnimGraphNode_StateMachine_0.FullBody.AnimStateNode_21.Locomotion.AnimGraphNode_StateMachine_0.TestInOrder.AnimStateTransitionNode_0.Transition
    src=State
    via=AnimStateTransitionNode_0
   }
   {
    dst=State_2
    rule=AnimGraph.AnimGraphNode_StateMachine_0.FullBody.AnimStateNode_21.Locomotion.AnimGraphNode_StateMachine_0.TestInOrder.AnimStateTransitionNode_1.Transition
    src=State_1
    via=AnimStateTransitionNode_1
   }
   {
    dst=State
    rule=AnimGraph.AnimGraphNode_StateMachine_0.FullBody.AnimStateNode_21.Locomotion.AnimGraphNode_StateMachine_0.TestInOrder.AnimStateTransitionNode_2.Transition
    src=State_2
    via=AnimStateTransitionNode_2
   }
  ]
 }
 State {
  path=AnimGraph.AnimGraphNode_StateMachine_0.FullBody.AnimStateNode_21.Locomotion.AnimGraphNode_StateMachine_0.TestInOrder.AnimStateNode_0.State
  nodes [Output Animation Pose(Result=Sequence Player 'LocomotionPose'() {bLoopAnimation=false, playRate=0.25, sequence=&LocomotionPose})]
 }
 State_1 {
  path=AnimGraph.AnimGraphNode_StateMachine_0.FullBody.AnimStateNode_21.Locomotion.AnimGraphNode_StateMachine_0.TestInOrder.AnimStateNode_1.State_1
  nodes [Output Animation Pose(Result=Sequence Player 'SprintBreak_01'() {bLoopAnimation=false, sequence=&SprintBreak_01})]
 }
 State_2 {
  path=AnimGraph.AnimGraphNode_StateMachine_0.FullBody.AnimStateNode_21.Locomotion.AnimGraphNode_StateMachine_0.TestInOrder.AnimStateNode_2.State_2
  nodes [Output Animation Pose(Result=Sequence Player 'SprintBreak_02'() {bLoopAnimation=false, sequence=&SprintBreak_02})]
 }
 Transition {
  path=AnimGraph.AnimGraphNode_StateMachine_0.FullBody.AnimStateNode_21.Locomotion.AnimGraphNode_StateMachine_0.TestInOrder.AnimStateTransitionNode_0.Transition
  nodes [Result()]
 }
 Transition {
  path=AnimGraph.AnimGraphNode_StateMachine_0.FullBody.AnimStateNode_21.Locomotion.AnimGraphNode_StateMachine_0.TestInOrder.AnimStateTransitionNode_1.Transition
  nodes [Result()]
 }
 Transition {
  path=AnimGraph.AnimGraphNode_StateMachine_0.FullBody.AnimStateNode_21.Locomotion.AnimGraphNode_StateMachine_0.TestInOrder.AnimStateTransitionNode_2.Transition
  nodes [Result()]
 }
 HunkerToLobby {
  path=AnimGraph.AnimGraphNode_StateMachine_0.FullBody.AnimStateNode_22.HunkerToLobby
  nodes [Output Animation Pose(Result=Sequence Player 'LocomotionToCrouch'() {sequence=&LocomotionToCrouch})]
 }
 Hunker {
  path=AnimGraph.AnimGraphNode_StateMachine_0.FullBody.AnimStateNode_23.Hunker
  nodes [Output Animation Pose(Result=Apply Additive(
     Base=Sequence Player 'Hunker_Pose'() {sequence=&Hunker_Pose}
     Additive="Sequence Player (additive) 'IdleStraightLookAround_Add'"() {sequence=&IdleStraightLookAround_Add}
    ) {alphaBoolBlend {blendOption=Linear}})]
 }
 LobbyToHunker {
  path=AnimGraph.AnimGraphNode_StateMachine_0.FullBody.AnimStateNode_24.LobbyToHunker
  nodes [Output Animation Pose(Result=Sequence Player 'CrouchIdleToIdle'() {sequence=/Game/Animation/Game/Pets/Chameleon/Backpack/Crouch/CrouchIdleToIdle})]
 }
 FrontendSkydive {
  path=AnimGraph.AnimGraphNode_StateMachine_0.FullBody.AnimStateNode_25.FrontendSkydive
  nodes [
   Apply Additive(
    Base=Sequence Player 'Dive_B_Down_Pose'() {sequence=/Game/Animation/Game/Pets/Chameleon/Backpack/Skydive/Dive_B_Down_Pose}
    Additive="Sequence Player (additive) 'Dive_B_Down_Add'"() {sequence=/Game/Animation/Game/Pets/Chameleon/Backpack/Skydive/Dive_B_Down_Add}
   ) {alphaBoolBlend {blendOption=Linear}}
   AnimGraphNode_StateMachine_0 = FrontEndSkydive() {is=title AnimGraphNode_StateMachine}
   Output Animation Pose(Result=@AnimGraphNode_StateMachine_0)
  ]
 }
 FrontEndSkydive {
  entry=SkydiveGlide
  path=AnimGraph.AnimGraphNode_StateMachine_0.FullBody.AnimStateNode_25.FrontendSkydive.AnimGraphNode_StateMachine_0.FrontEndSkydive
  ids=id
  nodes [
   AnimStateNode_0 = SkydiveGlide() {is=title AnimStateNode}
   AnimStateNode_19 = ToDive() {is=title AnimStateNode}
   AnimStateNode_33 = ToGlide() {is=title AnimStateNode}
   AnimStateNode_8 = SkydiveDive() {is=title AnimStateNode}
   AnimStateTransitionNode_0 = AnimStateTransitionNode()
   AnimStateTransitionNode_1 = AnimStateTransitionNode()
   AnimStateTransitionNode_29 = AnimStateTransitionNode()
   AnimStateTransitionNode_31 = AnimStateTransitionNode() {bAutomaticRuleBasedOnSequencePlayerInState=true}
   AnimStateTransitionNode_70 = AnimStateTransitionNode()
   AnimStateTransitionNode_71 = AnimStateTransitionNode() {bAutomaticRuleBasedOnSequencePlayerInState=true}
  ]
  links [
   {
    dst=ToGlide
    rule=AnimGraph.AnimGraphNode_StateMachine_0.FullBody.AnimStateNode_25.FrontendSkydive.AnimGraphNode_StateMachine_0.FrontEndSkydive.AnimStateTransitionNode_70.AnimationTransitionGraph_1
    src=SkydiveDive
    via=AnimStateTransitionNode_70
   }
   {
    dst=ToDive
    rule=AnimGraph.AnimGraphNode_StateMachine_0.FullBody.AnimStateNode_25.FrontendSkydive.AnimGraphNode_StateMachine_0.FrontEndSkydive.AnimStateTransitionNode_29.AnimationTransitionGraph_1
    src=SkydiveGlide
    via=AnimStateTransitionNode_29
   }
   {
    dst=SkydiveDive
    rule=AnimGraph.AnimGraphNode_StateMachine_0.FullBody.AnimStateNode_25.FrontendSkydive.AnimGraphNode_StateMachine_0.FrontEndSkydive.AnimStateTransitionNode_31.AnimationTransitionGraph_1
    src=ToDive
    via=AnimStateTransitionNode_31
   }
   {
    dst=ToGlide
    rule=AnimGraph.AnimGraphNode_StateMachine_0.FullBody.AnimStateNode_25.FrontendSkydive.AnimGraphNode_StateMachine_0.FrontEndSkydive.AnimStateTransitionNode_1.AnimationTransitionGraph_0
    src=ToDive
    via=AnimStateTransitionNode_1
   }
   {
    dst=SkydiveGlide
    rule=AnimGraph.AnimGraphNode_StateMachine_0.FullBody.AnimStateNode_25.FrontendSkydive.AnimGraphNode_StateMachine_0.FrontEndSkydive.AnimStateTransitionNode_71.AnimationTransitionGraph_1
    src=ToGlide
    via=AnimStateTransitionNode_71
   }
   {
    dst=ToDive
    rule=AnimGraph.AnimGraphNode_StateMachine_0.FullBody.AnimStateNode_25.FrontendSkydive.AnimGraphNode_StateMachine_0.FrontEndSkydive.AnimStateTransitionNode_0.AnimationTransitionGraph_0
    src=ToGlide
    via=AnimStateTransitionNode_0
   }
  ]
 }
 SkydiveGlide {
  path=AnimGraph.AnimGraphNode_StateMachine_0.FullBody.AnimStateNode_25.FrontendSkydive.AnimGraphNode_StateMachine_0.FrontEndSkydive.AnimStateNode_0.SkydiveGlide
  nodes [
   n1 = Break Fort Anim Input Skydiving(FortAnimInput_Skydiving=Get SkydivingState()) -> (
    ParaGilderRootOffset=~
    ParaGliderRootRotationOverride=~
    bUseParaGlideRootModifier=~
    bIsSkydivingFromLaunchPad=~
    bIsSkydivingFromBus=~
    bIsInVortex=~
    bIsDiving=~
    bIsParachuteOpen=~
    bIsSkydiving=~
    bIsParachuteLeaning=~
    bIsSkydiveLeaning=~
    bIsLeaning=~
    bIsSkydiveDiveMode=~
    bPlaySkydiveDrift=~
    bSkydiveDriftDelayActive=~
    bSkydiveDriftAnimAllowed=~
    bIsGliderRight=~
    bIsGliderCenter=~
    bIsGliderLeft=~
    bIsGliderForward=~
    bIsGliderBack=~
    bHasDeployedAndIsStandingUpright=~
    LocalAccelForward=~
    LocalAccelRight=~
    LocalVelocityRight=~
    SkydiveAimPitch=~
    SkydiveAimYaw=~
    DeployChuteAnimRate=~
    SkydiveDriftAnimRate=~
    SkydiveDriftAnimRateCurrent=~
    SkydiveFidgetAnimRate=~
    SkydiveFidgetAnimRateCurrent=~
    SkydiveAdditiveAlpha=~
    SkydiveDriftDelay=~
   )
   Blend Poses by bool(
    BlendPose_0=Apply Additive(
     Base=Blendspace Player 'BS_Glide_Lean'(X=@n1.LocalVelocityRight, Y=@n1.SkydiveAimPitch) {blendSpace=/Game/Animation/Game/Pets/Chameleon/Backpack/Skydive/BS_Glide_Lean}
     Additive="Sequence Player (additive) 'Glide_Add'"() {sequence=/Game/Animation/Game/Pets/Chameleon/Backpack/Skydive/Glide_Add}
    ) {alphaBoolBlend {blendOption=Linear}}
    BlendPose_1=Apply Additive(
     Base=Blendspace Player 'BS_Glide_2d'(X=@n1.SkydiveAimYaw, Y=@n1.SkydiveAimPitch) {blendSpace=/Game/Animation/Game/Pets/Chameleon/Backpack/Skydive/BS_Glide_2d}
     Additive="Sequence Player (additive) 'Glide_Add'"() {sequence=&Glide_Add}
    ) {alphaBoolBlend {blendOption=Linear}}
    bActiveValue=@n1.bIsSkydiveLeaning
   ) {blendType=Linear}
   Output Animation Pose(Result=Apply Additive(
     Base=Sequence Player 'Glide_Pose'() {sequence=/Game/Animation/Game/Pets/Chameleon/Backpack/Skydive/Glide_Pose}
     Additive="Sequence Player (additive) 'Glide_Add'"() {sequence=&Glide_Add}
    ) {alphaBoolBlend {blendOption=Linear}})
  ]
 }
 ToDive {
  path=AnimGraph.AnimGraphNode_StateMachine_0.FullBody.AnimStateNode_25.FrontendSkydive.AnimGraphNode_StateMachine_0.FrontEndSkydive.AnimStateNode_19.ToDive
  nodes [Output Animation Pose(Result=Sequence Player 'GlideToDive'() {sequence=/Game/Animation/Game/Pets/Chameleon/Backpack/Skydive/GlideToDive})]
 }
 ToGlide {
  path=AnimGraph.AnimGraphNode_StateMachine_0.FullBody.AnimStateNode_25.FrontendSkydive.AnimGraphNode_StateMachine_0.FrontEndSkydive.AnimStateNode_33.ToGlide
  nodes [Output Animation Pose(Result=Sequence Player 'DiveToGlide'() {sequence=/Game/Animation/Game/Pets/Chameleon/Backpack/Skydive/DiveToGlide})]
 }
 SkydiveDive {
  path=AnimGraph.AnimGraphNode_StateMachine_0.FullBody.AnimStateNode_25.FrontendSkydive.AnimGraphNode_StateMachine_0.FrontEndSkydive.AnimStateNode_8.SkydiveDive
  nodes [
   n1 = Break Fort Anim Input Skydiving(FortAnimInput_Skydiving=Get SkydivingState()) -> (
    ParaGilderRootOffset=~
    ParaGliderRootRotationOverride=~
    bUseParaGlideRootModifier=~
    bIsSkydivingFromLaunchPad=~
    bIsSkydivingFromBus=~
    bIsInVortex=~
    bIsDiving=~
    bIsParachuteOpen=~
    bIsSkydiving=~
    bIsParachuteLeaning=~
    bIsSkydiveLeaning=~
    bIsLeaning=~
    bIsSkydiveDiveMode=~
    bPlaySkydiveDrift=~
    bSkydiveDriftDelayActive=~
    bSkydiveDriftAnimAllowed=~
    bIsGliderRight=~
    bIsGliderCenter=~
    bIsGliderLeft=~
    bIsGliderForward=~
    bIsGliderBack=~
    bHasDeployedAndIsStandingUpright=~
    LocalAccelForward=~
    LocalAccelRight=~
    LocalVelocityRight=~
    SkydiveAimPitch=~
    SkydiveAimYaw=~
    DeployChuteAnimRate=~
    SkydiveDriftAnimRate=~
    SkydiveDriftAnimRateCurrent=~
    SkydiveFidgetAnimRate=~
    SkydiveFidgetAnimRateCurrent=~
    SkydiveAdditiveAlpha=~
    SkydiveDriftDelay=~
   )
   Blend Poses by bool(
    BlendPose_0=Apply Additive(
     Base=Blendspace Player 'BS_Dive_Lean'(X=@n1.LocalVelocityRight, Y=@n1.SkydiveAimPitch) {blendSpace=/Game/Animation/Game/Pets/Chameleon/Backpack/Skydive/BS_Dive_Lean}
     Additive=Blendspace Player 'GlideDive_Add'!(X=@n1.SkydiveAimPitch) -> (Pose) {blendSpace=/Game/Animation/Game/Pets/Chameleon/Backpack/Skydive/GlideDive_Add}
    ) {alphaBoolBlend {blendOption=Linear}}
    BlendPose_1=Apply Additive(
     Base=Blendspace Player 'BS_Dive_2D'(X=@n1.SkydiveAimYaw, Y=@n1.SkydiveAimPitch) {blendSpace=/Game/Animation/Game/Pets/Chameleon/Backpack/Skydive/BS_Dive_2D}
     Additive=Blendspace Player 'GlideDive_Add'(X=@n1.SkydiveAimPitch) {blendSpace=&GlideDive_Add}
    ) {alphaBoolBlend {blendOption=Linear}}
    bActiveValue=@n1.bIsSkydiveLeaning
   ) {blendType=Linear}
   Output Animation Pose(Result=Apply Additive(
     Base=Sequence Player 'Dive_B_Down_Pose'() {sequence=&Dive_B_Down_Pose}
     Additive="Sequence Player (additive) 'Dive_B_Down_Add'"() {sequence=&Dive_B_Down_Add}
    ) {alphaBoolBlend {blendOption=Linear}})
  ]
 }
 AnimationTransitionGraph_0 {
  path=AnimGraph.AnimGraphNode_StateMachine_0.FullBody.AnimStateNode_25.FrontendSkydive.AnimGraphNode_StateMachine_0.FrontEndSkydive.AnimStateTransitionNode_0.AnimationTransitionGraph_0
  nodes [
   Result(bCanEnterTransition=Get bLobbySkyDive_IsDiving() -> (bLobbySkyDive_IsDiving:bool))
   Break Fort Anim Input Skydiving(FortAnimInput_Skydiving=Get SkydivingState()) -> (
    ParaGilderRootOffset=~
    ParaGliderRootRotationOverride=~
    bUseParaGlideRootModifier=~
    bIsSkydivingFromLaunchPad=~
    bIsSkydivingFromBus=~
    bIsInVortex=~
    bIsDiving=~
    bIsParachuteOpen=~
    bIsSkydiving=~
    bIsParachuteLeaning=~
    bIsSkydiveLeaning=~
    bIsLeaning=~
    bIsSkydiveDiveMode=~
    bPlaySkydiveDrift=~
    bSkydiveDriftDelayActive=~
    bSkydiveDriftAnimAllowed=~
    bIsGliderRight=~
    bIsGliderCenter=~
    bIsGliderLeft=~
    bIsGliderForward=~
    bIsGliderBack=~
    bHasDeployedAndIsStandingUpright=~
    LocalAccelForward=~
    LocalAccelRight=~
    LocalVelocityRight=~
    SkydiveAimPitch=~
    SkydiveAimYaw=~
    DeployChuteAnimRate=~
    SkydiveDriftAnimRate=~
    SkydiveDriftAnimRateCurrent=~
    SkydiveFidgetAnimRate=~
    SkydiveFidgetAnimRateCurrent=~
    SkydiveAdditiveAlpha=~
    SkydiveDriftDelay=~
   )
  ]
 }
 AnimationTransitionGraph_0 {
  path=AnimGraph.AnimGraphNode_StateMachine_0.FullBody.AnimStateNode_25.FrontendSkydive.AnimGraphNode_StateMachine_0.FrontEndSkydive.AnimStateTransitionNode_1.AnimationTransitionGraph_0
  nodes [
   Result(bCanEnterTransition=NOT Boolean(A=Get bLobbySkyDive_IsDiving()))
   Break Fort Anim Input Skydiving(FortAnimInput_Skydiving=Get SkydivingState()) -> (
    ParaGilderRootOffset=~
    ParaGliderRootRotationOverride=~
    bUseParaGlideRootModifier=~
    bIsSkydivingFromLaunchPad=~
    bIsSkydivingFromBus=~
    bIsInVortex=~
    bIsDiving=~
    bIsParachuteOpen=~
    bIsSkydiving=~
    bIsParachuteLeaning=~
    bIsSkydiveLeaning=~
    bIsLeaning=~
    bIsSkydiveDiveMode=~
    bPlaySkydiveDrift=~
    bSkydiveDriftDelayActive=~
    bSkydiveDriftAnimAllowed=~
    bIsGliderRight=~
    bIsGliderCenter=~
    bIsGliderLeft=~
    bIsGliderForward=~
    bIsGliderBack=~
    bHasDeployedAndIsStandingUpright=~
    LocalAccelForward=~
    LocalAccelRight=~
    LocalVelocityRight=~
    SkydiveAimPitch=~
    SkydiveAimYaw=~
    DeployChuteAnimRate=~
    SkydiveDriftAnimRate=~
    SkydiveDriftAnimRateCurrent=~
    SkydiveFidgetAnimRate=~
    SkydiveFidgetAnimRateCurrent=~
    SkydiveAdditiveAlpha=~
    SkydiveDriftDelay=~
   )
  ]
 }
 AnimationTransitionGraph_1 {
  path=AnimGraph.AnimGraphNode_StateMachine_0.FullBody.AnimStateNode_25.FrontendSkydive.AnimGraphNode_StateMachine_0.FrontEndSkydive.AnimStateTransitionNode_29.AnimationTransitionGraph_1
  nodes [
   Result(bCanEnterTransition=Get bLobbySkyDive_IsDiving())
   Break Fort Anim Input Skydiving(FortAnimInput_Skydiving=Get SkydivingState()) -> (
    ParaGilderRootOffset=~
    ParaGliderRootRotationOverride=~
    bUseParaGlideRootModifier=~
    bIsSkydivingFromLaunchPad=~
    bIsSkydivingFromBus=~
    bIsInVortex=~
    bIsDiving=~
    bIsParachuteOpen=~
    bIsSkydiving=~
    bIsParachuteLeaning=~
    bIsSkydiveLeaning=~
    bIsLeaning=~
    bIsSkydiveDiveMode=~
    bPlaySkydiveDrift=~
    bSkydiveDriftDelayActive=~
    bSkydiveDriftAnimAllowed=~
    bIsGliderRight=~
    bIsGliderCenter=~
    bIsGliderLeft=~
    bIsGliderForward=~
    bIsGliderBack=~
    bHasDeployedAndIsStandingUpright=~
    LocalAccelForward=~
    LocalAccelRight=~
    LocalVelocityRight=~
    SkydiveAimPitch=~
    SkydiveAimYaw=~
    DeployChuteAnimRate=~
    SkydiveDriftAnimRate=~
    SkydiveDriftAnimRateCurrent=~
    SkydiveFidgetAnimRate=~
    SkydiveFidgetAnimRateCurrent=~
    SkydiveAdditiveAlpha=~
    SkydiveDriftDelay=~
   )
  ]
 }
 AnimationTransitionGraph_1 {
  path=AnimGraph.AnimGraphNode_StateMachine_0.FullBody.AnimStateNode_25.FrontendSkydive.AnimGraphNode_StateMachine_0.FrontEndSkydive.AnimStateTransitionNode_31.AnimationTransitionGraph_1
  nodes [Result()]
 }
 AnimationTransitionGraph_1 {
  path=AnimGraph.AnimGraphNode_StateMachine_0.FullBody.AnimStateNode_25.FrontendSkydive.AnimGraphNode_StateMachine_0.FrontEndSkydive.AnimStateTransitionNode_70.AnimationTransitionGraph_1
  nodes [
   Result(bCanEnterTransition=NOT Boolean(A=Get bLobbySkyDive_IsDiving()))
   Break Fort Anim Input Skydiving(FortAnimInput_Skydiving=Get SkydivingState()) -> (
    ParaGilderRootOffset=~
    ParaGliderRootRotationOverride=~
    bUseParaGlideRootModifier=~
    bIsSkydivingFromLaunchPad=~
    bIsSkydivingFromBus=~
    bIsInVortex=~
    bIsDiving=~
    bIsParachuteOpen=~
    bIsSkydiving=~
    bIsParachuteLeaning=~
    bIsSkydiveLeaning=~
    bIsLeaning=~
    bIsSkydiveDiveMode=~
    bPlaySkydiveDrift=~
    bSkydiveDriftDelayActive=~
    bSkydiveDriftAnimAllowed=~
    bIsGliderRight=~
    bIsGliderCenter=~
    bIsGliderLeft=~
    bIsGliderForward=~
    bIsGliderBack=~
    bHasDeployedAndIsStandingUpright=~
    LocalAccelForward=~
    LocalAccelRight=~
    LocalVelocityRight=~
    SkydiveAimPitch=~
    SkydiveAimYaw=~
    DeployChuteAnimRate=~
    SkydiveDriftAnimRate=~
    SkydiveDriftAnimRateCurrent=~
    SkydiveFidgetAnimRate=~
    SkydiveFidgetAnimRateCurrent=~
    SkydiveAdditiveAlpha=~
    SkydiveDriftDelay=~
   )
  ]
 }
 AnimationTransitionGraph_1 {
  path=AnimGraph.AnimGraphNode_StateMachine_0.FullBody.AnimStateNode_25.FrontendSkydive.AnimGraphNode_StateMachine_0.FrontEndSkydive.AnimStateTransitionNode_71.AnimationTransitionGraph_1
  nodes [Result()]
 }
 InAirHunker {
  path=AnimGraph.AnimGraphNode_StateMachine_0.FullBody.AnimStateNode_26.InAirHunker
  nodes [Output Animation Pose(Result=Apply Additive(
     Base=Sequence Player 'Hunker_Pose'() {sequence=&Hunker_Pose}
     Additive="Sequence Player (additive) 'InAirHunkerLookAround_Add'"() {sequence=/Game/Animation/Game/Pets/Chameleon/Backpack/Idle/InAirHunkerLookAround_Add}
    ) {alphaBoolBlend {blendOption=Linear}})]
 }
 LobbyHunker {
  path=AnimGraph.AnimGraphNode_StateMachine_0.FullBody.AnimStateNode_27.LobbyHunker
  nodes [Output Animation Pose(Result=Apply Additive(
     Base=Sequence Player 'Hunker_Pose'() {sequence=&Hunker_Pose}
     Additive="Sequence Player (additive) 'IdleStraightLookAround_Add'"() {sequence=&IdleStraightLookAround_Add}
    ) {alphaBoolBlend {blendOption=Linear}})]
 }
 CrouchToCrouchADS {
  path=AnimGraph.AnimGraphNode_StateMachine_0.FullBody.AnimStateNode_28.CrouchToCrouchADS
  nodes [Output Animation Pose(Result=Sequence Player 'CrouchToCrouchADS'() {sequence=/Game/Animation/Game/Pets/Chameleon/Backpack/Crouch/CrouchToCrouchADS})]
 }
 CrouchToCrouchADSNotMoving {
  path=AnimGraph.AnimGraphNode_StateMachine_0.FullBody.AnimStateNode_29.CrouchToCrouchADSNotMoving
  nodes [Output Animation Pose(Result=Sequence Player 'CrouchToCrouchADS'() {sequence=&CrouchToCrouchADS})]
 }
 LocomotionToIdle {
  path=AnimGraph.AnimGraphNode_StateMachine_0.FullBody.AnimStateNode_3.LocomotionToIdle
  nodes [Output Animation Pose(Result=Sequence Player 'LocomotionToIdle'() {sequence=/Game/Animation/Game/Pets/Chameleon/Backpack/Sprint/LocomotionToIdle})]
 }
 ADSToCrouch {
  path=AnimGraph.AnimGraphNode_StateMachine_0.FullBody.AnimStateNode_30.ADSToCrouch
  nodes [Output Animation Pose(Result=Sequence Player 'CrouchDownSightsToCrouch'() {sequence=/Game/Animation/Game/Pets/Chameleon/Backpack/Crouch/CrouchDownSightsToCrouch})]
 }
 CrouchToIdle {
  path=AnimGraph.AnimGraphNode_StateMachine_0.FullBody.AnimStateNode_31.CrouchToIdle
  nodes [Output Animation Pose(Result=Sequence Player 'CrouchIdleToIdle'() {bLoopAnimation=false, sequence=&CrouchIdleToIdle})]
 }
 DBNOToIdle_Old {
  path=AnimGraph.AnimGraphNode_StateMachine_0.FullBody.AnimStateNode_32.DBNOToIdle_Old
  nodes [Output Animation Pose(Result=Sequence Player 'DBNO_Idle_Outro'() {bLoopAnimation=false, sequence=&DBNO_Idle_Outro})]
 }
 SkyDive {
  path=AnimGraph.AnimGraphNode_StateMachine_0.FullBody.AnimStateNode_34.SkyDive
  nodes [
   AnimGraphNode_StateMachine_0 = SkydiveParachute() {is=title AnimGraphNode_StateMachine}
   Output Animation Pose(Result=@AnimGraphNode_StateMachine_0)
  ]
 }
 SkydiveParachute {
  entry=SkydiveStart
  path=AnimGraph.AnimGraphNode_StateMachine_0.FullBody.AnimStateNode_34.SkyDive.AnimGraphNode_StateMachine_0.SkydiveParachute
  ids=id
  nodes [
   AnimStateNode_1 = Parachute() {is=title AnimStateNode}
   AnimStateNode_2 = SkyDive()
   AnimStateNode_26 = SkidiveGlideToParachute() {is=title AnimStateNode}
   AnimStateNode_28 = SkydiveStart() {is=title AnimStateNode}
   AnimStateNode_30 = ParachuteToSkydiveGlide() {is=title AnimStateNode}
   AnimStateTransitionNode_26 = AnimStateTransitionNode()
   AnimStateTransitionNode_27 = AnimStateTransitionNode()
   AnimStateTransitionNode_28 = AnimStateTransitionNode() {bAutomaticRuleBasedOnSequencePlayerInState=true}
   AnimStateTransitionNode_60 = AnimStateTransitionNode() {bAutomaticRuleBasedOnSequencePlayerInState=true}
   AnimStateTransitionNode_61 = AnimStateTransitionNode()
  ]
  links [
   {
    dst=ParachuteToSkydiveGlide
    rule=AnimGraph.AnimGraphNode_StateMachine_0.FullBody.AnimStateNode_34.SkyDive.AnimGraphNode_StateMachine_0.SkydiveParachute.AnimStateTransitionNode_61.AnimationTransitionGraph_1
    src=Parachute
    via=AnimStateTransitionNode_61
   }
   {
    dst=SkyDive
    rule=AnimGraph.AnimGraphNode_StateMachine_0.FullBody.AnimStateNode_34.SkyDive.AnimGraphNode_StateMachine_0.SkydiveParachute.AnimStateTransitionNode_27.Transition
    src=ParachuteToSkydiveGlide
    via=AnimStateTransitionNode_27
   }
   {
    dst=Parachute
    rule=AnimGraph.AnimGraphNode_StateMachine_0.FullBody.AnimStateNode_34.SkyDive.AnimGraphNode_StateMachine_0.SkydiveParachute.AnimStateTransitionNode_60.AnimationTransitionGraph_1
    src=SkidiveGlideToParachute
    via=AnimStateTransitionNode_60
   }
   {
    dst=SkidiveGlideToParachute
    rule=AnimGraph.AnimGraphNode_StateMachine_0.FullBody.AnimStateNode_34.SkyDive.AnimGraphNode_StateMachine_0.SkydiveParachute.AnimStateTransitionNode_26.Transition
    src=SkyDive
    via=AnimStateTransitionNode_26
   }
   {
    dst=SkyDive
    rule=AnimGraph.AnimGraphNode_StateMachine_0.FullBody.AnimStateNode_34.SkyDive.AnimGraphNode_StateMachine_0.SkydiveParachute.AnimStateTransitionNode_28.Transition
    src=SkydiveStart
    via=AnimStateTransitionNode_28
   }
  ]
 }
 Parachute {
  path=AnimGraph.AnimGraphNode_StateMachine_0.FullBody.AnimStateNode_34.SkyDive.AnimGraphNode_StateMachine_0.SkydiveParachute.AnimStateNode_1.Parachute
  nodes [Output Animation Pose(Result=Sequence Player 'Idle'() {sequence=&Idle})]
 }
 SkyDive {
  path=AnimGraph.AnimGraphNode_StateMachine_0.FullBody.AnimStateNode_34.SkyDive.AnimGraphNode_StateMachine_0.SkydiveParachute.AnimStateNode_2.SkyDive
  nodes [
   AnimGraphNode_StateMachine_0 = Skydive() {is=title AnimGraphNode_StateMachine}
   Output Animation Pose(Result=@AnimGraphNode_StateMachine_0)
  ]
 }
 Skydive {
  entry=SkydiveGlide
  path=AnimGraph.AnimGraphNode_StateMachine_0.FullBody.AnimStateNode_34.SkyDive.AnimGraphNode_StateMachine_0.SkydiveParachute.AnimStateNode_2.SkyDive.AnimGraphNode_StateMachine_0.Skydive
  ids=id
  nodes [
   AnimStateNode_0 = SkydiveGlide()
   AnimStateNode_19 = ToDive()
   AnimStateNode_33 = ToGlide()
   AnimStateNode_8 = SkydiveDive()
   AnimStateTransitionNode_0 = AnimStateTransitionNode()
   AnimStateTransitionNode_1 = AnimStateTransitionNode()
   AnimStateTransitionNode_29 = AnimStateTransitionNode()
   AnimStateTransitionNode_31 = AnimStateTransitionNode() {bAutomaticRuleBasedOnSequencePlayerInState=true}
   AnimStateTransitionNode_70 = AnimStateTransitionNode()
   AnimStateTransitionNode_71 = AnimStateTransitionNode() {bAutomaticRuleBasedOnSequencePlayerInState=true}
  ]
  links [
   {
    dst=ToGlide
    rule=AnimGraph.AnimGraphNode_StateMachine_0.FullBody.AnimStateNode_34.SkyDive.AnimGraphNode_StateMachine_0.SkydiveParachute.AnimStateNode_2.SkyDive.AnimGraphNode_StateMachine_0.Skydive.AnimStateTransitionNode_70.AnimationTransitionGraph_1
    src=SkydiveDive
    via=AnimStateTransitionNode_70
   }
   {
    dst=ToDive
    rule=AnimGraph.AnimGraphNode_StateMachine_0.FullBody.AnimStateNode_34.SkyDive.AnimGraphNode_StateMachine_0.SkydiveParachute.AnimStateNode_2.SkyDive.AnimGraphNode_StateMachine_0.Skydive.AnimStateTransitionNode_29.AnimationTransitionGraph_1
    src=SkydiveGlide
    via=AnimStateTransitionNode_29
   }
   {
    dst=SkydiveDive
    rule=AnimGraph.AnimGraphNode_StateMachine_0.FullBody.AnimStateNode_34.SkyDive.AnimGraphNode_StateMachine_0.SkydiveParachute.AnimStateNode_2.SkyDive.AnimGraphNode_StateMachine_0.Skydive.AnimStateTransitionNode_31.AnimationTransitionGraph_1
    src=ToDive
    via=AnimStateTransitionNode_31
   }
   {
    dst=ToGlide
    rule=AnimGraph.AnimGraphNode_StateMachine_0.FullBody.AnimStateNode_34.SkyDive.AnimGraphNode_StateMachine_0.SkydiveParachute.AnimStateNode_2.SkyDive.AnimGraphNode_StateMachine_0.Skydive.AnimStateTransitionNode_1.AnimationTransitionGraph_0
    src=ToDive
    via=AnimStateTransitionNode_1
   }
   {
    dst=SkydiveGlide
    rule=AnimGraph.AnimGraphNode_StateMachine_0.FullBody.AnimStateNode_34.SkyDive.AnimGraphNode_StateMachine_0.SkydiveParachute.AnimStateNode_2.SkyDive.AnimGraphNode_StateMachine_0.Skydive.AnimStateTransitionNode_71.AnimationTransitionGraph_1
    src=ToGlide
    via=AnimStateTransitionNode_71
   }
   {
    dst=ToDive
    rule=AnimGraph.AnimGraphNode_StateMachine_0.FullBody.AnimStateNode_34.SkyDive.AnimGraphNode_StateMachine_0.SkydiveParachute.AnimStateNode_2.SkyDive.AnimGraphNode_StateMachine_0.Skydive.AnimStateTransitionNode_0.AnimationTransitionGraph_0
    src=ToGlide
    via=AnimStateTransitionNode_0
   }
  ]
 }
 SkydiveGlide {
  path=AnimGraph.AnimGraphNode_StateMachine_0.FullBody.AnimStateNode_34.SkyDive.AnimGraphNode_StateMachine_0.SkydiveParachute.AnimStateNode_2.SkyDive.AnimGraphNode_StateMachine_0.Skydive.AnimStateNode_0.SkydiveGlide
  nodes [
   n1 = Break Fort Anim Input Skydiving(FortAnimInput_Skydiving=Get SkydivingState()) -> (
    ParaGilderRootOffset=~
    ParaGliderRootRotationOverride=~
    bUseParaGlideRootModifier=~
    bIsSkydivingFromLaunchPad=~
    bIsSkydivingFromBus=~
    bIsInVortex=~
    bIsDiving=~
    bIsParachuteOpen=~
    bIsSkydiving=~
    bIsParachuteLeaning=~
    bIsSkydiveLeaning=~
    bIsLeaning=~
    bIsSkydiveDiveMode=~
    bPlaySkydiveDrift=~
    bSkydiveDriftDelayActive=~
    bSkydiveDriftAnimAllowed=~
    bIsGliderRight=~
    bIsGliderCenter=~
    bIsGliderLeft=~
    bIsGliderForward=~
    bIsGliderBack=~
    bHasDeployedAndIsStandingUpright=~
    LocalAccelForward=~
    LocalAccelRight=~
    LocalVelocityRight=~
    SkydiveAimPitch=~
    SkydiveAimYaw=~
    DeployChuteAnimRate=~
    SkydiveDriftAnimRate=~
    SkydiveDriftAnimRateCurrent=~
    SkydiveFidgetAnimRate=~
    SkydiveFidgetAnimRateCurrent=~
    SkydiveAdditiveAlpha=~
    SkydiveDriftDelay=~
   )
   n2 = Blendspace Player 'BS_Glide_Lean'(X=@n1.LocalVelocityRight, Y=@n1.SkydiveAimPitch) {blendSpace=&BS_Glide_Lean}
   n3 = "Sequence Player (additive) 'Glide_Add'"() {sequence=&Glide_Add}
   n4 = Blendspace Player 'BS_Glide_2d'(X=@n1.SkydiveAimYaw, Y=@n1.SkydiveAimPitch) {blendSpace=&BS_Glide_2d}
   n5 = "Sequence Player (additive) 'Glide_Add'"() {sequence=&Glide_Add}
   Output Animation Pose(Result=Blend Poses by bool(
     BlendPose_0=Apply Additive(Base=@n2, Additive=@n3) {alphaBoolBlend {blendOption=Linear}}
     BlendPose_1=Apply Additive(Base=@n4, Additive=@n5) {alphaBoolBlend {blendOption=Linear}}
     BlendTime_0=0.5
     BlendTime_1=0.5
     bActiveValue=@n1.bIsSkydiveLeaning
    ) {blendType=Linear})
  ]
 }
 ToDive {
  path=AnimGraph.AnimGraphNode_StateMachine_0.FullBody.AnimStateNode_34.SkyDive.AnimGraphNode_StateMachine_0.SkydiveParachute.AnimStateNode_2.SkyDive.AnimGraphNode_StateMachine_0.Skydive.AnimStateNode_19.ToDive
  nodes [Output Animation Pose(Result=Sequence Player 'GlideToDive'() {sequence=&GlideToDive})]
 }
 ToGlide {
  path=AnimGraph.AnimGraphNode_StateMachine_0.FullBody.AnimStateNode_34.SkyDive.AnimGraphNode_StateMachine_0.SkydiveParachute.AnimStateNode_2.SkyDive.AnimGraphNode_StateMachine_0.Skydive.AnimStateNode_33.ToGlide
  nodes [Output Animation Pose(Result=Sequence Player 'DiveToGlide'() {sequence=&DiveToGlide})]
 }
 SkydiveDive {
  path=AnimGraph.AnimGraphNode_StateMachine_0.FullBody.AnimStateNode_34.SkyDive.AnimGraphNode_StateMachine_0.SkydiveParachute.AnimStateNode_2.SkyDive.AnimGraphNode_StateMachine_0.Skydive.AnimStateNode_8.SkydiveDive
  nodes [
   n1 = Break Fort Anim Input Skydiving(FortAnimInput_Skydiving=Get SkydivingState()) -> (
    ParaGilderRootOffset=~
    ParaGliderRootRotationOverride=~
    bUseParaGlideRootModifier=~
    bIsSkydivingFromLaunchPad=~
    bIsSkydivingFromBus=~
    bIsInVortex=~
    bIsDiving=~
    bIsParachuteOpen=~
    bIsSkydiving=~
    bIsParachuteLeaning=~
    bIsSkydiveLeaning=~
    bIsLeaning=~
    bIsSkydiveDiveMode=~
    bPlaySkydiveDrift=~
    bSkydiveDriftDelayActive=~
    bSkydiveDriftAnimAllowed=~
    bIsGliderRight=~
    bIsGliderCenter=~
    bIsGliderLeft=~
    bIsGliderForward=~
    bIsGliderBack=~
    bHasDeployedAndIsStandingUpright=~
    LocalAccelForward=~
    LocalAccelRight=~
    LocalVelocityRight=~
    SkydiveAimPitch=~
    SkydiveAimYaw=~
    DeployChuteAnimRate=~
    SkydiveDriftAnimRate=~
    SkydiveDriftAnimRateCurrent=~
    SkydiveFidgetAnimRate=~
    SkydiveFidgetAnimRateCurrent=~
    SkydiveAdditiveAlpha=~
    SkydiveDriftDelay=~
   )
   n2 = Blendspace Player 'BS_Dive_Lean'(X=@n1.LocalVelocityRight, Y=@n1.SkydiveAimPitch) {blendSpace=&BS_Dive_Lean}
   n3 = Blendspace Player 'GlideDive_Add'(X=@n1.SkydiveAimPitch) {blendSpace=&GlideDive_Add}
   n4 = Blendspace Player 'BS_Dive_2D'(X=@n1.SkydiveAimYaw, Y=@n1.SkydiveAimPitch) {blendSpace=&BS_Dive_2D}
   n5 = Blendspace Player 'GlideDive_Add'(X=@n1.SkydiveAimPitch) {blendSpace=&GlideDive_Add}
   Output Animation Pose(Result=Blend Poses by bool(
     BlendPose_0=Apply Additive(Base=@n2, Additive=@n3) {alphaBoolBlend {blendOption=Linear}}
     BlendPose_1=Apply Additive(Base=@n4, Additive=@n5) {alphaBoolBlend {blendOption=Linear}}
     bActiveValue=@n1.bIsSkydiveLeaning
    ) {blendType=Linear})
  ]
 }
 AnimationTransitionGraph_0 {
  path=AnimGraph.AnimGraphNode_StateMachine_0.FullBody.AnimStateNode_34.SkyDive.AnimGraphNode_StateMachine_0.SkydiveParachute.AnimStateNode_2.SkyDive.AnimGraphNode_StateMachine_0.Skydive.AnimStateTransitionNode_0.AnimationTransitionGraph_0
  nodes [Result(bCanEnterTransition=Get bTransition_SkyDiveGliding_SkyDiveDiving() -> (bTransition_SkyDiveGliding_SkyDiveDiving:bool))]
 }
 AnimationTransitionGraph_0 {
  path=AnimGraph.AnimGraphNode_StateMachine_0.FullBody.AnimStateNode_34.SkyDive.AnimGraphNode_StateMachine_0.SkydiveParachute.AnimStateNode_2.SkyDive.AnimGraphNode_StateMachine_0.Skydive.AnimStateTransitionNode_1.AnimationTransitionGraph_0
  nodes [Result(bCanEnterTransition=NOT Boolean(A=Get bTransition_SkyDiveGliding_SkyDiveDiving()))]
 }
 AnimationTransitionGraph_1 {
  path=AnimGraph.AnimGraphNode_StateMachine_0.FullBody.AnimStateNode_34.SkyDive.AnimGraphNode_StateMachine_0.SkydiveParachute.AnimStateNode_2.SkyDive.AnimGraphNode_StateMachine_0.Skydive.AnimStateTransitionNode_29.AnimationTransitionGraph_1
  nodes [Result(bCanEnterTransition=Get bTransition_SkyDiveGliding_SkyDiveDiving())]
 }
 AnimationTransitionGraph_1 {
  path=AnimGraph.AnimGraphNode_StateMachine_0.FullBody.AnimStateNode_34.SkyDive.AnimGraphNode_StateMachine_0.SkydiveParachute.AnimStateNode_2.SkyDive.AnimGraphNode_StateMachine_0.Skydive.AnimStateTransitionNode_31.AnimationTransitionGraph_1
  nodes [Result()]
 }
 AnimationTransitionGraph_1 {
  path=AnimGraph.AnimGraphNode_StateMachine_0.FullBody.AnimStateNode_34.SkyDive.AnimGraphNode_StateMachine_0.SkydiveParachute.AnimStateNode_2.SkyDive.AnimGraphNode_StateMachine_0.Skydive.AnimStateTransitionNode_70.AnimationTransitionGraph_1
  nodes [Result(bCanEnterTransition=NOT Boolean(A=Get bTransition_SkyDiveGliding_SkyDiveDiving()))]
 }
 AnimationTransitionGraph_1 {
  path=AnimGraph.AnimGraphNode_StateMachine_0.FullBody.AnimStateNode_34.SkyDive.AnimGraphNode_StateMachine_0.SkydiveParachute.AnimStateNode_2.SkyDive.AnimGraphNode_StateMachine_0.Skydive.AnimStateTransitionNode_71.AnimationTransitionGraph_1
  nodes [Result()]
 }
 SkidiveGlideToParachute {
  path=AnimGraph.AnimGraphNode_StateMachine_0.FullBody.AnimStateNode_34.SkyDive.AnimGraphNode_StateMachine_0.SkydiveParachute.AnimStateNode_26.SkidiveGlideToParachute
  nodes [Output Animation Pose(Result=Sequence Player 'Para_Into'() {sequence=/Game/Animation/Game/Pets/Chameleon/Backpack/Skydive/Para_Into})]
 }
 SkydiveStart {
  path=AnimGraph.AnimGraphNode_StateMachine_0.FullBody.AnimStateNode_34.SkyDive.AnimGraphNode_StateMachine_0.SkydiveParachute.AnimStateNode_28.SkydiveStart
  nodes [Output Animation Pose(Result=Sequence Player 'SkydiveStart'() {bLoopAnimation=false, sequence=/Game/Animation/Game/Pets/Chameleon/Backpack/Skydive/SkydiveStart})]
 }
 ParachuteToSkydiveGlide {
  path=AnimGraph.AnimGraphNode_StateMachine_0.FullBody.AnimStateNode_34.SkyDive.AnimGraphNode_StateMachine_0.SkydiveParachute.AnimStateNode_30.ParachuteToSkydiveGlide
  nodes [Output Animation Pose(Result=Sequence Player 'SkydiveStart'() {sequence=&SkydiveStart})]
 }
 Transition {
  path=AnimGraph.AnimGraphNode_StateMachine_0.FullBody.AnimStateNode_34.SkyDive.AnimGraphNode_StateMachine_0.SkydiveParachute.AnimStateTransitionNode_26.Transition
  nodes [Result(bCanEnterTransition=Break Fort Anim Input Skydiving(FortAnimInput_Skydiving=Get SkydivingState()) -> (
     ParaGilderRootOffset=~
     ParaGliderRootRotationOverride=~
     bUseParaGlideRootModifier=~
     bIsSkydivingFromLaunchPad=~
     bIsSkydivingFromBus=~
     bIsInVortex=~
     bIsDiving=~
     bIsParachuteOpen=~
     bIsSkydiving=~
     bIsParachuteLeaning=~
     bIsSkydiveLeaning=~
     bIsLeaning=~
     bIsSkydiveDiveMode=~
     bPlaySkydiveDrift=~
     bSkydiveDriftDelayActive=~
     bSkydiveDriftAnimAllowed=~
     bIsGliderRight=~
     bIsGliderCenter=~
     bIsGliderLeft=~
     bIsGliderForward=~
     bIsGliderBack=~
     bHasDeployedAndIsStandingUpright=~
     LocalAccelForward=~
     LocalAccelRight=~
     LocalVelocityRight=~
     SkydiveAimPitch=~
     SkydiveAimYaw=~
     DeployChuteAnimRate=~
     SkydiveDriftAnimRate=~
     SkydiveDriftAnimRateCurrent=~
     SkydiveFidgetAnimRate=~
     SkydiveFidgetAnimRateCurrent=~
     SkydiveAdditiveAlpha=~
     SkydiveDriftDelay=~
    ).bIsParachuteOpen)]
 }
 Transition {
  path=AnimGraph.AnimGraphNode_StateMachine_0.FullBody.AnimStateNode_34.SkyDive.AnimGraphNode_StateMachine_0.SkydiveParachute.AnimStateTransitionNode_27.Transition
  nodes [Result(bCanEnterTransition=Get bTransition_ParachuteToSkydiveGlide_Skydive() -> (bTransition_ParachuteToSkydiveGlide_Skydive:bool))]
 }
 Transition {
  path=AnimGraph.AnimGraphNode_StateMachine_0.FullBody.AnimStateNode_34.SkyDive.AnimGraphNode_StateMachine_0.SkydiveParachute.AnimStateTransitionNode_28.Transition
  nodes [Result()]
 }
 AnimationTransitionGraph_1 {
  path=AnimGraph.AnimGraphNode_StateMachine_0.FullBody.AnimStateNode_34.SkyDive.AnimGraphNode_StateMachine_0.SkydiveParachute.AnimStateTransitionNode_60.AnimationTransitionGraph_1
  nodes [Result()]
 }
 AnimationTransitionGraph_1 {
  path=AnimGraph.AnimGraphNode_StateMachine_0.FullBody.AnimStateNode_34.SkyDive.AnimGraphNode_StateMachine_0.SkydiveParachute.AnimStateTransitionNode_61.AnimationTransitionGraph_1
  nodes [
   n1 = Get SkydivingState()
   Result(bCanEnterTransition=NOT Boolean(A=Break Fort Anim Input Skydiving(FortAnimInput_Skydiving=@n1) -> (
      ParaGilderRootOffset=~
      ParaGliderRootRotationOverride=~
      bUseParaGlideRootModifier=~
      bIsSkydivingFromLaunchPad=~
      bIsSkydivingFromBus=~
      bIsInVortex=~
      bIsDiving=~
      bIsParachuteOpen=~
      bIsSkydiving=~
      bIsParachuteLeaning=~
      bIsSkydiveLeaning=~
      bIsLeaning=~
      bIsSkydiveDiveMode=~
      bPlaySkydiveDrift=~
      bSkydiveDriftDelayActive=~
      bSkydiveDriftAnimAllowed=~
      bIsGliderRight=~
      bIsGliderCenter=~
      bIsGliderLeft=~
      bIsGliderForward=~
      bIsGliderBack=~
      bHasDeployedAndIsStandingUpright=~
      LocalAccelForward=~
      LocalAccelRight=~
      LocalVelocityRight=~
      SkydiveAimPitch=~
      SkydiveAimYaw=~
      DeployChuteAnimRate=~
      SkydiveDriftAnimRate=~
      SkydiveDriftAnimRateCurrent=~
      SkydiveFidgetAnimRate=~
      SkydiveFidgetAnimRateCurrent=~
      SkydiveAdditiveAlpha=~
      SkydiveDriftDelay=~
     ).bIsParachuteOpen))
  ]
 }
 StartJump {
  path=AnimGraph.AnimGraphNode_StateMachine_0.FullBody.AnimStateNode_4.StartJump
  nodes [
   Layered blend per bone(BlendPoses_0=Local Space Ref Pose(), BlendWeights_0="To Float (Boolean)"(InBool=Get bIsFox())) {layerSetup [{branchFilters [{boneName=faceAttach}]}]}
   n1 = Blend Poses by int!(
    BlendPose_0=Sequence Player 'JumpUp_01'() {sequence=/Game/Animation/Game/Pets/Chameleon/Backpack/Jump/JumpUp_01}
    BlendPose_1=Sequence Player 'JumpUp_02'() {sequence=/Game/Animation/Game/Pets/Chameleon/Backpack/Jump/JumpUp_02}
    BlendTime_0=0.1
    BlendTime_1=0.1
    ActiveChildIndex=Get JumpIndex() -> (JumpIndex:int=0)
   ) -> (Pose) {blendType=Linear, blendPose [{linkId=-1, sourceLinkId=-1}, {linkId=-1, sourceLinkId=-1}], blendTime [0.1,0.1]}
   Output Animation Pose(Result=Blend Poses by bool(
     BlendPose_0=@n1
     BlendPose_1=Sequence Player 'JumpUp_00'() {sequence=/Game/Animation/Game/Pets/Chameleon/Backpack/Jump/JumpUp_00}
     bActiveValue=Get bIsMovingDuringJumpInitiated() -> (bIsMovingDuringJumpInitiated:bool)
    ) {blendType=Linear})
  ]
 }
 JumpLoop {
  path=AnimGraph.AnimGraphNode_StateMachine_0.FullBody.AnimStateNode_5.JumpLoop
  nodes [
   Layered blend per bone(BlendPoses_0=Local Space Ref Pose(), BlendWeights_0="To Float (Boolean)"(InBool=Get bIsFox())) {layerSetup [{branchFilters [{boneName=faceAttach}]}]}
   n1 = Blend Poses by int(
    BlendPose_0=Sequence Player 'Jump_ApexLoop_01'() {sequence=/Game/Animation/Game/Pets/Chameleon/Backpack/Jump/Jump_ApexLoop_01}
    BlendPose_1=Sequence Player 'Jump_ApexLoop_02'() {sequence=/Game/Animation/Game/Pets/Chameleon/Backpack/Jump/Jump_ApexLoop_02}
    BlendTime_0=0.1
    BlendTime_1=0.1
    ActiveChildIndex=Get JumpIndex() -> (JumpIndex=0)
   ) {blendType=Linear, blendPose [{linkId=-1, sourceLinkId=-1}, {linkId=-1, sourceLinkId=-1}], blendTime [0.1,0.1]}
   Output Animation Pose(Result=Blend Poses by bool(
     BlendPose_0=@n1
     BlendPose_1=Sequence Player 'Jump_ApexLoop_00'() {sequence=/Game/Animation/Game/Pets/Chameleon/Backpack/Jump/Jump_ApexLoop_00}
     bActiveValue=Get bIsMovingDuringJumpInitiated()
    ) {blendType=Linear})
  ]
 }
 JumpFall {
  path=AnimGraph.AnimGraphNode_StateMachine_0.FullBody.AnimStateNode_6.JumpFall
  nodes [
   Layered blend per bone(BlendPoses_0=Local Space Ref Pose(), BlendWeights_0="To Float (Boolean)"(InBool=Get bIsFox())) {layerSetup [{branchFilters [{boneName=faceAttach}]}]}
   n1 = Blend Poses by int(
    BlendPose_0=Sequence Player 'JumpFall_01'() {bLoopAnimation=false, sequence=/Game/Animation/Game/Pets/Chameleon/Backpack/Jump/JumpFall_01}
    BlendPose_1=Sequence Player 'JumpFall_02'() {bLoopAnimation=false, sequence=/Game/Animation/Game/Pets/Chameleon/Backpack/Jump/JumpFall_02}
    BlendTime_0=0.1
    BlendTime_1=0.1
    ActiveChildIndex=Get JumpIndex() -> (JumpIndex=0)
   ) {blendType=Linear, blendPose [{linkId=-1, sourceLinkId=-1}, {linkId=-1, sourceLinkId=-1}], blendTime [0.1,0.1]}
   Output Animation Pose(Result=Blend Poses by bool(
     BlendPose_0=@n1
     BlendPose_1=Sequence Player 'JumpFall_00'() {bLoopAnimation=false, sequence=/Game/Animation/Game/Pets/Chameleon/Backpack/Jump/JumpFall_00}
     bActiveValue=Get bIsMovingDuringJumpInitiated()
    ) {blendType=Linear})
  ]
 }
 JumpLand {
  path=AnimGraph.AnimGraphNode_StateMachine_0.FullBody.AnimStateNode_7.JumpLand
  nodes [
   Layered blend per bone(BlendPoses_0=Local Space Ref Pose(), BlendWeights_0="To Float (Boolean)"(InBool=Get bIsFox())) {layerSetup [{branchFilters [{boneName=faceAttach}]}]}
   n1 = Blend Poses by int(
    BlendPose_0=Sequence Player 'JumpLand_01'!(bLoopAnimation:bool=False) -> (Pose) {bLoopAnimation=false, sequence=/Game/Animation/Game/Pets/Chameleon/Backpack/Jump/JumpLand_01}
    BlendPose_1=Sequence Player 'JumpLand_02'!(bLoopAnimation:bool=False) -> (Pose) {bLoopAnimation=false, sequence=/Game/Animation/Game/Pets/Chameleon/Backpack/Jump/JumpLand_02}
    BlendTime_0=0.1
    BlendTime_1=0.1
    ActiveChildIndex=Get JumpIndex() -> (JumpIndex=0)
   ) {blendType=Linear, blendPose [{linkId=-1, sourceLinkId=-1}, {linkId=-1, sourceLinkId=-1}], blendTime [0.1,0.1]}
   Output Animation Pose(Result=Blend Poses by bool(
     BlendPose_0=@n1
     BlendPose_1=Sequence Player 'JumpLand_00'!(bLoopAnimation:bool=False) -> (Pose) {bLoopAnimation=false, sequence=/Game/Animation/Game/Pets/Chameleon/Backpack/Jump/JumpLand_00}
     bActiveValue=Get bIsMovingDuringJumpInitiated()
    ) {blendType=Linear})
  ]
 }
 CrouchToLocomotion {
  path=AnimGraph.AnimGraphNode_StateMachine_0.FullBody.AnimStateNode_8.CrouchToLocomotion
  nodes [Output Animation Pose(Result=Sequence Player 'CrouchIdleToIdle'() {bLoopAnimation=false, sequence=&CrouchIdleToIdle})]
 }
 DBNO_old {
  path=AnimGraph.AnimGraphNode_StateMachine_0.FullBody.AnimStateNode_9.DBNO_old
  nodes [
   AnimGraphNode_StateMachine_0 = DBNO()
   Output Animation Pose(Result=@AnimGraphNode_StateMachine_0)
  ]
 }
 DBNO {
  entry=DNBO
  path=AnimGraph.AnimGraphNode_StateMachine_0.FullBody.AnimStateNode_9.DBNO_old.AnimGraphNode_StateMachine_0.DBNO
  ids=id
  nodes [
   AnimStateNode_0 = DNBO() {is=title AnimStateNode}
   AnimStateNode_1 = ToDBNOHunker() {is=title AnimStateNode}
   AnimStateNode_2 = DBNOHunker() {is=title AnimStateNode}
   AnimStateNode_3 = ToDBNO() {is=title AnimStateNode}
   AnimStateTransitionNode_0 = AnimStateTransitionNode()
   AnimStateTransitionNode_1 = AnimStateTransitionNode() {bAutomaticRuleBasedOnSequencePlayerInState=true, crossfadeDuration=0}
   AnimStateTransitionNode_2 = AnimStateTransitionNode()
   AnimStateTransitionNode_3 = AnimStateTransitionNode() {bAutomaticRuleBasedOnSequencePlayerInState=true, crossfadeDuration=0}
  ]
  links [
   {
    dst=ToDBNO
    rule=AnimGraph.AnimGraphNode_StateMachine_0.FullBody.AnimStateNode_9.DBNO_old.AnimGraphNode_StateMachine_0.DBNO.AnimStateTransitionNode_2.Transition
    src=DBNOHunker
    via=AnimStateTransitionNode_2
   }
   {
    dst=ToDBNOHunker
    rule=AnimGraph.AnimGraphNode_StateMachine_0.FullBody.AnimStateNode_9.DBNO_old.AnimGraphNode_StateMachine_0.DBNO.AnimStateTransitionNode_0.Transition
    src=DNBO
    via=AnimStateTransitionNode_0
   }
   {
    dst=DNBO
    rule=AnimGraph.AnimGraphNode_StateMachine_0.FullBody.AnimStateNode_9.DBNO_old.AnimGraphNode_StateMachine_0.DBNO.AnimStateTransitionNode_3.Transition
    src=ToDBNO
    via=AnimStateTransitionNode_3
   }
   {
    dst=DBNOHunker
    rule=AnimGraph.AnimGraphNode_StateMachine_0.FullBody.AnimStateNode_9.DBNO_old.AnimGraphNode_StateMachine_0.DBNO.AnimStateTransitionNode_1.Transition
    src=ToDBNOHunker
    via=AnimStateTransitionNode_1
   }
  ]
 }
 DNBO {
  path=AnimGraph.AnimGraphNode_StateMachine_0.FullBody.AnimStateNode_9.DBNO_old.AnimGraphNode_StateMachine_0.DBNO.AnimStateNode_0.DNBO
  nodes [Output Animation Pose(Result=Sequence Player 'DBNO_Idle'() {sequence=&DBNO_Idle})]
 }
 ToDBNOHunker {
  path=AnimGraph.AnimGraphNode_StateMachine_0.FullBody.AnimStateNode_9.DBNO_old.AnimGraphNode_StateMachine_0.DBNO.AnimStateNode_1.ToDBNOHunker
  nodes [Output Animation Pose(Result=Sequence Player 'DBNOToHunker'() {bLoopAnimation=false, sequence=&DBNOToHunker})]
 }
 DBNOHunker {
  path=AnimGraph.AnimGraphNode_StateMachine_0.FullBody.AnimStateNode_9.DBNO_old.AnimGraphNode_StateMachine_0.DBNO.AnimStateNode_2.DBNOHunker
  nodes [Output Animation Pose(Result=Sequence Player 'DBNO_Hunker'() {sequence=&DBNO_Hunker})]
 }
 ToDBNO {
  path=AnimGraph.AnimGraphNode_StateMachine_0.FullBody.AnimStateNode_9.DBNO_old.AnimGraphNode_StateMachine_0.DBNO.AnimStateNode_3.ToDBNO
  nodes [Output Animation Pose(Result=Sequence Player 'HunkerToDBNO'() {bLoopAnimation=false, sequence=&HunkerToDBNO})]
 }
 Transition {
  path=AnimGraph.AnimGraphNode_StateMachine_0.FullBody.AnimStateNode_9.DBNO_old.AnimGraphNode_StateMachine_0.DBNO.AnimStateTransitionNode_0.Transition
  nodes [Result(bCanEnterTransition=Get bIsMoving2D())]
 }
 Transition {
  path=AnimGraph.AnimGraphNode_StateMachine_0.FullBody.AnimStateNode_9.DBNO_old.AnimGraphNode_StateMachine_0.DBNO.AnimStateTransitionNode_1.Transition
  nodes [Result()]
 }
 Transition {
  path=AnimGraph.AnimGraphNode_StateMachine_0.FullBody.AnimStateNode_9.DBNO_old.AnimGraphNode_StateMachine_0.DBNO.AnimStateTransitionNode_2.Transition
  nodes [Result(bCanEnterTransition=NOT Boolean(A=Get bIsMoving2D()))]
 }
 Transition {
  path=AnimGraph.AnimGraphNode_StateMachine_0.FullBody.AnimStateNode_9.DBNO_old.AnimGraphNode_StateMachine_0.DBNO.AnimStateTransitionNode_3.Transition
  nodes [Result()]
 }
 Transition {
  path=AnimGraph.AnimGraphNode_StateMachine_0.FullBody.AnimStateTransitionNode_0.Transition
  nodes [Result(bCanEnterTransition=Get bIsCrouching())]
 }
 AnimationTransitionGraph_0 {
  path=AnimGraph.AnimGraphNode_StateMachine_0.FullBody.AnimStateTransitionNode_1.AnimationTransitionGraph_0
  nodes [Result(bCanEnterTransition=NOT Boolean(A=Get bIsFrontEndLobbyPartyMember() -> (bIsFrontEndLobbyPartyMember:bool)))]
 }
 Transition {
  path=AnimGraph.AnimGraphNode_StateMachine_0.FullBody.AnimStateTransitionNode_10.Transition
  nodes [Result(bCanEnterTransition=Get bIsSurfaceSwimming())]
 }
 Transition {path=AnimGraph.AnimGraphNode_StateMachine_0.FullBody.AnimStateTransitionNode_100.Transition, nodes [Result()]}
 Transition {
  path=AnimGraph.AnimGraphNode_StateMachine_0.FullBody.AnimStateTransitionNode_101.Transition
  nodes [Result(bCanEnterTransition=NOT Boolean(A=Get bIsTargeting()))]
 }
 Transition {
  path=AnimGraph.AnimGraphNode_StateMachine_0.FullBody.AnimStateTransitionNode_102.Transition
  nodes [Result(bCanEnterTransition=NOT Boolean(A=Get bIsTargeting()))]
 }
 Transition {
  path=AnimGraph.AnimGraphNode_StateMachine_0.FullBody.AnimStateTransitionNode_103.Transition
  nodes [Result(bCanEnterTransition=Get bIsHunker() -> (bIsHunker:bool))]
 }
 Transition {
  path=AnimGraph.AnimGraphNode_StateMachine_0.FullBody.AnimStateTransitionNode_104.Transition
  nodes [Result(bCanEnterTransition=Get bIsInAirNotJumping() -> (bIsInAirNotJumping:bool))]
 }
 AnimationTransitionGraph_2 {
  path=AnimGraph.AnimGraphNode_StateMachine_0.FullBody.AnimStateTransitionNode_105.AnimationTransitionGraph_2
  nodes [Result(bCanEnterTransition=NOT Boolean(A=Get bIsDBNO()))]
 }
 Transition {
  path=AnimGraph.AnimGraphNode_StateMachine_0.FullBody.AnimStateTransitionNode_106.Transition
  nodes [Result(bCanEnterTransition=Get bIsDBNO())]
 }
 Transition {
  path=AnimGraph.AnimGraphNode_StateMachine_0.FullBody.AnimStateTransitionNode_108.Transition
  nodes [Result(bCanEnterTransition=Get bIsDBNO())]
 }
 Transition {
  path=AnimGraph.AnimGraphNode_StateMachine_0.FullBody.AnimStateTransitionNode_109.Transition
  nodes [Result(bCanEnterTransition=NOT Boolean(A=Get bIsDBNO()))]
 }
 AnimationTransitionGraph_0 {
  path=AnimGraph.AnimGraphNode_StateMachine_0.FullBody.AnimStateTransitionNode_11.AnimationTransitionGraph_0
  nodes [Result(bCanEnterTransition=Get bIsOnGround() -> (bIsOnGround:bool))]
 }
 Transition {path=AnimGraph.AnimGraphNode_StateMachine_0.FullBody.AnimStateTransitionNode_110.Transition, nodes [Result()]}
 AnimationTransitionGraph_2 {path=AnimGraph.AnimGraphNode_StateMachine_0.FullBody.AnimStateTransitionNode_12.AnimationTransitionGraph_2, nodes [Result()]}
 AnimationTransitionGraph_0 {
  path=AnimGraph.AnimGraphNode_StateMachine_0.FullBody.AnimStateTransitionNode_13.AnimationTransitionGraph_0
  nodes [Result(bCanEnterTransition=Get bIsSlopeSliding() -> (bIsSlopeSliding:bool))]
 }
 Transition {
  path=AnimGraph.AnimGraphNode_StateMachine_0.FullBody.AnimStateTransitionNode_14.Transition
  nodes [Result(bCanEnterTransition=Get bIsSurfaceSwimming())]
 }
 AnimationTransitionGraph_0 {
  path=AnimGraph.AnimGraphNode_StateMachine_0.FullBody.AnimStateTransitionNode_16.AnimationTransitionGraph_0
  nodes [Result(bCanEnterTransition=Get bLandingPredicted() -> (bLandingPredicted:bool))]
 }
 AnimationTransitionGraph_0 {
  path=AnimGraph.AnimGraphNode_StateMachine_0.FullBody.AnimStateTransitionNode_17.AnimationTransitionGraph_0
  nodes [Result(bCanEnterTransition=Get bIsOnGround())]
 }
 AnimationTransitionGraph_0 {
  path=AnimGraph.AnimGraphNode_StateMachine_0.FullBody.AnimStateTransitionNode_18.AnimationTransitionGraph_0
  nodes [Result(bCanEnterTransition=Get bIsSlopeSliding())]
 }
 AnimationTransitionGraph_0 {path=AnimGraph.AnimGraphNode_StateMachine_0.FullBody.AnimStateTransitionNode_19.AnimationTransitionGraph_0, nodes [Result()]}
 AnimationTransitionGraph_1 {
  path=AnimGraph.AnimGraphNode_StateMachine_0.FullBody.AnimStateTransitionNode_2.AnimationTransitionGraph_1
  nodes [Result(bCanEnterTransition=Get bTransition_FrontEndSkydive_IdleLobby() -> (bTransition_FrontEndSkydive_IdleLobby:bool))]
 }
 Transition {
  path=AnimGraph.AnimGraphNode_StateMachine_0.FullBody.AnimStateTransitionNode_20.Transition
  nodes [Result(bCanEnterTransition=Get bTransition_DownSights_DownSightsToIdle() -> (bTransition_DownSights_DownSightsToIdle:bool))]
 }
 Transition {
  path=AnimGraph.AnimGraphNode_StateMachine_0.FullBody.AnimStateTransitionNode_21.Transition
  nodes [Result(bCanEnterTransition=NOT Boolean(A=Get bIsMoving2D()))]
 }
 Transition {
  path=AnimGraph.AnimGraphNode_StateMachine_0.FullBody.AnimStateTransitionNode_22.Transition
  nodes [Result(bCanEnterTransition=Get bTransition_Hunker_HunkerToCrouch() -> (bTransition_Hunker_HunkerToCrouch:bool))]
 }
 Transition {
  path=AnimGraph.AnimGraphNode_StateMachine_0.FullBody.AnimStateTransitionNode_23.Transition
  nodes [Result(bCanEnterTransition=Get bTransition_Locomotion_To_LocomotionToIdle() -> (bTransition_Locomotion_To_LocomotionToIdle:bool))]
 }
 Transition {
  path=AnimGraph.AnimGraphNode_StateMachine_0.FullBody.AnimStateTransitionNode_24.Transition
  nodes [Result(bCanEnterTransition=Get bIsFalling())]
 }
 Transition {path=AnimGraph.AnimGraphNode_StateMachine_0.FullBody.AnimStateTransitionNode_25.Transition, nodes [Result()]}
 Transition {
  path=AnimGraph.AnimGraphNode_StateMachine_0.FullBody.AnimStateTransitionNode_26.Transition
  nodes [Result(bCanEnterTransition=NOT Boolean(A=Get bIsMoving2D()))]
 }
 Transition {path=AnimGraph.AnimGraphNode_StateMachine_0.FullBody.AnimStateTransitionNode_27.Transition, nodes [Result()]}
 Transition {path=AnimGraph.AnimGraphNode_StateMachine_0.FullBody.AnimStateTransitionNode_28.Transition, nodes [Result()]}
 Transition {
  path=AnimGraph.AnimGraphNode_StateMachine_0.FullBody.AnimStateTransitionNode_29.Transition
  nodes [Result(bCanEnterTransition=Get bTransition_Crouching_CrouchToIdle() -> (bTransition_Crouching_CrouchToIdle:bool))]
 }
 Transition {path=AnimGraph.AnimGraphNode_StateMachine_0.FullBody.AnimStateTransitionNode_3.Transition, nodes [Result(bCanEnterTransition=true)]}
 Transition {
  path=AnimGraph.AnimGraphNode_StateMachine_0.FullBody.AnimStateTransitionNode_30.Transition
  nodes [Result(bCanEnterTransition=Get bIsJumping())]
 }
 AnimationTransitionGraph_0 {
  path=AnimGraph.AnimGraphNode_StateMachine_0.FullBody.AnimStateTransitionNode_31.AnimationTransitionGraph_0
  nodes [Result(bCanEnterTransition=Get bTransition_LobbyHunker_HunkerToLobby() -> (bTransition_LobbyHunker_HunkerToLobby:bool))]
 }
 Transition {
  path=AnimGraph.AnimGraphNode_StateMachine_0.FullBody.AnimStateTransitionNode_32.Transition
  nodes [
   Result(bCanEnterTransition=Get BlockAnimRule() -> (BlockAnimRule:bool))
   Get bIsCrouchMoving() -> (bIsCrouchMoving:bool)
  ]
 }
 AnimationTransitionGraph_0 {path=AnimGraph.AnimGraphNode_StateMachine_0.FullBody.AnimStateTransitionNode_33.AnimationTransitionGraph_0, nodes [Result()]}
 Transition {
  path=AnimGraph.AnimGraphNode_StateMachine_0.FullBody.AnimStateTransitionNode_34.Transition
  nodes [Result(bCanEnterTransition=Get bIsHunker())]
 }
 Transition {
  path=AnimGraph.AnimGraphNode_StateMachine_0.FullBody.AnimStateTransitionNode_35.Transition
  nodes [Result(bCanEnterTransition=Get bIsJumping())]
 }
 Transition {
  path=AnimGraph.AnimGraphNode_StateMachine_0.FullBody.AnimStateTransitionNode_36.Transition
  nodes [Result(bCanEnterTransition=Get bIsTargeting())]
 }
 Transition {
  path=AnimGraph.AnimGraphNode_StateMachine_0.FullBody.AnimStateTransitionNode_37.Transition
  nodes [Result(bCanEnterTransition=Get bIsTargeting())]
 }
 Transition {
  path=AnimGraph.AnimGraphNode_StateMachine_0.FullBody.AnimStateTransitionNode_38.Transition
  nodes [Result(bCanEnterTransition=Get bIsInAirNotJumping())]
 }
 Transition {
  path=AnimGraph.AnimGraphNode_StateMachine_0.FullBody.AnimStateTransitionNode_39.Transition
  nodes [Result(bCanEnterTransition=Get bTransition_Crouching_CrouchToCrouchADSNotMoving() -> (bTransition_Crouching_CrouchToCrouchADSNotMoving:bool))]
 }
 Transition {
  path=AnimGraph.AnimGraphNode_StateMachine_0.FullBody.AnimStateTransitionNode_4.Transition
  nodes [Result(bCanEnterTransition=Get bIsJumping())]
 }
 AnimationTransitionGraph_0 {
  path=AnimGraph.AnimGraphNode_StateMachine_0.FullBody.AnimStateTransitionNode_40.AnimationTransitionGraph_0
  nodes [Result(bCanEnterTransition=Get bIsCrouchMoving())]
 }
 AnimationTransitionGraph_0 {
  path=AnimGraph.AnimGraphNode_StateMachine_0.FullBody.AnimStateTransitionNode_41.AnimationTransitionGraph_0
  nodes [Result(bCanEnterTransition=Get bTransition_CrouchWalkingDownSights_DownSights() -> (bTransition_CrouchWalkingDownSights_DownSights:bool))]
 }
 Transition {
  path=AnimGraph.AnimGraphNode_StateMachine_0.FullBody.AnimStateTransitionNode_42.Transition
  nodes [Result(bCanEnterTransition=Get bIsTargeting())]
 }
 Transition {
  path=AnimGraph.AnimGraphNode_StateMachine_0.FullBody.AnimStateTransitionNode_43.Transition
  nodes [Result(bCanEnterTransition=NOT Boolean(A=Get bIsCrouching()))]
 }
 Transition {
  path=AnimGraph.AnimGraphNode_StateMachine_0.FullBody.AnimStateTransitionNode_44.Transition
  nodes [Result(bCanEnterTransition=Get bTransition_Crouch_CrouchToLocomotion() -> (bTransition_Crouch_CrouchToLocomotion:bool))]
 }
 Transition {
  path=AnimGraph.AnimGraphNode_StateMachine_0.FullBody.AnimStateTransitionNode_45.Transition
  nodes [Result(bCanEnterTransition=Get bIsTargeting())]
 }
 Transition {
  path=AnimGraph.AnimGraphNode_StateMachine_0.FullBody.AnimStateTransitionNode_46.Transition
  nodes [Result(bCanEnterTransition=Get bIsInAirNotJumping())]
 }
 Transition {
  path=AnimGraph.AnimGraphNode_StateMachine_0.FullBody.AnimStateTransitionNode_47.Transition
  nodes [Result(bCanEnterTransition=Get bTransition_Idle_IdleLobby() -> (bTransition_Idle_IdleLobby:bool))]
 }
 Transition {
  path=AnimGraph.AnimGraphNode_StateMachine_0.FullBody.AnimStateTransitionNode_48.Transition
  nodes [Result(bCanEnterTransition=Get bTransition_JumpFall_To_Skydive() -> (bTransition_JumpFall_To_Skydive:bool))]
 }
 Transition {
  path=AnimGraph.AnimGraphNode_StateMachine_0.FullBody.AnimStateTransitionNode_49.Transition
  nodes [Result(bCanEnterTransition=Get bIsJumping())]
 }
 AnimationTransitionGraph_0 {path=AnimGraph.AnimGraphNode_StateMachine_0.FullBody.AnimStateTransitionNode_5.AnimationTransitionGraph_0, nodes [Result()]}
 Transition {
  path=AnimGraph.AnimGraphNode_StateMachine_0.FullBody.AnimStateTransitionNode_50.Transition
  nodes [Result(bCanEnterTransition=Get bTransition_DownSightsToIdle_To_Locomotion() -> (bTransition_DownSightsToIdle_To_Locomotion:bool))]
 }
 Transition {
  path=AnimGraph.AnimGraphNode_StateMachine_0.FullBody.AnimStateTransitionNode_51.Transition
  nodes [Result(bCanEnterTransition=Get bIsHunker())]
 }
 Transition {
  path=AnimGraph.AnimGraphNode_StateMachine_0.FullBody.AnimStateTransitionNode_52.Transition
  nodes [Result(bCanEnterTransition=Get bIsFalling())]
 }
 Transition {
  path=AnimGraph.AnimGraphNode_StateMachine_0.FullBody.AnimStateTransitionNode_53.Transition
  nodes [Result(bCanEnterTransition=Get bIsSlopeSliding())]
 }
 Transition {path=AnimGraph.AnimGraphNode_StateMachine_0.FullBody.AnimStateTransitionNode_54.Transition, nodes [Result()]}
 Transition {path=AnimGraph.AnimGraphNode_StateMachine_0.FullBody.AnimStateTransitionNode_55.Transition, nodes [Result()]}
 Transition {
  path=AnimGraph.AnimGraphNode_StateMachine_0.FullBody.AnimStateTransitionNode_56.Transition
  nodes [Result(bCanEnterTransition=Get bIsCrouching())]
 }
 Transition {path=AnimGraph.AnimGraphNode_StateMachine_0.FullBody.AnimStateTransitionNode_57.Transition, nodes [Result()]}
 Transition {
  path=AnimGraph.AnimGraphNode_StateMachine_0.FullBody.AnimStateTransitionNode_58.Transition
  nodes [Result(bCanEnterTransition=Get bIsHunker())]
 }
 Transition {
  path=AnimGraph.AnimGraphNode_StateMachine_0.FullBody.AnimStateTransitionNode_59.Transition
  nodes [Result(bCanEnterTransition=Get bIsFalling())]
 }
 AnimationTransitionGraph_0 {
  path=AnimGraph.AnimGraphNode_StateMachine_0.FullBody.AnimStateTransitionNode_6.AnimationTransitionGraph_0
  nodes [Result(bCanEnterTransition=Get bLandingPredicted())]
 }
 Transition {
  path=AnimGraph.AnimGraphNode_StateMachine_0.FullBody.AnimStateTransitionNode_60.Transition
  nodes [Result(bCanEnterTransition=Get bTransition_IdleLobby_Idle() -> (bTransition_IdleLobby_Idle:bool))]
 }
 Transition {
  path=AnimGraph.AnimGraphNode_StateMachine_0.FullBody.AnimStateTransitionNode_61.Transition
  nodes [Result(bCanEnterTransition=Get bTransition_Hunker_Idle() -> (bTransition_Hunker_Idle:bool))]
 }
 Transition {
  path=AnimGraph.AnimGraphNode_StateMachine_0.FullBody.AnimStateTransitionNode_62.Transition
  nodes [Result(bCanEnterTransition=Get bIsJumping())]
 }
 Transition {
  path=AnimGraph.AnimGraphNode_StateMachine_0.FullBody.AnimStateTransitionNode_63.Transition
  nodes [Result(bCanEnterTransition=Get bTransition_JumpLand_To_Locomotion() -> (bTransition_JumpLand_To_Locomotion:bool))]
 }
 AnimationTransitionGraph_0 {
  path=AnimGraph.AnimGraphNode_StateMachine_0.FullBody.AnimStateTransitionNode_64.AnimationTransitionGraph_0
  nodes [Result(bCanEnterTransition=Get bTransition_FrontEndSkydive_LobbyHunker() -> (bTransition_FrontEndSkydive_LobbyHunker:bool))]
 }
 AnimationTransitionGraph_0 {
  path=AnimGraph.AnimGraphNode_StateMachine_0.FullBody.AnimStateTransitionNode_65.AnimationTransitionGraph_0
  nodes [Result(bCanEnterTransition=Get bEmotePlaying_WhereHunkeringIsBest())]
 }
 AnimationTransitionGraph_0 {path=AnimGraph.AnimGraphNode_StateMachine_0.FullBody.AnimStateTransitionNode_66.AnimationTransitionGraph_0, nodes [Result()]}
 AnimationTransitionGraph_0 {
  path=AnimGraph.AnimGraphNode_StateMachine_0.FullBody.AnimStateTransitionNode_67.AnimationTransitionGraph_0
  nodes [Result(bCanEnterTransition=Break Fort Anim Input Skydiving(FortAnimInput_Skydiving=Get SkydivingState()) -> (
     ParaGilderRootOffset=~
     ParaGliderRootRotationOverride=~
     bUseParaGlideRootModifier=~
     bIsSkydivingFromLaunchPad=~
     bIsSkydivingFromBus=~
     bIsInVortex=~
     bIsDiving=~
     bIsParachuteOpen=~
     bIsSkydiving=~
     bIsParachuteLeaning=~
     bIsSkydiveLeaning=~
     bIsLeaning=~
     bIsSkydiveDiveMode=~
     bPlaySkydiveDrift=~
     bSkydiveDriftDelayActive=~
     bSkydiveDriftAnimAllowed=~
     bIsGliderRight=~
     bIsGliderCenter=~
     bIsGliderLeft=~
     bIsGliderForward=~
     bIsGliderBack=~
     bHasDeployedAndIsStandingUpright=~
     LocalAccelForward=~
     LocalAccelRight=~
     LocalVelocityRight=~
     SkydiveAimPitch=~
     SkydiveAimYaw=~
     DeployChuteAnimRate=~
     SkydiveDriftAnimRate=~
     SkydiveDriftAnimRateCurrent=~
     SkydiveFidgetAnimRate=~
     SkydiveFidgetAnimRateCurrent=~
     SkydiveAdditiveAlpha=~
     SkydiveDriftDelay=~
    ).bIsSkydiving)]
 }
 AnimationTransitionGraph_0 {
  path=AnimGraph.AnimGraphNode_StateMachine_0.FullBody.AnimStateTransitionNode_68.AnimationTransitionGraph_0
  nodes [Result(bCanEnterTransition=true)]
 }
 AnimationTransitionGraph_0 {
  path=AnimGraph.AnimGraphNode_StateMachine_0.FullBody.AnimStateTransitionNode_69.AnimationTransitionGraph_0
  nodes [Result(bCanEnterTransition=true)]
 }
 AnimationTransitionGraph_0 {
  path=AnimGraph.AnimGraphNode_StateMachine_0.FullBody.AnimStateTransitionNode_7.AnimationTransitionGraph_0
  nodes [Result(bCanEnterTransition=Get bIsOnGround())]
 }
 Transition {
  path=AnimGraph.AnimGraphNode_StateMachine_0.FullBody.AnimStateTransitionNode_70.Transition
  nodes [Result(bCanEnterTransition=Get bTransition_Hunker_Locomotion() -> (bTransition_Hunker_Locomotion:bool))]
 }
 AnimationTransitionGraph_0 {
  path=AnimGraph.AnimGraphNode_StateMachine_0.FullBody.AnimStateTransitionNode_71.AnimationTransitionGraph_0
  nodes [Result(bCanEnterTransition=true)]
 }
 Transition {
  path=AnimGraph.AnimGraphNode_StateMachine_0.FullBody.AnimStateTransitionNode_72.Transition
  nodes [Result(bCanEnterTransition=Break Fort Anim Input Skydiving(FortAnimInput_Skydiving=Get SkydivingState()) -> (
     ParaGilderRootOffset=~
     ParaGliderRootRotationOverride=~
     bUseParaGlideRootModifier=~
     bIsSkydivingFromLaunchPad=~
     bIsSkydivingFromBus=~
     bIsInVortex=~
     bIsDiving=~
     bIsParachuteOpen=~
     bIsSkydiving=~
     bIsParachuteLeaning=~
     bIsSkydiveLeaning=~
     bIsLeaning=~
     bIsSkydiveDiveMode=~
     bPlaySkydiveDrift=~
     bSkydiveDriftDelayActive=~
     bSkydiveDriftAnimAllowed=~
     bIsGliderRight=~
     bIsGliderCenter=~
     bIsGliderLeft=~
     bIsGliderForward=~
     bIsGliderBack=~
     bHasDeployedAndIsStandingUpright=~
     LocalAccelForward=~
     LocalAccelRight=~
     LocalVelocityRight=~
     SkydiveAimPitch=~
     SkydiveAimYaw=~
     DeployChuteAnimRate=~
     SkydiveDriftAnimRate=~
     SkydiveDriftAnimRateCurrent=~
     SkydiveFidgetAnimRate=~
     SkydiveFidgetAnimRateCurrent=~
     SkydiveAdditiveAlpha=~
     SkydiveDriftDelay=~
    ).bIsSkydiving)]
 }
 AnimationTransitionGraph_0 {
  path=AnimGraph.AnimGraphNode_StateMachine_0.FullBody.AnimStateTransitionNode_73.AnimationTransitionGraph_0
  nodes [Result(bCanEnterTransition=true)]
 }
 AnimationTransitionGraph_1 {
  path=AnimGraph.AnimGraphNode_StateMachine_0.FullBody.AnimStateTransitionNode_74.AnimationTransitionGraph_1
  nodes [Result(bCanEnterTransition=Get bIsSkydivingInGame() -> (bIsSkydivingInGame:bool))]
 }
 Transition {
  path=AnimGraph.AnimGraphNode_StateMachine_0.FullBody.AnimStateTransitionNode_75.Transition
  nodes [Result(bCanEnterTransition=Get bIsCrouching())]
 }
 Transition {
  path=AnimGraph.AnimGraphNode_StateMachine_0.FullBody.AnimStateTransitionNode_76.Transition
  nodes [Result(bCanEnterTransition=Get bIsInAirNotJumping())]
 }
 Transition {path=AnimGraph.AnimGraphNode_StateMachine_0.FullBody.AnimStateTransitionNode_77.Transition, nodes [Result()]}
 Transition {
  path=AnimGraph.AnimGraphNode_StateMachine_0.FullBody.AnimStateTransitionNode_78.Transition
  nodes [Result(bCanEnterTransition=NOT Boolean(A=Get bIsCrouching()))]
 }
 Transition {
  path=AnimGraph.AnimGraphNode_StateMachine_0.FullBody.AnimStateTransitionNode_79.Transition
  nodes [Result(bCanEnterTransition=Get bTransition_SkyDive_Idle() -> (bTransition_SkyDive_Idle:bool))]
 }
 Transition {
  path=AnimGraph.AnimGraphNode_StateMachine_0.FullBody.AnimStateTransitionNode_8.Transition
  nodes [Result(bCanEnterTransition=Get bIsSurfaceSwimming())]
 }
 Transition {
  path=AnimGraph.AnimGraphNode_StateMachine_0.FullBody.AnimStateTransitionNode_80.Transition
  nodes [Result(bCanEnterTransition=Get bTransition_Idle_Locomotion() -> (bTransition_Idle_Locomotion:bool))]
 }
 Transition {
  path=AnimGraph.AnimGraphNode_StateMachine_0.FullBody.AnimStateTransitionNode_81.Transition
  nodes [Result(bCanEnterTransition=Get bIsInAirNotJumping())]
 }
 Transition {
  path=AnimGraph.AnimGraphNode_StateMachine_0.FullBody.AnimStateTransitionNode_82.Transition
  nodes [Result(bCanEnterTransition=Get bTransition_Locomotion_StartJump() -> (bTransition_Locomotion_StartJump:bool))]
 }
 Transition {
  path=AnimGraph.AnimGraphNode_StateMachine_0.FullBody.AnimStateTransitionNode_83.Transition
  nodes [Result(bCanEnterTransition=NOT Boolean(A=Get bIsInAirNotJumping()))]
 }
 Transition {
  path=AnimGraph.AnimGraphNode_StateMachine_0.FullBody.AnimStateTransitionNode_84.Transition
  nodes [Result(bCanEnterTransition=Get bIsTargeting())]
 }
 Transition {
  path=AnimGraph.AnimGraphNode_StateMachine_0.FullBody.AnimStateTransitionNode_85.Transition
  nodes [Result(bCanEnterTransition=Get bIsCrouching())]
 }
 Transition {
  path=AnimGraph.AnimGraphNode_StateMachine_0.FullBody.AnimStateTransitionNode_86.Transition
  nodes [Result(bCanEnterTransition=Get bTransition_LocomotionToIdle_To_Locomotion() -> (bTransition_LocomotionToIdle_To_Locomotion:bool))]
 }
 Transition {
  path=AnimGraph.AnimGraphNode_StateMachine_0.FullBody.AnimStateTransitionNode_87.Transition
  nodes [Result(bCanEnterTransition=Get bIsCrouching())]
 }
 Transition {
  path=AnimGraph.AnimGraphNode_StateMachine_0.FullBody.AnimStateTransitionNode_88.Transition
  nodes [Result(bCanEnterTransition=NOT Boolean(A=Get bIsTargeting()))]
 }
 Transition {
  path=AnimGraph.AnimGraphNode_StateMachine_0.FullBody.AnimStateTransitionNode_89.Transition
  nodes [Result(bCanEnterTransition=Get bIsCrouching())]
 }
 Transition {
  path=AnimGraph.AnimGraphNode_StateMachine_0.FullBody.AnimStateTransitionNode_9.Transition
  nodes [Result(bCanEnterTransition=Get bTransition_Locomotion_LocomotionHappy() -> (bTransition_Locomotion_LocomotionHappy:bool))]
 }
 Transition {
  path=AnimGraph.AnimGraphNode_StateMachine_0.FullBody.AnimStateTransitionNode_90.Transition
  nodes [Result(bCanEnterTransition=Get bIsTargeting())]
 }
 Transition {
  path=AnimGraph.AnimGraphNode_StateMachine_0.FullBody.AnimStateTransitionNode_91.Transition
  nodes [Result(bCanEnterTransition=Get bIsTargetingAndMoving() -> (bIsTargetingAndMoving:bool))]
 }
 Transition {
  path=AnimGraph.AnimGraphNode_StateMachine_0.FullBody.AnimStateTransitionNode_92.Transition
  nodes [Result(bCanEnterTransition=Get bIsTargeting())]
 }
 Transition {
  path=AnimGraph.AnimGraphNode_StateMachine_0.FullBody.AnimStateTransitionNode_93.Transition
  nodes [Result(bCanEnterTransition=Get bIsTargetingAndMoving())]
 }
 Transition {
  path=AnimGraph.AnimGraphNode_StateMachine_0.FullBody.AnimStateTransitionNode_94.Transition
  nodes [Result(bCanEnterTransition=Get bIsTargeting())]
 }
 Transition {
  path=AnimGraph.AnimGraphNode_StateMachine_0.FullBody.AnimStateTransitionNode_95.Transition
  nodes [Result(bCanEnterTransition=Get bIsTargetingAndMoving())]
 }
 Transition {path=AnimGraph.AnimGraphNode_StateMachine_0.FullBody.AnimStateTransitionNode_96.Transition, nodes [Result()]}
 Transition {
  path=AnimGraph.AnimGraphNode_StateMachine_0.FullBody.AnimStateTransitionNode_97.Transition
  nodes [Result(bCanEnterTransition=Get bTransition_CrouchToCrouchADSNotMoving_Crouching() -> (bTransition_CrouchToCrouchADSNotMoving_Crouching:bool))]
 }
 Transition {path=AnimGraph.AnimGraphNode_StateMachine_0.FullBody.AnimStateTransitionNode_98.Transition, nodes [Result()]}
 Transition {
  path=AnimGraph.AnimGraphNode_StateMachine_0.FullBody.AnimStateTransitionNode_99.Transition
  nodes [Result(bCanEnterTransition=Get bTransition_DownSights_ADSToCrouch() -> (bTransition_DownSights_ADSToCrouch:bool))]
 }
 FullBody_1 {
  entry=Main
  path=AnimGraph.AnimGraphNode_StateMachine_1.FullBody_1
  ids=id
  nodes [
   AnimStateNode_0 = Lobby() {is=title AnimStateNode}
   AnimStateNode_1 = Main() {is=title AnimStateNode}
   AnimStateNode_11 = DBNOToIdle()
   AnimStateNode_9 = DBNO() {is=title AnimStateNode}
   AnimStateTransitionNode_14 = AnimStateTransitionNode()
   AnimStateTransitionNode_21 = AnimStateTransitionNode()
   AnimStateTransitionNode_22 = AnimStateTransitionNode()
   AnimStateTransitionNode_46 = AnimStateTransitionNode()
   AnimStateTransitionNode_47 = AnimStateTransitionNode() {bAutomaticRuleBasedOnSequencePlayerInState=true}
  ]
  links [
   {
    dst=DBNOToIdle
    rule=AnimGraph.AnimGraphNode_StateMachine_1.FullBody_1.AnimStateTransitionNode_14.AnimationTransitionGraph_2
    src=DBNO
    via=AnimStateTransitionNode_14
   }
   {
    dst=Main
    rule=AnimGraph.AnimGraphNode_StateMachine_1.FullBody_1.AnimStateTransitionNode_47.Transition
    src=DBNOToIdle
    via=AnimStateTransitionNode_47
   }
   {
    dst=Main
    rule=AnimGraph.AnimGraphNode_StateMachine_1.FullBody_1.AnimStateTransitionNode_22.Transition
    src=Lobby
    via=AnimStateTransitionNode_22
   }
   {
    dst=DBNO
    rule=AnimGraph.AnimGraphNode_StateMachine_1.FullBody_1.AnimStateTransitionNode_46.Transition
    src=Main
    via=AnimStateTransitionNode_46
   }
   {
    dst=Lobby
    rule=AnimGraph.AnimGraphNode_StateMachine_1.FullBody_1.AnimStateTransitionNode_21.Transition
    src=Main
    via=AnimStateTransitionNode_21
   }
  ]
 }
 Lobby {
  path=AnimGraph.AnimGraphNode_StateMachine_1.FullBody_1.AnimStateNode_0.Lobby
  nodes [
   AnimGraphNode_StateMachine_0 = Lobby() {is=title AnimGraphNode_StateMachine}
   Output Animation Pose(Result=@AnimGraphNode_StateMachine_0)
  ]
 }
 Lobby {
  entry=IdleLobby
  path=AnimGraph.AnimGraphNode_StateMachine_1.FullBody_1.AnimStateNode_0.Lobby.AnimGraphNode_StateMachine_0.Lobby
  ids=id
  nodes [
   AnimStateConduitNode_2 = PopToSkydive()
   AnimStateNode_0 = IdleLobby()
   AnimStateNode_22 = HunkerToLobby()
   AnimStateNode_24 = LobbyToHunker()
   AnimStateNode_25 = FrontendSkydive()
   AnimStateNode_27 = LobbyHunker()
   AnimStateTransitionNode_1 = AnimStateTransitionNode() {crossfadeDuration=0, priorityOrder=0}
   AnimStateTransitionNode_2 = AnimStateTransitionNode() {crossfadeDuration=0}
   AnimStateTransitionNode_31 = AnimStateTransitionNode()
   AnimStateTransitionNode_33 = AnimStateTransitionNode() {bAutomaticRuleBasedOnSequencePlayerInState=true, crossfadeDuration=0.3}
   AnimStateTransitionNode_64 = AnimStateTransitionNode() {crossfadeDuration=0, priorityOrder=0}
   AnimStateTransitionNode_65 = AnimStateTransitionNode()
   AnimStateTransitionNode_66 = AnimStateTransitionNode() {bAutomaticRuleBasedOnSequencePlayerInState=true}
   AnimStateTransitionNode_67 = AnimStateTransitionNode() {crossfadeDuration=0}
   AnimStateTransitionNode_68 = AnimStateTransitionNode()
   AnimStateTransitionNode_69 = AnimStateTransitionNode()
   AnimStateTransitionNode_71 = AnimStateTransitionNode()
   AnimStateTransitionNode_73 = AnimStateTransitionNode()
  ]
  links [
   {
    dst=IdleLobby
    rule=AnimGraph.AnimGraphNode_StateMachine_1.FullBody_1.AnimStateNode_0.Lobby.AnimGraphNode_StateMachine_0.Lobby.AnimStateTransitionNode_2.AnimationTransitionGraph_1
    src=FrontendSkydive
    via=AnimStateTransitionNode_2
   }
   {
    dst=LobbyHunker
    rule=AnimGraph.AnimGraphNode_StateMachine_1.FullBody_1.AnimStateNode_0.Lobby.AnimGraphNode_StateMachine_0.Lobby.AnimStateTransitionNode_64.AnimationTransitionGraph_0
    src=FrontendSkydive
    via=AnimStateTransitionNode_64
   }
   {
    dst=IdleLobby
    rule=AnimGraph.AnimGraphNode_StateMachine_1.FullBody_1.AnimStateNode_0.Lobby.AnimGraphNode_StateMachine_0.Lobby.AnimStateTransitionNode_33.AnimationTransitionGraph_0
    src=HunkerToLobby
    via=AnimStateTransitionNode_33
   }
   {
    dst=PopToSkydive
    rule=AnimGraph.AnimGraphNode_StateMachine_1.FullBody_1.AnimStateNode_0.Lobby.AnimGraphNode_StateMachine_0.Lobby.AnimStateTransitionNode_71.AnimationTransitionGraph_0
    src=HunkerToLobby
    via=AnimStateTransitionNode_71
   }
   {
    dst=LobbyHunker
    rule=AnimGraph.AnimGraphNode_StateMachine_1.FullBody_1.AnimStateNode_0.Lobby.AnimGraphNode_StateMachine_0.Lobby.AnimStateTransitionNode_1.AnimationTransitionGraph_0
    src=IdleLobby
    via=AnimStateTransitionNode_1
   }
   {
    dst=LobbyToHunker
    rule=AnimGraph.AnimGraphNode_StateMachine_1.FullBody_1.AnimStateNode_0.Lobby.AnimGraphNode_StateMachine_0.Lobby.AnimStateTransitionNode_65.AnimationTransitionGraph_0
    src=IdleLobby
    via=AnimStateTransitionNode_65
   }
   {
    dst=PopToSkydive
    rule=AnimGraph.AnimGraphNode_StateMachine_1.FullBody_1.AnimStateNode_0.Lobby.AnimGraphNode_StateMachine_0.Lobby.AnimStateTransitionNode_68.AnimationTransitionGraph_0
    src=IdleLobby
    via=AnimStateTransitionNode_68
   }
   {
    dst=HunkerToLobby
    rule=AnimGraph.AnimGraphNode_StateMachine_1.FullBody_1.AnimStateNode_0.Lobby.AnimGraphNode_StateMachine_0.Lobby.AnimStateTransitionNode_31.AnimationTransitionGraph_0
    src=LobbyHunker
    via=AnimStateTransitionNode_31
   }
   {
    dst=PopToSkydive
    rule=AnimGraph.AnimGraphNode_StateMachine_1.FullBody_1.AnimStateNode_0.Lobby.AnimGraphNode_StateMachine_0.Lobby.AnimStateTransitionNode_73.AnimationTransitionGraph_0
    src=LobbyHunker
    via=AnimStateTransitionNode_73
   }
   {
    dst=LobbyHunker
    rule=AnimGraph.AnimGraphNode_StateMachine_1.FullBody_1.AnimStateNode_0.Lobby.AnimGraphNode_StateMachine_0.Lobby.AnimStateTransitionNode_66.AnimationTransitionGraph_0
    src=LobbyToHunker
    via=AnimStateTransitionNode_66
   }
   {
    dst=PopToSkydive
    rule=AnimGraph.AnimGraphNode_StateMachine_1.FullBody_1.AnimStateNode_0.Lobby.AnimGraphNode_StateMachine_0.Lobby.AnimStateTransitionNode_69.AnimationTransitionGraph_0
    src=LobbyToHunker
    via=AnimStateTransitionNode_69
   }
   {
    dst=FrontendSkydive
    rule=AnimGraph.AnimGraphNode_StateMachine_1.FullBody_1.AnimStateNode_0.Lobby.AnimGraphNode_StateMachine_0.Lobby.AnimStateTransitionNode_67.AnimationTransitionGraph_0
    src=PopToSkydive
    via=AnimStateTransitionNode_67
   }
  ]
 }
 PopToSkydive {
  path=AnimGraph.AnimGraphNode_StateMachine_1.FullBody_1.AnimStateNode_0.Lobby.AnimGraphNode_StateMachine_0.Lobby.AnimStateConduitNode_2.PopToSkydive
  nodes [Result(bCanEnterTransition=true)]
 }
 IdleLobby {
  path=AnimGraph.AnimGraphNode_StateMachine_1.FullBody_1.AnimStateNode_0.Lobby.AnimGraphNode_StateMachine_0.Lobby.AnimStateNode_0.IdleLobby
  nodes [Output Animation Pose(Result=Sequence Player 'Idle_Lobby'() {sequence=&Idle_Lobby})]
 }
 HunkerToLobby {
  path=AnimGraph.AnimGraphNode_StateMachine_1.FullBody_1.AnimStateNode_0.Lobby.AnimGraphNode_StateMachine_0.Lobby.AnimStateNode_22.HunkerToLobby
  nodes [Output Animation Pose(Result=Sequence Player 'LocomotionToCrouch'() {sequence=&LocomotionToCrouch})]
 }
 LobbyToHunker {
  path=AnimGraph.AnimGraphNode_StateMachine_1.FullBody_1.AnimStateNode_0.Lobby.AnimGraphNode_StateMachine_0.Lobby.AnimStateNode_24.LobbyToHunker
  nodes [Output Animation Pose(Result=Sequence Player 'CrouchIdleToIdle'() {sequence=&CrouchIdleToIdle})]
 }
 FrontendSkydive {
  path=AnimGraph.AnimGraphNode_StateMachine_1.FullBody_1.AnimStateNode_0.Lobby.AnimGraphNode_StateMachine_0.Lobby.AnimStateNode_25.FrontendSkydive
  nodes [
   Apply Additive(
    Base=Sequence Player 'Dive_B_Down_Pose'() {sequence=&Dive_B_Down_Pose}
    Additive="Sequence Player (additive) 'Dive_B_Down_Add'"() {sequence=&Dive_B_Down_Add}
   ) {alphaBoolBlend {blendOption=Linear}}
   AnimGraphNode_StateMachine_0 = FrontEndSkydive()
   Output Animation Pose(Result=@AnimGraphNode_StateMachine_0)
  ]
 }
 FrontEndSkydive {
  entry=SkydiveGlide
  path=AnimGraph.AnimGraphNode_StateMachine_1.FullBody_1.AnimStateNode_0.Lobby.AnimGraphNode_StateMachine_0.Lobby.AnimStateNode_25.FrontendSkydive.AnimGraphNode_StateMachine_0.FrontEndSkydive
  ids=id
  nodes [
   AnimStateNode_0 = SkydiveGlide()
   AnimStateNode_19 = ToDive()
   AnimStateNode_33 = ToGlide()
   AnimStateNode_8 = SkydiveDive()
   AnimStateTransitionNode_0 = AnimStateTransitionNode()
   AnimStateTransitionNode_1 = AnimStateTransitionNode()
   AnimStateTransitionNode_29 = AnimStateTransitionNode()
   AnimStateTransitionNode_31 = AnimStateTransitionNode() {bAutomaticRuleBasedOnSequencePlayerInState=true}
   AnimStateTransitionNode_70 = AnimStateTransitionNode()
   AnimStateTransitionNode_71 = AnimStateTransitionNode() {bAutomaticRuleBasedOnSequencePlayerInState=true}
  ]
  links [
   {
    dst=ToGlide
    rule=AnimGraph.AnimGraphNode_StateMachine_1.FullBody_1.AnimStateNode_0.Lobby.AnimGraphNode_StateMachine_0.Lobby.AnimStateNode_25.FrontendSkydive.AnimGraphNode_StateMachine_0.FrontEndSkydive.AnimStateTransitionNode_70.AnimationTransitionGraph_1
    src=SkydiveDive
    via=AnimStateTransitionNode_70
   }
   {
    dst=ToDive
    rule=AnimGraph.AnimGraphNode_StateMachine_1.FullBody_1.AnimStateNode_0.Lobby.AnimGraphNode_StateMachine_0.Lobby.AnimStateNode_25.FrontendSkydive.AnimGraphNode_StateMachine_0.FrontEndSkydive.AnimStateTransitionNode_29.AnimationTransitionGraph_1
    src=SkydiveGlide
    via=AnimStateTransitionNode_29
   }
   {
    dst=SkydiveDive
    rule=AnimGraph.AnimGraphNode_StateMachine_1.FullBody_1.AnimStateNode_0.Lobby.AnimGraphNode_StateMachine_0.Lobby.AnimStateNode_25.FrontendSkydive.AnimGraphNode_StateMachine_0.FrontEndSkydive.AnimStateTransitionNode_31.AnimationTransitionGraph_1
    src=ToDive
    via=AnimStateTransitionNode_31
   }
   {
    dst=ToGlide
    rule=AnimGraph.AnimGraphNode_StateMachine_1.FullBody_1.AnimStateNode_0.Lobby.AnimGraphNode_StateMachine_0.Lobby.AnimStateNode_25.FrontendSkydive.AnimGraphNode_StateMachine_0.FrontEndSkydive.AnimStateTransitionNode_1.AnimationTransitionGraph_0
    src=ToDive
    via=AnimStateTransitionNode_1
   }
   {
    dst=SkydiveGlide
    rule=AnimGraph.AnimGraphNode_StateMachine_1.FullBody_1.AnimStateNode_0.Lobby.AnimGraphNode_StateMachine_0.Lobby.AnimStateNode_25.FrontendSkydive.AnimGraphNode_StateMachine_0.FrontEndSkydive.AnimStateTransitionNode_71.AnimationTransitionGraph_1
    src=ToGlide
    via=AnimStateTransitionNode_71
   }
   {
    dst=ToDive
    rule=AnimGraph.AnimGraphNode_StateMachine_1.FullBody_1.AnimStateNode_0.Lobby.AnimGraphNode_StateMachine_0.Lobby.AnimStateNode_25.FrontendSkydive.AnimGraphNode_StateMachine_0.FrontEndSkydive.AnimStateTransitionNode_0.AnimationTransitionGraph_0
    src=ToGlide
    via=AnimStateTransitionNode_0
   }
  ]
 }
 SkydiveGlide {
  path=AnimGraph.AnimGraphNode_StateMachine_1.FullBody_1.AnimStateNode_0.Lobby.AnimGraphNode_StateMachine_0.Lobby.AnimStateNode_25.FrontendSkydive.AnimGraphNode_StateMachine_0.FrontEndSkydive.AnimStateNode_0.SkydiveGlide
  nodes [
   n1 = Break Fort Anim Input Skydiving(FortAnimInput_Skydiving=Get SkydivingState()) -> (
    ParaGilderRootOffset=~
    ParaGliderRootRotationOverride=~
    bUseParaGlideRootModifier=~
    bIsSkydivingFromLaunchPad=~
    bIsSkydivingFromBus=~
    bIsInVortex=~
    bIsDiving=~
    bIsParachuteOpen=~
    bIsSkydiving=~
    bIsParachuteLeaning=~
    bIsSkydiveLeaning=~
    bIsLeaning=~
    bIsSkydiveDiveMode=~
    bPlaySkydiveDrift=~
    bSkydiveDriftDelayActive=~
    bSkydiveDriftAnimAllowed=~
    bIsGliderRight=~
    bIsGliderCenter=~
    bIsGliderLeft=~
    bIsGliderForward=~
    bIsGliderBack=~
    bHasDeployedAndIsStandingUpright=~
    LocalAccelForward=~
    LocalAccelRight=~
    LocalVelocityRight=~
    SkydiveAimPitch=~
    SkydiveAimYaw=~
    DeployChuteAnimRate=~
    SkydiveDriftAnimRate=~
    SkydiveDriftAnimRateCurrent=~
    SkydiveFidgetAnimRate=~
    SkydiveFidgetAnimRateCurrent=~
    SkydiveAdditiveAlpha=~
    SkydiveDriftDelay=~
   )
   Blend Poses by bool(
    BlendPose_0=Apply Additive(
     Base=Blendspace Player 'BS_Glide_Lean'(X=@n1.LocalVelocityRight, Y=@n1.SkydiveAimPitch) {blendSpace=&BS_Glide_Lean}
     Additive="Sequence Player (additive) 'Glide_Add'"() {sequence=&Glide_Add}
    ) {alphaBoolBlend {blendOption=Linear}}
    BlendPose_1=Apply Additive(
     Base=Blendspace Player 'BS_Glide_2d'(X=@n1.SkydiveAimYaw, Y=@n1.SkydiveAimPitch) {blendSpace=&BS_Glide_2d}
     Additive="Sequence Player (additive) 'Glide_Add'"() {sequence=&Glide_Add}
    ) {alphaBoolBlend {blendOption=Linear}}
    bActiveValue=@n1.bIsSkydiveLeaning
   ) {blendType=Linear}
   Output Animation Pose(Result=Apply Additive(
     Base=Sequence Player 'Glide_Pose'() {sequence=&Glide_Pose}
     Additive="Sequence Player (additive) 'Glide_Add'"() {sequence=&Glide_Add}
    ) {alphaBoolBlend {blendOption=Linear}})
  ]
 }
 ToDive {
  path=AnimGraph.AnimGraphNode_StateMachine_1.FullBody_1.AnimStateNode_0.Lobby.AnimGraphNode_StateMachine_0.Lobby.AnimStateNode_25.FrontendSkydive.AnimGraphNode_StateMachine_0.FrontEndSkydive.AnimStateNode_19.ToDive
  nodes [Output Animation Pose(Result=Sequence Player 'GlideToDive'() {sequence=&GlideToDive})]
 }
 ToGlide {
  path=AnimGraph.AnimGraphNode_StateMachine_1.FullBody_1.AnimStateNode_0.Lobby.AnimGraphNode_StateMachine_0.Lobby.AnimStateNode_25.FrontendSkydive.AnimGraphNode_StateMachine_0.FrontEndSkydive.AnimStateNode_33.ToGlide
  nodes [Output Animation Pose(Result=Sequence Player 'DiveToGlide'() {sequence=&DiveToGlide})]
 }
 SkydiveDive {
  path=AnimGraph.AnimGraphNode_StateMachine_1.FullBody_1.AnimStateNode_0.Lobby.AnimGraphNode_StateMachine_0.Lobby.AnimStateNode_25.FrontendSkydive.AnimGraphNode_StateMachine_0.FrontEndSkydive.AnimStateNode_8.SkydiveDive
  nodes [
   n1 = Break Fort Anim Input Skydiving(FortAnimInput_Skydiving=Get SkydivingState()) -> (
    ParaGilderRootOffset=~
    ParaGliderRootRotationOverride=~
    bUseParaGlideRootModifier=~
    bIsSkydivingFromLaunchPad=~
    bIsSkydivingFromBus=~
    bIsInVortex=~
    bIsDiving=~
    bIsParachuteOpen=~
    bIsSkydiving=~
    bIsParachuteLeaning=~
    bIsSkydiveLeaning=~
    bIsLeaning=~
    bIsSkydiveDiveMode=~
    bPlaySkydiveDrift=~
    bSkydiveDriftDelayActive=~
    bSkydiveDriftAnimAllowed=~
    bIsGliderRight=~
    bIsGliderCenter=~
    bIsGliderLeft=~
    bIsGliderForward=~
    bIsGliderBack=~
    bHasDeployedAndIsStandingUpright=~
    LocalAccelForward=~
    LocalAccelRight=~
    LocalVelocityRight=~
    SkydiveAimPitch=~
    SkydiveAimYaw=~
    DeployChuteAnimRate=~
    SkydiveDriftAnimRate=~
    SkydiveDriftAnimRateCurrent=~
    SkydiveFidgetAnimRate=~
    SkydiveFidgetAnimRateCurrent=~
    SkydiveAdditiveAlpha=~
    SkydiveDriftDelay=~
   )
   Blend Poses by bool(
    BlendPose_0=Apply Additive(
     Base=Blendspace Player 'BS_Dive_Lean'(X=@n1.LocalVelocityRight, Y=@n1.SkydiveAimPitch) {blendSpace=&BS_Dive_Lean}
     Additive=Blendspace Player 'GlideDive_Add'(X=@n1.SkydiveAimPitch) {blendSpace=&GlideDive_Add}
    ) {alphaBoolBlend {blendOption=Linear}}
    BlendPose_1=Apply Additive(
     Base=Blendspace Player 'BS_Dive_2D'(X=@n1.SkydiveAimYaw, Y=@n1.SkydiveAimPitch) {blendSpace=&BS_Dive_2D}
     Additive=Blendspace Player 'GlideDive_Add'(X=@n1.SkydiveAimPitch) {blendSpace=&GlideDive_Add}
    ) {alphaBoolBlend {blendOption=Linear}}
    bActiveValue=@n1.bIsSkydiveLeaning
   ) {blendType=Linear}
   Output Animation Pose(Result=Apply Additive(
     Base=Sequence Player 'Dive_B_Down_Pose'() {sequence=&Dive_B_Down_Pose}
     Additive="Sequence Player (additive) 'Dive_B_Down_Add'"() {sequence=&Dive_B_Down_Add}
    ) {alphaBoolBlend {blendOption=Linear}})
  ]
 }
 AnimationTransitionGraph_0 {
  path=AnimGraph.AnimGraphNode_StateMachine_1.FullBody_1.AnimStateNode_0.Lobby.AnimGraphNode_StateMachine_0.Lobby.AnimStateNode_25.FrontendSkydive.AnimGraphNode_StateMachine_0.FrontEndSkydive.AnimStateTransitionNode_0.AnimationTransitionGraph_0
  nodes [
   Result(bCanEnterTransition=Get bLobbySkyDive_IsDiving())
   Break Fort Anim Input Skydiving(FortAnimInput_Skydiving=Get SkydivingState()) -> (
    ParaGilderRootOffset=~
    ParaGliderRootRotationOverride=~
    bUseParaGlideRootModifier=~
    bIsSkydivingFromLaunchPad=~
    bIsSkydivingFromBus=~
    bIsInVortex=~
    bIsDiving=~
    bIsParachuteOpen=~
    bIsSkydiving=~
    bIsParachuteLeaning=~
    bIsSkydiveLeaning=~
    bIsLeaning=~
    bIsSkydiveDiveMode=~
    bPlaySkydiveDrift=~
    bSkydiveDriftDelayActive=~
    bSkydiveDriftAnimAllowed=~
    bIsGliderRight=~
    bIsGliderCenter=~
    bIsGliderLeft=~
    bIsGliderForward=~
    bIsGliderBack=~
    bHasDeployedAndIsStandingUpright=~
    LocalAccelForward=~
    LocalAccelRight=~
    LocalVelocityRight=~
    SkydiveAimPitch=~
    SkydiveAimYaw=~
    DeployChuteAnimRate=~
    SkydiveDriftAnimRate=~
    SkydiveDriftAnimRateCurrent=~
    SkydiveFidgetAnimRate=~
    SkydiveFidgetAnimRateCurrent=~
    SkydiveAdditiveAlpha=~
    SkydiveDriftDelay=~
   )
  ]
 }
 AnimationTransitionGraph_0 {
  path=AnimGraph.AnimGraphNode_StateMachine_1.FullBody_1.AnimStateNode_0.Lobby.AnimGraphNode_StateMachine_0.Lobby.AnimStateNode_25.FrontendSkydive.AnimGraphNode_StateMachine_0.FrontEndSkydive.AnimStateTransitionNode_1.AnimationTransitionGraph_0
  nodes [
   Result(bCanEnterTransition=NOT Boolean(A=Get bLobbySkyDive_IsDiving()))
   Break Fort Anim Input Skydiving(FortAnimInput_Skydiving=Get SkydivingState()) -> (
    ParaGilderRootOffset=~
    ParaGliderRootRotationOverride=~
    bUseParaGlideRootModifier=~
    bIsSkydivingFromLaunchPad=~
    bIsSkydivingFromBus=~
    bIsInVortex=~
    bIsDiving=~
    bIsParachuteOpen=~
    bIsSkydiving=~
    bIsParachuteLeaning=~
    bIsSkydiveLeaning=~
    bIsLeaning=~
    bIsSkydiveDiveMode=~
    bPlaySkydiveDrift=~
    bSkydiveDriftDelayActive=~
    bSkydiveDriftAnimAllowed=~
    bIsGliderRight=~
    bIsGliderCenter=~
    bIsGliderLeft=~
    bIsGliderForward=~
    bIsGliderBack=~
    bHasDeployedAndIsStandingUpright=~
    LocalAccelForward=~
    LocalAccelRight=~
    LocalVelocityRight=~
    SkydiveAimPitch=~
    SkydiveAimYaw=~
    DeployChuteAnimRate=~
    SkydiveDriftAnimRate=~
    SkydiveDriftAnimRateCurrent=~
    SkydiveFidgetAnimRate=~
    SkydiveFidgetAnimRateCurrent=~
    SkydiveAdditiveAlpha=~
    SkydiveDriftDelay=~
   )
  ]
 }
 AnimationTransitionGraph_1 {
  path=AnimGraph.AnimGraphNode_StateMachine_1.FullBody_1.AnimStateNode_0.Lobby.AnimGraphNode_StateMachine_0.Lobby.AnimStateNode_25.FrontendSkydive.AnimGraphNode_StateMachine_0.FrontEndSkydive.AnimStateTransitionNode_29.AnimationTransitionGraph_1
  nodes [
   Result(bCanEnterTransition=Get bLobbySkyDive_IsDiving())
   Break Fort Anim Input Skydiving(FortAnimInput_Skydiving=Get SkydivingState()) -> (
    ParaGilderRootOffset=~
    ParaGliderRootRotationOverride=~
    bUseParaGlideRootModifier=~
    bIsSkydivingFromLaunchPad=~
    bIsSkydivingFromBus=~
    bIsInVortex=~
    bIsDiving=~
    bIsParachuteOpen=~
    bIsSkydiving=~
    bIsParachuteLeaning=~
    bIsSkydiveLeaning=~
    bIsLeaning=~
    bIsSkydiveDiveMode=~
    bPlaySkydiveDrift=~
    bSkydiveDriftDelayActive=~
    bSkydiveDriftAnimAllowed=~
    bIsGliderRight=~
    bIsGliderCenter=~
    bIsGliderLeft=~
    bIsGliderForward=~
    bIsGliderBack=~
    bHasDeployedAndIsStandingUpright=~
    LocalAccelForward=~
    LocalAccelRight=~
    LocalVelocityRight=~
    SkydiveAimPitch=~
    SkydiveAimYaw=~
    DeployChuteAnimRate=~
    SkydiveDriftAnimRate=~
    SkydiveDriftAnimRateCurrent=~
    SkydiveFidgetAnimRate=~
    SkydiveFidgetAnimRateCurrent=~
    SkydiveAdditiveAlpha=~
    SkydiveDriftDelay=~
   )
  ]
 }
 AnimationTransitionGraph_1 {
  path=AnimGraph.AnimGraphNode_StateMachine_1.FullBody_1.AnimStateNode_0.Lobby.AnimGraphNode_StateMachine_0.Lobby.AnimStateNode_25.FrontendSkydive.AnimGraphNode_StateMachine_0.FrontEndSkydive.AnimStateTransitionNode_31.AnimationTransitionGraph_1
  nodes [Result()]
 }
 AnimationTransitionGraph_1 {
  path=AnimGraph.AnimGraphNode_StateMachine_1.FullBody_1.AnimStateNode_0.Lobby.AnimGraphNode_StateMachine_0.Lobby.AnimStateNode_25.FrontendSkydive.AnimGraphNode_StateMachine_0.FrontEndSkydive.AnimStateTransitionNode_70.AnimationTransitionGraph_1
  nodes [
   Result(bCanEnterTransition=NOT Boolean(A=Get bLobbySkyDive_IsDiving()))
   Break Fort Anim Input Skydiving(FortAnimInput_Skydiving=Get SkydivingState()) -> (
    ParaGilderRootOffset=~
    ParaGliderRootRotationOverride=~
    bUseParaGlideRootModifier=~
    bIsSkydivingFromLaunchPad=~
    bIsSkydivingFromBus=~
    bIsInVortex=~
    bIsDiving=~
    bIsParachuteOpen=~
    bIsSkydiving=~
    bIsParachuteLeaning=~
    bIsSkydiveLeaning=~
    bIsLeaning=~
    bIsSkydiveDiveMode=~
    bPlaySkydiveDrift=~
    bSkydiveDriftDelayActive=~
    bSkydiveDriftAnimAllowed=~
    bIsGliderRight=~
    bIsGliderCenter=~
    bIsGliderLeft=~
    bIsGliderForward=~
    bIsGliderBack=~
    bHasDeployedAndIsStandingUpright=~
    LocalAccelForward=~
    LocalAccelRight=~
    LocalVelocityRight=~
    SkydiveAimPitch=~
    SkydiveAimYaw=~
    DeployChuteAnimRate=~
    SkydiveDriftAnimRate=~
    SkydiveDriftAnimRateCurrent=~
    SkydiveFidgetAnimRate=~
    SkydiveFidgetAnimRateCurrent=~
    SkydiveAdditiveAlpha=~
    SkydiveDriftDelay=~
   )
  ]
 }
 AnimationTransitionGraph_1 {
  path=AnimGraph.AnimGraphNode_StateMachine_1.FullBody_1.AnimStateNode_0.Lobby.AnimGraphNode_StateMachine_0.Lobby.AnimStateNode_25.FrontendSkydive.AnimGraphNode_StateMachine_0.FrontEndSkydive.AnimStateTransitionNode_71.AnimationTransitionGraph_1
  nodes [Result()]
 }
 LobbyHunker {
  path=AnimGraph.AnimGraphNode_StateMachine_1.FullBody_1.AnimStateNode_0.Lobby.AnimGraphNode_StateMachine_0.Lobby.AnimStateNode_27.LobbyHunker
  nodes [Output Animation Pose(Result=Apply Additive(
     Base=Sequence Player 'Hunker_Pose'() {sequence=&Hunker_Pose}
     Additive="Sequence Player (additive) 'IdleStraightLookAround_Add'"() {sequence=&IdleStraightLookAround_Add}
    ) {alphaBoolBlend {blendOption=Linear}})]
 }
 AnimationTransitionGraph_0 {
  path=AnimGraph.AnimGraphNode_StateMachine_1.FullBody_1.AnimStateNode_0.Lobby.AnimGraphNode_StateMachine_0.Lobby.AnimStateTransitionNode_1.AnimationTransitionGraph_0
  nodes [Result(bCanEnterTransition=NOT Boolean(A=Get bIsFrontEndLobbyPartyMember()))]
 }
 AnimationTransitionGraph_1 {
  path=AnimGraph.AnimGraphNode_StateMachine_1.FullBody_1.AnimStateNode_0.Lobby.AnimGraphNode_StateMachine_0.Lobby.AnimStateTransitionNode_2.AnimationTransitionGraph_1
  nodes [Result(bCanEnterTransition=Get bTransition_FrontEndSkydive_IdleLobby())]
 }
 AnimationTransitionGraph_0 {
  path=AnimGraph.AnimGraphNode_StateMachine_1.FullBody_1.AnimStateNode_0.Lobby.AnimGraphNode_StateMachine_0.Lobby.AnimStateTransitionNode_31.AnimationTransitionGraph_0
  nodes [Result(bCanEnterTransition=Get bTransition_LobbyHunker_HunkerToLobby())]
 }
 AnimationTransitionGraph_0 {
  path=AnimGraph.AnimGraphNode_StateMachine_1.FullBody_1.AnimStateNode_0.Lobby.AnimGraphNode_StateMachine_0.Lobby.AnimStateTransitionNode_33.AnimationTransitionGraph_0
  nodes [Result()]
 }
 AnimationTransitionGraph_0 {
  path=AnimGraph.AnimGraphNode_StateMachine_1.FullBody_1.AnimStateNode_0.Lobby.AnimGraphNode_StateMachine_0.Lobby.AnimStateTransitionNode_64.AnimationTransitionGraph_0
  nodes [Result(bCanEnterTransition=Get bTransition_FrontEndSkydive_LobbyHunker())]
 }
 AnimationTransitionGraph_0 {
  path=AnimGraph.AnimGraphNode_StateMachine_1.FullBody_1.AnimStateNode_0.Lobby.AnimGraphNode_StateMachine_0.Lobby.AnimStateTransitionNode_65.AnimationTransitionGraph_0
  nodes [Result(bCanEnterTransition=Get bEmotePlaying_WhereHunkeringIsBest())]
 }
 AnimationTransitionGraph_0 {
  path=AnimGraph.AnimGraphNode_StateMachine_1.FullBody_1.AnimStateNode_0.Lobby.AnimGraphNode_StateMachine_0.Lobby.AnimStateTransitionNode_66.AnimationTransitionGraph_0
  nodes [Result()]
 }
 AnimationTransitionGraph_0 {
  path=AnimGraph.AnimGraphNode_StateMachine_1.FullBody_1.AnimStateNode_0.Lobby.AnimGraphNode_StateMachine_0.Lobby.AnimStateTransitionNode_67.AnimationTransitionGraph_0
  nodes [Result(bCanEnterTransition=Break Fort Anim Input Skydiving(FortAnimInput_Skydiving=Get SkydivingState()) -> (
     ParaGilderRootOffset=~
     ParaGliderRootRotationOverride=~
     bUseParaGlideRootModifier=~
     bIsSkydivingFromLaunchPad=~
     bIsSkydivingFromBus=~
     bIsInVortex=~
     bIsDiving=~
     bIsParachuteOpen=~
     bIsSkydiving=~
     bIsParachuteLeaning=~
     bIsSkydiveLeaning=~
     bIsLeaning=~
     bIsSkydiveDiveMode=~
     bPlaySkydiveDrift=~
     bSkydiveDriftDelayActive=~
     bSkydiveDriftAnimAllowed=~
     bIsGliderRight=~
     bIsGliderCenter=~
     bIsGliderLeft=~
     bIsGliderForward=~
     bIsGliderBack=~
     bHasDeployedAndIsStandingUpright=~
     LocalAccelForward=~
     LocalAccelRight=~
     LocalVelocityRight=~
     SkydiveAimPitch=~
     SkydiveAimYaw=~
     DeployChuteAnimRate=~
     SkydiveDriftAnimRate=~
     SkydiveDriftAnimRateCurrent=~
     SkydiveFidgetAnimRate=~
     SkydiveFidgetAnimRateCurrent=~
     SkydiveAdditiveAlpha=~
     SkydiveDriftDelay=~
    ).bIsSkydiving)]
 }
 AnimationTransitionGraph_0 {
  path=AnimGraph.AnimGraphNode_StateMachine_1.FullBody_1.AnimStateNode_0.Lobby.AnimGraphNode_StateMachine_0.Lobby.AnimStateTransitionNode_68.AnimationTransitionGraph_0
  nodes [Result(bCanEnterTransition=true)]
 }
 AnimationTransitionGraph_0 {
  path=AnimGraph.AnimGraphNode_StateMachine_1.FullBody_1.AnimStateNode_0.Lobby.AnimGraphNode_StateMachine_0.Lobby.AnimStateTransitionNode_69.AnimationTransitionGraph_0
  nodes [Result(bCanEnterTransition=true)]
 }
 AnimationTransitionGraph_0 {
  path=AnimGraph.AnimGraphNode_StateMachine_1.FullBody_1.AnimStateNode_0.Lobby.AnimGraphNode_StateMachine_0.Lobby.AnimStateTransitionNode_71.AnimationTransitionGraph_0
  nodes [Result(bCanEnterTransition=true)]
 }
 AnimationTransitionGraph_0 {
  path=AnimGraph.AnimGraphNode_StateMachine_1.FullBody_1.AnimStateNode_0.Lobby.AnimGraphNode_StateMachine_0.Lobby.AnimStateTransitionNode_73.AnimationTransitionGraph_0
  nodes [Result(bCanEnterTransition=true)]
 }
 Main {
  path=AnimGraph.AnimGraphNode_StateMachine_1.FullBody_1.AnimStateNode_1.Main
  nodes [
   AnimGraphNode_StateMachine_0 = Main() {is=title AnimGraphNode_StateMachine}
   Output Animation Pose(Result=@AnimGraphNode_StateMachine_0)
  ]
 }
 Main {
  entry=OnGround
  path=AnimGraph.AnimGraphNode_StateMachine_1.FullBody_1.AnimStateNode_1.Main.AnimGraphNode_StateMachine_0.Main
  ids=id
  nodes [
   AnimStateConduitNode_0 = NotOnGround() {is=title AnimStateConduitNode}
   AnimStateNode_3 = InAir() {is=title AnimStateNode, bAlwaysResetOnEntry=true}
   AnimStateNode_35 = OnGround() {is=title AnimStateNode}
   AnimStateNode_4 = StartJump() {stateEntered {guid=#19, notifyName=StartJump_Entered_Test}}
   AnimStateNode_7 = JumpLand() {stateLeft {guid=#20, notifyName=JumpLandExit}}
   AnimStateTransitionNode_10 = AnimStateTransitionNode()
   AnimStateTransitionNode_11 = AnimStateTransitionNode()
   AnimStateTransitionNode_12 = AnimStateTransitionNode()
   AnimStateTransitionNode_17 = AnimStateTransitionNode()
   AnimStateTransitionNode_20 = AnimStateTransitionNode()
   AnimStateTransitionNode_5 = AnimStateTransitionNode() {bAutomaticRuleBasedOnSequencePlayerInState=true}
   AnimStateTransitionNode_8 = AnimStateTransitionNode() {bAutomaticRuleBasedOnSequencePlayerInState=true}
   AnimStateTransitionNode_9 = AnimStateTransitionNode()
  ]
  links [
   {
    dst=JumpLand
    rule=AnimGraph.AnimGraphNode_StateMachine_1.FullBody_1.AnimStateNode_1.Main.AnimGraphNode_StateMachine_0.Main.AnimStateTransitionNode_17.AnimationTransitionGraph_1
    src=InAir
    via=AnimStateTransitionNode_17
   }
   {
    dst=NotOnGround
    rule=AnimGraph.AnimGraphNode_StateMachine_1.FullBody_1.AnimStateNode_1.Main.AnimGraphNode_StateMachine_0.Main.AnimStateTransitionNode_12.AnimationTransitionGraph_1
    src=JumpLand
    via=AnimStateTransitionNode_12
   }
   {
    dst=OnGround
    rule=AnimGraph.AnimGraphNode_StateMachine_1.FullBody_1.AnimStateNode_1.Main.AnimGraphNode_StateMachine_0.Main.AnimStateTransitionNode_5.AnimationTransitionGraph_1
    src=JumpLand
    via=AnimStateTransitionNode_5
   }
   {
    dst=InAir
    rule=AnimGraph.AnimGraphNode_StateMachine_1.FullBody_1.AnimStateNode_1.Main.AnimGraphNode_StateMachine_0.Main.AnimStateTransitionNode_11.AnimationTransitionGraph_1
    src=NotOnGround
    via=AnimStateTransitionNode_11
   }
   {
    dst=StartJump
    rule=AnimGraph.AnimGraphNode_StateMachine_1.FullBody_1.AnimStateNode_1.Main.AnimGraphNode_StateMachine_0.Main.AnimStateTransitionNode_10.AnimationTransitionGraph_1
    src=NotOnGround
    via=AnimStateTransitionNode_10
   }
   {
    dst=NotOnGround
    rule=AnimGraph.AnimGraphNode_StateMachine_1.FullBody_1.AnimStateNode_1.Main.AnimGraphNode_StateMachine_0.Main.AnimStateTransitionNode_9.AnimationTransitionGraph_1
    src=OnGround
    via=AnimStateTransitionNode_9
   }
   {
    dst=InAir
    rule=AnimGraph.AnimGraphNode_StateMachine_1.FullBody_1.AnimStateNode_1.Main.AnimGraphNode_StateMachine_0.Main.AnimStateTransitionNode_20.AnimationTransitionGraph_1
    src=StartJump
    via=AnimStateTransitionNode_20
   }
   {
    dst=InAir
    rule=AnimGraph.AnimGraphNode_StateMachine_1.FullBody_1.AnimStateNode_1.Main.AnimGraphNode_StateMachine_0.Main.AnimStateTransitionNode_8.AnimationTransitionGraph_1
    src=StartJump
    via=AnimStateTransitionNode_8
   }
  ]
 }
 NotOnGround {
  path=AnimGraph.AnimGraphNode_StateMachine_1.FullBody_1.AnimStateNode_1.Main.AnimGraphNode_StateMachine_0.Main.AnimStateConduitNode_0.NotOnGround
  nodes [Result(bCanEnterTransition=true)]
 }
 InAir {
  path=AnimGraph.AnimGraphNode_StateMachine_1.FullBody_1.AnimStateNode_1.Main.AnimGraphNode_StateMachine_0.Main.AnimStateNode_3.InAir
  nodes [
   AnimGraphNode_StateMachine_0 = InAir() {is=title AnimGraphNode_StateMachine}
   Output Animation Pose(Result=@AnimGraphNode_StateMachine_0)
  ]
 }
 InAir {
  entry=Falling
  path=AnimGraph.AnimGraphNode_StateMachine_1.FullBody_1.AnimStateNode_1.Main.AnimGraphNode_StateMachine_0.Main.AnimStateNode_3.InAir.AnimGraphNode_StateMachine_0.InAir
  ids=id
  nodes [
   AnimStateNode_0 = Falling() {is=title AnimStateNode}
   AnimStateNode_26 = InAirHunker()
   AnimStateNode_34 = SkyDive()
   AnimStateTransitionNode_13 = AnimStateTransitionNode()
   AnimStateTransitionNode_18 = AnimStateTransitionNode()
   AnimStateTransitionNode_19 = AnimStateTransitionNode()
  ]
  links [
   {
    dst=InAirHunker
    rule=AnimGraph.AnimGraphNode_StateMachine_1.FullBody_1.AnimStateNode_1.Main.AnimGraphNode_StateMachine_0.Main.AnimStateNode_3.InAir.AnimGraphNode_StateMachine_0.InAir.AnimStateTransitionNode_18.Transition
    src=Falling
    via=AnimStateTransitionNode_18
   }
   {
    dst=SkyDive
    rule=AnimGraph.AnimGraphNode_StateMachine_1.FullBody_1.AnimStateNode_1.Main.AnimGraphNode_StateMachine_0.Main.AnimStateNode_3.InAir.AnimGraphNode_StateMachine_0.InAir.AnimStateTransitionNode_13.Transition
    src=Falling
    via=AnimStateTransitionNode_13
   }
   {
    dst=Falling
    rule=AnimGraph.AnimGraphNode_StateMachine_1.FullBody_1.AnimStateNode_1.Main.AnimGraphNode_StateMachine_0.Main.AnimStateNode_3.InAir.AnimGraphNode_StateMachine_0.InAir.AnimStateTransitionNode_19.Transition
    src=InAirHunker
    via=AnimStateTransitionNode_19
   }
  ]
 }
 Falling {
  path=AnimGraph.AnimGraphNode_StateMachine_1.FullBody_1.AnimStateNode_1.Main.AnimGraphNode_StateMachine_0.Main.AnimStateNode_3.InAir.AnimGraphNode_StateMachine_0.InAir.AnimStateNode_0.Falling
  nodes [
   AnimGraphNode_StateMachine_0 = Falling() {is=title AnimGraphNode_StateMachine}
   Output Animation Pose(Result=@AnimGraphNode_StateMachine_0)
  ]
 }
 Falling {
  entry=JumpLoop
  path=AnimGraph.AnimGraphNode_StateMachine_1.FullBody_1.AnimStateNode_1.Main.AnimGraphNode_StateMachine_0.Main.AnimStateNode_3.InAir.AnimGraphNode_StateMachine_0.InAir.AnimStateNode_0.Falling.AnimGraphNode_StateMachine_0.Falling
  ids=id
  nodes [
   AnimStateNode_5 = JumpLoop()
   AnimStateNode_6 = PrepForLanding() {is=title AnimStateNode}
   AnimStateTransitionNode_1 = AnimStateTransitionNode()
   AnimStateTransitionNode_2 = AnimStateTransitionNode()
   AnimStateTransitionNode_3 = AnimStateTransitionNode()
  ]
  links [
   {
    dst=PrepForLanding
    rule=AnimGraph.AnimGraphNode_StateMachine_1.FullBody_1.AnimStateNode_1.Main.AnimGraphNode_StateMachine_0.Main.AnimStateNode_3.InAir.AnimGraphNode_StateMachine_0.InAir.AnimStateNode_0.Falling.AnimGraphNode_StateMachine_0.Falling.AnimStateTransitionNode_1.Transition
    src=JumpLoop
    via=AnimStateTransitionNode_1
   }
   {
    dst=PrepForLanding
    rule=AnimGraph.AnimGraphNode_StateMachine_1.FullBody_1.AnimStateNode_1.Main.AnimGraphNode_StateMachine_0.Main.AnimStateNode_3.InAir.AnimGraphNode_StateMachine_0.InAir.AnimStateNode_0.Falling.AnimGraphNode_StateMachine_0.Falling.AnimStateTransitionNode_2.Transition
    src=JumpLoop
    via=AnimStateTransitionNode_2
   }
   {
    dst=PrepForLanding
    rule=AnimGraph.AnimGraphNode_StateMachine_1.FullBody_1.AnimStateNode_1.Main.AnimGraphNode_StateMachine_0.Main.AnimStateNode_3.InAir.AnimGraphNode_StateMachine_0.InAir.AnimStateNode_0.Falling.AnimGraphNode_StateMachine_0.Falling.AnimStateTransitionNode_3.Transition
    src=JumpLoop
    via=AnimStateTransitionNode_3
   }
  ]
 }
 JumpLoop {
  path=AnimGraph.AnimGraphNode_StateMachine_1.FullBody_1.AnimStateNode_1.Main.AnimGraphNode_StateMachine_0.Main.AnimStateNode_3.InAir.AnimGraphNode_StateMachine_0.InAir.AnimStateNode_0.Falling.AnimGraphNode_StateMachine_0.Falling.AnimStateNode_5.JumpLoop
  nodes [
   Layered blend per bone(BlendPoses_0=Local Space Ref Pose(), BlendWeights_0="To Float (Boolean)"(InBool=Get bIsFox())) {layerSetup [{branchFilters [{boneName=faceAttach}]}]}
   n1 = Blend Poses by int(
    BlendPose_0=Sequence Player 'Jump_ApexLoop_01'() {sequence=&Jump_ApexLoop_01}
    BlendPose_1=Sequence Player 'Jump_ApexLoop_02'() {sequence=&Jump_ApexLoop_02}
    BlendTime_0=0.1
    BlendTime_1=0.1
    ActiveChildIndex=Get JumpIndex() -> (JumpIndex=0)
   ) {blendType=Linear, blendPose [{linkId=-1, sourceLinkId=-1}, {linkId=-1, sourceLinkId=-1}], blendTime [0.1,0.1]}
   Output Animation Pose(Result=Blend Poses by bool(
     BlendPose_0=@n1
     BlendPose_1=Sequence Player 'Jump_ApexLoop_00'() {sequence=&Jump_ApexLoop_00}
     bActiveValue=Get bIsMovingDuringJumpInitiated()
    ) {blendType=Linear})
  ]
 }
 PrepForLanding {
  path=AnimGraph.AnimGraphNode_StateMachine_1.FullBody_1.AnimStateNode_1.Main.AnimGraphNode_StateMachine_0.Main.AnimStateNode_3.InAir.AnimGraphNode_StateMachine_0.InAir.AnimStateNode_0.Falling.AnimGraphNode_StateMachine_0.Falling.AnimStateNode_6.PrepForLanding
  nodes [
   n1 = Blend Poses by int(
    BlendPose_0=Sequence Player 'JumpFall_01'!(bLoopAnimation:bool=False) -> (Pose) {bLoopAnimation=false, sequence=&JumpFall_01}
    BlendPose_1=Sequence Player 'JumpFall_02'!(bLoopAnimation:bool=False) -> (Pose) {bLoopAnimation=false, sequence=&JumpFall_02}
    BlendTime_0=0.1
    BlendTime_1=0.1
    ActiveChildIndex=Get JumpIndex() -> (JumpIndex=0)
   ) {blendType=Linear, blendPose [{linkId=-1, sourceLinkId=-1}, {linkId=-1, sourceLinkId=-1}], blendTime [0.1,0.1]}
   Output Animation Pose(Result=Blend Poses by bool(
     BlendPose_0=@n1
     BlendPose_1=Sequence Player 'JumpFall_00'!(bLoopAnimation:bool=False) -> (Pose) {bLoopAnimation=false, sequence=&JumpFall_00}
     bActiveValue=Get bIsMovingDuringJumpInitiated()
    ) {blendType=Linear})
  ]
 }
 Transition {
  path=AnimGraph.AnimGraphNode_StateMachine_1.FullBody_1.AnimStateNode_1.Main.AnimGraphNode_StateMachine_0.Main.AnimStateNode_3.InAir.AnimGraphNode_StateMachine_0.InAir.AnimStateNode_0.Falling.AnimGraphNode_StateMachine_0.Falling.AnimStateTransitionNode_1.Transition
  nodes [Result(bCanEnterTransition=Get bLandingPredicted())]
 }
 Transition {
  path=AnimGraph.AnimGraphNode_StateMachine_1.FullBody_1.AnimStateNode_1.Main.AnimGraphNode_StateMachine_0.Main.AnimStateNode_3.InAir.AnimGraphNode_StateMachine_0.InAir.AnimStateNode_0.Falling.AnimGraphNode_StateMachine_0.Falling.AnimStateTransitionNode_2.Transition
  nodes [Result(bCanEnterTransition=Get bIsOnGround())]
 }
 Transition {
  path=AnimGraph.AnimGraphNode_StateMachine_1.FullBody_1.AnimStateNode_1.Main.AnimGraphNode_StateMachine_0.Main.AnimStateNode_3.InAir.AnimGraphNode_StateMachine_0.InAir.AnimStateNode_0.Falling.AnimGraphNode_StateMachine_0.Falling.AnimStateTransitionNode_3.Transition
  nodes [Result(bCanEnterTransition=Get bIsSlopeSliding())]
 }
 InAirHunker {
  path=AnimGraph.AnimGraphNode_StateMachine_1.FullBody_1.AnimStateNode_1.Main.AnimGraphNode_StateMachine_0.Main.AnimStateNode_3.InAir.AnimGraphNode_StateMachine_0.InAir.AnimStateNode_26.InAirHunker
  nodes [Output Animation Pose(Result=Apply Additive(
     Base=Sequence Player 'Hunker_Pose'() {sequence=&Hunker_Pose}
     Additive="Sequence Player (additive) 'IdleStraightLookAround_Add'"() {sequence=&IdleStraightLookAround_Add}
    ) {alphaBoolBlend {blendOption=Linear}})]
 }
 SkyDive {
  path=AnimGraph.AnimGraphNode_StateMachine_1.FullBody_1.AnimStateNode_1.Main.AnimGraphNode_StateMachine_0.Main.AnimStateNode_3.InAir.AnimGraphNode_StateMachine_0.InAir.AnimStateNode_34.SkyDive
  nodes [
   AnimGraphNode_StateMachine_0 = SkydiveParachute()
   Output Animation Pose(Result=@AnimGraphNode_StateMachine_0)
  ]
 }
 SkydiveParachute {
  entry=SkydiveStart
  path=AnimGraph.AnimGraphNode_StateMachine_1.FullBody_1.AnimStateNode_1.Main.AnimGraphNode_StateMachine_0.Main.AnimStateNode_3.InAir.AnimGraphNode_StateMachine_0.InAir.AnimStateNode_34.SkyDive.AnimGraphNode_StateMachine_0.SkydiveParachute
  ids=id
  nodes [
   AnimStateNode_1 = Parachute()
   AnimStateNode_2 = SkyDive()
   AnimStateNode_26 = SkidiveGlideToParachute()
   AnimStateNode_28 = SkydiveStart()
   AnimStateNode_30 = ParachuteToSkydiveGlide()
   AnimStateTransitionNode_26 = AnimStateTransitionNode()
   AnimStateTransitionNode_27 = AnimStateTransitionNode()
   AnimStateTransitionNode_28 = AnimStateTransitionNode() {bAutomaticRuleBasedOnSequencePlayerInState=true}
   AnimStateTransitionNode_60 = AnimStateTransitionNode() {bAutomaticRuleBasedOnSequencePlayerInState=true}
   AnimStateTransitionNode_61 = AnimStateTransitionNode()
  ]
  links [
   {
    dst=ParachuteToSkydiveGlide
    rule=AnimGraph.AnimGraphNode_StateMachine_1.FullBody_1.AnimStateNode_1.Main.AnimGraphNode_StateMachine_0.Main.AnimStateNode_3.InAir.AnimGraphNode_StateMachine_0.InAir.AnimStateNode_34.SkyDive.AnimGraphNode_StateMachine_0.SkydiveParachute.AnimStateTransitionNode_61.AnimationTransitionGraph_1
    src=Parachute
    via=AnimStateTransitionNode_61
   }
   {
    dst=SkyDive
    rule=AnimGraph.AnimGraphNode_StateMachine_1.FullBody_1.AnimStateNode_1.Main.AnimGraphNode_StateMachine_0.Main.AnimStateNode_3.InAir.AnimGraphNode_StateMachine_0.InAir.AnimStateNode_34.SkyDive.AnimGraphNode_StateMachine_0.SkydiveParachute.AnimStateTransitionNode_27.Transition
    src=ParachuteToSkydiveGlide
    via=AnimStateTransitionNode_27
   }
   {
    dst=Parachute
    rule=AnimGraph.AnimGraphNode_StateMachine_1.FullBody_1.AnimStateNode_1.Main.AnimGraphNode_StateMachine_0.Main.AnimStateNode_3.InAir.AnimGraphNode_StateMachine_0.InAir.AnimStateNode_34.SkyDive.AnimGraphNode_StateMachine_0.SkydiveParachute.AnimStateTransitionNode_60.AnimationTransitionGraph_1
    src=SkidiveGlideToParachute
    via=AnimStateTransitionNode_60
   }
   {
    dst=SkidiveGlideToParachute
    rule=AnimGraph.AnimGraphNode_StateMachine_1.FullBody_1.AnimStateNode_1.Main.AnimGraphNode_StateMachine_0.Main.AnimStateNode_3.InAir.AnimGraphNode_StateMachine_0.InAir.AnimStateNode_34.SkyDive.AnimGraphNode_StateMachine_0.SkydiveParachute.AnimStateTransitionNode_26.Transition
    src=SkyDive
    via=AnimStateTransitionNode_26
   }
   {
    dst=SkyDive
    rule=AnimGraph.AnimGraphNode_StateMachine_1.FullBody_1.AnimStateNode_1.Main.AnimGraphNode_StateMachine_0.Main.AnimStateNode_3.InAir.AnimGraphNode_StateMachine_0.InAir.AnimStateNode_34.SkyDive.AnimGraphNode_StateMachine_0.SkydiveParachute.AnimStateTransitionNode_28.Transition
    src=SkydiveStart
    via=AnimStateTransitionNode_28
   }
  ]
 }
 Parachute {
  path=AnimGraph.AnimGraphNode_StateMachine_1.FullBody_1.AnimStateNode_1.Main.AnimGraphNode_StateMachine_0.Main.AnimStateNode_3.InAir.AnimGraphNode_StateMachine_0.InAir.AnimStateNode_34.SkyDive.AnimGraphNode_StateMachine_0.SkydiveParachute.AnimStateNode_1.Parachute
  nodes [Output Animation Pose(Result=Sequence Player 'Idle'() {sequence=&Idle})]
 }
 SkyDive {
  path=AnimGraph.AnimGraphNode_StateMachine_1.FullBody_1.AnimStateNode_1.Main.AnimGraphNode_StateMachine_0.Main.AnimStateNode_3.InAir.AnimGraphNode_StateMachine_0.InAir.AnimStateNode_34.SkyDive.AnimGraphNode_StateMachine_0.SkydiveParachute.AnimStateNode_2.SkyDive
  nodes [
   AnimGraphNode_StateMachine_0 = Skydive()
   Output Animation Pose(Result=@AnimGraphNode_StateMachine_0)
  ]
 }
 Skydive {
  entry=SkydiveGlide
  path=AnimGraph.AnimGraphNode_StateMachine_1.FullBody_1.AnimStateNode_1.Main.AnimGraphNode_StateMachine_0.Main.AnimStateNode_3.InAir.AnimGraphNode_StateMachine_0.InAir.AnimStateNode_34.SkyDive.AnimGraphNode_StateMachine_0.SkydiveParachute.AnimStateNode_2.SkyDive.AnimGraphNode_StateMachine_0.Skydive
  ids=id
  nodes [
   AnimStateNode_0 = SkydiveGlide()
   AnimStateNode_19 = ToDive()
   AnimStateNode_33 = ToGlide()
   AnimStateNode_8 = SkydiveDive()
   AnimStateTransitionNode_0 = AnimStateTransitionNode()
   AnimStateTransitionNode_1 = AnimStateTransitionNode()
   AnimStateTransitionNode_29 = AnimStateTransitionNode()
   AnimStateTransitionNode_31 = AnimStateTransitionNode() {bAutomaticRuleBasedOnSequencePlayerInState=true}
   AnimStateTransitionNode_70 = AnimStateTransitionNode()
   AnimStateTransitionNode_71 = AnimStateTransitionNode() {bAutomaticRuleBasedOnSequencePlayerInState=true}
  ]
  links [
   {
    dst=ToGlide
    rule=AnimGraph.AnimGraphNode_StateMachine_1.FullBody_1.AnimStateNode_1.Main.AnimGraphNode_StateMachine_0.Main.AnimStateNode_3.InAir.AnimGraphNode_StateMachine_0.InAir.AnimStateNode_34.SkyDive.AnimGraphNode_StateMachine_0.SkydiveParachute.AnimStateNode_2.SkyDive.AnimGraphNode_StateMachine_0.Skydive.AnimStateTransitionNode_70.AnimationTransitionGraph_1
    src=SkydiveDive
    via=AnimStateTransitionNode_70
   }
   {
    dst=ToDive
    rule=AnimGraph.AnimGraphNode_StateMachine_1.FullBody_1.AnimStateNode_1.Main.AnimGraphNode_StateMachine_0.Main.AnimStateNode_3.InAir.AnimGraphNode_StateMachine_0.InAir.AnimStateNode_34.SkyDive.AnimGraphNode_StateMachine_0.SkydiveParachute.AnimStateNode_2.SkyDive.AnimGraphNode_StateMachine_0.Skydive.AnimStateTransitionNode_29.AnimationTransitionGraph_1
    src=SkydiveGlide
    via=AnimStateTransitionNode_29
   }
   {
    dst=SkydiveDive
    rule=AnimGraph.AnimGraphNode_StateMachine_1.FullBody_1.AnimStateNode_1.Main.AnimGraphNode_StateMachine_0.Main.AnimStateNode_3.InAir.AnimGraphNode_StateMachine_0.InAir.AnimStateNode_34.SkyDive.AnimGraphNode_StateMachine_0.SkydiveParachute.AnimStateNode_2.SkyDive.AnimGraphNode_StateMachine_0.Skydive.AnimStateTransitionNode_31.AnimationTransitionGraph_1
    src=ToDive
    via=AnimStateTransitionNode_31
   }
   {
    dst=ToGlide
    rule=AnimGraph.AnimGraphNode_StateMachine_1.FullBody_1.AnimStateNode_1.Main.AnimGraphNode_StateMachine_0.Main.AnimStateNode_3.InAir.AnimGraphNode_StateMachine_0.InAir.AnimStateNode_34.SkyDive.AnimGraphNode_StateMachine_0.SkydiveParachute.AnimStateNode_2.SkyDive.AnimGraphNode_StateMachine_0.Skydive.AnimStateTransitionNode_1.AnimationTransitionGraph_0
    src=ToDive
    via=AnimStateTransitionNode_1
   }
   {
    dst=SkydiveGlide
    rule=AnimGraph.AnimGraphNode_StateMachine_1.FullBody_1.AnimStateNode_1.Main.AnimGraphNode_StateMachine_0.Main.AnimStateNode_3.InAir.AnimGraphNode_StateMachine_0.InAir.AnimStateNode_34.SkyDive.AnimGraphNode_StateMachine_0.SkydiveParachute.AnimStateNode_2.SkyDive.AnimGraphNode_StateMachine_0.Skydive.AnimStateTransitionNode_71.AnimationTransitionGraph_1
    src=ToGlide
    via=AnimStateTransitionNode_71
   }
   {
    dst=ToDive
    rule=AnimGraph.AnimGraphNode_StateMachine_1.FullBody_1.AnimStateNode_1.Main.AnimGraphNode_StateMachine_0.Main.AnimStateNode_3.InAir.AnimGraphNode_StateMachine_0.InAir.AnimStateNode_34.SkyDive.AnimGraphNode_StateMachine_0.SkydiveParachute.AnimStateNode_2.SkyDive.AnimGraphNode_StateMachine_0.Skydive.AnimStateTransitionNode_0.AnimationTransitionGraph_0
    src=ToGlide
    via=AnimStateTransitionNode_0
   }
  ]
 }
 SkydiveGlide {
  path=AnimGraph.AnimGraphNode_StateMachine_1.FullBody_1.AnimStateNode_1.Main.AnimGraphNode_StateMachine_0.Main.AnimStateNode_3.InAir.AnimGraphNode_StateMachine_0.InAir.AnimStateNode_34.SkyDive.AnimGraphNode_StateMachine_0.SkydiveParachute.AnimStateNode_2.SkyDive.AnimGraphNode_StateMachine_0.Skydive.AnimStateNode_0.SkydiveGlide
  nodes [
   n1 = Break Fort Anim Input Skydiving(FortAnimInput_Skydiving=Get SkydivingState()) -> (
    ParaGilderRootOffset=~
    ParaGliderRootRotationOverride=~
    bUseParaGlideRootModifier=~
    bIsSkydivingFromLaunchPad=~
    bIsSkydivingFromBus=~
    bIsInVortex=~
    bIsDiving=~
    bIsParachuteOpen=~
    bIsSkydiving=~
    bIsParachuteLeaning=~
    bIsSkydiveLeaning=~
    bIsLeaning=~
    bIsSkydiveDiveMode=~
    bPlaySkydiveDrift=~
    bSkydiveDriftDelayActive=~
    bSkydiveDriftAnimAllowed=~
    bIsGliderRight=~
    bIsGliderCenter=~
    bIsGliderLeft=~
    bIsGliderForward=~
    bIsGliderBack=~
    bHasDeployedAndIsStandingUpright=~
    LocalAccelForward=~
    LocalAccelRight=~
    LocalVelocityRight=~
    SkydiveAimPitch=~
    SkydiveAimYaw=~
    DeployChuteAnimRate=~
    SkydiveDriftAnimRate=~
    SkydiveDriftAnimRateCurrent=~
    SkydiveFidgetAnimRate=~
    SkydiveFidgetAnimRateCurrent=~
    SkydiveAdditiveAlpha=~
    SkydiveDriftDelay=~
   )
   n2 = Blendspace Player 'BS_Glide_Lean'(X=@n1.LocalVelocityRight, Y=@n1.SkydiveAimPitch) {blendSpace=&BS_Glide_Lean}
   n3 = "Sequence Player (additive) 'Glide_Add'"() {sequence=&Glide_Add}
   n4 = Blendspace Player 'BS_Glide_2d'(X=@n1.SkydiveAimYaw, Y=@n1.SkydiveAimPitch) {blendSpace=&BS_Glide_2d}
   n5 = "Sequence Player (additive) 'Glide_Add'"() {sequence=&Glide_Add}
   Output Animation Pose(Result=Blend Poses by bool(
     BlendPose_0=Apply Additive(Base=@n2, Additive=@n3) {alphaBoolBlend {blendOption=Linear}}
     BlendPose_1=Apply Additive(Base=@n4, Additive=@n5) {alphaBoolBlend {blendOption=Linear}}
     bActiveValue=@n1.bIsSkydiveLeaning
    ) {blendType=Linear})
  ]
 }
 ToDive {
  path=AnimGraph.AnimGraphNode_StateMachine_1.FullBody_1.AnimStateNode_1.Main.AnimGraphNode_StateMachine_0.Main.AnimStateNode_3.InAir.AnimGraphNode_StateMachine_0.InAir.AnimStateNode_34.SkyDive.AnimGraphNode_StateMachine_0.SkydiveParachute.AnimStateNode_2.SkyDive.AnimGraphNode_StateMachine_0.Skydive.AnimStateNode_19.ToDive
  nodes [Output Animation Pose(Result=Sequence Player 'GlideToDive'() {sequence=&GlideToDive})]
 }
 ToGlide {
  path=AnimGraph.AnimGraphNode_StateMachine_1.FullBody_1.AnimStateNode_1.Main.AnimGraphNode_StateMachine_0.Main.AnimStateNode_3.InAir.AnimGraphNode_StateMachine_0.InAir.AnimStateNode_34.SkyDive.AnimGraphNode_StateMachine_0.SkydiveParachute.AnimStateNode_2.SkyDive.AnimGraphNode_StateMachine_0.Skydive.AnimStateNode_33.ToGlide
  nodes [Output Animation Pose(Result=Sequence Player 'DiveToGlide'() {sequence=&DiveToGlide})]
 }
 SkydiveDive {
  path=AnimGraph.AnimGraphNode_StateMachine_1.FullBody_1.AnimStateNode_1.Main.AnimGraphNode_StateMachine_0.Main.AnimStateNode_3.InAir.AnimGraphNode_StateMachine_0.InAir.AnimStateNode_34.SkyDive.AnimGraphNode_StateMachine_0.SkydiveParachute.AnimStateNode_2.SkyDive.AnimGraphNode_StateMachine_0.Skydive.AnimStateNode_8.SkydiveDive
  nodes [
   n1 = Break Fort Anim Input Skydiving(FortAnimInput_Skydiving=Get SkydivingState()) -> (
    ParaGilderRootOffset=~
    ParaGliderRootRotationOverride=~
    bUseParaGlideRootModifier=~
    bIsSkydivingFromLaunchPad=~
    bIsSkydivingFromBus=~
    bIsInVortex=~
    bIsDiving=~
    bIsParachuteOpen=~
    bIsSkydiving=~
    bIsParachuteLeaning=~
    bIsSkydiveLeaning=~
    bIsLeaning=~
    bIsSkydiveDiveMode=~
    bPlaySkydiveDrift=~
    bSkydiveDriftDelayActive=~
    bSkydiveDriftAnimAllowed=~
    bIsGliderRight=~
    bIsGliderCenter=~
    bIsGliderLeft=~
    bIsGliderForward=~
    bIsGliderBack=~
    bHasDeployedAndIsStandingUpright=~
    LocalAccelForward=~
    LocalAccelRight=~
    LocalVelocityRight=~
    SkydiveAimPitch=~
    SkydiveAimYaw=~
    DeployChuteAnimRate=~
    SkydiveDriftAnimRate=~
    SkydiveDriftAnimRateCurrent=~
    SkydiveFidgetAnimRate=~
    SkydiveFidgetAnimRateCurrent=~
    SkydiveAdditiveAlpha=~
    SkydiveDriftDelay=~
   )
   n2 = Blendspace Player 'BS_Dive_Lean'(X=@n1.LocalVelocityRight, Y=@n1.SkydiveAimPitch) {blendSpace=&BS_Dive_Lean}
   n3 = Blendspace Player 'GlideDive_Add'(X=@n1.SkydiveAimPitch) {blendSpace=&GlideDive_Add}
   n4 = Blendspace Player 'BS_Dive_2D'(X=@n1.SkydiveAimYaw, Y=@n1.SkydiveAimPitch) {blendSpace=&BS_Dive_2D}
   n5 = Blendspace Player 'GlideDive_Add'(X=@n1.SkydiveAimPitch) {blendSpace=&GlideDive_Add}
   Output Animation Pose(Result=Blend Poses by bool(
     BlendPose_0=Apply Additive(Base=@n2, Additive=@n3) {alphaBoolBlend {blendOption=Linear}}
     BlendPose_1=Apply Additive(Base=@n4, Additive=@n5) {alphaBoolBlend {blendOption=Linear}}
     bActiveValue=@n1.bIsSkydiveLeaning
    ) {blendType=Linear})
  ]
 }
 AnimationTransitionGraph_0 {
  path=AnimGraph.AnimGraphNode_StateMachine_1.FullBody_1.AnimStateNode_1.Main.AnimGraphNode_StateMachine_0.Main.AnimStateNode_3.InAir.AnimGraphNode_StateMachine_0.InAir.AnimStateNode_34.SkyDive.AnimGraphNode_StateMachine_0.SkydiveParachute.AnimStateNode_2.SkyDive.AnimGraphNode_StateMachine_0.Skydive.AnimStateTransitionNode_0.AnimationTransitionGraph_0
  nodes [Result(bCanEnterTransition=Break Fort Anim Input Skydiving(FortAnimInput_Skydiving=Get SkydivingState()) -> (
     ParaGilderRootOffset=~
     ParaGliderRootRotationOverride=~
     bUseParaGlideRootModifier=~
     bIsSkydivingFromLaunchPad=~
     bIsSkydivingFromBus=~
     bIsInVortex=~
     bIsDiving=~
     bIsParachuteOpen=~
     bIsSkydiving=~
     bIsParachuteLeaning=~
     bIsSkydiveLeaning=~
     bIsLeaning=~
     bIsSkydiveDiveMode=~
     bPlaySkydiveDrift=~
     bSkydiveDriftDelayActive=~
     bSkydiveDriftAnimAllowed=~
     bIsGliderRight=~
     bIsGliderCenter=~
     bIsGliderLeft=~
     bIsGliderForward=~
     bIsGliderBack=~
     bHasDeployedAndIsStandingUpright=~
     LocalAccelForward=~
     LocalAccelRight=~
     LocalVelocityRight=~
     SkydiveAimPitch=~
     SkydiveAimYaw=~
     DeployChuteAnimRate=~
     SkydiveDriftAnimRate=~
     SkydiveDriftAnimRateCurrent=~
     SkydiveFidgetAnimRate=~
     SkydiveFidgetAnimRateCurrent=~
     SkydiveAdditiveAlpha=~
     SkydiveDriftDelay=~
    ).bIsDiving)]
 }
 AnimationTransitionGraph_0 {
  path=AnimGraph.AnimGraphNode_StateMachine_1.FullBody_1.AnimStateNode_1.Main.AnimGraphNode_StateMachine_0.Main.AnimStateNode_3.InAir.AnimGraphNode_StateMachine_0.InAir.AnimStateNode_34.SkyDive.AnimGraphNode_StateMachine_0.SkydiveParachute.AnimStateNode_2.SkyDive.AnimGraphNode_StateMachine_0.Skydive.AnimStateTransitionNode_1.AnimationTransitionGraph_0
  nodes [
   n1 = Get SkydivingState()
   Result(bCanEnterTransition=NOT Boolean(A=Break Fort Anim Input Skydiving(FortAnimInput_Skydiving=@n1) -> (
      ParaGilderRootOffset=~
      ParaGliderRootRotationOverride=~
      bUseParaGlideRootModifier=~
      bIsSkydivingFromLaunchPad=~
      bIsSkydivingFromBus=~
      bIsInVortex=~
      bIsDiving=~
      bIsParachuteOpen=~
      bIsSkydiving=~
      bIsParachuteLeaning=~
      bIsSkydiveLeaning=~
      bIsLeaning=~
      bIsSkydiveDiveMode=~
      bPlaySkydiveDrift=~
      bSkydiveDriftDelayActive=~
      bSkydiveDriftAnimAllowed=~
      bIsGliderRight=~
      bIsGliderCenter=~
      bIsGliderLeft=~
      bIsGliderForward=~
      bIsGliderBack=~
      bHasDeployedAndIsStandingUpright=~
      LocalAccelForward=~
      LocalAccelRight=~
      LocalVelocityRight=~
      SkydiveAimPitch=~
      SkydiveAimYaw=~
      DeployChuteAnimRate=~
      SkydiveDriftAnimRate=~
      SkydiveDriftAnimRateCurrent=~
      SkydiveFidgetAnimRate=~
      SkydiveFidgetAnimRateCurrent=~
      SkydiveAdditiveAlpha=~
      SkydiveDriftDelay=~
     ).bIsDiving))
  ]
 }
 AnimationTransitionGraph_1 {
  path=AnimGraph.AnimGraphNode_StateMachine_1.FullBody_1.AnimStateNode_1.Main.AnimGraphNode_StateMachine_0.Main.AnimStateNode_3.InAir.AnimGraphNode_StateMachine_0.InAir.AnimStateNode_34.SkyDive.AnimGraphNode_StateMachine_0.SkydiveParachute.AnimStateNode_2.SkyDive.AnimGraphNode_StateMachine_0.Skydive.AnimStateTransitionNode_29.AnimationTransitionGraph_1
  nodes [Result(bCanEnterTransition=Break Fort Anim Input Skydiving(FortAnimInput_Skydiving=Get SkydivingState()) -> (
     ParaGilderRootOffset=~
     ParaGliderRootRotationOverride=~
     bUseParaGlideRootModifier=~
     bIsSkydivingFromLaunchPad=~
     bIsSkydivingFromBus=~
     bIsInVortex=~
     bIsDiving=~
     bIsParachuteOpen=~
     bIsSkydiving=~
     bIsParachuteLeaning=~
     bIsSkydiveLeaning=~
     bIsLeaning=~
     bIsSkydiveDiveMode=~
     bPlaySkydiveDrift=~
     bSkydiveDriftDelayActive=~
     bSkydiveDriftAnimAllowed=~
     bIsGliderRight=~
     bIsGliderCenter=~
     bIsGliderLeft=~
     bIsGliderForward=~
     bIsGliderBack=~
     bHasDeployedAndIsStandingUpright=~
     LocalAccelForward=~
     LocalAccelRight=~
     LocalVelocityRight=~
     SkydiveAimPitch=~
     SkydiveAimYaw=~
     DeployChuteAnimRate=~
     SkydiveDriftAnimRate=~
     SkydiveDriftAnimRateCurrent=~
     SkydiveFidgetAnimRate=~
     SkydiveFidgetAnimRateCurrent=~
     SkydiveAdditiveAlpha=~
     SkydiveDriftDelay=~
    ).bIsDiving)]
 }
 AnimationTransitionGraph_1 {
  path=AnimGraph.AnimGraphNode_StateMachine_1.FullBody_1.AnimStateNode_1.Main.AnimGraphNode_StateMachine_0.Main.AnimStateNode_3.InAir.AnimGraphNode_StateMachine_0.InAir.AnimStateNode_34.SkyDive.AnimGraphNode_StateMachine_0.SkydiveParachute.AnimStateNode_2.SkyDive.AnimGraphNode_StateMachine_0.Skydive.AnimStateTransitionNode_31.AnimationTransitionGraph_1
  nodes [Result()]
 }
 AnimationTransitionGraph_1 {
  path=AnimGraph.AnimGraphNode_StateMachine_1.FullBody_1.AnimStateNode_1.Main.AnimGraphNode_StateMachine_0.Main.AnimStateNode_3.InAir.AnimGraphNode_StateMachine_0.InAir.AnimStateNode_34.SkyDive.AnimGraphNode_StateMachine_0.SkydiveParachute.AnimStateNode_2.SkyDive.AnimGraphNode_StateMachine_0.Skydive.AnimStateTransitionNode_70.AnimationTransitionGraph_1
  nodes [
   n1 = Get SkydivingState()
   Result(bCanEnterTransition=NOT Boolean(A=Break Fort Anim Input Skydiving(FortAnimInput_Skydiving=@n1) -> (
      ParaGilderRootOffset=~
      ParaGliderRootRotationOverride=~
      bUseParaGlideRootModifier=~
      bIsSkydivingFromLaunchPad=~
      bIsSkydivingFromBus=~
      bIsInVortex=~
      bIsDiving=~
      bIsParachuteOpen=~
      bIsSkydiving=~
      bIsParachuteLeaning=~
      bIsSkydiveLeaning=~
      bIsLeaning=~
      bIsSkydiveDiveMode=~
      bPlaySkydiveDrift=~
      bSkydiveDriftDelayActive=~
      bSkydiveDriftAnimAllowed=~
      bIsGliderRight=~
      bIsGliderCenter=~
      bIsGliderLeft=~
      bIsGliderForward=~
      bIsGliderBack=~
      bHasDeployedAndIsStandingUpright=~
      LocalAccelForward=~
      LocalAccelRight=~
      LocalVelocityRight=~
      SkydiveAimPitch=~
      SkydiveAimYaw=~
      DeployChuteAnimRate=~
      SkydiveDriftAnimRate=~
      SkydiveDriftAnimRateCurrent=~
      SkydiveFidgetAnimRate=~
      SkydiveFidgetAnimRateCurrent=~
      SkydiveAdditiveAlpha=~
      SkydiveDriftDelay=~
     ).bIsDiving))
  ]
 }
 AnimationTransitionGraph_1 {
  path=AnimGraph.AnimGraphNode_StateMachine_1.FullBody_1.AnimStateNode_1.Main.AnimGraphNode_StateMachine_0.Main.AnimStateNode_3.InAir.AnimGraphNode_StateMachine_0.InAir.AnimStateNode_34.SkyDive.AnimGraphNode_StateMachine_0.SkydiveParachute.AnimStateNode_2.SkyDive.AnimGraphNode_StateMachine_0.Skydive.AnimStateTransitionNode_71.AnimationTransitionGraph_1
  nodes [Result()]
 }
 SkidiveGlideToParachute {
  path=AnimGraph.AnimGraphNode_StateMachine_1.FullBody_1.AnimStateNode_1.Main.AnimGraphNode_StateMachine_0.Main.AnimStateNode_3.InAir.AnimGraphNode_StateMachine_0.InAir.AnimStateNode_34.SkyDive.AnimGraphNode_StateMachine_0.SkydiveParachute.AnimStateNode_26.SkidiveGlideToParachute
  nodes [Output Animation Pose(Result=Sequence Player 'Para_Into'() {sequence=&Para_Into})]
 }
 SkydiveStart {
  path=AnimGraph.AnimGraphNode_StateMachine_1.FullBody_1.AnimStateNode_1.Main.AnimGraphNode_StateMachine_0.Main.AnimStateNode_3.InAir.AnimGraphNode_StateMachine_0.InAir.AnimStateNode_34.SkyDive.AnimGraphNode_StateMachine_0.SkydiveParachute.AnimStateNode_28.SkydiveStart
  nodes [Output Animation Pose(Result=Sequence Player 'SkydiveStart'() {bLoopAnimation=false, sequence=&SkydiveStart})]
 }
 ParachuteToSkydiveGlide {
  path=AnimGraph.AnimGraphNode_StateMachine_1.FullBody_1.AnimStateNode_1.Main.AnimGraphNode_StateMachine_0.Main.AnimStateNode_3.InAir.AnimGraphNode_StateMachine_0.InAir.AnimStateNode_34.SkyDive.AnimGraphNode_StateMachine_0.SkydiveParachute.AnimStateNode_30.ParachuteToSkydiveGlide
  nodes [Output Animation Pose(Result=Sequence Player 'SkydiveStart'() {sequence=&SkydiveStart})]
 }
 Transition {
  path=AnimGraph.AnimGraphNode_StateMachine_1.FullBody_1.AnimStateNode_1.Main.AnimGraphNode_StateMachine_0.Main.AnimStateNode_3.InAir.AnimGraphNode_StateMachine_0.InAir.AnimStateNode_34.SkyDive.AnimGraphNode_StateMachine_0.SkydiveParachute.AnimStateTransitionNode_26.Transition
  nodes [Result(bCanEnterTransition=Break Fort Anim Input Skydiving(FortAnimInput_Skydiving=Get SkydivingState()) -> (
     ParaGilderRootOffset=~
     ParaGliderRootRotationOverride=~
     bUseParaGlideRootModifier=~
     bIsSkydivingFromLaunchPad=~
     bIsSkydivingFromBus=~
     bIsInVortex=~
     bIsDiving=~
     bIsParachuteOpen=~
     bIsSkydiving=~
     bIsParachuteLeaning=~
     bIsSkydiveLeaning=~
     bIsLeaning=~
     bIsSkydiveDiveMode=~
     bPlaySkydiveDrift=~
     bSkydiveDriftDelayActive=~
     bSkydiveDriftAnimAllowed=~
     bIsGliderRight=~
     bIsGliderCenter=~
     bIsGliderLeft=~
     bIsGliderForward=~
     bIsGliderBack=~
     bHasDeployedAndIsStandingUpright=~
     LocalAccelForward=~
     LocalAccelRight=~
     LocalVelocityRight=~
     SkydiveAimPitch=~
     SkydiveAimYaw=~
     DeployChuteAnimRate=~
     SkydiveDriftAnimRate=~
     SkydiveDriftAnimRateCurrent=~
     SkydiveFidgetAnimRate=~
     SkydiveFidgetAnimRateCurrent=~
     SkydiveAdditiveAlpha=~
     SkydiveDriftDelay=~
    ).bIsParachuteOpen)]
 }
 Transition {
  path=AnimGraph.AnimGraphNode_StateMachine_1.FullBody_1.AnimStateNode_1.Main.AnimGraphNode_StateMachine_0.Main.AnimStateNode_3.InAir.AnimGraphNode_StateMachine_0.InAir.AnimStateNode_34.SkyDive.AnimGraphNode_StateMachine_0.SkydiveParachute.AnimStateTransitionNode_27.Transition
  nodes [
   n1 = Break Fort Anim Input Skydiving(FortAnimInput_Skydiving=Get SkydivingState()) -> (
    ParaGilderRootOffset=~
    ParaGliderRootRotationOverride=~
    bUseParaGlideRootModifier=~
    bIsSkydivingFromLaunchPad=~
    bIsSkydivingFromBus=~
    bIsInVortex=~
    bIsDiving=~
    bIsParachuteOpen=~
    bIsSkydiving=~
    bIsParachuteLeaning=~
    bIsSkydiveLeaning=~
    bIsLeaning=~
    bIsSkydiveDiveMode=~
    bPlaySkydiveDrift=~
    bSkydiveDriftDelayActive=~
    bSkydiveDriftAnimAllowed=~
    bIsGliderRight=~
    bIsGliderCenter=~
    bIsGliderLeft=~
    bIsGliderForward=~
    bIsGliderBack=~
    bHasDeployedAndIsStandingUpright=~
    LocalAccelForward=~
    LocalAccelRight=~
    LocalVelocityRight=~
    SkydiveAimPitch=~
    SkydiveAimYaw=~
    DeployChuteAnimRate=~
    SkydiveDriftAnimRate=~
    SkydiveDriftAnimRateCurrent=~
    SkydiveFidgetAnimRate=~
    SkydiveFidgetAnimRateCurrent=~
    SkydiveAdditiveAlpha=~
    SkydiveDriftDelay=~
   )
   n2 = "Get Relevant Anim Time Remaining (ParachuteToSkydiveGlide)"!() -> (ReturnValue:float) {is=title K2Node_AnimGetter}
   Result(bCanEnterTransition=OR Boolean(A="float <= float"!(A:double=@n2, B:double=0.2) -> (ReturnValue:bool) {is=title K2Node_CallFunction}, B=@n1.bIsDiving))
   AND Boolean(
    A=@n1.bIsDiving
    B="float > float"(A="Get Relevant Anim Time (ParachuteToSkydiveGlide)"!() -> (ReturnValue:float) {is=title K2Node_AnimGetter}, B=0.7) {is=title K2Node_CallFunction}
   )
  ]
 }
 Transition {
  path=AnimGraph.AnimGraphNode_StateMachine_1.FullBody_1.AnimStateNode_1.Main.AnimGraphNode_StateMachine_0.Main.AnimStateNode_3.InAir.AnimGraphNode_StateMachine_0.InAir.AnimStateNode_34.SkyDive.AnimGraphNode_StateMachine_0.SkydiveParachute.AnimStateTransitionNode_28.Transition
  nodes [Result()]
 }
 AnimationTransitionGraph_1 {
  path=AnimGraph.AnimGraphNode_StateMachine_1.FullBody_1.AnimStateNode_1.Main.AnimGraphNode_StateMachine_0.Main.AnimStateNode_3.InAir.AnimGraphNode_StateMachine_0.InAir.AnimStateNode_34.SkyDive.AnimGraphNode_StateMachine_0.SkydiveParachute.AnimStateTransitionNode_60.AnimationTransitionGraph_1
  nodes [Result()]
 }
 AnimationTransitionGraph_1 {
  path=AnimGraph.AnimGraphNode_StateMachine_1.FullBody_1.AnimStateNode_1.Main.AnimGraphNode_StateMachine_0.Main.AnimStateNode_3.InAir.AnimGraphNode_StateMachine_0.InAir.AnimStateNode_34.SkyDive.AnimGraphNode_StateMachine_0.SkydiveParachute.AnimStateTransitionNode_61.AnimationTransitionGraph_1
  nodes [
   n1 = Get SkydivingState()
   Result(bCanEnterTransition=NOT Boolean(A=Break Fort Anim Input Skydiving(FortAnimInput_Skydiving=@n1) -> (
      ParaGilderRootOffset=~
      ParaGliderRootRotationOverride=~
      bUseParaGlideRootModifier=~
      bIsSkydivingFromLaunchPad=~
      bIsSkydivingFromBus=~
      bIsInVortex=~
      bIsDiving=~
      bIsParachuteOpen=~
      bIsSkydiving=~
      bIsParachuteLeaning=~
      bIsSkydiveLeaning=~
      bIsLeaning=~
      bIsSkydiveDiveMode=~
      bPlaySkydiveDrift=~
      bSkydiveDriftDelayActive=~
      bSkydiveDriftAnimAllowed=~
      bIsGliderRight=~
      bIsGliderCenter=~
      bIsGliderLeft=~
      bIsGliderForward=~
      bIsGliderBack=~
      bHasDeployedAndIsStandingUpright=~
      LocalAccelForward=~
      LocalAccelRight=~
      LocalVelocityRight=~
      SkydiveAimPitch=~
      SkydiveAimYaw=~
      DeployChuteAnimRate=~
      SkydiveDriftAnimRate=~
      SkydiveDriftAnimRateCurrent=~
      SkydiveFidgetAnimRate=~
      SkydiveFidgetAnimRateCurrent=~
      SkydiveAdditiveAlpha=~
      SkydiveDriftDelay=~
     ).bIsParachuteOpen))
  ]
 }
 Transition {
  path=AnimGraph.AnimGraphNode_StateMachine_1.FullBody_1.AnimStateNode_1.Main.AnimGraphNode_StateMachine_0.Main.AnimStateNode_3.InAir.AnimGraphNode_StateMachine_0.InAir.AnimStateTransitionNode_13.Transition
  nodes [Result(bCanEnterTransition=Get bIsSkydivingInGame())]
 }
 Transition {
  path=AnimGraph.AnimGraphNode_StateMachine_1.FullBody_1.AnimStateNode_1.Main.AnimGraphNode_StateMachine_0.Main.AnimStateNode_3.InAir.AnimGraphNode_StateMachine_0.InAir.AnimStateTransitionNode_18.Transition
  nodes [Result(bCanEnterTransition=Get bIsInAirNotJumping())]
 }
 Transition {
  path=AnimGraph.AnimGraphNode_StateMachine_1.FullBody_1.AnimStateNode_1.Main.AnimGraphNode_StateMachine_0.Main.AnimStateNode_3.InAir.AnimGraphNode_StateMachine_0.InAir.AnimStateTransitionNode_19.Transition
  nodes [Result(bCanEnterTransition=NOT Boolean(A=Get bIsInAirNotJumping()))]
 }
 OnGround {
  path=AnimGraph.AnimGraphNode_StateMachine_1.FullBody_1.AnimStateNode_1.Main.AnimGraphNode_StateMachine_0.Main.AnimStateNode_35.OnGround
  nodes [
   AnimGraphNode_StateMachine_0 = OnGround() {is=title AnimGraphNode_StateMachine}
   Output Animation Pose(Result=@AnimGraphNode_StateMachine_0)
  ]
 }
 OnGround {
  entry=Normal
  path=AnimGraph.AnimGraphNode_StateMachine_1.FullBody_1.AnimStateNode_1.Main.AnimGraphNode_StateMachine_0.Main.AnimStateNode_35.OnGround.AnimGraphNode_StateMachine_0.OnGround
  ids=id
  nodes [
   AnimStateConduitNode_0 = IsTargeting() {is=title AnimStateConduitNode}
   AnimStateConduitNode_1 = IsNotTargeting() {is=title AnimStateConduitNode}
   AnimStateConduitNode_2 = IsCrouching() {is=title AnimStateConduitNode}
   AnimStateNode_0 = DownSights()
   AnimStateNode_13 = IdleToDownSights()
   AnimStateNode_14 = DownSightsToIdle()
   AnimStateNode_28 = CrouchMovingToDownSights() {is=title AnimStateNode}
   AnimStateNode_29 = CrouchToDownSights() {is=title AnimStateNode}
   AnimStateNode_30 = DownSightsToCrouch() {is=title AnimStateNode, stateEntered {guid=#21, notifyName=Entered_DownSightsToCrouch}}
   AnimStateNode_4 = Normal() {is=title AnimStateNode}
   AnimStateNode_5 = Hunker()
   AnimStateTransitionNode_27 = AnimStateTransitionNode()
   AnimStateTransitionNode_28 = AnimStateTransitionNode()
   AnimStateTransitionNode_29 = AnimStateTransitionNode() {bAutomaticRuleBasedOnSequencePlayerInState=true}
   AnimStateTransitionNode_32 = AnimStateTransitionNode() {bAutomaticRuleBasedOnSequencePlayerInState=true}
   AnimStateTransitionNode_33 = AnimStateTransitionNode() {bAutomaticRuleBasedOnSequencePlayerInState=true}
   AnimStateTransitionNode_36 = AnimStateTransitionNode()
   AnimStateTransitionNode_37 = AnimStateTransitionNode()
   AnimStateTransitionNode_38 = AnimStateTransitionNode()
   AnimStateTransitionNode_39 = AnimStateTransitionNode() {bAutomaticRuleBasedOnSequencePlayerInState=true}
   AnimStateTransitionNode_40 = AnimStateTransitionNode() {bAutomaticRuleBasedOnSequencePlayerInState=true}
   AnimStateTransitionNode_41 = AnimStateTransitionNode()
   AnimStateTransitionNode_42 = AnimStateTransitionNode()
   AnimStateTransitionNode_43 = AnimStateTransitionNode()
   AnimStateTransitionNode_8 = AnimStateTransitionNode()
   AnimStateTransitionNode_9 = AnimStateTransitionNode()
  ]
  links [
   {
    dst=DownSights
    rule=AnimGraph.AnimGraphNode_StateMachine_1.FullBody_1.AnimStateNode_1.Main.AnimGraphNode_StateMachine_0.Main.AnimStateNode_35.OnGround.AnimGraphNode_StateMachine_0.OnGround.AnimStateTransitionNode_33.Transition
    src=CrouchMovingToDownSights
    via=AnimStateTransitionNode_33
   }
   {
    dst=DownSights
    rule=AnimGraph.AnimGraphNode_StateMachine_1.FullBody_1.AnimStateNode_1.Main.AnimGraphNode_StateMachine_0.Main.AnimStateNode_35.OnGround.AnimGraphNode_StateMachine_0.OnGround.AnimStateTransitionNode_32.Transition
    src=CrouchToDownSights
    via=AnimStateTransitionNode_32
   }
   {
    dst=IsNotTargeting
    rule=AnimGraph.AnimGraphNode_StateMachine_1.FullBody_1.AnimStateNode_1.Main.AnimGraphNode_StateMachine_0.Main.AnimStateNode_35.OnGround.AnimGraphNode_StateMachine_0.OnGround.AnimStateTransitionNode_36.Transition
    src=DownSights
    via=AnimStateTransitionNode_36
   }
   {
    dst=Normal
    rule=AnimGraph.AnimGraphNode_StateMachine_1.FullBody_1.AnimStateNode_1.Main.AnimGraphNode_StateMachine_0.Main.AnimStateNode_35.OnGround.AnimGraphNode_StateMachine_0.OnGround.AnimStateTransitionNode_39.Transition
    src=DownSightsToCrouch
    via=AnimStateTransitionNode_39
   }
   {
    dst=Normal
    rule=AnimGraph.AnimGraphNode_StateMachine_1.FullBody_1.AnimStateNode_1.Main.AnimGraphNode_StateMachine_0.Main.AnimStateNode_35.OnGround.AnimGraphNode_StateMachine_0.OnGround.AnimStateTransitionNode_40.Transition
    src=DownSightsToIdle
    via=AnimStateTransitionNode_40
   }
   {
    dst=Normal
    rule=AnimGraph.AnimGraphNode_StateMachine_1.FullBody_1.AnimStateNode_1.Main.AnimGraphNode_StateMachine_0.Main.AnimStateNode_35.OnGround.AnimGraphNode_StateMachine_0.OnGround.AnimStateTransitionNode_9.Transition
    src=Hunker
    via=AnimStateTransitionNode_9
   }
   {
    dst=DownSights
    rule=AnimGraph.AnimGraphNode_StateMachine_1.FullBody_1.AnimStateNode_1.Main.AnimGraphNode_StateMachine_0.Main.AnimStateNode_35.OnGround.AnimGraphNode_StateMachine_0.OnGround.AnimStateTransitionNode_29.Transition
    src=IdleToDownSights
    via=AnimStateTransitionNode_29
   }
   {
    dst=CrouchMovingToDownSights
    rule=AnimGraph.AnimGraphNode_StateMachine_1.FullBody_1.AnimStateNode_1.Main.AnimGraphNode_StateMachine_0.Main.AnimStateNode_35.OnGround.AnimGraphNode_StateMachine_0.OnGround.AnimStateTransitionNode_43.Transition
    src=IsCrouching
    via=AnimStateTransitionNode_43
   }
   {
    dst=CrouchToDownSights
    rule=AnimGraph.AnimGraphNode_StateMachine_1.FullBody_1.AnimStateNode_1.Main.AnimGraphNode_StateMachine_0.Main.AnimStateNode_35.OnGround.AnimGraphNode_StateMachine_0.OnGround.AnimStateTransitionNode_42.Transition
    src=IsCrouching
    via=AnimStateTransitionNode_42
   }
   {
    dst=DownSightsToCrouch
    rule=AnimGraph.AnimGraphNode_StateMachine_1.FullBody_1.AnimStateNode_1.Main.AnimGraphNode_StateMachine_0.Main.AnimStateNode_35.OnGround.AnimGraphNode_StateMachine_0.OnGround.AnimStateTransitionNode_38.Transition
    src=IsNotTargeting
    via=AnimStateTransitionNode_38
   }
   {
    dst=DownSightsToIdle
    rule=AnimGraph.AnimGraphNode_StateMachine_1.FullBody_1.AnimStateNode_1.Main.AnimGraphNode_StateMachine_0.Main.AnimStateNode_35.OnGround.AnimGraphNode_StateMachine_0.OnGround.AnimStateTransitionNode_37.Transition
    src=IsNotTargeting
    via=AnimStateTransitionNode_37
   }
   {
    dst=IdleToDownSights
    rule=AnimGraph.AnimGraphNode_StateMachine_1.FullBody_1.AnimStateNode_1.Main.AnimGraphNode_StateMachine_0.Main.AnimStateNode_35.OnGround.AnimGraphNode_StateMachine_0.OnGround.AnimStateTransitionNode_28.Transition
    src=IsTargeting
    via=AnimStateTransitionNode_28
   }
   {
    dst=IsCrouching
    rule=AnimGraph.AnimGraphNode_StateMachine_1.FullBody_1.AnimStateNode_1.Main.AnimGraphNode_StateMachine_0.Main.AnimStateNode_35.OnGround.AnimGraphNode_StateMachine_0.OnGround.AnimStateTransitionNode_41.Transition
    src=IsTargeting
    via=AnimStateTransitionNode_41
   }
   {
    dst=Hunker
    rule=AnimGraph.AnimGraphNode_StateMachine_1.FullBody_1.AnimStateNode_1.Main.AnimGraphNode_StateMachine_0.Main.AnimStateNode_35.OnGround.AnimGraphNode_StateMachine_0.OnGround.AnimStateTransitionNode_8.Transition
    src=Normal
    via=AnimStateTransitionNode_8
   }
   {
    dst=IsTargeting
    rule=AnimGraph.AnimGraphNode_StateMachine_1.FullBody_1.AnimStateNode_1.Main.AnimGraphNode_StateMachine_0.Main.AnimStateNode_35.OnGround.AnimGraphNode_StateMachine_0.OnGround.AnimStateTransitionNode_27.Transition
    src=Normal
    via=AnimStateTransitionNode_27
   }
  ]
 }
 IsTargeting {
  path=AnimGraph.AnimGraphNode_StateMachine_1.FullBody_1.AnimStateNode_1.Main.AnimGraphNode_StateMachine_0.Main.AnimStateNode_35.OnGround.AnimGraphNode_StateMachine_0.OnGround.AnimStateConduitNode_0.IsTargeting
  nodes [Result(bCanEnterTransition=true)]
 }
 IsNotTargeting {
  path=AnimGraph.AnimGraphNode_StateMachine_1.FullBody_1.AnimStateNode_1.Main.AnimGraphNode_StateMachine_0.Main.AnimStateNode_35.OnGround.AnimGraphNode_StateMachine_0.OnGround.AnimStateConduitNode_1.IsNotTargeting
  nodes [Result(bCanEnterTransition=true)]
 }
 IsCrouching {
  path=AnimGraph.AnimGraphNode_StateMachine_1.FullBody_1.AnimStateNode_1.Main.AnimGraphNode_StateMachine_0.Main.AnimStateNode_35.OnGround.AnimGraphNode_StateMachine_0.OnGround.AnimStateConduitNode_2.IsCrouching
  nodes [Result(bCanEnterTransition=true)]
 }
 DownSights {
  path=AnimGraph.AnimGraphNode_StateMachine_1.FullBody_1.AnimStateNode_1.Main.AnimGraphNode_StateMachine_0.Main.AnimStateNode_35.OnGround.AnimGraphNode_StateMachine_0.OnGround.AnimStateNode_0.DownSights
  nodes [
   Blend Poses by bool(
    BlendPose_0=Sequence Player 'CrouchWalkDownSights'() {sequence=&CrouchWalkDownSights}
    BlendPose_1=Sequence Player 'DownSights'() {sequence=&DownSights}
    BlendTime_0=0.2
    BlendTime_1=0.2
    bActiveValue=Get bIsMoving2D()
   ) {blendType=Linear}
   AnimGraphNode_StateMachine_0 = DownSights() {is=title AnimGraphNode_StateMachine}
   Output Animation Pose(Result=@AnimGraphNode_StateMachine_0)
  ]
 }
 DownSights {
  entry=DownSights
  path=AnimGraph.AnimGraphNode_StateMachine_1.FullBody_1.AnimStateNode_1.Main.AnimGraphNode_StateMachine_0.Main.AnimStateNode_35.OnGround.AnimGraphNode_StateMachine_0.OnGround.AnimStateNode_0.DownSights.AnimGraphNode_StateMachine_0.DownSights
  ids=id
  nodes [
   AnimStateNode_0 = CrouchWalkDownSights() {is=title AnimStateNode}
   AnimStateNode_12 = DownSights() {is=title AnimStateNode}
   AnimStateTransitionNode_34 = AnimStateTransitionNode()
   AnimStateTransitionNode_35 = AnimStateTransitionNode()
  ]
  links [
   {
    dst=DownSights
    rule=AnimGraph.AnimGraphNode_StateMachine_1.FullBody_1.AnimStateNode_1.Main.AnimGraphNode_StateMachine_0.Main.AnimStateNode_35.OnGround.AnimGraphNode_StateMachine_0.OnGround.AnimStateNode_0.DownSights.AnimGraphNode_StateMachine_0.DownSights.AnimStateTransitionNode_35.Transition
    src=CrouchWalkDownSights
    via=AnimStateTransitionNode_35
   }
   {
    dst=CrouchWalkDownSights
    rule=AnimGraph.AnimGraphNode_StateMachine_1.FullBody_1.AnimStateNode_1.Main.AnimGraphNode_StateMachine_0.Main.AnimStateNode_35.OnGround.AnimGraphNode_StateMachine_0.OnGround.AnimStateNode_0.DownSights.AnimGraphNode_StateMachine_0.DownSights.AnimStateTransitionNode_34.Transition
    src=DownSights
    via=AnimStateTransitionNode_34
   }
  ]
 }
 CrouchWalkDownSights {
  path=AnimGraph.AnimGraphNode_StateMachine_1.FullBody_1.AnimStateNode_1.Main.AnimGraphNode_StateMachine_0.Main.AnimStateNode_35.OnGround.AnimGraphNode_StateMachine_0.OnGround.AnimStateNode_0.DownSights.AnimGraphNode_StateMachine_0.DownSights.AnimStateNode_0.CrouchWalkDownSights
  nodes [Output Animation Pose(Result=Sequence Player 'CrouchWalkDownSights'() {sequence=&CrouchWalkDownSights})]
 }
 DownSights {
  path=AnimGraph.AnimGraphNode_StateMachine_1.FullBody_1.AnimStateNode_1.Main.AnimGraphNode_StateMachine_0.Main.AnimStateNode_35.OnGround.AnimGraphNode_StateMachine_0.OnGround.AnimStateNode_0.DownSights.AnimGraphNode_StateMachine_0.DownSights.AnimStateNode_12.DownSights
  nodes [Output Animation Pose(Result=Sequence Player 'DownSights'() {sequence=&DownSights})]
 }
 Transition {
  path=AnimGraph.AnimGraphNode_StateMachine_1.FullBody_1.AnimStateNode_1.Main.AnimGraphNode_StateMachine_0.Main.AnimStateNode_35.OnGround.AnimGraphNode_StateMachine_0.OnGround.AnimStateNode_0.DownSights.AnimGraphNode_StateMachine_0.DownSights.AnimStateTransitionNode_34.Transition
  nodes [Result(bCanEnterTransition=Get bIsCrouchMoving())]
 }
 Transition {
  path=AnimGraph.AnimGraphNode_StateMachine_1.FullBody_1.AnimStateNode_1.Main.AnimGraphNode_StateMachine_0.Main.AnimStateNode_35.OnGround.AnimGraphNode_StateMachine_0.OnGround.AnimStateNode_0.DownSights.AnimGraphNode_StateMachine_0.DownSights.AnimStateTransitionNode_35.Transition
  nodes [Result(bCanEnterTransition=Get bTransition_CrouchWalkingDownSights_DownSights())]
 }
 IdleToDownSights {
  path=AnimGraph.AnimGraphNode_StateMachine_1.FullBody_1.AnimStateNode_1.Main.AnimGraphNode_StateMachine_0.Main.AnimStateNode_35.OnGround.AnimGraphNode_StateMachine_0.OnGround.AnimStateNode_13.IdleToDownSights
  nodes [Output Animation Pose(Result=Sequence Player 'IdleToDownSights'!(bLoopAnimation:bool=False) -> (Pose) {bLoopAnimation=false, sequence=&IdleToDownSights})]
 }
 DownSightsToIdle {
  path=AnimGraph.AnimGraphNode_StateMachine_1.FullBody_1.AnimStateNode_1.Main.AnimGraphNode_StateMachine_0.Main.AnimStateNode_35.OnGround.AnimGraphNode_StateMachine_0.OnGround.AnimStateNode_14.DownSightsToIdle
  nodes [Output Animation Pose(Result=Sequence Player 'DownSightsToIdle'!(bLoopAnimation:bool=False) -> (Pose) {bLoopAnimation=false, sequence=&DownSightsToIdle})]
 }
 CrouchMovingToDownSights {
  path=AnimGraph.AnimGraphNode_StateMachine_1.FullBody_1.AnimStateNode_1.Main.AnimGraphNode_StateMachine_0.Main.AnimStateNode_35.OnGround.AnimGraphNode_StateMachine_0.OnGround.AnimStateNode_28.CrouchMovingToDownSights
  nodes [Output Animation Pose(Result=Sequence Player 'CrouchToCrouchADS'!(bLoopAnimation:bool=False) -> (Pose) {bLoopAnimation=false, sequence=&CrouchToCrouchADS})]
 }
 CrouchToDownSights {
  path=AnimGraph.AnimGraphNode_StateMachine_1.FullBody_1.AnimStateNode_1.Main.AnimGraphNode_StateMachine_0.Main.AnimStateNode_35.OnGround.AnimGraphNode_StateMachine_0.OnGround.AnimStateNode_29.CrouchToDownSights
  nodes [Output Animation Pose(Result=Sequence Player 'CrouchToCrouchADS'(bLoopAnimation=False) {bLoopAnimation=false, sequence=&CrouchToCrouchADS})]
 }
 DownSightsToCrouch {
  path=AnimGraph.AnimGraphNode_StateMachine_1.FullBody_1.AnimStateNode_1.Main.AnimGraphNode_StateMachine_0.Main.AnimStateNode_35.OnGround.AnimGraphNode_StateMachine_0.OnGround.AnimStateNode_30.DownSightsToCrouch
  nodes [Output Animation Pose(Result=Sequence Player 'CrouchDownSightsToCrouch'!(bLoopAnimation:bool=False) -> (Pose) {bLoopAnimation=false, sequence=&CrouchDownSightsToCrouch})]
 }
 Normal {
  path=AnimGraph.AnimGraphNode_StateMachine_1.FullBody_1.AnimStateNode_1.Main.AnimGraphNode_StateMachine_0.Main.AnimStateNode_35.OnGround.AnimGraphNode_StateMachine_0.OnGround.AnimStateNode_4.Normal
  nodes [
   AnimGraphNode_StateMachine_0 = Normal() {is=title AnimGraphNode_StateMachine}
   Output Animation Pose(Result=@AnimGraphNode_StateMachine_0)
  ]
 }
 Normal {
  entry=Standing
  path=AnimGraph.AnimGraphNode_StateMachine_1.FullBody_1.AnimStateNode_1.Main.AnimGraphNode_StateMachine_0.Main.AnimStateNode_35.OnGround.AnimGraphNode_StateMachine_0.OnGround.AnimStateNode_4.Normal.AnimGraphNode_StateMachine_0.Normal
  ids=id
  nodes [
   AnimStateNode_0 = Standing() {is=title AnimStateNode}
   AnimStateNode_1 = Crouching() {stateEntered {guid=#22, notifyName=Entered_Crouching}}
   AnimStateNode_2 = ToCrouch() {is=title AnimStateNode}
   AnimStateNode_3 = ToStanding() {is=title AnimStateNode}
   AnimStateTransitionNode_2 = AnimStateTransitionNode()
   AnimStateTransitionNode_3 = AnimStateTransitionNode() {bAutomaticRuleBasedOnSequencePlayerInState=true}
   AnimStateTransitionNode_4 = AnimStateTransitionNode()
   AnimStateTransitionNode_44 = AnimStateTransitionNode()
   AnimStateTransitionNode_5 = AnimStateTransitionNode() {bAutomaticRuleBasedOnSequencePlayerInState=true}
   AnimStateTransitionNode_6 = AnimStateTransitionNode()
   AnimStateTransitionNode_7 = AnimStateTransitionNode()
  ]
  links [
   {
    dst=ToStanding
    rule=AnimGraph.AnimGraphNode_StateMachine_1.FullBody_1.AnimStateNode_1.Main.AnimGraphNode_StateMachine_0.Main.AnimStateNode_35.OnGround.AnimGraphNode_StateMachine_0.OnGround.AnimStateNode_4.Normal.AnimGraphNode_StateMachine_0.Normal.AnimStateTransitionNode_4.AnimationTransitionGraph_1
    src=Crouching
    via=AnimStateTransitionNode_4
   }
   {
    dst=Crouching
    rule=AnimGraph.AnimGraphNode_StateMachine_1.FullBody_1.AnimStateNode_1.Main.AnimGraphNode_StateMachine_0.Main.AnimStateNode_35.OnGround.AnimGraphNode_StateMachine_0.OnGround.AnimStateNode_4.Normal.AnimGraphNode_StateMachine_0.Normal.AnimStateTransitionNode_44.Transition
    src=Standing
    via=AnimStateTransitionNode_44
   }
   {
    dst=ToCrouch
    rule=AnimGraph.AnimGraphNode_StateMachine_1.FullBody_1.AnimStateNode_1.Main.AnimGraphNode_StateMachine_0.Main.AnimStateNode_35.OnGround.AnimGraphNode_StateMachine_0.OnGround.AnimStateNode_4.Normal.AnimGraphNode_StateMachine_0.Normal.AnimStateTransitionNode_2.AnimationTransitionGraph_1
    src=Standing
    via=AnimStateTransitionNode_2
   }
   {
    dst=Crouching
    rule=AnimGraph.AnimGraphNode_StateMachine_1.FullBody_1.AnimStateNode_1.Main.AnimGraphNode_StateMachine_0.Main.AnimStateNode_35.OnGround.AnimGraphNode_StateMachine_0.OnGround.AnimStateNode_4.Normal.AnimGraphNode_StateMachine_0.Normal.AnimStateTransitionNode_3.AnimationTransitionGraph_1
    src=ToCrouch
    via=AnimStateTransitionNode_3
   }
   {
    dst=ToStanding
    rule=AnimGraph.AnimGraphNode_StateMachine_1.FullBody_1.AnimStateNode_1.Main.AnimGraphNode_StateMachine_0.Main.AnimStateNode_35.OnGround.AnimGraphNode_StateMachine_0.OnGround.AnimStateNode_4.Normal.AnimGraphNode_StateMachine_0.Normal.AnimStateTransitionNode_6.AnimationTransitionGraph_1
    src=ToCrouch
    via=AnimStateTransitionNode_6
   }
   {
    dst=Standing
    rule=AnimGraph.AnimGraphNode_StateMachine_1.FullBody_1.AnimStateNode_1.Main.AnimGraphNode_StateMachine_0.Main.AnimStateNode_35.OnGround.AnimGraphNode_StateMachine_0.OnGround.AnimStateNode_4.Normal.AnimGraphNode_StateMachine_0.Normal.AnimStateTransitionNode_5.AnimationTransitionGraph_1
    src=ToStanding
    via=AnimStateTransitionNode_5
   }
   {
    dst=ToCrouch
    rule=AnimGraph.AnimGraphNode_StateMachine_1.FullBody_1.AnimStateNode_1.Main.AnimGraphNode_StateMachine_0.Main.AnimStateNode_35.OnGround.AnimGraphNode_StateMachine_0.OnGround.AnimStateNode_4.Normal.AnimGraphNode_StateMachine_0.Normal.AnimStateTransitionNode_7.AnimationTransitionGraph_1
    src=ToStanding
    via=AnimStateTransitionNode_7
   }
  ]
 }
 Standing {
  path=AnimGraph.AnimGraphNode_StateMachine_1.FullBody_1.AnimStateNode_1.Main.AnimGraphNode_StateMachine_0.Main.AnimStateNode_35.OnGround.AnimGraphNode_StateMachine_0.OnGround.AnimStateNode_4.Normal.AnimGraphNode_StateMachine_0.Normal.AnimStateNode_0.Standing
  nodes [
   AnimGraphNode_StateMachine_0 = Standing() {is=title AnimGraphNode_StateMachine}
   Output Animation Pose(Result=@AnimGraphNode_StateMachine_0)
  ]
 }
 Standing {
  entry=standing_still
  path=AnimGraph.AnimGraphNode_StateMachine_1.FullBody_1.AnimStateNode_1.Main.AnimGraphNode_StateMachine_0.Main.AnimStateNode_35.OnGround.AnimGraphNode_StateMachine_0.OnGround.AnimStateNode_4.Normal.AnimGraphNode_StateMachine_0.Normal.AnimStateNode_0.Standing.AnimGraphNode_StateMachine_0.Standing
  ids=id
  nodes [
   AnimStateNode_20 = standing_still() {is=title AnimStateNode}
   AnimStateNode_3 = movement_to_idle() {is=title AnimStateNode}
   AnimStateNode_32 = movement() {is=title AnimStateNode}
   AnimStateTransitionNode_107 = AnimStateTransitionNode()
   AnimStateTransitionNode_109 = AnimStateTransitionNode() {bAutomaticRuleBasedOnSequencePlayerInState=true}
   AnimStateTransitionNode_7 = AnimStateTransitionNode()
   AnimStateTransitionNode_8 = AnimStateTransitionNode()
  ]
  links [
   {
    dst=movement_to_idle
    rule=AnimGraph.AnimGraphNode_StateMachine_1.FullBody_1.AnimStateNode_1.Main.AnimGraphNode_StateMachine_0.Main.AnimStateNode_35.OnGround.AnimGraphNode_StateMachine_0.OnGround.AnimStateNode_4.Normal.AnimGraphNode_StateMachine_0.Normal.AnimStateNode_0.Standing.AnimGraphNode_StateMachine_0.Standing.AnimStateTransitionNode_107.AnimationTransitionGraph_0
    src=movement
    via=AnimStateTransitionNode_107
   }
   {
    dst=movement
    rule=AnimGraph.AnimGraphNode_StateMachine_1.FullBody_1.AnimStateNode_1.Main.AnimGraphNode_StateMachine_0.Main.AnimStateNode_35.OnGround.AnimGraphNode_StateMachine_0.OnGround.AnimStateNode_4.Normal.AnimGraphNode_StateMachine_0.Normal.AnimStateNode_0.Standing.AnimGraphNode_StateMachine_0.Standing.AnimStateTransitionNode_8.AnimationTransitionGraph_0
    src=movement_to_idle
    via=AnimStateTransitionNode_8
   }
   {
    dst=standing_still
    rule=AnimGraph.AnimGraphNode_StateMachine_1.FullBody_1.AnimStateNode_1.Main.AnimGraphNode_StateMachine_0.Main.AnimStateNode_35.OnGround.AnimGraphNode_StateMachine_0.OnGround.AnimStateNode_4.Normal.AnimGraphNode_StateMachine_0.Normal.AnimStateNode_0.Standing.AnimGraphNode_StateMachine_0.Standing.AnimStateTransitionNode_109.AnimationTransitionGraph_0
    src=movement_to_idle
    via=AnimStateTransitionNode_109
   }
   {
    dst=movement
    rule=AnimGraph.AnimGraphNode_StateMachine_1.FullBody_1.AnimStateNode_1.Main.AnimGraphNode_StateMachine_0.Main.AnimStateNode_35.OnGround.AnimGraphNode_StateMachine_0.OnGround.AnimStateNode_4.Normal.AnimGraphNode_StateMachine_0.Normal.AnimStateNode_0.Standing.AnimGraphNode_StateMachine_0.Standing.AnimStateTransitionNode_7.AnimationTransitionGraph_0
    src=standing_still
    via=AnimStateTransitionNode_7
   }
  ]
 }
 standing_still {
  path=AnimGraph.AnimGraphNode_StateMachine_1.FullBody_1.AnimStateNode_1.Main.AnimGraphNode_StateMachine_0.Main.AnimStateNode_35.OnGround.AnimGraphNode_StateMachine_0.OnGround.AnimStateNode_4.Normal.AnimGraphNode_StateMachine_0.Normal.AnimStateNode_0.Standing.AnimGraphNode_StateMachine_0.Standing.AnimStateNode_20.standing_still
  nodes [
   AnimGraphNode_StateMachine_0 = Idle()
   Output Animation Pose(Result=@AnimGraphNode_StateMachine_0)
  ]
 }
 Idle {
  entry=InnerIdle
  path=AnimGraph.AnimGraphNode_StateMachine_1.FullBody_1.AnimStateNode_1.Main.AnimGraphNode_StateMachine_0.Main.AnimStateNode_35.OnGround.AnimGraphNode_StateMachine_0.OnGround.AnimStateNode_4.Normal.AnimGraphNode_StateMachine_0.Normal.AnimStateNode_0.Standing.AnimGraphNode_StateMachine_0.Standing.AnimStateNode_20.standing_still.AnimGraphNode_StateMachine_0.Idle
  ids=id
  nodes [
   AnimStateNode_0 = TestIdle_01_2()
   AnimStateNode_1 = Emote()
   AnimStateNode_2 = Happy()
   AnimStateNode_3 = IdleToHappy()
   AnimStateNode_4 = HappyToIdle()
   AnimStateNode_5 = InnerIdle()
   AnimStateNode_6 = TestIdle_01()
   AnimStateNode_7 = TestIdle_01_1()
   AnimStateTransitionNode_0 = AnimStateTransitionNode() {bAutomaticRuleBasedOnSequencePlayerInState=true}
   AnimStateTransitionNode_1 = AnimStateTransitionNode() {bAutomaticRuleBasedOnSequencePlayerInState=true}
   AnimStateTransitionNode_2 = AnimStateTransitionNode() {bAutomaticRuleBasedOnSequencePlayerInState=true}
   AnimStateTransitionNode_3 = AnimStateTransitionNode()
   AnimStateTransitionNode_4 = AnimStateTransitionNode() {bAutomaticRuleBasedOnSequencePlayerInState=true}
   AnimStateTransitionNode_6 = AnimStateTransitionNode()
   AnimStateTransitionNode_7 = AnimStateTransitionNode() {bAutomaticRuleBasedOnSequencePlayerInState=true}
   AnimStateTransitionNode_8 = AnimStateTransitionNode() {crossfadeDuration=0.1}
   AnimStateTransitionNode_9 = AnimStateTransitionNode() {bAutomaticRuleBasedOnSequencePlayerInState=true, crossfadeDuration=0}
  ]
  links [
   {
    dst=InnerIdle
    rule=AnimGraph.AnimGraphNode_StateMachine_1.FullBody_1.AnimStateNode_1.Main.AnimGraphNode_StateMachine_0.Main.AnimStateNode_35.OnGround.AnimGraphNode_StateMachine_0.OnGround.AnimStateNode_4.Normal.AnimGraphNode_StateMachine_0.Normal.AnimStateNode_0.Standing.AnimGraphNode_StateMachine_0.Standing.AnimStateNode_20.standing_still.AnimGraphNode_StateMachine_0.Idle.AnimStateTransitionNode_4.Transition
    src=Emote
    via=AnimStateTransitionNode_4
   }
   {
    dst=HappyToIdle
    rule=AnimGraph.AnimGraphNode_StateMachine_1.FullBody_1.AnimStateNode_1.Main.AnimGraphNode_StateMachine_0.Main.AnimStateNode_35.OnGround.AnimGraphNode_StateMachine_0.OnGround.AnimStateNode_4.Normal.AnimGraphNode_StateMachine_0.Normal.AnimStateNode_0.Standing.AnimGraphNode_StateMachine_0.Standing.AnimStateNode_20.standing_still.AnimGraphNode_StateMachine_0.Idle.AnimStateTransitionNode_8.AnimationTransitionGraph_0
    src=Happy
    via=AnimStateTransitionNode_8
   }
   {
    dst=InnerIdle
    rule=AnimGraph.AnimGraphNode_StateMachine_1.FullBody_1.AnimStateNode_1.Main.AnimGraphNode_StateMachine_0.Main.AnimStateNode_35.OnGround.AnimGraphNode_StateMachine_0.OnGround.AnimStateNode_4.Normal.AnimGraphNode_StateMachine_0.Normal.AnimStateNode_0.Standing.AnimGraphNode_StateMachine_0.Standing.AnimStateNode_20.standing_still.AnimGraphNode_StateMachine_0.Idle.AnimStateTransitionNode_9.AnimationTransitionGraph_0
    src=HappyToIdle
    via=AnimStateTransitionNode_9
   }
   {
    dst=Happy
    rule=AnimGraph.AnimGraphNode_StateMachine_1.FullBody_1.AnimStateNode_1.Main.AnimGraphNode_StateMachine_0.Main.AnimStateNode_35.OnGround.AnimGraphNode_StateMachine_0.OnGround.AnimStateNode_4.Normal.AnimGraphNode_StateMachine_0.Normal.AnimStateNode_0.Standing.AnimGraphNode_StateMachine_0.Standing.AnimStateNode_20.standing_still.AnimGraphNode_StateMachine_0.Idle.AnimStateTransitionNode_7.AnimationTransitionGraph_0
    src=IdleToHappy
    via=AnimStateTransitionNode_7
   }
   {
    dst=Emote
    rule=AnimGraph.AnimGraphNode_StateMachine_1.FullBody_1.AnimStateNode_1.Main.AnimGraphNode_StateMachine_0.Main.AnimStateNode_35.OnGround.AnimGraphNode_StateMachine_0.OnGround.AnimStateNode_4.Normal.AnimGraphNode_StateMachine_0.Normal.AnimStateNode_0.Standing.AnimGraphNode_StateMachine_0.Standing.AnimStateNode_20.standing_still.AnimGraphNode_StateMachine_0.Idle.AnimStateTransitionNode_3.Transition
    src=InnerIdle
    via=AnimStateTransitionNode_3
   }
   {
    dst=IdleToHappy
    rule=AnimGraph.AnimGraphNode_StateMachine_1.FullBody_1.AnimStateNode_1.Main.AnimGraphNode_StateMachine_0.Main.AnimStateNode_35.OnGround.AnimGraphNode_StateMachine_0.OnGround.AnimStateNode_4.Normal.AnimGraphNode_StateMachine_0.Normal.AnimStateNode_0.Standing.AnimGraphNode_StateMachine_0.Standing.AnimStateNode_20.standing_still.AnimGraphNode_StateMachine_0.Idle.AnimStateTransitionNode_6.AnimationTransitionGraph_0
    src=InnerIdle
    via=AnimStateTransitionNode_6
   }
   {
    dst=TestIdle_01_1
    rule=AnimGraph.AnimGraphNode_StateMachine_1.FullBody_1.AnimStateNode_1.Main.AnimGraphNode_StateMachine_0.Main.AnimStateNode_35.OnGround.AnimGraphNode_StateMachine_0.OnGround.AnimStateNode_4.Normal.AnimGraphNode_StateMachine_0.Normal.AnimStateNode_0.Standing.AnimGraphNode_StateMachine_0.Standing.AnimStateNode_20.standing_still.AnimGraphNode_StateMachine_0.Idle.AnimStateTransitionNode_0.Transition
    src=TestIdle_01
    via=AnimStateTransitionNode_0
   }
   {
    dst=TestIdle_01_2
    rule=AnimGraph.AnimGraphNode_StateMachine_1.FullBody_1.AnimStateNode_1.Main.AnimGraphNode_StateMachine_0.Main.AnimStateNode_35.OnGround.AnimGraphNode_StateMachine_0.OnGround.AnimStateNode_4.Normal.AnimGraphNode_StateMachine_0.Normal.AnimStateNode_0.Standing.AnimGraphNode_StateMachine_0.Standing.AnimStateNode_20.standing_still.AnimGraphNode_StateMachine_0.Idle.AnimStateTransitionNode_1.Transition
    src=TestIdle_01_1
    via=AnimStateTransitionNode_1
   }
   {
    dst=TestIdle_01
    rule=AnimGraph.AnimGraphNode_StateMachine_1.FullBody_1.AnimStateNode_1.Main.AnimGraphNode_StateMachine_0.Main.AnimStateNode_35.OnGround.AnimGraphNode_StateMachine_0.OnGround.AnimStateNode_4.Normal.AnimGraphNode_StateMachine_0.Normal.AnimStateNode_0.Standing.AnimGraphNode_StateMachine_0.Standing.AnimStateNode_20.standing_still.AnimGraphNode_StateMachine_0.Idle.AnimStateTransitionNode_2.Transition
    src=TestIdle_01_2
    via=AnimStateTransitionNode_2
   }
  ]
 }
 TestIdle_01_2 {
  path=AnimGraph.AnimGraphNode_StateMachine_1.FullBody_1.AnimStateNode_1.Main.AnimGraphNode_StateMachine_0.Main.AnimStateNode_35.OnGround.AnimGraphNode_StateMachine_0.OnGround.AnimStateNode_4.Normal.AnimGraphNode_StateMachine_0.Normal.AnimStateNode_0.Standing.AnimGraphNode_StateMachine_0.Standing.AnimStateNode_20.standing_still.AnimGraphNode_StateMachine_0.Idle.AnimStateNode_0.TestIdle_01_2
  nodes [Output Animation Pose(Result=Sequence Player 'IdleBreak_02'() {sequence=&IdleBreak_02})]
 }
 Emote {
  path=AnimGraph.AnimGraphNode_StateMachine_1.FullBody_1.AnimStateNode_1.Main.AnimGraphNode_StateMachine_0.Main.AnimStateNode_35.OnGround.AnimGraphNode_StateMachine_0.OnGround.AnimStateNode_4.Normal.AnimGraphNode_StateMachine_0.Normal.AnimStateNode_0.Standing.AnimGraphNode_StateMachine_0.Standing.AnimStateNode_20.standing_still.AnimGraphNode_StateMachine_0.Idle.AnimStateNode_1.Emote
  nodes [Output Animation Pose(Result=Sequence Player 'Emote'() {bLoopAnimation=false, sequence=&Emote})]
 }
 Happy {
  path=AnimGraph.AnimGraphNode_StateMachine_1.FullBody_1.AnimStateNode_1.Main.AnimGraphNode_StateMachine_0.Main.AnimStateNode_35.OnGround.AnimGraphNode_StateMachine_0.OnGround.AnimStateNode_4.Normal.AnimGraphNode_StateMachine_0.Normal.AnimStateNode_0.Standing.AnimGraphNode_StateMachine_0.Standing.AnimStateNode_20.standing_still.AnimGraphNode_StateMachine_0.Idle.AnimStateNode_2.Happy
  nodes [Output Animation Pose(Result=Sequence Player 'Happy'() {sequence=&Happy})]
 }
 IdleToHappy {
  path=AnimGraph.AnimGraphNode_StateMachine_1.FullBody_1.AnimStateNode_1.Main.AnimGraphNode_StateMachine_0.Main.AnimStateNode_35.OnGround.AnimGraphNode_StateMachine_0.OnGround.AnimStateNode_4.Normal.AnimGraphNode_StateMachine_0.Normal.AnimStateNode_0.Standing.AnimGraphNode_StateMachine_0.Standing.AnimStateNode_20.standing_still.AnimGraphNode_StateMachine_0.Idle.AnimStateNode_3.IdleToHappy
  nodes [Output Animation Pose(Result=Sequence Player 'IdleToHappy'() {sequence=&IdleToHappy})]
 }
 HappyToIdle {
  path=AnimGraph.AnimGraphNode_StateMachine_1.FullBody_1.AnimStateNode_1.Main.AnimGraphNode_StateMachine_0.Main.AnimStateNode_35.OnGround.AnimGraphNode_StateMachine_0.OnGround.AnimStateNode_4.Normal.AnimGraphNode_StateMachine_0.Normal.AnimStateNode_0.Standing.AnimGraphNode_StateMachine_0.Standing.AnimStateNode_20.standing_still.AnimGraphNode_StateMachine_0.Idle.AnimStateNode_4.HappyToIdle
  nodes [Output Animation Pose(Result=Sequence Player 'HappyToIdle'() {bLoopAnimation=false, sequence=&HappyToIdle})]
 }
 InnerIdle {
  path=AnimGraph.AnimGraphNode_StateMachine_1.FullBody_1.AnimStateNode_1.Main.AnimGraphNode_StateMachine_0.Main.AnimStateNode_35.OnGround.AnimGraphNode_StateMachine_0.OnGround.AnimStateNode_4.Normal.AnimGraphNode_StateMachine_0.Normal.AnimStateNode_0.Standing.AnimGraphNode_StateMachine_0.Standing.AnimStateNode_20.standing_still.AnimGraphNode_StateMachine_0.Idle.AnimStateNode_5.InnerIdle
  nodes [
   n1 = Random Sequence Player() {
    bShuffleMode=true
    entries [
     {
      chanceToPlay=0.2
      maxLoopCount=2
      maxPlayRate=1
      minLoopCount=0
      minPlayRate=1
      sequence=&Idle
      blendIn {blendOption=Linear, blendTime=0.2, customCurve=null}
     }
     {
      chanceToPlay=0.1
      maxLoopCount=0
      maxPlayRate=1
      minLoopCount=0
      minPlayRate=1
      sequence=&IdleBreak_02
      blendIn {blendOption=Linear, blendTime=0.2, customCurve=null}
     }
     {
      chanceToPlay=0.1
      maxLoopCount=0
      maxPlayRate=1
      minLoopCount=0
      minPlayRate=1
      sequence=&IdleBreak_01
      blendIn {blendOption=Linear, blendTime=0.2, customCurve=null}
     }
     {
      chanceToPlay=0.2
      maxLoopCount=2
      maxPlayRate=1
      minLoopCount=0
      minPlayRate=1
      sequence=&Idle
      blendIn {blendOption=Linear, blendTime=0.2, customCurve=null}
     }
     {
      chanceToPlay=0.2
      maxLoopCount=2
      maxPlayRate=1
      minLoopCount=0
      minPlayRate=1
      sequence=&Idle
      blendIn {blendOption=Linear, blendTime=0.2, customCurve=null}
     }
     {
      chanceToPlay=0.2
      maxLoopCount=2
      maxPlayRate=1
      minLoopCount=0
      minPlayRate=1
      sequence=&Idle
      blendIn {blendOption=Linear, blendTime=0.2, customCurve=null}
     }
     {
      chanceToPlay=1
      maxLoopCount=0
      maxPlayRate=1
      minLoopCount=0
      minPlayRate=1
      sequence=&Idle
      blendIn {blendOption=Linear, blendTime=0.2, customCurve=null}
     }
     {
      chanceToPlay=1
      maxLoopCount=0
      maxPlayRate=1
      minLoopCount=0
      minPlayRate=1
      sequence=&Idle
      blendIn {blendOption=Linear, blendTime=0.2, customCurve=null}
     }
     {
      chanceToPlay=1
      maxLoopCount=0
      maxPlayRate=1
      minLoopCount=0
      minPlayRate=1
      sequence=&Idle
      blendIn {blendOption=Linear, blendTime=0.2, customCurve=null}
     }
     {
      chanceToPlay=1
      maxLoopCount=0
      maxPlayRate=1
      minLoopCount=0
      minPlayRate=1
      sequence=&Idle
      blendIn {blendOption=Linear, blendTime=0.2, customCurve=null}
     }
     {
      chanceToPlay=1
      maxLoopCount=0
      maxPlayRate=1
      minLoopCount=0
      minPlayRate=1
      sequence=&Idle
      blendIn {blendOption=Linear, blendTime=0.2, customCurve=null}
     }
    ]
   }
   Output Animation Pose(Result=@n1)
  ]
 }
 TestIdle_01 {
  path=AnimGraph.AnimGraphNode_StateMachine_1.FullBody_1.AnimStateNode_1.Main.AnimGraphNode_StateMachine_0.Main.AnimStateNode_35.OnGround.AnimGraphNode_StateMachine_0.OnGround.AnimStateNode_4.Normal.AnimGraphNode_StateMachine_0.Normal.AnimStateNode_0.Standing.AnimGraphNode_StateMachine_0.Standing.AnimStateNode_20.standing_still.AnimGraphNode_StateMachine_0.Idle.AnimStateNode_6.TestIdle_01
  nodes [Output Animation Pose(Result=Sequence Player 'Idle'() {sequence=&Idle})]
 }
 TestIdle_01_1 {
  path=AnimGraph.AnimGraphNode_StateMachine_1.FullBody_1.AnimStateNode_1.Main.AnimGraphNode_StateMachine_0.Main.AnimStateNode_35.OnGround.AnimGraphNode_StateMachine_0.OnGround.AnimStateNode_4.Normal.AnimGraphNode_StateMachine_0.Normal.AnimStateNode_0.Standing.AnimGraphNode_StateMachine_0.Standing.AnimStateNode_20.standing_still.AnimGraphNode_StateMachine_0.Idle.AnimStateNode_7.TestIdle_01_1
  nodes [Output Animation Pose(Result=Sequence Player 'IdleBreak_01'() {sequence=&IdleBreak_01})]
 }
 Transition {
  path=AnimGraph.AnimGraphNode_StateMachine_1.FullBody_1.AnimStateNode_1.Main.AnimGraphNode_StateMachine_0.Main.AnimStateNode_35.OnGround.AnimGraphNode_StateMachine_0.OnGround.AnimStateNode_4.Normal.AnimGraphNode_StateMachine_0.Normal.AnimStateNode_0.Standing.AnimGraphNode_StateMachine_0.Standing.AnimStateNode_20.standing_still.AnimGraphNode_StateMachine_0.Idle.AnimStateTransitionNode_0.Transition
  nodes [Result()]
 }
 Transition {
  path=AnimGraph.AnimGraphNode_StateMachine_1.FullBody_1.AnimStateNode_1.Main.AnimGraphNode_StateMachine_0.Main.AnimStateNode_35.OnGround.AnimGraphNode_StateMachine_0.OnGround.AnimStateNode_4.Normal.AnimGraphNode_StateMachine_0.Normal.AnimStateNode_0.Standing.AnimGraphNode_StateMachine_0.Standing.AnimStateNode_20.standing_still.AnimGraphNode_StateMachine_0.Idle.AnimStateTransitionNode_1.Transition
  nodes [Result()]
 }
 Transition {
  path=AnimGraph.AnimGraphNode_StateMachine_1.FullBody_1.AnimStateNode_1.Main.AnimGraphNode_StateMachine_0.Main.AnimStateNode_35.OnGround.AnimGraphNode_StateMachine_0.OnGround.AnimStateNode_4.Normal.AnimGraphNode_StateMachine_0.Normal.AnimStateNode_0.Standing.AnimGraphNode_StateMachine_0.Standing.AnimStateNode_20.standing_still.AnimGraphNode_StateMachine_0.Idle.AnimStateTransitionNode_2.Transition
  nodes [Result()]
 }
 Transition {
  path=AnimGraph.AnimGraphNode_StateMachine_1.FullBody_1.AnimStateNode_1.Main.AnimGraphNode_StateMachine_0.Main.AnimStateNode_35.OnGround.AnimGraphNode_StateMachine_0.OnGround.AnimStateNode_4.Normal.AnimGraphNode_StateMachine_0.Normal.AnimStateNode_0.Standing.AnimGraphNode_StateMachine_0.Standing.AnimStateNode_20.standing_still.AnimGraphNode_StateMachine_0.Idle.AnimStateTransitionNode_3.Transition
  nodes [Result(bCanEnterTransition=Get bEmotePlaying_WhereHunkeringIsBest())]
 }
 Transition {
  path=AnimGraph.AnimGraphNode_StateMachine_1.FullBody_1.AnimStateNode_1.Main.AnimGraphNode_StateMachine_0.Main.AnimStateNode_35.OnGround.AnimGraphNode_StateMachine_0.OnGround.AnimStateNode_4.Normal.AnimGraphNode_StateMachine_0.Normal.AnimStateNode_0.Standing.AnimGraphNode_StateMachine_0.Standing.AnimStateNode_20.standing_still.AnimGraphNode_StateMachine_0.Idle.AnimStateTransitionNode_4.Transition
  nodes [Result(bCanEnterTransition=NOT Boolean(A=Get bEmotePlaying_WhereHunkeringIsBest()))]
 }
 AnimationTransitionGraph_0 {
  path=AnimGraph.AnimGraphNode_StateMachine_1.FullBody_1.AnimStateNode_1.Main.AnimGraphNode_StateMachine_0.Main.AnimStateNode_35.OnGround.AnimGraphNode_StateMachine_0.OnGround.AnimStateNode_4.Normal.AnimGraphNode_StateMachine_0.Normal.AnimStateNode_0.Standing.AnimGraphNode_StateMachine_0.Standing.AnimStateNode_20.standing_still.AnimGraphNode_StateMachine_0.Idle.AnimStateTransitionNode_6.AnimationTransitionGraph_0
  nodes [Result(bCanEnterTransition=Get bTransition_InnerIdle_IdleToHappy())]
 }
 AnimationTransitionGraph_0 {
  path=AnimGraph.AnimGraphNode_StateMachine_1.FullBody_1.AnimStateNode_1.Main.AnimGraphNode_StateMachine_0.Main.AnimStateNode_35.OnGround.AnimGraphNode_StateMachine_0.OnGround.AnimStateNode_4.Normal.AnimGraphNode_StateMachine_0.Normal.AnimStateNode_0.Standing.AnimGraphNode_StateMachine_0.Standing.AnimStateNode_20.standing_still.AnimGraphNode_StateMachine_0.Idle.AnimStateTransitionNode_7.AnimationTransitionGraph_0
  nodes [Result()]
 }
 AnimationTransitionGraph_0 {
  path=AnimGraph.AnimGraphNode_StateMachine_1.FullBody_1.AnimStateNode_1.Main.AnimGraphNode_StateMachine_0.Main.AnimStateNode_35.OnGround.AnimGraphNode_StateMachine_0.OnGround.AnimStateNode_4.Normal.AnimGraphNode_StateMachine_0.Normal.AnimStateNode_0.Standing.AnimGraphNode_StateMachine_0.Standing.AnimStateNode_20.standing_still.AnimGraphNode_StateMachine_0.Idle.AnimStateTransitionNode_8.AnimationTransitionGraph_0
  nodes [
   n1 = Get bIsMoving2D()
   n2 = "Get Relevant Anim Time Remaining (Happy)"!() -> (ReturnValue:float) {is=title K2Node_AnimGetter}
   n3 = Get bIsHappy()
   Result(bCanEnterTransition=AND Boolean!(
     A="float < float"!(A:double=@n2, B:double=0.1) -> (ReturnValue:bool) {is=title K2Node_CallFunction}
     B=NOT Boolean(A=@n3)
     C:bool=NOT Boolean(A=@n1)
    ) -> (ReturnValue))
  ]
 }
 AnimationTransitionGraph_0 {
  path=AnimGraph.AnimGraphNode_StateMachine_1.FullBody_1.AnimStateNode_1.Main.AnimGraphNode_StateMachine_0.Main.AnimStateNode_35.OnGround.AnimGraphNode_StateMachine_0.OnGround.AnimStateNode_4.Normal.AnimGraphNode_StateMachine_0.Normal.AnimStateNode_0.Standing.AnimGraphNode_StateMachine_0.Standing.AnimStateNode_20.standing_still.AnimGraphNode_StateMachine_0.Idle.AnimStateTransitionNode_9.AnimationTransitionGraph_0
  nodes [Result()]
 }
 movement_to_idle {
  path=AnimGraph.AnimGraphNode_StateMachine_1.FullBody_1.AnimStateNode_1.Main.AnimGraphNode_StateMachine_0.Main.AnimStateNode_35.OnGround.AnimGraphNode_StateMachine_0.OnGround.AnimStateNode_4.Normal.AnimGraphNode_StateMachine_0.Normal.AnimStateNode_0.Standing.AnimGraphNode_StateMachine_0.Standing.AnimStateNode_3.movement_to_idle
  nodes [Output Animation Pose(Result=Sequence Player 'LocomotionToIdle'() {sequence=&LocomotionToIdle})]
 }
 movement {
  path=AnimGraph.AnimGraphNode_StateMachine_1.FullBody_1.AnimStateNode_1.Main.AnimGraphNode_StateMachine_0.Main.AnimStateNode_35.OnGround.AnimGraphNode_StateMachine_0.OnGround.AnimStateNode_4.Normal.AnimGraphNode_StateMachine_0.Normal.AnimStateNode_0.Standing.AnimGraphNode_StateMachine_0.Standing.AnimStateNode_32.movement
  nodes [
   AnimGraphNode_StateMachine_0 = TestInOrder()
   n1 = Random Sequence Player() {
    bShuffleMode=true
    entries [
     {
      chanceToPlay=0.2
      maxLoopCount=4
      maxPlayRate=1
      minLoopCount=2
      minPlayRate=1
      sequence=&LocomotionPose
      blendIn {blendOption=Linear, blendTime=0.2, customCurve=null}
     }
     {
      chanceToPlay=0.1
      maxLoopCount=0
      maxPlayRate=1
      minLoopCount=0
      minPlayRate=1
      sequence=&SprintBreak_01
      blendIn {blendOption=Linear, blendTime=0.2, customCurve=null}
     }
     {
      chanceToPlay=0.1
      maxLoopCount=0
      maxPlayRate=1
      minLoopCount=0
      minPlayRate=1
      sequence=&SprintBreak_02
      blendIn {blendOption=Linear, blendTime=0.2, customCurve=null}
     }
     {
      chanceToPlay=0.2
      maxLoopCount=4
      maxPlayRate=1
      minLoopCount=2
      minPlayRate=1
      sequence=&LocomotionPose
      blendIn {blendOption=Linear, blendTime=0.2, customCurve=null}
     }
     {
      chanceToPlay=0.2
      maxLoopCount=4
      maxPlayRate=1
      minLoopCount=2
      minPlayRate=1
      sequence=&LocomotionPose
      blendIn {blendOption=Linear, blendTime=0.2, customCurve=null}
     }
     {
      chanceToPlay=0.2
      maxLoopCount=4
      maxPlayRate=1
      minLoopCount=2
      minPlayRate=1
      sequence=&LocomotionPose
      blendIn {blendOption=Linear, blendTime=0.2, customCurve=null}
     }
     {
      chanceToPlay=1
      maxLoopCount=0
      maxPlayRate=1
      minLoopCount=0
      minPlayRate=1
      sequence=&LocomotionPose
      blendIn {blendOption=Linear, blendTime=0.2, customCurve=null}
     }
     {
      chanceToPlay=1
      maxLoopCount=0
      maxPlayRate=1
      minLoopCount=0
      minPlayRate=1
      sequence=&LocomotionPose
      blendIn {blendOption=Linear, blendTime=0.2, customCurve=null}
     }
     {
      chanceToPlay=1
      maxLoopCount=0
      maxPlayRate=1
      minLoopCount=0
      minPlayRate=1
      sequence=&LocomotionPose
      blendIn {blendOption=Linear, blendTime=0.2, customCurve=null}
     }
     {
      chanceToPlay=1
      maxLoopCount=0
      maxPlayRate=1
      minLoopCount=0
      minPlayRate=1
      sequence=&LocomotionPose
      blendIn {blendOption=Linear, blendTime=0.2, customCurve=null}
     }
     {
      chanceToPlay=1
      maxLoopCount=0
      maxPlayRate=1
      minLoopCount=0
      minPlayRate=1
      sequence=&LocomotionPose
      blendIn {blendOption=Linear, blendTime=0.2, customCurve=null}
     }
    ]
   }
   Output Animation Pose(Result=@n1)
  ]
 }
 TestInOrder {
  entry=State
  path=AnimGraph.AnimGraphNode_StateMachine_1.FullBody_1.AnimStateNode_1.Main.AnimGraphNode_StateMachine_0.Main.AnimStateNode_35.OnGround.AnimGraphNode_StateMachine_0.OnGround.AnimStateNode_4.Normal.AnimGraphNode_StateMachine_0.Normal.AnimStateNode_0.Standing.AnimGraphNode_StateMachine_0.Standing.AnimStateNode_32.movement.AnimGraphNode_StateMachine_0.TestInOrder
  ids=id
  nodes [
   AnimStateNode_0 = State()
   AnimStateNode_1 = State_1()
   AnimStateNode_2 = State_2()
   AnimStateTransitionNode_0 = AnimStateTransitionNode() {bAutomaticRuleBasedOnSequencePlayerInState=true}
   AnimStateTransitionNode_1 = AnimStateTransitionNode() {bAutomaticRuleBasedOnSequencePlayerInState=true}
   AnimStateTransitionNode_2 = AnimStateTransitionNode() {bAutomaticRuleBasedOnSequencePlayerInState=true}
  ]
  links [
   {
    dst=State_1
    rule=AnimGraph.AnimGraphNode_StateMachine_1.FullBody_1.AnimStateNode_1.Main.AnimGraphNode_StateMachine_0.Main.AnimStateNode_35.OnGround.AnimGraphNode_StateMachine_0.OnGround.AnimStateNode_4.Normal.AnimGraphNode_StateMachine_0.Normal.AnimStateNode_0.Standing.AnimGraphNode_StateMachine_0.Standing.AnimStateNode_32.movement.AnimGraphNode_StateMachine_0.TestInOrder.AnimStateTransitionNode_0.Transition
    src=State
    via=AnimStateTransitionNode_0
   }
   {
    dst=State_2
    rule=AnimGraph.AnimGraphNode_StateMachine_1.FullBody_1.AnimStateNode_1.Main.AnimGraphNode_StateMachine_0.Main.AnimStateNode_35.OnGround.AnimGraphNode_StateMachine_0.OnGround.AnimStateNode_4.Normal.AnimGraphNode_StateMachine_0.Normal.AnimStateNode_0.Standing.AnimGraphNode_StateMachine_0.Standing.AnimStateNode_32.movement.AnimGraphNode_StateMachine_0.TestInOrder.AnimStateTransitionNode_1.Transition
    src=State_1
    via=AnimStateTransitionNode_1
   }
   {
    dst=State
    rule=AnimGraph.AnimGraphNode_StateMachine_1.FullBody_1.AnimStateNode_1.Main.AnimGraphNode_StateMachine_0.Main.AnimStateNode_35.OnGround.AnimGraphNode_StateMachine_0.OnGround.AnimStateNode_4.Normal.AnimGraphNode_StateMachine_0.Normal.AnimStateNode_0.Standing.AnimGraphNode_StateMachine_0.Standing.AnimStateNode_32.movement.AnimGraphNode_StateMachine_0.TestInOrder.AnimStateTransitionNode_2.Transition
    src=State_2
    via=AnimStateTransitionNode_2
   }
  ]
 }
 State {
  path=AnimGraph.AnimGraphNode_StateMachine_1.FullBody_1.AnimStateNode_1.Main.AnimGraphNode_StateMachine_0.Main.AnimStateNode_35.OnGround.AnimGraphNode_StateMachine_0.OnGround.AnimStateNode_4.Normal.AnimGraphNode_StateMachine_0.Normal.AnimStateNode_0.Standing.AnimGraphNode_StateMachine_0.Standing.AnimStateNode_32.movement.AnimGraphNode_StateMachine_0.TestInOrder.AnimStateNode_0.State
  nodes [Output Animation Pose(Result=Sequence Player 'LocomotionPose'() {bLoopAnimation=false, playRate=0.25, sequence=&LocomotionPose})]
 }
 State_1 {
  path=AnimGraph.AnimGraphNode_StateMachine_1.FullBody_1.AnimStateNode_1.Main.AnimGraphNode_StateMachine_0.Main.AnimStateNode_35.OnGround.AnimGraphNode_StateMachine_0.OnGround.AnimStateNode_4.Normal.AnimGraphNode_StateMachine_0.Normal.AnimStateNode_0.Standing.AnimGraphNode_StateMachine_0.Standing.AnimStateNode_32.movement.AnimGraphNode_StateMachine_0.TestInOrder.AnimStateNode_1.State_1
  nodes [Output Animation Pose(Result=Sequence Player 'SprintBreak_01'() {bLoopAnimation=false, sequence=&SprintBreak_01})]
 }
 State_2 {
  path=AnimGraph.AnimGraphNode_StateMachine_1.FullBody_1.AnimStateNode_1.Main.AnimGraphNode_StateMachine_0.Main.AnimStateNode_35.OnGround.AnimGraphNode_StateMachine_0.OnGround.AnimStateNode_4.Normal.AnimGraphNode_StateMachine_0.Normal.AnimStateNode_0.Standing.AnimGraphNode_StateMachine_0.Standing.AnimStateNode_32.movement.AnimGraphNode_StateMachine_0.TestInOrder.AnimStateNode_2.State_2
  nodes [Output Animation Pose(Result=Sequence Player 'SprintBreak_02'() {bLoopAnimation=false, sequence=&SprintBreak_02})]
 }
 Transition {
  path=AnimGraph.AnimGraphNode_StateMachine_1.FullBody_1.AnimStateNode_1.Main.AnimGraphNode_StateMachine_0.Main.AnimStateNode_35.OnGround.AnimGraphNode_StateMachine_0.OnGround.AnimStateNode_4.Normal.AnimGraphNode_StateMachine_0.Normal.AnimStateNode_0.Standing.AnimGraphNode_StateMachine_0.Standing.AnimStateNode_32.movement.AnimGraphNode_StateMachine_0.TestInOrder.AnimStateTransitionNode_0.Transition
  nodes [Result()]
 }
 Transition {
  path=AnimGraph.AnimGraphNode_StateMachine_1.FullBody_1.AnimStateNode_1.Main.AnimGraphNode_StateMachine_0.Main.AnimStateNode_35.OnGround.AnimGraphNode_StateMachine_0.OnGround.AnimStateNode_4.Normal.AnimGraphNode_StateMachine_0.Normal.AnimStateNode_0.Standing.AnimGraphNode_StateMachine_0.Standing.AnimStateNode_32.movement.AnimGraphNode_StateMachine_0.TestInOrder.AnimStateTransitionNode_1.Transition
  nodes [Result()]
 }
 Transition {
  path=AnimGraph.AnimGraphNode_StateMachine_1.FullBody_1.AnimStateNode_1.Main.AnimGraphNode_StateMachine_0.Main.AnimStateNode_35.OnGround.AnimGraphNode_StateMachine_0.OnGround.AnimStateNode_4.Normal.AnimGraphNode_StateMachine_0.Normal.AnimStateNode_0.Standing.AnimGraphNode_StateMachine_0.Standing.AnimStateNode_32.movement.AnimGraphNode_StateMachine_0.TestInOrder.AnimStateTransitionNode_2.Transition
  nodes [Result()]
 }
 AnimationTransitionGraph_0 {
  path=AnimGraph.AnimGraphNode_StateMachine_1.FullBody_1.AnimStateNode_1.Main.AnimGraphNode_StateMachine_0.Main.AnimStateNode_35.OnGround.AnimGraphNode_StateMachine_0.OnGround.AnimStateNode_4.Normal.AnimGraphNode_StateMachine_0.Normal.AnimStateNode_0.Standing.AnimGraphNode_StateMachine_0.Standing.AnimStateTransitionNode_107.AnimationTransitionGraph_0
  nodes [Result(bCanEnterTransition=NOT Boolean(A=Get bIsMoving2D()))]
 }
 AnimationTransitionGraph_0 {
  path=AnimGraph.AnimGraphNode_StateMachine_1.FullBody_1.AnimStateNode_1.Main.AnimGraphNode_StateMachine_0.Main.AnimStateNode_35.OnGround.AnimGraphNode_StateMachine_0.OnGround.AnimStateNode_4.Normal.AnimGraphNode_StateMachine_0.Normal.AnimStateNode_0.Standing.AnimGraphNode_StateMachine_0.Standing.AnimStateTransitionNode_109.AnimationTransitionGraph_0
  nodes [Result()]
 }
 AnimationTransitionGraph_0 {
  path=AnimGraph.AnimGraphNode_StateMachine_1.FullBody_1.AnimStateNode_1.Main.AnimGraphNode_StateMachine_0.Main.AnimStateNode_35.OnGround.AnimGraphNode_StateMachine_0.OnGround.AnimStateNode_4.Normal.AnimGraphNode_StateMachine_0.Normal.AnimStateNode_0.Standing.AnimGraphNode_StateMachine_0.Standing.AnimStateTransitionNode_7.AnimationTransitionGraph_0
  nodes [Result(bCanEnterTransition=Get bTransition_Idle_Locomotion())]
 }
 AnimationTransitionGraph_0 {
  path=AnimGraph.AnimGraphNode_StateMachine_1.FullBody_1.AnimStateNode_1.Main.AnimGraphNode_StateMachine_0.Main.AnimStateNode_35.OnGround.AnimGraphNode_StateMachine_0.OnGround.AnimStateNode_4.Normal.AnimGraphNode_StateMachine_0.Normal.AnimStateNode_0.Standing.AnimGraphNode_StateMachine_0.Standing.AnimStateTransitionNode_8.AnimationTransitionGraph_0
  nodes [Result(bCanEnterTransition=Get bIsMoving2D())]
 }
 Crouching {
  path=AnimGraph.AnimGraphNode_StateMachine_1.FullBody_1.AnimStateNode_1.Main.AnimGraphNode_StateMachine_0.Main.AnimStateNode_35.OnGround.AnimGraphNode_StateMachine_0.OnGround.AnimStateNode_4.Normal.AnimGraphNode_StateMachine_0.Normal.AnimStateNode_1.Crouching
  nodes [
   AnimGraphNode_StateMachine_0 = Crouch() {is=title AnimGraphNode_StateMachine}
   Output Animation Pose(Result=@AnimGraphNode_StateMachine_0)
  ]
 }
 Crouch {
  entry=Crouch
  path=AnimGraph.AnimGraphNode_StateMachine_1.FullBody_1.AnimStateNode_1.Main.AnimGraphNode_StateMachine_0.Main.AnimStateNode_35.OnGround.AnimGraphNode_StateMachine_0.OnGround.AnimStateNode_4.Normal.AnimGraphNode_StateMachine_0.Normal.AnimStateNode_1.Crouching.AnimGraphNode_StateMachine_0.Crouch
  ids=id
  nodes [
   AnimStateNode_0 = Blank()
   AnimStateNode_1 = toCrouchWalk()
   AnimStateNode_2 = CrouchSprinting()
   AnimStateNode_27 = CrouchWalk()
   AnimStateNode_29 = Crouch() {is=title AnimStateNode}
   AnimStateNode_3 = LocomotionCrouchHappy()
   AnimStateTransitionNode_0 = AnimStateTransitionNode()
   AnimStateTransitionNode_1 = AnimStateTransitionNode() {bAutomaticRuleBasedOnSequencePlayerInState=true}
   AnimStateTransitionNode_11 = AnimStateTransitionNode()
   AnimStateTransitionNode_2 = AnimStateTransitionNode()
   AnimStateTransitionNode_3 = AnimStateTransitionNode()
   AnimStateTransitionNode_4 = AnimStateTransitionNode()
   AnimStateTransitionNode_47 = AnimStateTransitionNode()
   AnimStateTransitionNode_5 = AnimStateTransitionNode()
   AnimStateTransitionNode_55 = AnimStateTransitionNode()
   AnimStateTransitionNode_6 = AnimStateTransitionNode() {bAutomaticRuleBasedOnSequencePlayerInState=true}
   AnimStateTransitionNode_7 = AnimStateTransitionNode()
  ]
  links [
   {
    dst=Crouch
    rule=AnimGraph.AnimGraphNode_StateMachine_1.FullBody_1.AnimStateNode_1.Main.AnimGraphNode_StateMachine_0.Main.AnimStateNode_35.OnGround.AnimGraphNode_StateMachine_0.OnGround.AnimStateNode_4.Normal.AnimGraphNode_StateMachine_0.Normal.AnimStateNode_1.Crouching.AnimGraphNode_StateMachine_0.Crouch.AnimStateTransitionNode_11.Transition
    src=Blank
    via=AnimStateTransitionNode_11
   }
   {
    dst=CrouchWalk
    rule=AnimGraph.AnimGraphNode_StateMachine_1.FullBody_1.AnimStateNode_1.Main.AnimGraphNode_StateMachine_0.Main.AnimStateNode_35.OnGround.AnimGraphNode_StateMachine_0.OnGround.AnimStateNode_4.Normal.AnimGraphNode_StateMachine_0.Normal.AnimStateNode_1.Crouching.AnimGraphNode_StateMachine_0.Crouch.AnimStateTransitionNode_7.Transition
    src=Blank
    via=AnimStateTransitionNode_7
   }
   {
    dst=toCrouchWalk
    rule=AnimGraph.AnimGraphNode_StateMachine_1.FullBody_1.AnimStateNode_1.Main.AnimGraphNode_StateMachine_0.Main.AnimStateNode_35.OnGround.AnimGraphNode_StateMachine_0.OnGround.AnimStateNode_4.Normal.AnimGraphNode_StateMachine_0.Normal.AnimStateNode_1.Crouching.AnimGraphNode_StateMachine_0.Crouch.AnimStateTransitionNode_0.Transition
    src=Crouch
    via=AnimStateTransitionNode_0
   }
   {
    dst=CrouchWalk
    rule=AnimGraph.AnimGraphNode_StateMachine_1.FullBody_1.AnimStateNode_1.Main.AnimGraphNode_StateMachine_0.Main.AnimStateNode_35.OnGround.AnimGraphNode_StateMachine_0.OnGround.AnimStateNode_4.Normal.AnimGraphNode_StateMachine_0.Normal.AnimStateNode_1.Crouching.AnimGraphNode_StateMachine_0.Crouch.AnimStateTransitionNode_55.AnimationTransitionGraph_1
    src=CrouchSprinting
    via=AnimStateTransitionNode_55
   }
   {
    dst=LocomotionCrouchHappy
    rule=AnimGraph.AnimGraphNode_StateMachine_1.FullBody_1.AnimStateNode_1.Main.AnimGraphNode_StateMachine_0.Main.AnimStateNode_35.OnGround.AnimGraphNode_StateMachine_0.OnGround.AnimStateNode_4.Normal.AnimGraphNode_StateMachine_0.Normal.AnimStateNode_1.Crouching.AnimGraphNode_StateMachine_0.Crouch.AnimStateTransitionNode_3.Transition
    src=CrouchSprinting
    via=AnimStateTransitionNode_3
   }
   {
    dst=Crouch
    rule=AnimGraph.AnimGraphNode_StateMachine_1.FullBody_1.AnimStateNode_1.Main.AnimGraphNode_StateMachine_0.Main.AnimStateNode_35.OnGround.AnimGraphNode_StateMachine_0.OnGround.AnimStateNode_4.Normal.AnimGraphNode_StateMachine_0.Normal.AnimStateNode_1.Crouching.AnimGraphNode_StateMachine_0.Crouch.AnimStateTransitionNode_4.Transition
    src=CrouchWalk
    via=AnimStateTransitionNode_4
   }
   {
    dst=CrouchSprinting
    rule=AnimGraph.AnimGraphNode_StateMachine_1.FullBody_1.AnimStateNode_1.Main.AnimGraphNode_StateMachine_0.Main.AnimStateNode_35.OnGround.AnimGraphNode_StateMachine_0.OnGround.AnimStateNode_4.Normal.AnimGraphNode_StateMachine_0.Normal.AnimStateNode_1.Crouching.AnimGraphNode_StateMachine_0.Crouch.AnimStateTransitionNode_47.AnimationTransitionGraph_1
    src=CrouchWalk
    via=AnimStateTransitionNode_47
   }
   {
    dst=LocomotionCrouchHappy
    rule=AnimGraph.AnimGraphNode_StateMachine_1.FullBody_1.AnimStateNode_1.Main.AnimGraphNode_StateMachine_0.Main.AnimStateNode_35.OnGround.AnimGraphNode_StateMachine_0.OnGround.AnimStateNode_4.Normal.AnimGraphNode_StateMachine_0.Normal.AnimStateNode_1.Crouching.AnimGraphNode_StateMachine_0.Crouch.AnimStateTransitionNode_2.Transition
    src=CrouchWalk
    via=AnimStateTransitionNode_2
   }
   {
    dst=CrouchWalk
    rule=AnimGraph.AnimGraphNode_StateMachine_1.FullBody_1.AnimStateNode_1.Main.AnimGraphNode_StateMachine_0.Main.AnimStateNode_35.OnGround.AnimGraphNode_StateMachine_0.OnGround.AnimStateNode_4.Normal.AnimGraphNode_StateMachine_0.Normal.AnimStateNode_1.Crouching.AnimGraphNode_StateMachine_0.Crouch.AnimStateTransitionNode_6.Transition
    src=LocomotionCrouchHappy
    via=AnimStateTransitionNode_6
   }
   {
    dst=Crouch
    rule=AnimGraph.AnimGraphNode_StateMachine_1.FullBody_1.AnimStateNode_1.Main.AnimGraphNode_StateMachine_0.Main.AnimStateNode_35.OnGround.AnimGraphNode_StateMachine_0.OnGround.AnimStateNode_4.Normal.AnimGraphNode_StateMachine_0.Normal.AnimStateNode_1.Crouching.AnimGraphNode_StateMachine_0.Crouch.AnimStateTransitionNode_5.Transition
    src=toCrouchWalk
    via=AnimStateTransitionNode_5
   }
   {
    dst=CrouchWalk
    rule=AnimGraph.AnimGraphNode_StateMachine_1.FullBody_1.AnimStateNode_1.Main.AnimGraphNode_StateMachine_0.Main.AnimStateNode_35.OnGround.AnimGraphNode_StateMachine_0.OnGround.AnimStateNode_4.Normal.AnimGraphNode_StateMachine_0.Normal.AnimStateNode_1.Crouching.AnimGraphNode_StateMachine_0.Crouch.AnimStateTransitionNode_1.Transition
    src=toCrouchWalk
    via=AnimStateTransitionNode_1
   }
  ]
 }
 Blank {
  path=AnimGraph.AnimGraphNode_StateMachine_1.FullBody_1.AnimStateNode_1.Main.AnimGraphNode_StateMachine_0.Main.AnimStateNode_35.OnGround.AnimGraphNode_StateMachine_0.OnGround.AnimStateNode_4.Normal.AnimGraphNode_StateMachine_0.Normal.AnimStateNode_1.Crouching.AnimGraphNode_StateMachine_0.Crouch.AnimStateNode_0.Blank
  nodes [Output Animation Pose()]
 }
 toCrouchWalk {
  path=AnimGraph.AnimGraphNode_StateMachine_1.FullBody_1.AnimStateNode_1.Main.AnimGraphNode_StateMachine_0.Main.AnimStateNode_35.OnGround.AnimGraphNode_StateMachine_0.OnGround.AnimStateNode_4.Normal.AnimGraphNode_StateMachine_0.Normal.AnimStateNode_1.Crouching.AnimGraphNode_StateMachine_0.Crouch.AnimStateNode_1.toCrouchWalk
  nodes [Output Animation Pose(Result=Sequence Player 'CrouchIdleToWalk'() {bLoopAnimation=false, sequence=&CrouchIdleToWalk})]
 }
 CrouchSprinting {
  path=AnimGraph.AnimGraphNode_StateMachine_1.FullBody_1.AnimStateNode_1.Main.AnimGraphNode_StateMachine_0.Main.AnimStateNode_35.OnGround.AnimGraphNode_StateMachine_0.OnGround.AnimStateNode_4.Normal.AnimGraphNode_StateMachine_0.Normal.AnimStateNode_1.Crouching.AnimGraphNode_StateMachine_0.Crouch.AnimStateNode_2.CrouchSprinting
  nodes [Output Animation Pose(Result=Sequence Player 'CrouchSprint'() {sequence=&CrouchSprint})]
 }
 CrouchWalk {
  path=AnimGraph.AnimGraphNode_StateMachine_1.FullBody_1.AnimStateNode_1.Main.AnimGraphNode_StateMachine_0.Main.AnimStateNode_35.OnGround.AnimGraphNode_StateMachine_0.OnGround.AnimStateNode_4.Normal.AnimGraphNode_StateMachine_0.Normal.AnimStateNode_1.Crouching.AnimGraphNode_StateMachine_0.Crouch.AnimStateNode_27.CrouchWalk
  nodes [Output Animation Pose(Result=Sequence Player 'CrouchWalk'() {sequence=&CrouchWalk})]
 }
 Crouch {
  path=AnimGraph.AnimGraphNode_StateMachine_1.FullBody_1.AnimStateNode_1.Main.AnimGraphNode_StateMachine_0.Main.AnimStateNode_35.OnGround.AnimGraphNode_StateMachine_0.OnGround.AnimStateNode_4.Normal.AnimGraphNode_StateMachine_0.Normal.AnimStateNode_1.Crouching.AnimGraphNode_StateMachine_0.Crouch.AnimStateNode_29.Crouch
  nodes [
   AnimGraphNode_StateMachine_0 = Crouch() {is=title AnimGraphNode_StateMachine}
   Output Animation Pose(Result=@AnimGraphNode_StateMachine_0)
  ]
 }
 Crouch {
  entry=Crouch
  path=AnimGraph.AnimGraphNode_StateMachine_1.FullBody_1.AnimStateNode_1.Main.AnimGraphNode_StateMachine_0.Main.AnimStateNode_35.OnGround.AnimGraphNode_StateMachine_0.OnGround.AnimStateNode_4.Normal.AnimGraphNode_StateMachine_0.Normal.AnimStateNode_1.Crouching.AnimGraphNode_StateMachine_0.Crouch.AnimStateNode_29.Crouch.AnimGraphNode_StateMachine_0.Crouch
  ids=id
  nodes [
   AnimStateNode_0 = Crouch() {is=title AnimStateNode}
   AnimStateNode_2 = Happy()
   AnimStateNode_3 = CrouchToHappy()
   AnimStateNode_4 = HappyToCrouch()
   AnimStateTransitionNode_4 = AnimStateTransitionNode()
   AnimStateTransitionNode_5 = AnimStateTransitionNode() {bAutomaticRuleBasedOnSequencePlayerInState=true}
   AnimStateTransitionNode_6 = AnimStateTransitionNode()
   AnimStateTransitionNode_7 = AnimStateTransitionNode() {bAutomaticRuleBasedOnSequencePlayerInState=true, crossfadeDuration=0}
  ]
  links [
   {
    dst=CrouchToHappy
    rule=AnimGraph.AnimGraphNode_StateMachine_1.FullBody_1.AnimStateNode_1.Main.AnimGraphNode_StateMachine_0.Main.AnimStateNode_35.OnGround.AnimGraphNode_StateMachine_0.OnGround.AnimStateNode_4.Normal.AnimGraphNode_StateMachine_0.Normal.AnimStateNode_1.Crouching.AnimGraphNode_StateMachine_0.Crouch.AnimStateNode_29.Crouch.AnimGraphNode_StateMachine_0.Crouch.AnimStateTransitionNode_4.Transition
    src=Crouch
    via=AnimStateTransitionNode_4
   }
   {
    dst=Happy
    rule=AnimGraph.AnimGraphNode_StateMachine_1.FullBody_1.AnimStateNode_1.Main.AnimGraphNode_StateMachine_0.Main.AnimStateNode_35.OnGround.AnimGraphNode_StateMachine_0.OnGround.AnimStateNode_4.Normal.AnimGraphNode_StateMachine_0.Normal.AnimStateNode_1.Crouching.AnimGraphNode_StateMachine_0.Crouch.AnimStateNode_29.Crouch.AnimGraphNode_StateMachine_0.Crouch.AnimStateTransitionNode_5.Transition
    src=CrouchToHappy
    via=AnimStateTransitionNode_5
   }
   {
    dst=HappyToCrouch
    rule=AnimGraph.AnimGraphNode_StateMachine_1.FullBody_1.AnimStateNode_1.Main.AnimGraphNode_StateMachine_0.Main.AnimStateNode_35.OnGround.AnimGraphNode_StateMachine_0.OnGround.AnimStateNode_4.Normal.AnimGraphNode_StateMachine_0.Normal.AnimStateNode_1.Crouching.AnimGraphNode_StateMachine_0.Crouch.AnimStateNode_29.Crouch.AnimGraphNode_StateMachine_0.Crouch.AnimStateTransitionNode_6.Transition
    src=Happy
    via=AnimStateTransitionNode_6
   }
   {
    dst=Crouch
    rule=AnimGraph.AnimGraphNode_StateMachine_1.FullBody_1.AnimStateNode_1.Main.AnimGraphNode_StateMachine_0.Main.AnimStateNode_35.OnGround.AnimGraphNode_StateMachine_0.OnGround.AnimStateNode_4.Normal.AnimGraphNode_StateMachine_0.Normal.AnimStateNode_1.Crouching.AnimGraphNode_StateMachine_0.Crouch.AnimStateNode_29.Crouch.AnimGraphNode_StateMachine_0.Crouch.AnimStateTransitionNode_7.Transition
    src=HappyToCrouch
    via=AnimStateTransitionNode_7
   }
  ]
 }
 Crouch {
  path=AnimGraph.AnimGraphNode_StateMachine_1.FullBody_1.AnimStateNode_1.Main.AnimGraphNode_StateMachine_0.Main.AnimStateNode_35.OnGround.AnimGraphNode_StateMachine_0.OnGround.AnimStateNode_4.Normal.AnimGraphNode_StateMachine_0.Normal.AnimStateNode_1.Crouching.AnimGraphNode_StateMachine_0.Crouch.AnimStateNode_29.Crouch.AnimGraphNode_StateMachine_0.Crouch.AnimStateNode_0.Crouch
  nodes [
   AnimGraphNode_StateMachine_0 = IdleLoop()
   Output Animation Pose(Result=@AnimGraphNode_StateMachine_0)
  ]
 }
 IdleLoop {
  entry=Idle1
  path=AnimGraph.AnimGraphNode_StateMachine_1.FullBody_1.AnimStateNode_1.Main.AnimGraphNode_StateMachine_0.Main.AnimStateNode_35.OnGround.AnimGraphNode_StateMachine_0.OnGround.AnimStateNode_4.Normal.AnimGraphNode_StateMachine_0.Normal.AnimStateNode_1.Crouching.AnimGraphNode_StateMachine_0.Crouch.AnimStateNode_29.Crouch.AnimGraphNode_StateMachine_0.Crouch.AnimStateNode_0.Crouch.AnimGraphNode_StateMachine_0.IdleLoop
  ids=id
  nodes [
   AnimStateNode_0 = Idle1() {
    bAlwaysResetOnEntry=true
    stateEntered {guid=#23, notifyName=RollRandomIdle2_Test}
    stateFullyBlended {guid=#24, notifyName=CrouchIdle1_FullyBlended}
   }
   AnimStateNode_1 = Idle2() {
    bAlwaysResetOnEntry=true
    stateEntered {guid=#25, notifyName=RollRandomIdle_Test}
    stateFullyBlended {guid=#26, notifyName=CrouchIdle2_FullyBlended}
   }
   AnimStateNode_2 = State()
   AnimStateNode_3 = State_1()
   AnimStateNode_4 = State_2()
   AnimStateTransitionNode_0 = AnimStateTransitionNode() {bAutomaticRuleBasedOnSequencePlayerInState=true}
   AnimStateTransitionNode_1 = AnimStateTransitionNode() {bAutomaticRuleBasedOnSequencePlayerInState=true}
   AnimStateTransitionNode_2 = AnimStateTransitionNode() {bAutomaticRuleBasedOnSequencePlayerInState=true}
   AnimStateTransitionNode_3 = AnimStateTransitionNode() {bAutomaticRuleBasedOnSequencePlayerInState=true}
   AnimStateTransitionNode_4 = AnimStateTransitionNode() {bAutomaticRuleBasedOnSequencePlayerInState=true}
  ]
  links [
   {
    dst=Idle2
    rule=AnimGraph.AnimGraphNode_StateMachine_1.FullBody_1.AnimStateNode_1.Main.AnimGraphNode_StateMachine_0.Main.AnimStateNode_35.OnGround.AnimGraphNode_StateMachine_0.OnGround.AnimStateNode_4.Normal.AnimGraphNode_StateMachine_0.Normal.AnimStateNode_1.Crouching.AnimGraphNode_StateMachine_0.Crouch.AnimStateNode_29.Crouch.AnimGraphNode_StateMachine_0.Crouch.AnimStateNode_0.Crouch.AnimGraphNode_StateMachine_0.IdleLoop.AnimStateTransitionNode_0.Transition
    src=Idle1
    via=AnimStateTransitionNode_0
   }
   {
    dst=Idle1
    rule=AnimGraph.AnimGraphNode_StateMachine_1.FullBody_1.AnimStateNode_1.Main.AnimGraphNode_StateMachine_0.Main.AnimStateNode_35.OnGround.AnimGraphNode_StateMachine_0.OnGround.AnimStateNode_4.Normal.AnimGraphNode_StateMachine_0.Normal.AnimStateNode_1.Crouching.AnimGraphNode_StateMachine_0.Crouch.AnimStateNode_29.Crouch.AnimGraphNode_StateMachine_0.Crouch.AnimStateNode_0.Crouch.AnimGraphNode_StateMachine_0.IdleLoop.AnimStateTransitionNode_1.Transition
    src=Idle2
    via=AnimStateTransitionNode_1
   }
   {
    dst=State_1
    rule=AnimGraph.AnimGraphNode_StateMachine_1.FullBody_1.AnimStateNode_1.Main.AnimGraphNode_StateMachine_0.Main.AnimStateNode_35.OnGround.AnimGraphNode_StateMachine_0.OnGround.AnimStateNode_4.Normal.AnimGraphNode_StateMachine_0.Normal.AnimStateNode_1.Crouching.AnimGraphNode_StateMachine_0.Crouch.AnimStateNode_29.Crouch.AnimGraphNode_StateMachine_0.Crouch.AnimStateNode_0.Crouch.AnimGraphNode_StateMachine_0.IdleLoop.AnimStateTransitionNode_2.Transition
    src=State
    via=AnimStateTransitionNode_2
   }
   {
    dst=State_2
    rule=AnimGraph.AnimGraphNode_StateMachine_1.FullBody_1.AnimStateNode_1.Main.AnimGraphNode_StateMachine_0.Main.AnimStateNode_35.OnGround.AnimGraphNode_StateMachine_0.OnGround.AnimStateNode_4.Normal.AnimGraphNode_StateMachine_0.Normal.AnimStateNode_1.Crouching.AnimGraphNode_StateMachine_0.Crouch.AnimStateNode_29.Crouch.AnimGraphNode_StateMachine_0.Crouch.AnimStateNode_0.Crouch.AnimGraphNode_StateMachine_0.IdleLoop.AnimStateTransitionNode_3.Transition
    src=State_1
    via=AnimStateTransitionNode_3
   }
   {
    dst=State
    rule=AnimGraph.AnimGraphNode_StateMachine_1.FullBody_1.AnimStateNode_1.Main.AnimGraphNode_StateMachine_0.Main.AnimStateNode_35.OnGround.AnimGraphNode_StateMachine_0.OnGround.AnimStateNode_4.Normal.AnimGraphNode_StateMachine_0.Normal.AnimStateNode_1.Crouching.AnimGraphNode_StateMachine_0.Crouch.AnimStateNode_29.Crouch.AnimGraphNode_StateMachine_0.Crouch.AnimStateNode_0.Crouch.AnimGraphNode_StateMachine_0.IdleLoop.AnimStateTransitionNode_4.Transition
    src=State_2
    via=AnimStateTransitionNode_4
   }
  ]
 }
 Idle1 {
  path=AnimGraph.AnimGraphNode_StateMachine_1.FullBody_1.AnimStateNode_1.Main.AnimGraphNode_StateMachine_0.Main.AnimStateNode_35.OnGround.AnimGraphNode_StateMachine_0.OnGround.AnimStateNode_4.Normal.AnimGraphNode_StateMachine_0.Normal.AnimStateNode_1.Crouching.AnimGraphNode_StateMachine_0.Crouch.AnimStateNode_29.Crouch.AnimGraphNode_StateMachine_0.Crouch.AnimStateNode_0.Crouch.AnimGraphNode_StateMachine_0.IdleLoop.AnimStateNode_0.Idle1
  nodes [
   n1 = Blend Poses by int!(
    BlendPose_0=Sequence Player 'CrouchBreak_01'() {bLoopAnimation=false, sequence=&CrouchBreak_01}
    BlendPose_1=Sequence Player 'CrouchBreak_02'() {bLoopAnimation=false, sequence=&CrouchBreak_02}
    BlendPose_2:Pose Link Structure=Apply Additive(
     Base=Sequence Player 'Crouch_Pose'!(PlayRate:float=0.005) -> (Pose) {bIgnoreForRelevancyTest=true, sequence=&Crouch_Pose}
     Additive="Sequence Player (additive) 'IdleStraightLookAround_Add'"() {bLoopAnimation=false, sequence=&IdleStraightLookAround_Add}
    ) {alphaBoolBlend {blendOption=Linear}}
    BlendTime_0=0
    BlendTime_1=0
    BlendTime_2:float=0
    ActiveChildIndex=Get RandomIdle() -> (RandomIdle=0)
   ) -> (Pose) {
    blendType=Linear
    blendPose [{linkId=-1, sourceLinkId=-1}, {linkId=-1, sourceLinkId=-1}, {linkId=-1, sourceLinkId=-1}]
    blendTime [0.1,0.1,0.1]
   }
   Output Animation Pose(Result=@n1)
  ]
 }
 Idle2 {
  path=AnimGraph.AnimGraphNode_StateMachine_1.FullBody_1.AnimStateNode_1.Main.AnimGraphNode_StateMachine_0.Main.AnimStateNode_35.OnGround.AnimGraphNode_StateMachine_0.OnGround.AnimStateNode_4.Normal.AnimGraphNode_StateMachine_0.Normal.AnimStateNode_1.Crouching.AnimGraphNode_StateMachine_0.Crouch.AnimStateNode_29.Crouch.AnimGraphNode_StateMachine_0.Crouch.AnimStateNode_0.Crouch.AnimGraphNode_StateMachine_0.IdleLoop.AnimStateNode_1.Idle2
  nodes [
   n1 = Blend Poses by int(
    BlendPose_0=Sequence Player 'CrouchBreak_01'() {bLoopAnimation=false, sequence=&CrouchBreak_01}
    BlendPose_1=Sequence Player 'CrouchBreak_02'() {bLoopAnimation=false, sequence=&CrouchBreak_02}
    BlendPose_2=Apply Additive(
     Base=Sequence Player 'Crouch_Pose'(PlayRate=0.005) {bIgnoreForRelevancyTest=true, sequence=&Crouch_Pose}
     Additive="Sequence Player (additive) 'IdleStraightLookAround_Add'"() {bLoopAnimation=false, sequence=&IdleStraightLookAround_Add}
    ) {alphaBoolBlend {blendOption=Linear}}
    BlendTime_0=0
    BlendTime_1=0
    BlendTime_2=0
    ActiveChildIndex=Get RandomIdle2() -> (RandomIdle2=0)
   ) {
    blendType=Linear
    blendPose [{linkId=-1, sourceLinkId=-1}, {linkId=-1, sourceLinkId=-1}, {linkId=-1, sourceLinkId=-1}]
    blendTime [0.1,0.1,0.1]
   }
   Output Animation Pose(Result=@n1)
  ]
 }
 State {
  path=AnimGraph.AnimGraphNode_StateMachine_1.FullBody_1.AnimStateNode_1.Main.AnimGraphNode_StateMachine_0.Main.AnimStateNode_35.OnGround.AnimGraphNode_StateMachine_0.OnGround.AnimStateNode_4.Normal.AnimGraphNode_StateMachine_0.Normal.AnimStateNode_1.Crouching.AnimGraphNode_StateMachine_0.Crouch.AnimStateNode_29.Crouch.AnimGraphNode_StateMachine_0.Crouch.AnimStateNode_0.Crouch.AnimGraphNode_StateMachine_0.IdleLoop.AnimStateNode_2.State
  nodes [
   Sequence Player 'CrouchBreak_01'() {bLoopAnimation=false, sequence=&CrouchBreak_01}
   Sequence Player 'CrouchBreak_02'() {bLoopAnimation=false, sequence=&CrouchBreak_02}
   Output Animation Pose(Result=Apply Additive(
     Base=Sequence Player 'Crouch_Pose'!() -> (Pose) {bIgnoreForRelevancyTest=true, bLoopAnimation=false, sequence=&Crouch_Pose}
     Additive="Sequence Player (additive) 'IdleStraightLookAround_Add'"() {bLoopAnimation=false, sequence=&IdleStraightLookAround_Add}
    ) {alphaBoolBlend {blendOption=Linear}})
  ]
 }
 State_1 {
  path=AnimGraph.AnimGraphNode_StateMachine_1.FullBody_1.AnimStateNode_1.Main.AnimGraphNode_StateMachine_0.Main.AnimStateNode_35.OnGround.AnimGraphNode_StateMachine_0.OnGround.AnimStateNode_4.Normal.AnimGraphNode_StateMachine_0.Normal.AnimStateNode_1.Crouching.AnimGraphNode_StateMachine_0.Crouch.AnimStateNode_29.Crouch.AnimGraphNode_StateMachine_0.Crouch.AnimStateNode_0.Crouch.AnimGraphNode_StateMachine_0.IdleLoop.AnimStateNode_3.State_1
  nodes [
   Apply Additive(
    Base=Sequence Player 'Crouch_Pose'() {bIgnoreForRelevancyTest=true, bLoopAnimation=false, sequence=&Crouch_Pose}
    Additive="Sequence Player (additive) 'IdleStraightLookAround_Add'"() {bLoopAnimation=false, sequence=&IdleStraightLookAround_Add}
   ) {alphaBoolBlend {blendOption=Linear}}
   Sequence Player 'CrouchBreak_02'() {bLoopAnimation=false, sequence=&CrouchBreak_02}
   Output Animation Pose(Result=Sequence Player 'CrouchBreak_01'() {bLoopAnimation=false, sequence=&CrouchBreak_01})
  ]
 }
 State_2 {
  path=AnimGraph.AnimGraphNode_StateMachine_1.FullBody_1.AnimStateNode_1.Main.AnimGraphNode_StateMachine_0.Main.AnimStateNode_35.OnGround.AnimGraphNode_StateMachine_0.OnGround.AnimStateNode_4.Normal.AnimGraphNode_StateMachine_0.Normal.AnimStateNode_1.Crouching.AnimGraphNode_StateMachine_0.Crouch.AnimStateNode_29.Crouch.AnimGraphNode_StateMachine_0.Crouch.AnimStateNode_0.Crouch.AnimGraphNode_StateMachine_0.IdleLoop.AnimStateNode_4.State_2
  nodes [
   Apply Additive(
    Base=Sequence Player 'Crouch_Pose'() {bIgnoreForRelevancyTest=true, bLoopAnimation=false, sequence=&Crouch_Pose}
    Additive="Sequence Player (additive) 'IdleStraightLookAround_Add'"() {bLoopAnimation=false, sequence=&IdleStraightLookAround_Add}
   ) {alphaBoolBlend {blendOption=Linear}}
   Sequence Player 'CrouchBreak_01'() {bLoopAnimation=false, sequence=&CrouchBreak_01}
   Output Animation Pose(Result=Sequence Player 'CrouchBreak_02'() {bLoopAnimation=false, sequence=&CrouchBreak_02})
  ]
 }
 Transition {
  path=AnimGraph.AnimGraphNode_StateMachine_1.FullBody_1.AnimStateNode_1.Main.AnimGraphNode_StateMachine_0.Main.AnimStateNode_35.OnGround.AnimGraphNode_StateMachine_0.OnGround.AnimStateNode_4.Normal.AnimGraphNode_StateMachine_0.Normal.AnimStateNode_1.Crouching.AnimGraphNode_StateMachine_0.Crouch.AnimStateNode_29.Crouch.AnimGraphNode_StateMachine_0.Crouch.AnimStateNode_0.Crouch.AnimGraphNode_StateMachine_0.IdleLoop.AnimStateTransitionNode_0.Transition
  nodes [Result()]
 }
 Transition {
  path=AnimGraph.AnimGraphNode_StateMachine_1.FullBody_1.AnimStateNode_1.Main.AnimGraphNode_StateMachine_0.Main.AnimStateNode_35.OnGround.AnimGraphNode_StateMachine_0.OnGround.AnimStateNode_4.Normal.AnimGraphNode_StateMachine_0.Normal.AnimStateNode_1.Crouching.AnimGraphNode_StateMachine_0.Crouch.AnimStateNode_29.Crouch.AnimGraphNode_StateMachine_0.Crouch.AnimStateNode_0.Crouch.AnimGraphNode_StateMachine_0.IdleLoop.AnimStateTransitionNode_1.Transition
  nodes [Result()]
 }
 Transition {
  path=AnimGraph.AnimGraphNode_StateMachine_1.FullBody_1.AnimStateNode_1.Main.AnimGraphNode_StateMachine_0.Main.AnimStateNode_35.OnGround.AnimGraphNode_StateMachine_0.OnGround.AnimStateNode_4.Normal.AnimGraphNode_StateMachine_0.Normal.AnimStateNode_1.Crouching.AnimGraphNode_StateMachine_0.Crouch.AnimStateNode_29.Crouch.AnimGraphNode_StateMachine_0.Crouch.AnimStateNode_0.Crouch.AnimGraphNode_StateMachine_0.IdleLoop.AnimStateTransitionNode_2.Transition
  nodes [Result()]
 }
 Transition {
  path=AnimGraph.AnimGraphNode_StateMachine_1.FullBody_1.AnimStateNode_1.Main.AnimGraphNode_StateMachine_0.Main.AnimStateNode_35.OnGround.AnimGraphNode_StateMachine_0.OnGround.AnimStateNode_4.Normal.AnimGraphNode_StateMachine_0.Normal.AnimStateNode_1.Crouching.AnimGraphNode_StateMachine_0.Crouch.AnimStateNode_29.Crouch.AnimGraphNode_StateMachine_0.Crouch.AnimStateNode_0.Crouch.AnimGraphNode_StateMachine_0.IdleLoop.AnimStateTransitionNode_3.Transition
  nodes [Result()]
 }
 Transition {
  path=AnimGraph.AnimGraphNode_StateMachine_1.FullBody_1.AnimStateNode_1.Main.AnimGraphNode_StateMachine_0.Main.AnimStateNode_35.OnGround.AnimGraphNode_StateMachine_0.OnGround.AnimStateNode_4.Normal.AnimGraphNode_StateMachine_0.Normal.AnimStateNode_1.Crouching.AnimGraphNode_StateMachine_0.Crouch.AnimStateNode_29.Crouch.AnimGraphNode_StateMachine_0.Crouch.AnimStateNode_0.Crouch.AnimGraphNode_StateMachine_0.IdleLoop.AnimStateTransitionNode_4.Transition
  nodes [Result()]
 }
 Happy {
  path=AnimGraph.AnimGraphNode_StateMachine_1.FullBody_1.AnimStateNode_1.Main.AnimGraphNode_StateMachine_0.Main.AnimStateNode_35.OnGround.AnimGraphNode_StateMachine_0.OnGround.AnimStateNode_4.Normal.AnimGraphNode_StateMachine_0.Normal.AnimStateNode_1.Crouching.AnimGraphNode_StateMachine_0.Crouch.AnimStateNode_29.Crouch.AnimGraphNode_StateMachine_0.Crouch.AnimStateNode_2.Happy
  nodes [Output Animation Pose(Result=Sequence Player 'Happy'() {sequence=&Happy})]
 }
 CrouchToHappy {
  path=AnimGraph.AnimGraphNode_StateMachine_1.FullBody_1.AnimStateNode_1.Main.AnimGraphNode_StateMachine_0.Main.AnimStateNode_35.OnGround.AnimGraphNode_StateMachine_0.OnGround.AnimStateNode_4.Normal.AnimGraphNode_StateMachine_0.Normal.AnimStateNode_1.Crouching.AnimGraphNode_StateMachine_0.Crouch.AnimStateNode_29.Crouch.AnimGraphNode_StateMachine_0.Crouch.AnimStateNode_3.CrouchToHappy
  nodes [Output Animation Pose(Result=Sequence Player 'CrouchToHappy'() {sequence=&CrouchToHappy})]
 }
 HappyToCrouch {
  path=AnimGraph.AnimGraphNode_StateMachine_1.FullBody_1.AnimStateNode_1.Main.AnimGraphNode_StateMachine_0.Main.AnimStateNode_35.OnGround.AnimGraphNode_StateMachine_0.OnGround.AnimStateNode_4.Normal.AnimGraphNode_StateMachine_0.Normal.AnimStateNode_1.Crouching.AnimGraphNode_StateMachine_0.Crouch.AnimStateNode_29.Crouch.AnimGraphNode_StateMachine_0.Crouch.AnimStateNode_4.HappyToCrouch
  nodes [Output Animation Pose(Result=Sequence Player 'HappyToCrouch'() {bLoopAnimation=false, sequence=&HappyToCrouch})]
 }
 Transition {
  path=AnimGraph.AnimGraphNode_StateMachine_1.FullBody_1.AnimStateNode_1.Main.AnimGraphNode_StateMachine_0.Main.AnimStateNode_35.OnGround.AnimGraphNode_StateMachine_0.OnGround.AnimStateNode_4.Normal.AnimGraphNode_StateMachine_0.Normal.AnimStateNode_1.Crouching.AnimGraphNode_StateMachine_0.Crouch.AnimStateNode_29.Crouch.AnimGraphNode_StateMachine_0.Crouch.AnimStateTransitionNode_4.Transition
  nodes [Result(bCanEnterTransition=Get bIsHappy())]
 }
 Transition {
  path=AnimGraph.AnimGraphNode_StateMachine_1.FullBody_1.AnimStateNode_1.Main.AnimGraphNode_StateMachine_0.Main.AnimStateNode_35.OnGround.AnimGraphNode_StateMachine_0.OnGround.AnimStateNode_4.Normal.AnimGraphNode_StateMachine_0.Normal.AnimStateNode_1.Crouching.AnimGraphNode_StateMachine_0.Crouch.AnimStateNode_29.Crouch.AnimGraphNode_StateMachine_0.Crouch.AnimStateTransitionNode_5.Transition
  nodes [Result()]
 }
 Transition {
  path=AnimGraph.AnimGraphNode_StateMachine_1.FullBody_1.AnimStateNode_1.Main.AnimGraphNode_StateMachine_0.Main.AnimStateNode_35.OnGround.AnimGraphNode_StateMachine_0.OnGround.AnimStateNode_4.Normal.AnimGraphNode_StateMachine_0.Normal.AnimStateNode_1.Crouching.AnimGraphNode_StateMachine_0.Crouch.AnimStateNode_29.Crouch.AnimGraphNode_StateMachine_0.Crouch.AnimStateTransitionNode_6.Transition
  nodes [
   n1 = "Get Relevant Anim Time Remaining (Happy)"()
   n2 = Get bIsHappy()
   Result(bCanEnterTransition=AND Boolean!(A="float < float"(A=@n1, B=0.1), B=NOT Boolean(A=@n2)) -> (ReturnValue))
  ]
 }
 Transition {
  path=AnimGraph.AnimGraphNode_StateMachine_1.FullBody_1.AnimStateNode_1.Main.AnimGraphNode_StateMachine_0.Main.AnimStateNode_35.OnGround.AnimGraphNode_StateMachine_0.OnGround.AnimStateNode_4.Normal.AnimGraphNode_StateMachine_0.Normal.AnimStateNode_1.Crouching.AnimGraphNode_StateMachine_0.Crouch.AnimStateNode_29.Crouch.AnimGraphNode_StateMachine_0.Crouch.AnimStateTransitionNode_7.Transition
  nodes [Result()]
 }
 LocomotionCrouchHappy {
  path=AnimGraph.AnimGraphNode_StateMachine_1.FullBody_1.AnimStateNode_1.Main.AnimGraphNode_StateMachine_0.Main.AnimStateNode_35.OnGround.AnimGraphNode_StateMachine_0.OnGround.AnimStateNode_4.Normal.AnimGraphNode_StateMachine_0.Normal.AnimStateNode_1.Crouching.AnimGraphNode_StateMachine_0.Crouch.AnimStateNode_3.LocomotionCrouchHappy
  nodes [Output Animation Pose(Result=Sequence Player 'LocomotionCrouchHappy'() {sequence=&LocomotionCrouchHappy})]
 }
 Transition {
  path=AnimGraph.AnimGraphNode_StateMachine_1.FullBody_1.AnimStateNode_1.Main.AnimGraphNode_StateMachine_0.Main.AnimStateNode_35.OnGround.AnimGraphNode_StateMachine_0.OnGround.AnimStateNode_4.Normal.AnimGraphNode_StateMachine_0.Normal.AnimStateNode_1.Crouching.AnimGraphNode_StateMachine_0.Crouch.AnimStateTransitionNode_0.Transition
  nodes [Result(bCanEnterTransition=Get bIsMoving2D())]
 }
 Transition {
  path=AnimGraph.AnimGraphNode_StateMachine_1.FullBody_1.AnimStateNode_1.Main.AnimGraphNode_StateMachine_0.Main.AnimStateNode_35.OnGround.AnimGraphNode_StateMachine_0.OnGround.AnimStateNode_4.Normal.AnimGraphNode_StateMachine_0.Normal.AnimStateNode_1.Crouching.AnimGraphNode_StateMachine_0.Crouch.AnimStateTransitionNode_1.Transition
  nodes [Result()]
 }
 Transition {
  path=AnimGraph.AnimGraphNode_StateMachine_1.FullBody_1.AnimStateNode_1.Main.AnimGraphNode_StateMachine_0.Main.AnimStateNode_35.OnGround.AnimGraphNode_StateMachine_0.OnGround.AnimStateNode_4.Normal.AnimGraphNode_StateMachine_0.Normal.AnimStateNode_1.Crouching.AnimGraphNode_StateMachine_0.Crouch.AnimStateTransitionNode_11.Transition
  nodes [Result(bCanEnterTransition=NOT Boolean(A=Get bIsMoving2D()))]
 }
 Transition {
  path=AnimGraph.AnimGraphNode_StateMachine_1.FullBody_1.AnimStateNode_1.Main.AnimGraphNode_StateMachine_0.Main.AnimStateNode_35.OnGround.AnimGraphNode_StateMachine_0.OnGround.AnimStateNode_4.Normal.AnimGraphNode_StateMachine_0.Normal.AnimStateNode_1.Crouching.AnimGraphNode_StateMachine_0.Crouch.AnimStateTransitionNode_2.Transition
  nodes [Result(bCanEnterTransition=Get bIsHappy())]
 }
 Transition {
  path=AnimGraph.AnimGraphNode_StateMachine_1.FullBody_1.AnimStateNode_1.Main.AnimGraphNode_StateMachine_0.Main.AnimStateNode_35.OnGround.AnimGraphNode_StateMachine_0.OnGround.AnimStateNode_4.Normal.AnimGraphNode_StateMachine_0.Normal.AnimStateNode_1.Crouching.AnimGraphNode_StateMachine_0.Crouch.AnimStateTransitionNode_3.Transition
  nodes [Result(bCanEnterTransition=Get bIsHappy())]
 }
 Transition {
  path=AnimGraph.AnimGraphNode_StateMachine_1.FullBody_1.AnimStateNode_1.Main.AnimGraphNode_StateMachine_0.Main.AnimStateNode_35.OnGround.AnimGraphNode_StateMachine_0.OnGround.AnimStateNode_4.Normal.AnimGraphNode_StateMachine_0.Normal.AnimStateNode_1.Crouching.AnimGraphNode_StateMachine_0.Crouch.AnimStateTransitionNode_4.Transition
  nodes [Result(bCanEnterTransition=NOT Boolean(A=Get bIsMoving2D()))]
 }
 AnimationTransitionGraph_1 {
  path=AnimGraph.AnimGraphNode_StateMachine_1.FullBody_1.AnimStateNode_1.Main.AnimGraphNode_StateMachine_0.Main.AnimStateNode_35.OnGround.AnimGraphNode_StateMachine_0.OnGround.AnimStateNode_4.Normal.AnimGraphNode_StateMachine_0.Normal.AnimStateNode_1.Crouching.AnimGraphNode_StateMachine_0.Crouch.AnimStateTransitionNode_47.AnimationTransitionGraph_1
  nodes [Result(bCanEnterTransition=Get bIsCrouchSprinting())]
 }
 Transition {
  path=AnimGraph.AnimGraphNode_StateMachine_1.FullBody_1.AnimStateNode_1.Main.AnimGraphNode_StateMachine_0.Main.AnimStateNode_35.OnGround.AnimGraphNode_StateMachine_0.OnGround.AnimStateNode_4.Normal.AnimGraphNode_StateMachine_0.Normal.AnimStateNode_1.Crouching.AnimGraphNode_StateMachine_0.Crouch.AnimStateTransitionNode_5.Transition
  nodes [Result(bCanEnterTransition=NOT Boolean(A=Get bIsMoving2D()))]
 }
 AnimationTransitionGraph_1 {
  path=AnimGraph.AnimGraphNode_StateMachine_1.FullBody_1.AnimStateNode_1.Main.AnimGraphNode_StateMachine_0.Main.AnimStateNode_35.OnGround.AnimGraphNode_StateMachine_0.OnGround.AnimStateNode_4.Normal.AnimGraphNode_StateMachine_0.Normal.AnimStateNode_1.Crouching.AnimGraphNode_StateMachine_0.Crouch.AnimStateTransitionNode_55.AnimationTransitionGraph_1
  nodes [Result(bCanEnterTransition=NOT Boolean(A=Get bIsCrouchSprinting()))]
 }
 Transition {
  path=AnimGraph.AnimGraphNode_StateMachine_1.FullBody_1.AnimStateNode_1.Main.AnimGraphNode_StateMachine_0.Main.AnimStateNode_35.OnGround.AnimGraphNode_StateMachine_0.OnGround.AnimStateNode_4.Normal.AnimGraphNode_StateMachine_0.Normal.AnimStateNode_1.Crouching.AnimGraphNode_StateMachine_0.Crouch.AnimStateTransitionNode_6.Transition
  nodes [Result()]
 }
 Transition {
  path=AnimGraph.AnimGraphNode_StateMachine_1.FullBody_1.AnimStateNode_1.Main.AnimGraphNode_StateMachine_0.Main.AnimStateNode_35.OnGround.AnimGraphNode_StateMachine_0.OnGround.AnimStateNode_4.Normal.AnimGraphNode_StateMachine_0.Normal.AnimStateNode_1.Crouching.AnimGraphNode_StateMachine_0.Crouch.AnimStateTransitionNode_7.Transition
  nodes [Result(bCanEnterTransition=Get bIsMoving2D())]
 }
 ToCrouch {
  path=AnimGraph.AnimGraphNode_StateMachine_1.FullBody_1.AnimStateNode_1.Main.AnimGraphNode_StateMachine_0.Main.AnimStateNode_35.OnGround.AnimGraphNode_StateMachine_0.OnGround.AnimStateNode_4.Normal.AnimGraphNode_StateMachine_0.Normal.AnimStateNode_2.ToCrouch
  nodes [Output Animation Pose(Result=Sequence Player 'LocomotionToCrouch'() {bLoopAnimation=false, sequence=&LocomotionToCrouch})]
 }
 ToStanding {
  path=AnimGraph.AnimGraphNode_StateMachine_1.FullBody_1.AnimStateNode_1.Main.AnimGraphNode_StateMachine_0.Main.AnimStateNode_35.OnGround.AnimGraphNode_StateMachine_0.OnGround.AnimStateNode_4.Normal.AnimGraphNode_StateMachine_0.Normal.AnimStateNode_3.ToStanding
  nodes [Output Animation Pose(Result=Sequence Player 'CrouchIdleToIdle'() {bLoopAnimation=false, sequence=&CrouchIdleToIdle})]
 }
 AnimationTransitionGraph_1 {
  path=AnimGraph.AnimGraphNode_StateMachine_1.FullBody_1.AnimStateNode_1.Main.AnimGraphNode_StateMachine_0.Main.AnimStateNode_35.OnGround.AnimGraphNode_StateMachine_0.OnGround.AnimStateNode_4.Normal.AnimGraphNode_StateMachine_0.Normal.AnimStateTransitionNode_2.AnimationTransitionGraph_1
  nodes [Result(bCanEnterTransition=Get bIsCrouching())]
 }
 AnimationTransitionGraph_1 {
  path=AnimGraph.AnimGraphNode_StateMachine_1.FullBody_1.AnimStateNode_1.Main.AnimGraphNode_StateMachine_0.Main.AnimStateNode_35.OnGround.AnimGraphNode_StateMachine_0.OnGround.AnimStateNode_4.Normal.AnimGraphNode_StateMachine_0.Normal.AnimStateTransitionNode_3.AnimationTransitionGraph_1
  nodes [Result()]
 }
 AnimationTransitionGraph_1 {
  path=AnimGraph.AnimGraphNode_StateMachine_1.FullBody_1.AnimStateNode_1.Main.AnimGraphNode_StateMachine_0.Main.AnimStateNode_35.OnGround.AnimGraphNode_StateMachine_0.OnGround.AnimStateNode_4.Normal.AnimGraphNode_StateMachine_0.Normal.AnimStateTransitionNode_4.AnimationTransitionGraph_1
  nodes [Result(bCanEnterTransition=NOT Boolean(A=Get bIsCrouching()))]
 }
 Transition {
  path=AnimGraph.AnimGraphNode_StateMachine_1.FullBody_1.AnimStateNode_1.Main.AnimGraphNode_StateMachine_0.Main.AnimStateNode_35.OnGround.AnimGraphNode_StateMachine_0.OnGround.AnimStateNode_4.Normal.AnimGraphNode_StateMachine_0.Normal.AnimStateTransitionNode_44.Transition
  nodes [Result(bCanEnterTransition=Get bAllowSkippingIdleToCrouchTransition() -> (bAllowSkippingIdleToCrouchTransition:bool))]
 }
 AnimationTransitionGraph_1 {
  path=AnimGraph.AnimGraphNode_StateMachine_1.FullBody_1.AnimStateNode_1.Main.AnimGraphNode_StateMachine_0.Main.AnimStateNode_35.OnGround.AnimGraphNode_StateMachine_0.OnGround.AnimStateNode_4.Normal.AnimGraphNode_StateMachine_0.Normal.AnimStateTransitionNode_5.AnimationTransitionGraph_1
  nodes [Result()]
 }
 AnimationTransitionGraph_1 {
  path=AnimGraph.AnimGraphNode_StateMachine_1.FullBody_1.AnimStateNode_1.Main.AnimGraphNode_StateMachine_0.Main.AnimStateNode_35.OnGround.AnimGraphNode_StateMachine_0.OnGround.AnimStateNode_4.Normal.AnimGraphNode_StateMachine_0.Normal.AnimStateTransitionNode_6.AnimationTransitionGraph_1
  nodes [Result(bCanEnterTransition=NOT Boolean(A=Get bIsCrouching()))]
 }
 AnimationTransitionGraph_1 {
  path=AnimGraph.AnimGraphNode_StateMachine_1.FullBody_1.AnimStateNode_1.Main.AnimGraphNode_StateMachine_0.Main.AnimStateNode_35.OnGround.AnimGraphNode_StateMachine_0.OnGround.AnimStateNode_4.Normal.AnimGraphNode_StateMachine_0.Normal.AnimStateTransitionNode_7.AnimationTransitionGraph_1
  nodes [Result(bCanEnterTransition=Get bIsCrouching())]
 }
 Hunker {
  path=AnimGraph.AnimGraphNode_StateMachine_1.FullBody_1.AnimStateNode_1.Main.AnimGraphNode_StateMachine_0.Main.AnimStateNode_35.OnGround.AnimGraphNode_StateMachine_0.OnGround.AnimStateNode_5.Hunker
  nodes [Output Animation Pose(Result=Apply Additive(
     Base=Sequence Player 'Hunker_Pose'() {sequence=&Hunker_Pose}
     Additive="Sequence Player (additive) 'IdleStraightLookAround_Add'"() {sequence=&IdleStraightLookAround_Add}
    ) {alphaBoolBlend {blendOption=Linear}})]
 }
 Transition {
  path=AnimGraph.AnimGraphNode_StateMachine_1.FullBody_1.AnimStateNode_1.Main.AnimGraphNode_StateMachine_0.Main.AnimStateNode_35.OnGround.AnimGraphNode_StateMachine_0.OnGround.AnimStateTransitionNode_27.Transition
  nodes [Result(bCanEnterTransition=Get bIsTargeting())]
 }
 Transition {
  path=AnimGraph.AnimGraphNode_StateMachine_1.FullBody_1.AnimStateNode_1.Main.AnimGraphNode_StateMachine_0.Main.AnimStateNode_35.OnGround.AnimGraphNode_StateMachine_0.OnGround.AnimStateTransitionNode_28.Transition
  nodes [Result(bCanEnterTransition=NOT Boolean(A=Get bIsCrouching()))]
 }
 Transition {
  path=AnimGraph.AnimGraphNode_StateMachine_1.FullBody_1.AnimStateNode_1.Main.AnimGraphNode_StateMachine_0.Main.AnimStateNode_35.OnGround.AnimGraphNode_StateMachine_0.OnGround.AnimStateTransitionNode_29.Transition
  nodes [Result()]
 }
 Transition {
  path=AnimGraph.AnimGraphNode_StateMachine_1.FullBody_1.AnimStateNode_1.Main.AnimGraphNode_StateMachine_0.Main.AnimStateNode_35.OnGround.AnimGraphNode_StateMachine_0.OnGround.AnimStateTransitionNode_32.Transition
  nodes [Result()]
 }
 Transition {
  path=AnimGraph.AnimGraphNode_StateMachine_1.FullBody_1.AnimStateNode_1.Main.AnimGraphNode_StateMachine_0.Main.AnimStateNode_35.OnGround.AnimGraphNode_StateMachine_0.OnGround.AnimStateTransitionNode_33.Transition
  nodes [Result()]
 }
 Transition {
  path=AnimGraph.AnimGraphNode_StateMachine_1.FullBody_1.AnimStateNode_1.Main.AnimGraphNode_StateMachine_0.Main.AnimStateNode_35.OnGround.AnimGraphNode_StateMachine_0.OnGround.AnimStateTransitionNode_36.Transition
  nodes [Result(bCanEnterTransition=NOT Boolean(A=Get bIsTargeting()))]
 }
 Transition {
  path=AnimGraph.AnimGraphNode_StateMachine_1.FullBody_1.AnimStateNode_1.Main.AnimGraphNode_StateMachine_0.Main.AnimStateNode_35.OnGround.AnimGraphNode_StateMachine_0.OnGround.AnimStateTransitionNode_37.Transition
  nodes [Result(bCanEnterTransition=NOT Boolean(A=Get bIsCrouching()))]
 }
 Transition {
  path=AnimGraph.AnimGraphNode_StateMachine_1.FullBody_1.AnimStateNode_1.Main.AnimGraphNode_StateMachine_0.Main.AnimStateNode_35.OnGround.AnimGraphNode_StateMachine_0.OnGround.AnimStateTransitionNode_38.Transition
  nodes [Result(bCanEnterTransition=Get bIsCrouching())]
 }
 Transition {
  path=AnimGraph.AnimGraphNode_StateMachine_1.FullBody_1.AnimStateNode_1.Main.AnimGraphNode_StateMachine_0.Main.AnimStateNode_35.OnGround.AnimGraphNode_StateMachine_0.OnGround.AnimStateTransitionNode_39.Transition
  nodes [Result()]
 }
 Transition {
  path=AnimGraph.AnimGraphNode_StateMachine_1.FullBody_1.AnimStateNode_1.Main.AnimGraphNode_StateMachine_0.Main.AnimStateNode_35.OnGround.AnimGraphNode_StateMachine_0.OnGround.AnimStateTransitionNode_40.Transition
  nodes [Result()]
 }
 Transition {
  path=AnimGraph.AnimGraphNode_StateMachine_1.FullBody_1.AnimStateNode_1.Main.AnimGraphNode_StateMachine_0.Main.AnimStateNode_35.OnGround.AnimGraphNode_StateMachine_0.OnGround.AnimStateTransitionNode_41.Transition
  nodes [Result(bCanEnterTransition=Get bIsCrouching())]
 }
 Transition {
  path=AnimGraph.AnimGraphNode_StateMachine_1.FullBody_1.AnimStateNode_1.Main.AnimGraphNode_StateMachine_0.Main.AnimStateNode_35.OnGround.AnimGraphNode_StateMachine_0.OnGround.AnimStateTransitionNode_42.Transition
  nodes [Result(bCanEnterTransition=NOT Boolean(A=Get bIsMoving2D()))]
 }
 Transition {
  path=AnimGraph.AnimGraphNode_StateMachine_1.FullBody_1.AnimStateNode_1.Main.AnimGraphNode_StateMachine_0.Main.AnimStateNode_35.OnGround.AnimGraphNode_StateMachine_0.OnGround.AnimStateTransitionNode_43.Transition
  nodes [Result(bCanEnterTransition=Get bIsMoving2D())]
 }
 Transition {
  path=AnimGraph.AnimGraphNode_StateMachine_1.FullBody_1.AnimStateNode_1.Main.AnimGraphNode_StateMachine_0.Main.AnimStateNode_35.OnGround.AnimGraphNode_StateMachine_0.OnGround.AnimStateTransitionNode_8.Transition
  nodes [Result(bCanEnterTransition=Get bIsHunker())]
 }
 Transition {
  path=AnimGraph.AnimGraphNode_StateMachine_1.FullBody_1.AnimStateNode_1.Main.AnimGraphNode_StateMachine_0.Main.AnimStateNode_35.OnGround.AnimGraphNode_StateMachine_0.OnGround.AnimStateTransitionNode_9.Transition
  nodes [Result(bCanEnterTransition=NOT Boolean(A=Get bIsHunker()))]
 }
 StartJump {
  path=AnimGraph.AnimGraphNode_StateMachine_1.FullBody_1.AnimStateNode_1.Main.AnimGraphNode_StateMachine_0.Main.AnimStateNode_4.StartJump
  nodes [
   Layered blend per bone(BlendPoses_0=Local Space Ref Pose(), BlendWeights_0="To Float (Boolean)"(InBool=Get bIsFox())) {layerSetup [{branchFilters [{boneName=faceAttach}]}]}
   n1 = Blend Poses by int!(
    BlendPose_0=Sequence Player 'JumpUp_01'() {sequence=&JumpUp_01}
    BlendPose_1=Sequence Player 'JumpUp_02'() {sequence=&JumpUp_02}
    BlendTime_0=0.1
    BlendTime_1=0.1
    ActiveChildIndex=Get JumpIndex() -> (JumpIndex=0)
   ) -> (Pose) {blendType=Linear, blendPose [{linkId=-1, sourceLinkId=-1}, {linkId=-1, sourceLinkId=-1}], blendTime [0.1,0.1]}
   Output Animation Pose(Result=Blend Poses by bool(
     BlendPose_0=@n1
     BlendPose_1=Sequence Player 'JumpUp_00'() {sequence=&JumpUp_00}
     bActiveValue=Get bIsMovingDuringJumpInitiated()
    ) {blendType=Linear})
  ]
 }
 JumpLand {
  path=AnimGraph.AnimGraphNode_StateMachine_1.FullBody_1.AnimStateNode_1.Main.AnimGraphNode_StateMachine_0.Main.AnimStateNode_7.JumpLand
  nodes [
   Layered blend per bone(BlendPoses_0=Local Space Ref Pose(), BlendWeights_0="To Float (Boolean)"(InBool=Get bIsFox())) {layerSetup [{branchFilters [{boneName=faceAttach}]}]}
   n1 = Blend Poses by int(
    BlendPose_0=Sequence Player 'JumpLand_01'(bLoopAnimation=False) {bLoopAnimation=false, sequence=&JumpLand_01}
    BlendPose_1=Sequence Player 'JumpLand_02'(bLoopAnimation=False) {bLoopAnimation=false, sequence=&JumpLand_02}
    BlendTime_0=0.1
    BlendTime_1=0.1
    ActiveChildIndex=Get JumpIndex() -> (JumpIndex=0)
   ) {blendType=Linear, blendPose [{linkId=-1, sourceLinkId=-1}, {linkId=-1, sourceLinkId=-1}], blendTime [0.1,0.1]}
   Output Animation Pose(Result=Blend Poses by bool(
     BlendPose_0=@n1
     BlendPose_1=Sequence Player 'JumpLand_00'(bLoopAnimation=False) {bLoopAnimation=false, sequence=&JumpLand_00}
     bActiveValue=Get bIsMovingDuringJumpInitiated()
    ) {blendType=Linear})
  ]
 }
 AnimationTransitionGraph_1 {
  path=AnimGraph.AnimGraphNode_StateMachine_1.FullBody_1.AnimStateNode_1.Main.AnimGraphNode_StateMachine_0.Main.AnimStateTransitionNode_10.AnimationTransitionGraph_1
  nodes [Result(bCanEnterTransition=Get bIsJumping())]
 }
 AnimationTransitionGraph_1 {
  path=AnimGraph.AnimGraphNode_StateMachine_1.FullBody_1.AnimStateNode_1.Main.AnimGraphNode_StateMachine_0.Main.AnimStateTransitionNode_11.AnimationTransitionGraph_1
  nodes [Result(bCanEnterTransition=NOT Boolean(A=Get bIsJumping()))]
 }
 AnimationTransitionGraph_1 {
  path=AnimGraph.AnimGraphNode_StateMachine_1.FullBody_1.AnimStateNode_1.Main.AnimGraphNode_StateMachine_0.Main.AnimStateTransitionNode_12.AnimationTransitionGraph_1
  nodes [Result(bCanEnterTransition=NOT Boolean(A=Get bIsOnGround()))]
 }
 AnimationTransitionGraph_1 {
  path=AnimGraph.AnimGraphNode_StateMachine_1.FullBody_1.AnimStateNode_1.Main.AnimGraphNode_StateMachine_0.Main.AnimStateTransitionNode_17.AnimationTransitionGraph_1
  nodes [Result(bCanEnterTransition=Get bIsOnGround())]
 }
 AnimationTransitionGraph_1 {
  path=AnimGraph.AnimGraphNode_StateMachine_1.FullBody_1.AnimStateNode_1.Main.AnimGraphNode_StateMachine_0.Main.AnimStateTransitionNode_20.AnimationTransitionGraph_1
  nodes [Result(bCanEnterTransition=Get bIsInAirNotJumping())]
 }
 AnimationTransitionGraph_1 {
  path=AnimGraph.AnimGraphNode_StateMachine_1.FullBody_1.AnimStateNode_1.Main.AnimGraphNode_StateMachine_0.Main.AnimStateTransitionNode_5.AnimationTransitionGraph_1
  nodes [Result()]
 }
 AnimationTransitionGraph_1 {
  path=AnimGraph.AnimGraphNode_StateMachine_1.FullBody_1.AnimStateNode_1.Main.AnimGraphNode_StateMachine_0.Main.AnimStateTransitionNode_8.AnimationTransitionGraph_1
  nodes [Result()]
 }
 AnimationTransitionGraph_1 {
  path=AnimGraph.AnimGraphNode_StateMachine_1.FullBody_1.AnimStateNode_1.Main.AnimGraphNode_StateMachine_0.Main.AnimStateTransitionNode_9.AnimationTransitionGraph_1
  nodes [Result(bCanEnterTransition=NOT Boolean(A=Get bIsOnGround()))]
 }
 DBNOToIdle {
  path=AnimGraph.AnimGraphNode_StateMachine_1.FullBody_1.AnimStateNode_11.DBNOToIdle
  nodes [Output Animation Pose(Result=Sequence Player 'DBNO_Idle_Outro'() {bLoopAnimation=false, sequence=&DBNO_Idle_Outro})]
 }
 DBNO {
  path=AnimGraph.AnimGraphNode_StateMachine_1.FullBody_1.AnimStateNode_9.DBNO
  nodes [
   AnimGraphNode_StateMachine_0 = DBNO() {is=title AnimGraphNode_StateMachine}
   Output Animation Pose(Result=@AnimGraphNode_StateMachine_0)
  ]
 }
 DBNO {
  entry=IdleToDBNO
  path=AnimGraph.AnimGraphNode_StateMachine_1.FullBody_1.AnimStateNode_9.DBNO.AnimGraphNode_StateMachine_0.DBNO
  ids=id
  nodes [
   AnimStateNode_0 = DNBO()
   AnimStateNode_1 = ToDBNOHunker()
   AnimStateNode_10 = IdleToDBNO()
   AnimStateNode_2 = DBNOHunker()
   AnimStateNode_3 = ToDBNO()
   AnimStateTransitionNode_0 = AnimStateTransitionNode()
   AnimStateTransitionNode_1 = AnimStateTransitionNode() {bAutomaticRuleBasedOnSequencePlayerInState=true, crossfadeDuration=0}
   AnimStateTransitionNode_2 = AnimStateTransitionNode()
   AnimStateTransitionNode_3 = AnimStateTransitionNode() {bAutomaticRuleBasedOnSequencePlayerInState=true, crossfadeDuration=0}
   AnimStateTransitionNode_45 = AnimStateTransitionNode() {bAutomaticRuleBasedOnSequencePlayerInState=true}
  ]
  links [
   {
    dst=ToDBNO
    rule=AnimGraph.AnimGraphNode_StateMachine_1.FullBody_1.AnimStateNode_9.DBNO.AnimGraphNode_StateMachine_0.DBNO.AnimStateTransitionNode_2.Transition
    src=DBNOHunker
    via=AnimStateTransitionNode_2
   }
   {
    dst=ToDBNOHunker
    rule=AnimGraph.AnimGraphNode_StateMachine_1.FullBody_1.AnimStateNode_9.DBNO.AnimGraphNode_StateMachine_0.DBNO.AnimStateTransitionNode_0.Transition
    src=DNBO
    via=AnimStateTransitionNode_0
   }
   {
    dst=DNBO
    rule=AnimGraph.AnimGraphNode_StateMachine_1.FullBody_1.AnimStateNode_9.DBNO.AnimGraphNode_StateMachine_0.DBNO.AnimStateTransitionNode_45.Transition
    src=IdleToDBNO
    via=AnimStateTransitionNode_45
   }
   {
    dst=DNBO
    rule=AnimGraph.AnimGraphNode_StateMachine_1.FullBody_1.AnimStateNode_9.DBNO.AnimGraphNode_StateMachine_0.DBNO.AnimStateTransitionNode_3.Transition
    src=ToDBNO
    via=AnimStateTransitionNode_3
   }
   {
    dst=DBNOHunker
    rule=AnimGraph.AnimGraphNode_StateMachine_1.FullBody_1.AnimStateNode_9.DBNO.AnimGraphNode_StateMachine_0.DBNO.AnimStateTransitionNode_1.Transition
    src=ToDBNOHunker
    via=AnimStateTransitionNode_1
   }
  ]
 }
 DNBO {
  path=AnimGraph.AnimGraphNode_StateMachine_1.FullBody_1.AnimStateNode_9.DBNO.AnimGraphNode_StateMachine_0.DBNO.AnimStateNode_0.DNBO
  nodes [Output Animation Pose(Result=Sequence Player 'DBNO_Idle'() {sequence=&DBNO_Idle})]
 }
 ToDBNOHunker {
  path=AnimGraph.AnimGraphNode_StateMachine_1.FullBody_1.AnimStateNode_9.DBNO.AnimGraphNode_StateMachine_0.DBNO.AnimStateNode_1.ToDBNOHunker
  nodes [Output Animation Pose(Result=Sequence Player 'DBNOToHunker'() {bLoopAnimation=false, sequence=&DBNOToHunker})]
 }
 IdleToDBNO {
  path=AnimGraph.AnimGraphNode_StateMachine_1.FullBody_1.AnimStateNode_9.DBNO.AnimGraphNode_StateMachine_0.DBNO.AnimStateNode_10.IdleToDBNO
  nodes [Output Animation Pose(Result=Sequence Player 'DBNO_Idle_Intro'() {sequence=&DBNO_Idle_Intro})]
 }
 DBNOHunker {
  path=AnimGraph.AnimGraphNode_StateMachine_1.FullBody_1.AnimStateNode_9.DBNO.AnimGraphNode_StateMachine_0.DBNO.AnimStateNode_2.DBNOHunker
  nodes [Output Animation Pose(Result=Sequence Player 'DBNO_Hunker'() {sequence=&DBNO_Hunker})]
 }
 ToDBNO {
  path=AnimGraph.AnimGraphNode_StateMachine_1.FullBody_1.AnimStateNode_9.DBNO.AnimGraphNode_StateMachine_0.DBNO.AnimStateNode_3.ToDBNO
  nodes [Output Animation Pose(Result=Sequence Player 'HunkerToDBNO'() {bLoopAnimation=false, sequence=&HunkerToDBNO})]
 }
 Transition {
  path=AnimGraph.AnimGraphNode_StateMachine_1.FullBody_1.AnimStateNode_9.DBNO.AnimGraphNode_StateMachine_0.DBNO.AnimStateTransitionNode_0.Transition
  nodes [Result(bCanEnterTransition=Get bIsMoving2D())]
 }
 Transition {
  path=AnimGraph.AnimGraphNode_StateMachine_1.FullBody_1.AnimStateNode_9.DBNO.AnimGraphNode_StateMachine_0.DBNO.AnimStateTransitionNode_1.Transition
  nodes [Result()]
 }
 Transition {
  path=AnimGraph.AnimGraphNode_StateMachine_1.FullBody_1.AnimStateNode_9.DBNO.AnimGraphNode_StateMachine_0.DBNO.AnimStateTransitionNode_2.Transition
  nodes [Result(bCanEnterTransition=NOT Boolean(A=Get bIsMoving2D()))]
 }
 Transition {
  path=AnimGraph.AnimGraphNode_StateMachine_1.FullBody_1.AnimStateNode_9.DBNO.AnimGraphNode_StateMachine_0.DBNO.AnimStateTransitionNode_3.Transition
  nodes [Result()]
 }
 Transition {
  path=AnimGraph.AnimGraphNode_StateMachine_1.FullBody_1.AnimStateNode_9.DBNO.AnimGraphNode_StateMachine_0.DBNO.AnimStateTransitionNode_45.Transition
  nodes [Result()]
 }
 AnimationTransitionGraph_2 {
  path=AnimGraph.AnimGraphNode_StateMachine_1.FullBody_1.AnimStateTransitionNode_14.AnimationTransitionGraph_2
  nodes [Result(bCanEnterTransition=NOT Boolean(A=Get bIsDBNO()))]
 }
 Transition {
  path=AnimGraph.AnimGraphNode_StateMachine_1.FullBody_1.AnimStateTransitionNode_21.Transition
  nodes [Result(bCanEnterTransition=Get bTransition_Idle_IdleLobby())]
 }
 Transition {
  path=AnimGraph.AnimGraphNode_StateMachine_1.FullBody_1.AnimStateTransitionNode_22.Transition
  nodes [Result(bCanEnterTransition=Get bTransition_IdleLobby_Idle())]
 }
 Transition {
  path=AnimGraph.AnimGraphNode_StateMachine_1.FullBody_1.AnimStateTransitionNode_46.Transition
  nodes [Result(bCanEnterTransition=Get bIsDBNO())]
 }
 Transition {path=AnimGraph.AnimGraphNode_StateMachine_1.FullBody_1.AnimStateTransitionNode_47.Transition, nodes [Result()]}
 EventGraph {nodes [
   LobbySkyDive_IsDiving() {is=title K2Node_CustomEvent}
   > Set bLobbySkyDive_IsDiving(bLobbySkyDive_IsDiving:bool=true) -> (Output_Get:bool)
   LobbySkyDive_IsGliding() {is=title K2Node_CustomEvent}
   > Set bLobbySkyDive_IsDiving(bLobbySkyDive_IsDiving=false)
   n1 = AnimNotify_StartJump_Entered_Test() {is=title K2Node_Event}
   Set bIsMovingDuringJumpInitiated(execute=[@n1.then, @n4.then], bIsMovingDuringJumpInitiated:bool=Get bIsMoving2D()) -> (Output_Get:bool)
   > Set JumpIndex(JumpIndex:int=RandomInteger!(Max:int=2) -> (ReturnValue:int=0)) -> (Output_Get:int=0)
   n2 = AnimNotify_RollRandomIdle() {is=title K2Node_Event}
   n3 = "Clamp (Integer)"(Value=RandomIntegerInRange!(Min:int=0, Max:int=11) -> (ReturnValue:int=0), Min=0, Max=2) -> (ReturnValue=0)
   Set RandomIdle(execute=[@n2.then, @n6.then], RandomIdle:int=@n3) -> (Output_Get:int=0)
   n4 = AnimNotify_StartJump_Entered() {is=title K2Node_Event}
   n5 = AnimNotify_RollRandomIdle2() {is=title K2Node_Event}
   Set RandomIdle2(execute=[@n5.then, @n7.then], RandomIdle2:int=@n3) -> (Output_Get:int=0)
   AnimNotify_Entered_Crouching() {is=title K2Node_Event}
   > Set bAllowSkippingIdleToCrouchTransition(bAllowSkippingIdleToCrouchTransition:bool=false) -> (Output_Get:bool)
   n6 = AnimNotify_CrouchIdle1_FullyBlended() {is=title K2Node_Event}
   n7 = AnimNotify_CrouchIdle2_FullyBlended() {is=title K2Node_Event}
   AnimNotify_Entered_DownSightsToCrouch() {is=title K2Node_Event}
   > Set bAllowSkippingIdleToCrouchTransition(bAllowSkippingIdleToCrouchTransition=true)
   n8 = Flip Flop!(None:Exec) -> (A:Exec, B:Exec, IsA:bool=~) {is=title K2Node_MacroInstance}
   > Set JumpIndex(JumpIndex=0) -> (Output_Get=0)
   Set JumpIndex(execute=@n8.B, JumpIndex=1) -> (Output_Get=0)
  ]}
 PetFullOverrideLayer {nodes [Output Pose(Result=Input Pose()) {layerGroup=PetOverride}]}
 PetSkeletalControlLayer {nodes [
   n1 = Input Pose!() -> (Pose, LookAtOverrideInput:double, LookAtTargetLocationInput:Vector, LegIKAlphaOverride:double, HandIKAlphaOverride:double) {name=Pose}
   n2 = "Transform (Modify) Bone - Bone: head"!(ComponentPose=Local To Component(), Alpha=@n1.LookAtOverrideInput, Rotation="0,0,0") -> (Pose) {rotationMode=BMM_Replace, rotationSpace=BCS_BoneSpace, alphaBoolBlend {blendOption=Linear}, boneToModify {boneName=head}}
   n3 = Look At - Bone: R_eye!(
    ComponentPose=Look At - Bone: head!(ComponentPose=@n2, Alpha=@n1.LookAtOverrideInput, LookAtLocation:Vector=@n1.LookAtTargetLocationInput) -> (Pose) {lookAtClamp=50, alphaBoolBlend {blendOption=Linear}, boneToModify {boneName=head}, lookAtTarget=~}
    Alpha=@n1.LookAtOverrideInput
    LookAtLocation:Vector=@n1.LookAtTargetLocationInput
   ) -> (Pose) {
    lODThreshold=2
    lookAtClamp=40
    alphaBoolBlend {blendOption=Linear}
    boneToModify {boneName=R_eye}
    lookAt_Axis {axis {y=-1}}
    lookUp_Axis {axis {x=0, z=1}}
    lookAtTarget=~
   }
   n4 = Look At - Bone: L_eye!(ComponentPose=@n3, Alpha=@n1.LookAtOverrideInput, LookAtLocation:Vector=@n1.LookAtTargetLocationInput) -> (Pose) {
    lODThreshold=2
    lookAtClamp=40
    alphaBoolBlend {blendOption=Linear}
    boneToModify {boneName=L_eye}
    lookAt_Axis {axis {y=-1}}
    lookUp_Axis {axis {x=0, z=1}}
    lookAtTarget=~
   }
   n5 = Leg IK!(ComponentPose:Component Space Pose Link Structure=@n4, Alpha:float=@n1.LegIKAlphaOverride) -> (Pose:Component Space Pose Link Structure) {
    is=title AnimGraphNode_LegIK
    alphaBoolBlend {blendOption=Linear}
    legsDefinition [
     {
      bEnableKneeTwistCorrection=false
      bEnableRotationLimit=false
      footBoneForwardAxis=Y
      hingeRotationAxis=Z
      minRotationAngle=15
      numBonesInLimb=2
      twistOffsetCurveName=None
      fKFootBone {boneName=foot_l}
      iKFootBone {boneName=ik_foot_l}
     }
     {
      bEnableKneeTwistCorrection=false
      bEnableRotationLimit=false
      footBoneForwardAxis=Y
      hingeRotationAxis=Z
      minRotationAngle=15
      numBonesInLimb=2
      twistOffsetCurveName=None
      fKFootBone {boneName=foot_r}
      iKFootBone {boneName=ik_foot_r}
     }
    ]
   }
   n6 = Leg IK(ComponentPose=@n5, Alpha=@n1.HandIKAlphaOverride) {
    alphaBoolBlend {blendOption=Linear}
    legsDefinition [
     {
      bEnableKneeTwistCorrection=false
      bEnableRotationLimit=false
      footBoneForwardAxis=Y
      hingeRotationAxis=Z
      minRotationAngle=15
      numBonesInLimb=2
      twistOffsetCurveName=None
      fKFootBone {boneName=hand_l}
      iKFootBone {boneName=ik_hand_l}
     }
     {
      bEnableKneeTwistCorrection=false
      bEnableRotationLimit=false
      footBoneForwardAxis=Y
      hingeRotationAxis=Z
      minRotationAngle=15
      numBonesInLimb=2
      twistOffsetCurveName=None
      fKFootBone {boneName=hand_r}
      iKFootBone {boneName=ik_hand_r}
     }
    ]
   }
   Component To Local(ComponentPose=@n6)
   Output Pose(Result=@n1) {layerGroup=PetOverride}
  ]}
]
variables [
 bAllowSkippingIdleToCrouchTransition {type=boolean}
 bHasPlayerLocationBeenSet {type=boolean, default=true}
 bIsFox {type=boolean}
 bIsJumpLandComplete_0 {type=boolean, default=true}
 bIsMovingDuringJumpInitiated {type=boolean}
 BlockAnimRule {type=boolean}
 DownSightsToLocomotionTimeRemaining_b {type=number, default=0.33}
 FacePoseAsset {type=PoseAsset, default=/Game/Animation/Game/Pets/Chameleon/Pet_FacialPose_Export_PoseAsset}
 HandIKWeight {type=number, default=1}
 JumpIndex {type=integer}
 PlayerLocation {type=Vector}
 RandomIdle {type=integer, default=2}
 RandomIdle2 {type=integer, default=2}
]
