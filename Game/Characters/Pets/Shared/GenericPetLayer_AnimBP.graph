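# Generic pet animation layer Blueprint on the Chameleon skeleton, supplying PetOverride layers for procedural look-at and limb IK.
# PetSkeletalControlLayer resets head bone rotation, then chains clamped Look At on head and both eyes, driven by the layer's target and alpha inputs.
# Leg IK nodes then solve feet and hands toward IK bones, each pair with its own alpha.
# PetFullOverrideLayer is unwired, and the main AnimGraph and EventGraph contain no logic.
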AnimBlueprint /Game/Characters/Pets/Shared/GenericPetLayer_AnimBP
target_skeleton=/Game/Characters/Pets/Chameleon/Meshes/Chameleon_Skeleton
graphs [
 AnimGraph {nodes [Output Pose() {layerGroup=None}]}
 EventGraph {nodes [
   TryGetPawnOwner()
   Event BlueprintUpdateAnimation()
  ]}
 PetFullOverrideLayer {nodes [
   Input Pose()
   Output Pose() {layerGroup=PetOverride}
  ]}
 PetSkeletalControlLayer {nodes [
   n1 = Input Pose!() -> (Pose, LookAtOverrideInput:double, LookAtTargetLocationInput:Vector, LegIKAlphaOverride:double, HandIKAlphaOverride:double) {name=Pose}
   n2 = "Transform (Modify) Bone - Bone: head"!(ComponentPose=Local To Component(LocalPose=@n1), Alpha=@n1.LookAtOverrideInput, Rotation="0,0,0") -> (Pose) {rotationMode=BMM_Replace, rotationSpace=BCS_BoneSpace, alphaBoolBlend {blendOption=Linear}, boneToModify {boneName=head}}
   n3 = Look At - Bone: R_eye!(
    ComponentPose=Look At - Bone: head!(ComponentPose=@n2, Alpha=@n1.LookAtOverrideInput, LookAtLocation:Vector=@n1.LookAtTargetLocationInput) -> (Pose) {lookAtClamp=50, alphaBoolBlend {blendOption=Linear}, boneToModify {boneName=head}, lookAtTarget=~}
    Alpha=@n1.LookAtOverrideInput
    LookAtLocation:Vector=@n1.LookAtTargetLocationInput
   ) -> (Pose) {
    lODThreshold=2
    lookAtClamp=40
    alphaBoolBlend {blendOption=Linear}
    boneToModify {boneName=R_eye}
    lookAt_Axis {axis {y=-1}}
    lookUp_Axis {axis {x=0, z=1}}
    lookAtTarget=~
   }
   n4 = Look At - Bone: L_eye!(ComponentPose=@n3, Alpha=@n1.LookAtOverrideInput, LookAtLocation:Vector=@n1.LookAtTargetLocationInput) -> (Pose) {
    lODThreshold=2
    lookAtClamp=40
    alphaBoolBlend {blendOption=Linear}
    boneToModify {boneName=L_eye}
    lookAt_Axis {axis {y=-1}}
    lookUp_Axis {axis {x=0, z=1}}
    lookAtTarget=~
   }
   n5 = Leg IK!(ComponentPose:Component Space Pose Link Structure=@n4, Alpha:float=@n1.LegIKAlphaOverride) -> (Pose:Component Space Pose Link Structure) {
    is=title AnimGraphNode_LegIK
    alphaBoolBlend {blendOption=Linear}
    legsDefinition [
     {
      bEnableKneeTwistCorrection=false
      bEnableRotationLimit=false
      footBoneForwardAxis=Y
      hingeRotationAxis=Z
      minRotationAngle=15
      numBonesInLimb=2
      twistOffsetCurveName=None
      fKFootBone {boneName=foot_l}
      iKFootBone {boneName=ik_foot_l}
     }
     {
      bEnableKneeTwistCorrection=false
      bEnableRotationLimit=false
      footBoneForwardAxis=Y
      hingeRotationAxis=Z
      minRotationAngle=15
      numBonesInLimb=2
      twistOffsetCurveName=None
      fKFootBone {boneName=foot_r}
      iKFootBone {boneName=ik_foot_r}
     }
    ]
   }
   n6 = Leg IK(ComponentPose=@n5, Alpha=@n1.HandIKAlphaOverride) {
    alphaBoolBlend {blendOption=Linear}
    legsDefinition [
     {
      bEnableKneeTwistCorrection=false
      bEnableRotationLimit=false
      footBoneForwardAxis=Y
      hingeRotationAxis=Z
      minRotationAngle=15
      numBonesInLimb=2
      twistOffsetCurveName=None
      fKFootBone {boneName=hand_l}
      iKFootBone {boneName=ik_hand_l}
     }
     {
      bEnableKneeTwistCorrection=false
      bEnableRotationLimit=false
      footBoneForwardAxis=Y
      hingeRotationAxis=Z
      minRotationAngle=15
      numBonesInLimb=2
      twistOffsetCurveName=None
      fKFootBone {boneName=hand_r}
      iKFootBone {boneName=ik_hand_r}
     }
    ]
   }
   Output Pose(Result=Component To Local(ComponentPose=@n6)) {layerGroup=PetOverride}
  ]}
]
variables []
